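# Material function that shifts hue and scales saturation and value of the Base Color within a Material Attributes input.
# A Custom HLSL node applies a single matrix-style HSV transform using luma weights and sine/cosine of the hue shift, avoiding an RGB-to-HSV round trip.
# A Strength input lerps between original and transformed color, which is written back into the attributes for output.
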MaterialFunction /BaseMaterial/Materials/Functions/MF_DifColor_TransformHSV
graph {nodes [
  MaterialExpressionGetMaterialAttributes_0 = GetMaterialAttributes(None=FunctionInput() {bUsePreviewValueAsDefault=true, inputName=Input MA, inputType=FunctionInput_MaterialAttributes}) {attributeGetTypes [#1]}
  MaterialExpressionFunctionInput_1 = FunctionInput() {bUsePreviewValueAsDefault=true, inputName=HueShift, inputType=FunctionInput_Scalar, sortPriority=1}
  MaterialExpressionFunctionInput_2 = FunctionInput() {bUsePreviewValueAsDefault=true, inputName=SaturationMult, inputType=FunctionInput_Scalar, sortPriority=2, previewValue=<1 0 0 1>}
  MaterialExpressionFunctionInput_3 = FunctionInput() {bUsePreviewValueAsDefault=true, inputName=ValueMult, inputType=FunctionInput_Scalar, sortPriority=3, previewValue=<1 0 0 1>}
  n1 = Custom!(
   Color=@MaterialExpressionGetMaterialAttributes_0.BaseColor
   h=@MaterialExpressionFunctionInput_1
   s=@MaterialExpressionFunctionInput_2
   v=@MaterialExpressionFunctionInput_3
  ) {
   code="//Color Transform HSV\r\n//https://beesbuzz.biz/code/16-hsv-color-transforms\r\n\r\nfloat vsu = v * s * cos(h * 6.28318530718);\r\nfloat vsw = v * s * sin(h * 6.28318530718);\r\n\r\nfloat3 result = float3( 0.,0.,0. );\r\n\r\nresult.r =  (.299*v + .701*vsu + .168*vsw) * Color.r\r\n        +   (.587*v - .587*vsu + .330*vsw) * Color.g\r\n        +   (.114*v - .114*vsu - .497*vsw) * Color.b;\r\nresult.g =  (.299*v - .299*vsu - .328*vsw) * Color.r\r\n        +   (.587*v + .413*vsu + .035*vsw) * Color.g\r\n        +   (.114*v - .114*vsu + .292*vsw) * Color.b;\r\nresult.b =  (.299*v - .300*vsu + 1.25*vsw) * Color.r\r\n        +   (.587*v - .588*vsu - 1.05*vsw) * Color.g\r\n        +   (.114*v + .886*vsu - .203*vsw) * Color.b;\r\nreturn result;"
   description=TransformHSV
   inputs [
    {
     inputName=Color
     input {expression=&self:MaterialExpressionGetMaterialAttributes_0, outputIndex=1}
    }
    {inputName=h, input {expression=&self:MaterialExpressionFunctionInput_1}}
    {inputName=s, input {expression=&self:MaterialExpressionFunctionInput_2}}
    {inputName=v, input {expression=&self:MaterialExpressionFunctionInput_3}}
   ]
  }
  n2 = FunctionInput() {bUsePreviewValueAsDefault=true, inputName=Strength, inputType=FunctionInput_Scalar, sortPriority=4, previewValue=<1 0 0 1>}
  FunctionOutput(None=SetMaterialAttributes!(
    MaterialAttributes=@MaterialExpressionGetMaterialAttributes_0.MaterialAttributes
    BaseColor=LinearInterpolate(A=@MaterialExpressionGetMaterialAttributes_0.BaseColor, B=@n1, Alpha=@n2)
   ) {attributeSetTypes [#1]}) {outputName=Output MA}
 ]}
