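# Library material function that turns a scalar gradient into a soft band-selection mask centered on a Mask Pivot.
# A Custom HLSL node subtracts the clamped pivot from the gradient and takes the difference of two smoothsteps, with separate half-width, low-side and high-side falloff controls.
# The result is normalized by the band's center value and saturated.
# Code comments note a pivot of 1 suits single-direction dot(direction, normal) masks.
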MaterialFunction /BaseMaterial/Materials/Functions/MF_Mask_BandSelect
graph {
 properties {bExposeToLibrary=true, libraryCategoriesText [Base Material]}
 nodes [
  MaterialExpressionFunctionInput_0 = FunctionInput(Preview=ComponentMask(None=TextureCoordinate()) {r=true}) {bUsePreviewValueAsDefault=true, inputName=Gradient, inputType=FunctionInput_Scalar}
  MaterialExpressionFunctionInput_3 = FunctionInput() {bUsePreviewValueAsDefault=true, inputName=Mask Pivot, inputType=FunctionInput_Scalar, sortPriority=1, previewValue=<1 0 0 1>}
  MaterialExpressionFunctionInput_2 = FunctionInput() {bUsePreviewValueAsDefault=true, inputName="Mask Controls (Width/FalloffLow/FalloffHigh)", sortPriority=2, previewValue=<0.5 0.5 0.5 0>}
  n1 = Custom!(grad=@MaterialExpressionFunctionInput_0, pivot=@MaterialExpressionFunctionInput_3, falloff=@MaterialExpressionFunctionInput_2) {
   code="/*\r\npivot is the true center of the selection band\r\nwhen used with a dot(direction, normal) should probably be 1 \r\nfor single direction control where the default values \r\npivot=1 and falloff.xyz=0.5 reduce to a basic smoothstep\r\n*/\r\n\r\n//Allow pre-adjustments to the gradient\r\nfloat delta = grad - clamp(pivot, -0.99999, 0.99999);\r\n\r\n//falloff.x is the half-width of band (distance from center to edge)\r\n//define a safe minimum half-width\r\nfalloff.x = max(falloff.x, 0.01);\r\n\r\n//falloff.y is the falloff distance on the negative side of the band\r\n//falloff.z is the falloff distance on the positive side of the band\r\n\r\nfloat neg = smoothstep(  falloff.x - falloff.z,  falloff.x + falloff.z, delta);\r\nfloat pos = smoothstep( -falloff.x - falloff.y, -falloff.x + falloff.y, delta);\r\n\r\nfloat res = pos - neg;\r\n\r\nconst float fnmax = smoothstep( -falloff.x - falloff.y, -falloff.x + falloff.y, 0.f) -\r\n                    smoothstep(  falloff.x - falloff.z,  falloff.x + falloff.z, 0.f);\r\n\r\nreturn saturate( res / max(fnmax, 1e-6) );"
   description=Band Select
   showCode=true
   inputs [
    {inputName=grad, input {expression=&self:MaterialExpressionFunctionInput_0}}
    {inputName=pivot, input {expression=&self:MaterialExpressionFunctionInput_3}}
    {inputName=falloff, input {expression=&self:MaterialExpressionFunctionInput_2}}
   ]
  }
  FunctionOutput(None=@n1) {outputName=Mask}
 ]
}
