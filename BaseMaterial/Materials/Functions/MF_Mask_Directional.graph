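# MF_Mask_Directional is a library material function in Base Material that builds a mask from how closely a normal aligns with a direction.
# It takes the dot product of Direction and Normal, defaulting to up and the world-space vertex normal.
# A subtract and divide remap that result using the split Input Min/Max pair.
# The output is unclamped, preserving negative values as its description notes.
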MaterialFunction /BaseMaterial/Materials/Functions/MF_Mask_Directional
graph {
 properties {bExposeToLibrary=true, libraryCategoriesText [Base Material]}
 nodes [
  n1 = DotProduct(
   A=FunctionInput(Preview=Constant3Vector() {constant=<rgba 0 0 1 0>}) {bUsePreviewValueAsDefault=true, inputName=Direction}
   B=FunctionInput(Preview=VertexNormalWS()) {bUsePreviewValueAsDefault=true, inputName=Normal, sortPriority=1}
  )
  MaterialExpressionFunctionInput_2 = FunctionInput() {bUsePreviewValueAsDefault=true, inputName=Input Min/Max, inputType=FunctionInput_Vector2, sortPriority=2, previewValue=<0 1 0 1>}
  n2 = Convert(None=@MaterialExpressionFunctionInput_2) {
   nodeName=Break Float2
   convertInputs [{
     type=Vector2
     expressionInput {
      expression=&self:MaterialExpressionFunctionInput_2
      inputName=None
      mask=0
      maskA=0
      maskB=0
      maskG=0
      maskR=0
      outputIndex=0
     }
    }]
   convertOutputs [{type=Scalar}, {type=Scalar}]
  }
  FunctionOutput(None=Divide(A=Subtract(A=@n1, B=@n2), B=Subtract(A=@n2, B=@n2))) {description=preserves negative and vector scale, outputName=Mask}
 ]
}
