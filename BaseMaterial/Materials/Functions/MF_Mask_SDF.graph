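# Library material function that outputs a capsule or box signed-distance field, named as a mask, for use in materials.
# Camera-relative world position is transformed to instance space, shifted by Shape Offset, then rotated by yaw and pitch in custom HLSL.
# The rotated point feeds both tapered shapes, adapted from Inigo Quilez's distance functions, with extents and corner radius inputs.
# A truncated selector picks capsule (0) or box (1), defaulting to capsule otherwise.
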MaterialFunction /BaseMaterial/Materials/Functions/MF_Mask_SDF
graph {
 properties {bExposeToLibrary=true, libraryCategoriesText [Base Material]}
 nodes [
  n1 = FunctionInput() {bUsePreviewValueAsDefault=true, desc="0 = capsule, 1 = box", inputName=select SDF, inputType=FunctionInput_Scalar, sortPriority=-5}
  MaterialExpressionTruncate_0 = Truncate!(None=@n1) {is=class MaterialExpressionTruncate}
  MaterialExpressionAdd_3 = Add(
   A=FunctionInput() {bUsePreviewValueAsDefault=true, inputName=Shape Offset}
   B=TransformPosition(None=WorldPosition() {worldPositionShaderOffset=WPT_CameraRelative}.XYZ) {transformSourceType=TRANSFORMPOSSOURCE_TranslatedWorld, transformType=TRANSFORMPOSSOURCE_Instance}
  ) {constB=0}
  MaterialExpressionFunctionInput_2 = FunctionInput() {bUsePreviewValueAsDefault=true, inputName=Cooridnate Yaw, inputType=FunctionInput_Scalar, sortPriority=5}
  MaterialExpressionFunctionInput_3 = FunctionInput() {bUsePreviewValueAsDefault=true, inputName=Cooridnate Pitch, inputType=FunctionInput_Scalar, sortPriority=6}
  MaterialExpressionCustom_3 = Custom!(p=@MaterialExpressionAdd_3, yaw=@MaterialExpressionFunctionInput_2, pitch=@MaterialExpressionFunctionInput_3) {
   code="struct MyFunctions\r\n{\r\n    float3 RotateAboutAxis(float3 vec, float3 axis, float angle)\r\n    {\r\n        float s = sin( radians(angle) );\r\n        float c = cos( radians(angle) );\r\n        return vec * c + cross(axis, vec) * s + axis * dot(axis, vec) * (1.0 - c);\r\n    }\r\n};\r\n\r\n//Initialize Struct\r\nMyFunctions CallFunc;\r\n\r\n//Generate Rotation Matix\r\nfloat3 right = normalize( CallFunc.RotateAboutAxis( float3(1.,0.,0.), \r\n                                                    float3(0.,0.,1.), yaw ) );\r\nfloat3 up = normalize( CallFunc.RotateAboutAxis(    float3(0.,0.,1.), right, pitch + 90 ) );\r\nfloat3 forward = cross(right,up);\r\n\r\n//Calculate inverse matrix\r\nfloat3x3 invRot = transpose( float3x3(right, up, forward) );\r\n\r\nreturn mul(p, invRot);"
   description=" Axis-Angle Rotator"
   showCode=true
   inputs [
    {inputName=p, input {expression=&self:MaterialExpressionAdd_3}}
    {inputName=yaw, input {expression=&self:MaterialExpressionFunctionInput_2}}
    {inputName=pitch, input {expression=&self:MaterialExpressionFunctionInput_3}}
   ]
  }
  MaterialExpressionFunctionInput_4 = FunctionInput() {bUsePreviewValueAsDefault=true, inputName=Shape Extents, sortPriority=10}
  MaterialExpressionComponentMask_0 = ComponentMask(None=@MaterialExpressionFunctionInput_4) {g=true}
  MaterialExpressionFunctionInput_5 = FunctionInput() {bUsePreviewValueAsDefault=true, inputName=Corner Radius, inputType=FunctionInput_Scalar, sortPriority=11}
  MaterialExpressionFunctionInput_6 = FunctionInput() {bUsePreviewValueAsDefault=true, inputName=Taper, inputType=FunctionInput_Scalar, sortPriority=12}
  MaterialExpressionCustom_0 = Custom!(
   p=@MaterialExpressionCustom_3
   len=@MaterialExpressionComponentMask_0
   r=@MaterialExpressionFunctionInput_5
   taper=@MaterialExpressionFunctionInput_6
  ) {
   code="//modified from https://iquilezles.org/articles/distfunctions/\r\n//len is line length and r is cap radius\r\n\r\nconst float hlen = 0.5 * len;\r\nconst float y = clamp(p.y, -hlen, hlen);\r\n\r\nfloat2 d = float2(length(p.xz/(1. + y * taper/100.)), p.y - y);\r\nreturn length(d) - r;"
   description=CapsuleSDF
   outputType=CMOT_Float1
   showCode=true
   inputs [
    {inputName=p, input {expression=&self:MaterialExpressionCustom_3}}
    {inputName=len, input {expression=&self:MaterialExpressionComponentMask_0}}
    {inputName=r, input {expression=&self:MaterialExpressionFunctionInput_5}}
    {inputName=taper, input {expression=&self:MaterialExpressionFunctionInput_6}}
   ]
  }
  MaterialExpressionCustom_4 = Custom!(
   p=@MaterialExpressionCustom_3
   e=@MaterialExpressionFunctionInput_4
   r=@MaterialExpressionFunctionInput_5
   taper=@MaterialExpressionFunctionInput_6
  ) {
   code="//modified from https://iquilezles.org/articles/distfunctions/\r\n//e is box extents, and r is corner radius \r\nfloat np = 1. + p.z * taper/100;\r\np /= np;\r\n\r\nconst float3 d = abs(p) - (e - r);\r\nconst float box = length( max(d, 0.) ) - r;\r\n\r\nreturn box;"
   description=BoxSDF
   outputType=CMOT_Float1
   showCode=true
   inputs [
    {inputName=p, input {expression=&self:MaterialExpressionCustom_3}}
    {inputName=e, input {expression=&self:MaterialExpressionFunctionInput_4}}
    {inputName=r, input {expression=&self:MaterialExpressionFunctionInput_5}}
    {inputName=taper, input {expression=&self:MaterialExpressionFunctionInput_6}}
   ]
  }
  n2 = Custom!(selector=@MaterialExpressionTruncate_0, input0=@MaterialExpressionCustom_0, input1=@MaterialExpressionCustom_4) {
   code="[branch]\r\nif(selector == 0)\r\nreturn input0;\r\nelse if (selector == 1)\r\nreturn input1;\r\nelse\r\nreturn input0;"
   description=Selector
   showCode=true
   inputs [
    {inputName=selector, input {expression=&self:MaterialExpressionTruncate_0}}
    {inputName=input0, input {expression=&self:MaterialExpressionCustom_0}}
    {inputName=input1, input {expression=&self:MaterialExpressionCustom_4}}
   ]
  }
  FunctionOutput(None=@n2)
 ]
}
