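# Material function in the Base Material library that blends two scalar inputs into a compensated weighted average.
# The output is (Weight1 * Input1 + Weight2 * (2 - Weight1) * Input2) * 0.5, so lowering Weight1 raises the share of Input2.
# Math nodes compute the result, while an unconnected CompensatedAverage Custom HLSL node holds the same formula.
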MaterialFunction /BaseMaterial/Materials/Functions/MF_Compensated_Average
graph {
 properties {bExposeToLibrary=true, libraryCategoriesText [Base Material]}
 nodes [
  Custom!(x1, x2, w1, w2) {
   code="return (w1 * x1 + w2 * (2.0 - w1) * x2) * 0.5;"
   description=CompensatedAverage
   outputType=CMOT_Float1
   showCode=true
   inputs [
    {inputName=x1, input {expression=null}}
    {inputName=x2, input {expression=null}}
    {inputName=w1, input {expression=null}}
    {inputName=w2, input {expression=null}}
   ]
  }
  n1 = FunctionInput() {bUsePreviewValueAsDefault=true, inputName=Input1, inputType=FunctionInput_Scalar, previewValue=<0.5 0 0 1>}
  n2 = FunctionInput() {bUsePreviewValueAsDefault=true, inputName=Weight1, inputType=FunctionInput_Scalar, sortPriority=2, previewValue=<0.5 0 0 1>}
  n3 = Multiply(A=@n1, B=@n2)
  n4 = FunctionInput() {bUsePreviewValueAsDefault=true, inputName=Input2, inputType=FunctionInput_Scalar, sortPriority=1, previewValue=<0.5 0 0 1>}
  n5 = FunctionInput() {bUsePreviewValueAsDefault=true, inputName=Weight2, inputType=FunctionInput_Scalar, sortPriority=3, previewValue=<0.5 0 0 1>}
  n6 = Multiply(A=@n4, B=Multiply(A=Subtract(B=@n2) {constA=2}, B=@n5))
  FunctionOutput(None=Multiply(A=Add(A=@n3, B=@n6)) {constB=0.5})
 ]
}
