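# Library material function that builds a view mask from camera distance and viewing angle, exposed under Base Material.
# Camera-relative position length goes through a SmoothStep bounded by the black/white depth clamp.
# Custom HLSL computes view-dependent coverage from Thickness, Coverage and overridable N·V, using Smith-Schlick G ("Practical Multilayered Materials", Siggraph 2017).
# MF_Compensated_Average blends both by Depth and Angle Strength; non-negative Thickness is also output.
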MaterialFunction /BaseMaterial/Materials/Functions/MF_Mask_View
graph {
 properties {bExposeToLibrary=true, libraryCategoriesText [Base Material]}
 nodes [
  MaterialExpressionFunctionInput_1 = FunctionInput() {
   bUsePreviewValueAsDefault=true
   inputName="Depth Clamp (Black/White)"
   inputType=FunctionInput_Vector2
   sortPriority=10
   previewValue=<1000 100 0 1>
  }
  n1 = Convert(None=@MaterialExpressionFunctionInput_1) {
   nodeName=Break Float2
   convertInputs [{
     type=Vector2
     expressionInput {
      expression=&self:MaterialExpressionFunctionInput_1
      inputName=None
      mask=0
      maskA=0
      maskB=0
      maskG=0
      maskR=0
      outputIndex=0
     }
    }]
   convertOutputs [{type=Scalar}, {type=Scalar}]
  }
  n2 = Length(None=WorldPosition() {worldPositionShaderOffset=WPT_CameraRelative}.XYZ) {is=class MaterialExpressionLength}
  MaterialExpressionMax_0 = Max(A=FunctionInput(Preview=Constant() {r=0.01}) {bUsePreviewValueAsDefault=true, inputName=Thickness, inputType=FunctionInput_Scalar}) {constB=0}
  MaterialExpressionSaturate_1 = Saturate(None=FunctionInput(Preview=Constant()) {bUsePreviewValueAsDefault=true, inputName=Coverage, inputType=FunctionInput_Scalar, sortPriority=1})
  n3 = VertexNormalWS()
  n4 = Normalize(VectorInput=Subtract(A=CameraPositionWS(), B=WorldPosition().XYZ))
  MaterialExpressionFunctionInput_3 = FunctionInput(Preview=Saturate(None=DotProduct(A=@n3, B=@n4))) {bUsePreviewValueAsDefault=true, inputName=FresnelOverride, inputType=FunctionInput_Scalar, sortPriority=2}
  n5 = Custom!(Thickness=@MaterialExpressionMax_0, Coverage=@MaterialExpressionSaturate_1, NoV=@MaterialExpressionFunctionInput_3) {
   code="\t// Reference \"Practical Multilayered Materials\" Siggraph 2017\r\n\tconst float x = NoV;\r\n\tconst float t = Thickness;\t\r\n\tconst float alpha = Coverage;\t\r\n\tconst float g = 1.0f - ( x / ( x * ( 1.0 - t ) + t ) ); // Smith-Schlick G\r\n\treturn saturate( lerp( alpha, 1.0f, g ) ) ;"
   description=ViewDependentCoverage
   outputType=CMOT_Float1
   inputs [
    {inputName=Thickness, input {expression=&self:MaterialExpressionMax_0}}
    {inputName=Coverage, input {expression=&self:MaterialExpressionSaturate_1}}
    {inputName=NoV, input {expression=&self:MaterialExpressionFunctionInput_3}}
   ]
  }
  n6 = FunctionInput() {bUsePreviewValueAsDefault=true, inputName=Depth Strength, inputType=FunctionInput_Scalar, sortPriority=11, previewValue=<1 0 0 1>}
  n7 = FunctionInput() {bUsePreviewValueAsDefault=true, inputName=Angle Strength, inputType=FunctionInput_Scalar, sortPriority=3, previewValue=<1 0 0 1>}
  FunctionOutput(None=MaterialFunctionCall!(Input1=SmoothStep(Min=@n1, Max=@n1, Value=@n2), Input2=@n5, Weight1=@n6, Weight2=@n7) {materialFunction=&/MF_Compensated_Average}.Result)
  FunctionOutput(None=@MaterialExpressionMax_0) {description=The resulting thickness. Should be unchanged compare to the input., outputName=Thickness, sortPriority=1}
  PixelDepth()
 ]
}
