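# Material function that flattens the normal of an incoming Material Attributes set by a scalar Flatness amount.
# It reads the Normal attribute and lerps it toward a default surface normal, then writes it back while passing the other attributes through.
# A Custom HLSL node supplies that default, using a compile-time check to return tangent-space up or the world-space vertex normal, so both normal spaces are supported.
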MaterialFunction /BaseMaterial/Materials/Functions/MF_Normal_Flatten
graph {nodes [
  n1 = GetMaterialAttributes(None=FunctionInput() {bUsePreviewValueAsDefault=true, inputName=Input MA, inputType=FunctionInput_MaterialAttributes}) {attributeGetTypes [#1]}
  n2 = Custom() {
   code="{\r\n##if MATERIAL_TANGENTSPACENORMAL\r\n\tconst float3 N = float3(0.f,0.f,1.f);\r\n##else\r\n\tconst float3 N = Parameters.TangentToWorld[2];\r\n##endif\r\n\treturn N;\r\n}"
   description=Default Normal
   inputs []
  }
  n3 = FunctionInput() {bUsePreviewValueAsDefault=true, inputName=Flatness, inputType=FunctionInput_Scalar, sortPriority=1}
  FunctionOutput(None=SetMaterialAttributes!(MaterialAttributes=@n1.MaterialAttributes, Normal=LinearInterpolate(A=@n1.Normal, B=@n2, Alpha=@n3)) {attributeSetTypes [#1]}) {outputName=Output MA}
 ]}
