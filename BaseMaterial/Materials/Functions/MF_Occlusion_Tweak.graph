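# Material function that tweaks the ambient occlusion channel of an incoming Material Attributes set and passes the attributes on.
# The AO Strength input is inverted, added to the existing occlusion and saturated, so lower strength lightens the occlusion.
# A Fresnel term then blends that result toward 1, fading occlusion out at grazing view angles.
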MaterialFunction /BaseMaterial/Materials/Functions/MF_Occlusion_Tweak
graph {nodes [
  n1 = GetMaterialAttributes(None=FunctionInput() {bUsePreviewValueAsDefault=true, inputName=Input MA, inputType=FunctionInput_MaterialAttributes}) {attributeGetTypes [#1]}
  n2 = FunctionInput() {bUsePreviewValueAsDefault=true, inputName=AO Strength, inputType=FunctionInput_Scalar, sortPriority=1}
  n3 = Saturate(None=Add(A=@n1.AmbientOcclusion, B=OneMinus(None=@n2)) {constB=0.5})
  n4 = Fresnel() {baseReflectFraction=0, exponent=2}
  FunctionOutput(None=SetMaterialAttributes!(MaterialAttributes=@n1.MaterialAttributes, AmbientOcclusion=LinearInterpolate(A=@n3, Alpha=@n4) {constA=1}) {attributeSetTypes [#1]}) {outputName=Output MA}
 ]}
