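# Material function that builds biplanar projection coordinates from position, reference normal, tiling and rotation, after Inigo Quilez's method.
# It picks the normal's major and median axes, outputting sign-corrected UVs, tangents and bitangents for both, plus blend weights shaped by BlendFactor.
# UV derivatives come from DDX/DDY, with a shader-stage switch for vertex-shader use.
# The logic is built in nodes; an unconnected Custom node keeps the HLSL as reference.
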MaterialFunction /BaseMaterial/Materials/Functions/MF_Coordinate_Biplanar
graph {
 expressions {
  MaterialExpressionAbs_0 {class=MaterialExpressionAbs, inputs [None {from=MaterialExpressionFunctionInput_1, output=""}], properties {}}
  MaterialExpressionAdd_0 {
   class=MaterialExpressionAdd
   inputs [
    A {from=MaterialExpressionComponentMask_1}
    B {from=""}
   ]
   properties {constB=0.0033}
  }
  MaterialExpressionAdd_1 {
   class=MaterialExpressionAdd
   inputs [
    A {from=MaterialExpressionComponentMask_2}
    B {from=""}
   ]
   properties {constB=0.0033}
  }
  MaterialExpressionAdd_13 {
   class=MaterialExpressionAdd
   inputs [
    A {from=MaterialExpressionMultiply_33}
    B {from=MaterialExpressionMultiply_31}
   ]
  }
  MaterialExpressionAdd_15 {
   class=MaterialExpressionAdd
   inputs [
    A {from=MaterialExpressionMultiply_47}
    B {from=MaterialExpressionMultiply_45}
   ]
  }
  MaterialExpressionAdd_2 {
   class=MaterialExpressionAdd
   inputs [
    A {from=MaterialExpressionAbs_0}
    B {from=MaterialExpressionConstant3Vector_0}
   ]
  }
  MaterialExpressionAppendVector_10 {
   class=MaterialExpressionAppendVector
   inputs [
    A {from=MaterialExpressionLinearInterpolate_62, output=""}
    B {from=MaterialExpressionLinearInterpolate_64, output=""}
   ]
  }
  MaterialExpressionAppendVector_11 {
   class=MaterialExpressionAppendVector
   inputs [
    A {from=MaterialExpressionReroute_114, output=""}
    B {from=MaterialExpressionReroute_116, output=""}
   ]
  }
  MaterialExpressionAppendVector_12 {
   class=MaterialExpressionAppendVector
   inputs [
    A {from=MaterialExpressionNamedRerouteUsage_20, output=""}
    B {from=MaterialExpressionNamedRerouteUsage_21, output=""}
   ]
  }
  MaterialExpressionAppendVector_2 {
   class=MaterialExpressionAppendVector
   inputs [
    A {from=MaterialExpressionLinearInterpolate_8, output=""}
    B {from=MaterialExpressionLinearInterpolate_10, output=""}
   ]
  }
  MaterialExpressionAppendVector_5 {
   class=MaterialExpressionAppendVector
   inputs [
    A {from=MaterialExpressionReroute_36, output=""}
    B {from=MaterialExpressionReroute_53, output=""}
   ]
  }
  MaterialExpressionComponentMask_0 {class=MaterialExpressionComponentMask, inputs [None {from=MaterialExpressionNamedRerouteUsage_2}], properties {g=true, r=true}}
  MaterialExpressionComponentMask_1 {class=MaterialExpressionComponentMask, inputs [None {from=MaterialExpressionNamedRerouteUsage_0}], properties {b=true}}
  MaterialExpressionComponentMask_2 {class=MaterialExpressionComponentMask, inputs [None {from=MaterialExpressionNamedRerouteUsage_3}], properties {b=true}}
  MaterialExpressionComponentMask_4 {class=MaterialExpressionComponentMask, inputs [None {from=MaterialExpressionNamedRerouteUsage_1}], properties {g=true, r=true}}
  MaterialExpressionConstant2Vector_0 {class=MaterialExpressionConstant2Vector}
  MaterialExpressionConstant2Vector_1 {class=MaterialExpressionConstant2Vector}
  MaterialExpressionConstant3Vector_0 {class=MaterialExpressionConstant3Vector, properties {constant=<rgba 0 0 0.0033 0>}}
  MaterialExpressionConstant3Vector_1 {class=MaterialExpressionConstant3Vector, properties {constant=<rgba 0 1 2 0>}}
  MaterialExpressionConstant3Vector_10 {class=MaterialExpressionConstant3Vector, properties {constant=<rgba 0 0 1 0>}}
  MaterialExpressionConstant3Vector_14 {class=MaterialExpressionConstant3Vector, properties {constant=<rgba 0 -1 0 0>}}
  MaterialExpressionConstant3Vector_15 {class=MaterialExpressionConstant3Vector, properties {constant=<rgba 1 0 0 0>}}
  MaterialExpressionConstant3Vector_18 {class=MaterialExpressionConstant3Vector, properties {constant=<rgba 0 -1 0 0>}}
  MaterialExpressionConstant3Vector_19 {class=MaterialExpressionConstant3Vector, properties {constant=<rgba 1 0 0 0>}}
  MaterialExpressionConstant3Vector_2 {class=MaterialExpressionConstant3Vector, properties {constant=<rgba 1 2 0 0>}}
  MaterialExpressionConstant3Vector_3 {class=MaterialExpressionConstant3Vector, properties {constant=<rgba 2 0 1 0>}}
  MaterialExpressionConstant3Vector_4 {class=MaterialExpressionConstant3Vector, properties {constant=<rgba 0 1 2 0>}}
  MaterialExpressionConstant3Vector_5 {class=MaterialExpressionConstant3Vector, properties {constant=<rgba 1 2 0 0>}}
  MaterialExpressionConstant3Vector_6 {class=MaterialExpressionConstant3Vector, properties {constant=<rgba 2 0 1 0>}}
  MaterialExpressionConstant3Vector_7 {class=MaterialExpressionConstant3Vector, properties {constant=<rgba 3 3 3 0>}}
  MaterialExpressionConstant3Vector_8 {class=MaterialExpressionConstant3Vector, properties {constant=<rgba 1 0 0 0>}}
  MaterialExpressionConstant3Vector_9 {class=MaterialExpressionConstant3Vector, properties {constant=<rgba 0 1 0 0>}}
  MaterialExpressionConstant4Vector_0 {
   class=MaterialExpressionConstant4Vector
   inputs []
   properties {desc="force highest mip in VS - might be possible to do better (see vd1/vd2)"}
  }
  MaterialExpressionConstant_10 {class=MaterialExpressionConstant, properties {r=3}}
  MaterialExpressionConstant_11 {class=MaterialExpressionConstant, properties {r=1}}
  MaterialExpressionConstant_3 {class=MaterialExpressionConstant, properties {r=1}}
  MaterialExpressionConstant_8 {class=MaterialExpressionConstant, properties {r=1}}
  MaterialExpressionConstant_9 {class=MaterialExpressionConstant, properties {r=1}}
  MaterialExpressionConvert_0 {
   class=MaterialExpressionConvert
   inputs [None {from=MaterialExpressionDDX_0}]
   properties {
    convertInputs [{
      type=Vector3
      expressionInput {
       expression=&self:MaterialExpressionDDX_0
       inputName=None
       mask=0
       maskA=0
       maskB=0
       maskG=0
       maskR=0
       outputIndex=0
      }
     }]
    convertOutputs [{type=Vector2}, {type=Vector2}, {type=Vector2}]
   }
  }
  MaterialExpressionConvert_1 {
   class=MaterialExpressionConvert
   inputs [None {from=MaterialExpressionDDY_0}]
   properties {
    convertInputs [{
      type=Vector3
      expressionInput {
       expression=&self:MaterialExpressionDDY_0
       inputName=None
       mask=0
       maskA=0
       maskB=0
       maskG=0
       maskR=0
       outputIndex=0
      }
     }]
    convertOutputs [{type=Vector2}, {type=Vector2}, {type=Vector2}]
   }
  }
  MaterialExpressionConvert_10 {
   class=MaterialExpressionConvert
   inputs [
    None {from=MaterialExpressionMultiply_7}
    None {from=MaterialExpressionMultiply_9}
   ]
   properties {
    convertInputs [
     {
      type=Scalar
      expressionInput {
       expression=&self:MaterialExpressionMultiply_7
       inputName=None
       mask=0
       maskA=0
       maskB=0
       maskG=0
       maskR=0
       outputIndex=0
      }
     }
     {
      type=Scalar
      expressionInput {
       expression=&self:MaterialExpressionMultiply_9
       inputName=None
       mask=0
       maskA=0
       maskB=0
       maskG=0
       maskR=0
       outputIndex=0
      }
     }
    ]
    convertOutputs [{type=Vector4, defaultValue=<rgba 0 0 0 1>}]
   }
  }
  MaterialExpressionConvert_11 {
   class=MaterialExpressionConvert
   inputs [
    None {from=MaterialExpressionMultiply_7}
    None {from=MaterialExpressionMultiply_8}
   ]
   properties {
    convertInputs [
     {
      type=Scalar
      expressionInput {
       expression=&self:MaterialExpressionMultiply_7
       inputName=None
       mask=0
       maskA=0
       maskB=0
       maskG=0
       maskR=0
       outputIndex=0
      }
     }
     {
      type=Scalar
      expressionInput {
       expression=&self:MaterialExpressionMultiply_8
       inputName=None
       mask=0
       maskA=0
       maskB=0
       maskG=0
       maskR=0
       outputIndex=0
      }
     }
    ]
    convertOutputs [{type=Vector4, defaultValue=<rgba 0 0 0 1>}]
   }
  }
  MaterialExpressionConvert_17 {
   class=MaterialExpressionConvert
   inputs [None {from=MaterialExpressionNamedRerouteUsage_6}]
   properties {
    nodeName=Break Float3
    convertInputs [{
      type=Vector3
      expressionInput {
       expression=&self:MaterialExpressionNamedRerouteUsage_6
       inputName=None
       mask=0
       maskA=0
       maskB=0
       maskG=0
       maskR=0
       outputIndex=0
      }
     }]
    convertOutputs [{type=Scalar}, {type=Scalar}, {type=Scalar}]
   }
  }
  MaterialExpressionConvert_18 {
   class=MaterialExpressionConvert
   inputs [None {from=MaterialExpressionNamedRerouteUsage_7}]
   properties {
    nodeName=Break Float3
    convertInputs [{
      type=Vector3
      expressionInput {
       expression=&self:MaterialExpressionNamedRerouteUsage_7
       inputName=None
       mask=0
       maskA=0
       maskB=0
       maskG=0
       maskR=0
       outputIndex=0
      }
     }]
    convertOutputs [{type=Scalar}, {type=Scalar}, {type=Scalar}]
   }
  }
  MaterialExpressionConvert_19 {
   class=MaterialExpressionConvert
   inputs [None {from=MaterialExpressionLinearInterpolate_12}]
   properties {
    convertInputs [{
      type=Vector2
      expressionInput {
       expression=&self:MaterialExpressionLinearInterpolate_12
       inputName=None
       mask=0
       maskA=0
       maskB=0
       maskG=0
       maskR=0
       outputIndex=0
      }
     }]
    convertOutputs [{type=Vector2}]
   }
  }
  MaterialExpressionConvert_2 {
   class=MaterialExpressionConvert
   inputs [
    None {from=MaterialExpressionMultiply_2}
    None {from=MaterialExpressionMultiply_1}
   ]
   properties {
    convertInputs [
     {
      type=Vector2
      expressionInput {
       expression=&self:MaterialExpressionMultiply_2
       inputName=None
       mask=0
       maskA=0
       maskB=0
       maskG=0
       maskR=0
       outputIndex=0
      }
     }
     {
      type=Vector2
      expressionInput {
       expression=&self:MaterialExpressionMultiply_1
       inputName=None
       mask=0
       maskA=0
       maskB=0
       maskG=0
       maskR=0
       outputIndex=0
      }
     }
    ]
    convertOutputs [{type=Vector4, defaultValue=<rgba 0 0 0 1>}]
   }
  }
  MaterialExpressionConvert_20 {
   class=MaterialExpressionConvert
   inputs [None {from=MaterialExpressionNamedRerouteUsage_8}]
   properties {
    nodeName=Break Float3
    convertInputs [{
      type=Vector3
      expressionInput {
       expression=&self:MaterialExpressionNamedRerouteUsage_8
       inputName=None
       mask=0
       maskA=0
       maskB=0
       maskG=0
       maskR=0
       outputIndex=0
      }
     }]
    convertOutputs [{type=Scalar}, {type=Scalar}, {type=Scalar}]
   }
  }
  MaterialExpressionConvert_21 {
   class=MaterialExpressionConvert
   inputs [
    None {from=MaterialExpressionConvert_20}
    None {from=MaterialExpressionMultiply_20}
    None {from=MaterialExpressionConvert_20}
    None {from=MaterialExpressionConstant_3}
   ]
   properties {
    desc=csign
    convertInputs [
     {
      type=Scalar
      expressionInput {
       expression=&self:MaterialExpressionConvert_20
       inputName=None
       mask=0
       maskA=0
       maskB=0
       maskG=0
       maskR=0
       outputIndex=0
      }
     }
     {
      type=Scalar
      expressionInput {
       expression=&self:MaterialExpressionMultiply_20
       inputName=None
       mask=0
       maskA=0
       maskB=0
       maskG=0
       maskR=0
       outputIndex=0
      }
     }
     {
      type=Scalar
      expressionInput {
       expression=&self:MaterialExpressionConvert_20
       inputName=None
       mask=0
       maskA=0
       maskB=0
       maskG=0
       maskR=0
       outputIndex=2
      }
     }
     {
      type=Scalar
      expressionInput {
       expression=&self:MaterialExpressionConstant_3
       inputName=None
       mask=0
       maskA=0
       maskB=0
       maskG=0
       maskR=0
       outputIndex=0
      }
     }
    ]
    convertOutputs [{type=Vector2}, {type=Vector2}, {type=Vector2}]
   }
  }
  MaterialExpressionConvert_42 {
   class=MaterialExpressionConvert
   inputs [None {from=MaterialExpressionReroute_73}]
   properties {
    nodeName=Break Float3
    convertInputs [{
      type=Vector3
      expressionInput {
       expression=&self:MaterialExpressionReroute_73
       inputName=None
       mask=0
       maskA=0
       maskB=0
       maskG=0
       maskR=0
       outputIndex=0
      }
     }]
    convertOutputs [{type=Scalar}, {type=Scalar}, {type=Scalar}]
   }
  }
  MaterialExpressionConvert_43 {
   class=MaterialExpressionConvert
   inputs [None {from=MaterialExpressionNamedRerouteUsage_13}]
   properties {
    nodeName=Break Float3
    convertInputs [{
      type=Vector3
      expressionInput {
       expression=&self:MaterialExpressionNamedRerouteUsage_13
       inputName=None
       mask=0
       maskA=0
       maskB=0
       maskG=0
       maskR=0
       outputIndex=0
      }
     }]
    convertOutputs [{type=Scalar}, {type=Scalar}, {type=Scalar}]
   }
  }
  MaterialExpressionConvert_44 {
   class=MaterialExpressionConvert
   inputs [None {from=MaterialExpressionNamedRerouteUsage_11}]
   properties {
    nodeName=Break Float3
    convertInputs [{
      type=Vector3
      expressionInput {
       expression=&self:MaterialExpressionNamedRerouteUsage_11
       inputName=None
       mask=0
       maskA=0
       maskB=0
       maskG=0
       maskR=0
       outputIndex=0
      }
     }]
    convertOutputs [{type=Scalar}, {type=Scalar}, {type=Scalar}]
   }
  }
  MaterialExpressionConvert_45 {
   class=MaterialExpressionConvert
   inputs [None {from=MaterialExpressionLinearInterpolate_66}]
   properties {
    convertInputs [{
      type=Vector2
      expressionInput {
       expression=&self:MaterialExpressionLinearInterpolate_66
       inputName=None
       mask=0
       maskA=0
       maskB=0
       maskG=0
       maskR=0
       outputIndex=0
      }
     }]
    convertOutputs [{type=Vector2}]
   }
  }
  MaterialExpressionConvert_46 {
   class=MaterialExpressionConvert
   inputs [None {from=MaterialExpressionNamedRerouteUsage_12}]
   properties {
    nodeName=Break Float3
    convertInputs [{
      type=Vector3
      expressionInput {
       expression=&self:MaterialExpressionNamedRerouteUsage_12
       inputName=None
       mask=0
       maskA=0
       maskB=0
       maskG=0
       maskR=0
       outputIndex=0
      }
     }]
    convertOutputs [{type=Scalar}, {type=Scalar}, {type=Scalar}]
   }
  }
  MaterialExpressionConvert_47 {
   class=MaterialExpressionConvert
   inputs [
    None {from=MaterialExpressionConvert_46}
    None {from=MaterialExpressionMultiply_43}
    None {from=MaterialExpressionConvert_46}
    None {from=MaterialExpressionConstant_8}
   ]
   properties {
    desc=csign
    convertInputs [
     {
      type=Scalar
      expressionInput {
       expression=&self:MaterialExpressionConvert_46
       inputName=None
       mask=0
       maskA=0
       maskB=0
       maskG=0
       maskR=0
       outputIndex=0
      }
     }
     {
      type=Scalar
      expressionInput {
       expression=&self:MaterialExpressionMultiply_43
       inputName=None
       mask=0
       maskA=0
       maskB=0
       maskG=0
       maskR=0
       outputIndex=0
      }
     }
     {
      type=Scalar
      expressionInput {
       expression=&self:MaterialExpressionConvert_46
       inputName=None
       mask=0
       maskA=0
       maskB=0
       maskG=0
       maskR=0
       outputIndex=2
      }
     }
     {
      type=Scalar
      expressionInput {
       expression=&self:MaterialExpressionConstant_8
       inputName=None
       mask=0
       maskA=0
       maskB=0
       maskG=0
       maskR=0
       outputIndex=0
      }
     }
    ]
    convertOutputs [{type=Vector2}, {type=Vector2}, {type=Vector2}]
   }
  }
  MaterialExpressionConvert_48 {
   class=MaterialExpressionConvert
   inputs [None {from=MaterialExpressionReroute_117}]
   properties {
    nodeName=Break Float3
    convertInputs [{
      type=Vector3
      expressionInput {
       expression=&self:MaterialExpressionReroute_117
       inputName=None
       mask=0
       maskA=0
       maskB=0
       maskG=0
       maskR=0
       outputIndex=0
      }
     }]
    convertOutputs [{type=Scalar}, {type=Scalar}, {type=Scalar}]
   }
  }
  MaterialExpressionConvert_5 {
   class=MaterialExpressionConvert
   inputs [
    None {from=MaterialExpressionMultiply_10}
    None {from=MaterialExpressionMultiply_6}
   ]
   properties {
    convertInputs [
     {
      type=Vector2
      expressionInput {
       expression=&self:MaterialExpressionMultiply_10
       inputName=None
       mask=0
       maskA=0
       maskB=0
       maskG=0
       maskR=0
       outputIndex=0
      }
     }
     {
      type=Vector2
      expressionInput {
       expression=&self:MaterialExpressionMultiply_6
       inputName=None
       mask=0
       maskA=0
       maskB=0
       maskG=0
       maskR=0
       outputIndex=0
      }
     }
    ]
    convertOutputs [{type=Vector4, defaultValue=<rgba 0 0 0 1>}]
   }
  }
  MaterialExpressionConvert_6 {
   class=MaterialExpressionConvert
   inputs [None {from=MaterialExpressionFunctionInput_4}]
   properties {
    nodeName=Break Float3
    convertInputs [{
      type=Vector3
      expressionInput {
       expression=&self:MaterialExpressionFunctionInput_4
       inputName=None
       mask=0
       maskA=0
       maskB=0
       maskG=0
       maskR=0
       outputIndex=0
      }
     }]
    convertOutputs [{type=Scalar}, {type=Scalar}, {type=Scalar}]
   }
  }
  MaterialExpressionConvert_7 {
   class=MaterialExpressionConvert
   inputs [
    None {from=MaterialExpressionMultiply_8}
    None {from=MaterialExpressionMultiply_9}
   ]
   properties {
    convertInputs [
     {
      type=Scalar
      expressionInput {
       expression=&self:MaterialExpressionMultiply_8
       inputName=None
       mask=0
       maskA=0
       maskB=0
       maskG=0
       maskR=0
       outputIndex=0
      }
     }
     {
      type=Scalar
      expressionInput {
       expression=&self:MaterialExpressionMultiply_9
       inputName=None
       mask=0
       maskA=0
       maskB=0
       maskG=0
       maskR=0
       outputIndex=0
      }
     }
    ]
    convertOutputs [{type=Vector4, defaultValue=<rgba 0 0 0 1>}]
   }
  }
  MaterialExpressionConvert_8 {
   class=MaterialExpressionConvert
   inputs [None {from=MaterialExpressionNamedRerouteDeclaration_2}]
   properties {
    nodeName=Break Float3
    convertInputs [{
      type=Vector3
      expressionInput {
       expression=&self:MaterialExpressionNamedRerouteDeclaration_2
       inputName=None
       mask=0
       maskA=0
       maskB=0
       maskG=0
       maskR=0
       outputIndex=0
      }
     }]
    convertOutputs [{type=Scalar}, {type=Scalar}, {type=Scalar}]
   }
  }
  MaterialExpressionConvert_9 {
   class=MaterialExpressionConvert
   inputs [None {from=MaterialExpressionNamedRerouteDeclaration_2}]
   properties {
    nodeName=Break Float3
    convertInputs [{
      type=Vector3
      expressionInput {
       expression=&self:MaterialExpressionNamedRerouteDeclaration_2
       inputName=None
       mask=0
       maskA=0
       maskB=0
       maskG=0
       maskR=0
       outputIndex=0
      }
     }]
    convertOutputs [{type=Scalar}, {type=Scalar}, {type=Scalar}]
   }
  }
  MaterialExpressionCrossProduct_0 {
   class=MaterialExpressionCrossProduct
   inputs [
    A {from=MaterialExpressionReroute_76, output=""}
    B {from=MaterialExpressionNormalize_0, output=""}
   ]
   properties {}
  }
  MaterialExpressionCrossProduct_2 {
   class=MaterialExpressionCrossProduct
   inputs [
    A {from=MaterialExpressionReroute_120, output=""}
    B {from=MaterialExpressionNormalize_4, output=""}
   ]
   properties {}
  }
  MaterialExpressionCustom_1 {
   class=MaterialExpressionCustom
   inputs [
    p {from="", output=""}
    dpx {from="", output=""}
    dpy {from="", output=""}
    n {from="", output=""}
    sn {from="", output=""}
    k {from="", output=""}
   ]
   properties {
    code="//HLSL for reference\r\n// modified for UE from iquilez https://iquilezles.org/articles/biplanar/\r\n// more reading here https://ryandowlingsoka.com/unreal/triplanar-dither-biplanar/\r\n// last bit of ref https://www.youtube.com/watch?v=Cq5H59G-DHI\r\n\r\nfloat3 dpdx = dpx;\r\nfloat3 dpdy = dpy;\r\nconst float3 signs = sn;\r\n\r\nconst float2 csign[3] = { sn.x, 1.f, \r\n                         -sn.y, 1.f,\r\n                          sn.z, 1.f };\r\n\r\nn = abs(n) + float3(0.f,0.f, 0.0033);\r\n\r\n// determine major axis (in x; yz are following axis)\r\nint3 ma =   (n.x>n.y &&&& n.x>n.z) ?  int3(0,1,2) :\r\n            (n.y>n.z)            ?  int3(1,2,0) :\r\n                                    int3(2,0,1) ;\r\n// determine minor axis (in x; yz are following axis)\r\nint3 mi =   (n.x<n.y &&&& n.x<n.z) ?  int3(0,1,2) :\r\n            (n.y<n.z)            ?  int3(1,2,0) :\r\n                                    int3(2,0,1) ;\r\n// determine median axis (in x;  yz are following axis)\r\nint3 me = int3(3,3,3) - mi - ma;\r\n\r\nfloat3 tanDir[3] = { float3(0.f, -1.f, 0.f), \r\n                     float3(1.f,  0.f, 0.f), \r\n                     float3(1.f,  0.f, 0.f) };\r\n\r\nuv1 = (ma.x != 2) ? float2( p[ma.y], p[ma.z] ) : float2( -p[ma.y], -p[ma.z] );\r\n\r\n\r\nuv1 =  ma.x == 1 ? uv1.yx : uv1; //account for Zup\r\nuv1 *= csign[ma.x];  //correct coordinate dir\r\n\r\nout_majoraxis = float3(csign[ma.x], ma.x);\r\n\r\nfloat3 Tangent1 = tanDir[ma.x] * signs[ma.x];\r\nTangent1 += n * ( -1.f * dot(Tangent1, n) );\r\nuv1_tangent = normalize( Tangent1 );\r\nuv1_bitangent = normalize( signs[ma.x] * cross(n,Tangent1) );\r\n\r\nuv2 = (me.x != 2) ? float2( p[me.y], p[me.z] ) : float2( -p[me.y], -p[me.z] );\r\n\r\nuv2 = me.x == 1 ? uv2.yx : uv2; //account for Zup\r\nuv2 *= csign[me.x];  //correct coordinate dir\r\n\r\nout_medianaxis = float3(csign[me.x], me.x);\r\n\r\nfloat3 Tangent2 = tanDir[me.x] * signs[me.x];\r\nTangent2 += n * ( -1.f * dot(Tangent2, n) );\r\nuv2_tangent = normalize( Tangent2 );\r\nuv2_bitangent = normalize( signs[me.x] * cross(n,Tangent2) );\r\n\r\n// blend factors\r\nfloat2 w = float2(n[ma.x],n[me.x]);\r\nstatic const float fac = 1.f/sqrt(3.f) - 0.0033;\r\n// make local support\r\nw = saturate( (w-fac)/(1.0-fac) );\r\n// shape transition\r\nw  = pow(w, 2*k);\r\nw /= dot(w, 1);\r\n// blend and return\r\nreturn w;"
    description=Biplanar
    outputType=CMOT_Float2
    showCode=true
    additionalOutputs [
     {outputName=uv1, outputType=CMOT_Float2}
     {outputName=out_majoraxis, outputType=CMOT_Float3}
     {outputName=uv1_tangent, outputType=CMOT_Float3}
     {outputName=uv1_bitangent, outputType=CMOT_Float3}
     {outputName=uv2, outputType=CMOT_Float2}
     {outputName=out_medianaxis, outputType=CMOT_Float3}
     {outputName=uv2_tangent, outputType=CMOT_Float3}
     {outputName=uv2_bitangent, outputType=CMOT_Float3}
    ]
    inputs [
     {inputName=p, input {expression=null, inputName=None, mask=0, maskA=0, maskB=0, maskG=0, maskR=0, outputIndex=0}}
     {inputName=dpx, input {expression=null, inputName=None, mask=0, maskA=0, maskB=0, maskG=0, maskR=0, outputIndex=0}}
     {inputName=dpy, input {expression=null, inputName=None, mask=0, maskA=0, maskB=0, maskG=0, maskR=0, outputIndex=0}}
     {inputName=n, input {expression=null, inputName=None, mask=0, maskA=0, maskB=0, maskG=0, maskR=0, outputIndex=0}}
     {inputName=sn, input {expression=null, inputName=None, mask=0, maskA=0, maskB=0, maskG=0, maskR=0, outputIndex=0}}
     {inputName=k, input {expression=null, inputName=None, mask=0, maskA=0, maskB=0, maskG=0, maskR=0, outputIndex=0}}
    ]
   }
  }
  MaterialExpressionDDX_0 {class=MaterialExpressionDDX, inputs [Value {from=MaterialExpressionMultiply_0, output=""}], properties {}}
  MaterialExpressionDDY_0 {class=MaterialExpressionDDY, inputs [Value {from=MaterialExpressionMultiply_0, output=""}], properties {}}
  MaterialExpressionDivide_0 {
   class=MaterialExpressionDivide
   inputs [
    A {from=MaterialExpressionConstant_9}
    B {from=MaterialExpressionSquareRoot_0}
   ]
  }
  MaterialExpressionDivide_1 {
   class=MaterialExpressionDivide
   inputs [
    A {from=MaterialExpressionWorldPosition_1, output=XYZ}
    B {from=""}
   ]
   properties {constB=128}
  }
  MaterialExpressionDivide_2 {
   class=MaterialExpressionDivide
   inputs [
    A {from=MaterialExpressionSubtract_14}
    B {from=MaterialExpressionOneMinus_16}
   ]
  }
  MaterialExpressionDivide_3 {
   class=MaterialExpressionDivide
   inputs [
    A {from=MaterialExpressionPower_0}
    B {from=MaterialExpressionDotProduct_3}
   ]
  }
  MaterialExpressionDotProduct_0 {
   class=MaterialExpressionDotProduct
   inputs [
    A {from=MaterialExpressionReroute_74, output=""}
    B {from=MaterialExpressionMultiply_31, output=""}
   ]
   properties {}
  }
  MaterialExpressionDotProduct_2 {
   class=MaterialExpressionDotProduct
   inputs [
    A {from=MaterialExpressionReroute_118, output=""}
    B {from=MaterialExpressionMultiply_45, output=""}
   ]
   properties {}
  }
  MaterialExpressionDotProduct_3 {
   class=MaterialExpressionDotProduct
   inputs [
    A {from=MaterialExpressionPower_0, output=""}
    B {from=MaterialExpressionConstant_11, output=""}
   ]
   properties {}
  }
  MaterialExpressionFunctionInput_0 {
   class=MaterialExpressionFunctionInput
   inputs [Preview {from=MaterialExpressionDivide_1, output=""}]
   properties {bUsePreviewValueAsDefault=true, inputName=Position, sortPriority=-1}
  }
  MaterialExpressionFunctionInput_1 {
   class=MaterialExpressionFunctionInput
   inputs [Preview {from=MaterialExpressionVertexNormalWS_0, output=""}]
   properties {bUsePreviewValueAsDefault=true, inputName=RefNorm, sortPriority=1}
  }
  MaterialExpressionFunctionInput_2 {
   class=MaterialExpressionFunctionInput
   inputs [Preview {from="", output=""}]
   properties {bUsePreviewValueAsDefault=true, inputName=BlendFactor, inputType=FunctionInput_Scalar, sortPriority=2, previewValue=<0.5 0 0 1>}
  }
  MaterialExpressionFunctionInput_3 {
   class=MaterialExpressionFunctionInput
   inputs [Preview {from="", output=""}]
   properties {bUsePreviewValueAsDefault=true, inputName="Rotation (Turns)", inputType=FunctionInput_Scalar, sortPriority=2}
  }
  MaterialExpressionFunctionInput_4 {
   class=MaterialExpressionFunctionInput
   inputs [Preview {from="", output=""}]
   properties {bUsePreviewValueAsDefault=true, inputName=Tiling, previewValue=<1 1 1 1>}
  }
  MaterialExpressionFunctionOutput_0 {
   class=MaterialExpressionFunctionOutput
   inputs [None {from=MaterialExpressionNormalize_1, output=""}]
   properties {outputName=UV1_bitangent, sortPriority=3}
  }
  MaterialExpressionFunctionOutput_1 {
   class=MaterialExpressionFunctionOutput
   inputs [None {from=MaterialExpressionMaterialFunctionCall_1, output=Result}]
   properties {outputName=UV1}
  }
  MaterialExpressionFunctionOutput_2 {
   class=MaterialExpressionFunctionOutput
   inputs [None {from=MaterialExpressionShaderStageSwitch_1, output=""}]
   properties {outputName=UV1_derivs, sortPriority=1}
  }
  MaterialExpressionFunctionOutput_3 {
   class=MaterialExpressionFunctionOutput
   inputs [None {from=MaterialExpressionNormalize_0, output=""}]
   properties {outputName=UV1_tangent, sortPriority=2}
  }
  MaterialExpressionFunctionOutput_4 {
   class=MaterialExpressionFunctionOutput
   inputs [None {from=MaterialExpressionMaterialFunctionCall_0, output=Result}]
   properties {outputName=UV2, sortPriority=5}
  }
  MaterialExpressionFunctionOutput_5 {
   class=MaterialExpressionFunctionOutput
   inputs [None {from=MaterialExpressionShaderStageSwitch_0, output=""}]
   properties {outputName=UV2_derivs, sortPriority=6}
  }
  MaterialExpressionFunctionOutput_6 {
   class=MaterialExpressionFunctionOutput
   inputs [None {from=MaterialExpressionNormalize_4, output=""}]
   properties {outputName=UV2_tangent, sortPriority=7}
  }
  MaterialExpressionFunctionOutput_7 {
   class=MaterialExpressionFunctionOutput
   inputs [None {from=MaterialExpressionDivide_3, output=""}]
   properties {outputName=Blend, sortPriority=10}
  }
  MaterialExpressionFunctionOutput_8 {
   class=MaterialExpressionFunctionOutput
   inputs [None {from=MaterialExpressionNormalize_5, output=""}]
   properties {outputName=UV2_bitangent, sortPriority=7}
  }
  MaterialExpressionIf_0 {
   class=MaterialExpressionIf
   inputs [
    A {from=MaterialExpressionConvert_8, output=""}
    B {from=MaterialExpressionConvert_8, output=""}
    "A > B" {from=MaterialExpressionIf_1, output=""}
    "A == B" {from="", output=""}
    "A < B" {from=MaterialExpressionIf_2, output=""}
   ]
   properties {desc=ma}
  }
  MaterialExpressionIf_1 {
   class=MaterialExpressionIf
   inputs [
    A {from=MaterialExpressionConvert_8, output=""}
    B {from=MaterialExpressionConvert_8, output=""}
    "A > B" {from=MaterialExpressionConstant3Vector_1, output=""}
    "A == B" {from="", output=""}
    "A < B" {from=MaterialExpressionIf_2, output=""}
   ]
   properties {}
  }
  MaterialExpressionIf_2 {
   class=MaterialExpressionIf
   inputs [
    A {from=MaterialExpressionConvert_8, output=""}
    B {from=MaterialExpressionConvert_8, output=""}
    "A > B" {from=MaterialExpressionConstant3Vector_2, output=""}
    "A == B" {from="", output=""}
    "A < B" {from=MaterialExpressionConstant3Vector_3, output=""}
   ]
   properties {}
  }
  MaterialExpressionIf_3 {
   class=MaterialExpressionIf
   inputs [
    A {from=MaterialExpressionConvert_9, output=""}
    B {from=MaterialExpressionConvert_9, output=""}
    "A > B" {from=MaterialExpressionIf_5, output=""}
    "A == B" {from="", output=""}
    "A < B" {from=MaterialExpressionIf_4, output=""}
   ]
   properties {desc=mi}
  }
  MaterialExpressionIf_4 {
   class=MaterialExpressionIf
   inputs [
    A {from=MaterialExpressionConvert_9, output=""}
    B {from=MaterialExpressionConvert_9, output=""}
    "A > B" {from=MaterialExpressionIf_5, output=""}
    "A == B" {from="", output=""}
    "A < B" {from=MaterialExpressionConstant3Vector_4, output=""}
   ]
   properties {}
  }
  MaterialExpressionIf_5 {
   class=MaterialExpressionIf
   inputs [
    A {from=MaterialExpressionConvert_9, output=""}
    B {from=MaterialExpressionConvert_9, output=""}
    "A > B" {from=MaterialExpressionConstant3Vector_6, output=""}
    "A == B" {from="", output=""}
    "A < B" {from=MaterialExpressionConstant3Vector_5, output=""}
   ]
   properties {}
  }
  MaterialExpressionLength_1 {class=MaterialExpressionLength, inputs [None {from=MaterialExpressionTransformPosition_0, output=""}], properties {}}
  MaterialExpressionLength_2 {class=MaterialExpressionLength, inputs [None {from=MaterialExpressionTransformPosition_1, output=""}], properties {}}
  MaterialExpressionLength_3 {class=MaterialExpressionLength, inputs [None {from=MaterialExpressionTransformPosition_2, output=""}], properties {}}
  MaterialExpressionLinearInterpolate_1 {
   class=MaterialExpressionLinearInterpolate
   inputs [
    A {from=MaterialExpressionNamedRerouteUsage_23}
    B {from=MaterialExpressionNamedRerouteUsage_24}
    Alpha {from=MaterialExpressionStep_2}
   ]
  }
  MaterialExpressionLinearInterpolate_10 {
   class=MaterialExpressionLinearInterpolate
   inputs [
    A {from=MaterialExpressionReroute_24}
    B {from=MaterialExpressionLinearInterpolate_11}
    Alpha {from=MaterialExpressionReroute_19}
   ]
  }
  MaterialExpressionLinearInterpolate_11 {
   class=MaterialExpressionLinearInterpolate
   inputs [
    A {from=MaterialExpressionReroute_25}
    B {from=MaterialExpressionReroute_26}
    Alpha {from=MaterialExpressionStep_11}
   ]
  }
  MaterialExpressionLinearInterpolate_12 {
   class=MaterialExpressionLinearInterpolate
   inputs [
    A {from=MaterialExpressionMultiply_19}
    B {from=MaterialExpressionAppendVector_2}
    Alpha {from=MaterialExpressionOneMinus_6}
   ]
  }
  MaterialExpressionLinearInterpolate_13 {
   class=MaterialExpressionLinearInterpolate
   inputs [
    A {from=MaterialExpressionLinearInterpolate_12}
    B {from=MaterialExpressionConvert_19}
    Alpha {from=MaterialExpressionSaturate_6}
   ]
  }
  MaterialExpressionLinearInterpolate_14 {
   class=MaterialExpressionLinearInterpolate
   inputs [
    A {from=MaterialExpressionConvert_21}
    B {from=MaterialExpressionConvert_21}
    Alpha {from=MaterialExpressionStep_15}
   ]
  }
  MaterialExpressionLinearInterpolate_15 {
   class=MaterialExpressionLinearInterpolate
   inputs [
    A {from=MaterialExpressionConvert_21}
    B {from=MaterialExpressionLinearInterpolate_14}
    Alpha {from=MaterialExpressionReroute_28}
   ]
  }
  MaterialExpressionLinearInterpolate_16 {
   class=MaterialExpressionLinearInterpolate
   inputs [
    A {from=MaterialExpressionNamedRerouteUsage_25}
    B {from=MaterialExpressionLinearInterpolate_7}
    Alpha {from=MaterialExpressionStep_3}
   ]
   properties {desc=ddyA}
  }
  MaterialExpressionLinearInterpolate_17 {
   class=MaterialExpressionLinearInterpolate
   inputs [
    A {from=MaterialExpressionNamedRerouteUsage_30}
    B {from=MaterialExpressionNamedRerouteUsage_31}
    Alpha {from=MaterialExpressionStep_1}
   ]
  }
  MaterialExpressionLinearInterpolate_18 {
   class=MaterialExpressionLinearInterpolate
   inputs [
    A {from=MaterialExpressionNamedRerouteUsage_29}
    B {from=MaterialExpressionLinearInterpolate_17}
    Alpha {from=MaterialExpressionStep_0}
   ]
   properties {desc=ddxA}
  }
  MaterialExpressionLinearInterpolate_19 {
   class=MaterialExpressionLinearInterpolate
   inputs [
    A {from=MaterialExpressionNamedRerouteUsage_33}
    B {from=MaterialExpressionNamedRerouteUsage_34}
    Alpha {from=MaterialExpressionStep_1}
   ]
  }
  MaterialExpressionLinearInterpolate_2 {
   class=MaterialExpressionLinearInterpolate
   inputs [
    A {from=MaterialExpressionNamedRerouteUsage_10}
    B {from=MaterialExpressionLinearInterpolate_3}
    Alpha {from=MaterialExpressionReroute_12}
   ]
   properties {desc=vd2}
  }
  MaterialExpressionLinearInterpolate_20 {
   class=MaterialExpressionLinearInterpolate
   inputs [
    A {from=MaterialExpressionNamedRerouteUsage_32}
    B {from=MaterialExpressionLinearInterpolate_19}
    Alpha {from=MaterialExpressionStep_0}
   ]
   properties {desc=ddyA}
  }
  MaterialExpressionLinearInterpolate_3 {
   class=MaterialExpressionLinearInterpolate
   inputs [
    A {from=MaterialExpressionNamedRerouteUsage_17}
    B {from=MaterialExpressionNamedRerouteUsage_19}
    Alpha {from=MaterialExpressionStep_2}
   ]
  }
  MaterialExpressionLinearInterpolate_4 {
   class=MaterialExpressionLinearInterpolate
   inputs [
    A {from=MaterialExpressionNamedRerouteUsage_9}
    B {from=MaterialExpressionLinearInterpolate_5}
    Alpha {from=MaterialExpressionReroute_0}
   ]
   properties {desc=vd1}
  }
  MaterialExpressionLinearInterpolate_42 {
   class=MaterialExpressionLinearInterpolate
   inputs [
    A {from=MaterialExpressionConstant3Vector_14}
    B {from=MaterialExpressionConstant3Vector_15}
    Alpha {from=MaterialExpressionStep_16}
   ]
   properties {desc=tanDir}
  }
  MaterialExpressionLinearInterpolate_43 {
   class=MaterialExpressionLinearInterpolate
   inputs [
    A {from=MaterialExpressionConvert_20}
    B {from=MaterialExpressionConvert_20}
    Alpha {from=MaterialExpressionStep_15}
   ]
  }
  MaterialExpressionLinearInterpolate_44 {
   class=MaterialExpressionLinearInterpolate
   inputs [
    A {from=MaterialExpressionConvert_20}
    B {from=MaterialExpressionLinearInterpolate_43}
    Alpha {from=MaterialExpressionReroute_28}
   ]
  }
  MaterialExpressionLinearInterpolate_5 {
   class=MaterialExpressionLinearInterpolate
   inputs [
    A {from=MaterialExpressionNamedRerouteUsage_16}
    B {from=MaterialExpressionNamedRerouteUsage_18}
    Alpha {from=MaterialExpressionStep_4}
   ]
  }
  MaterialExpressionLinearInterpolate_6 {
   class=MaterialExpressionLinearInterpolate
   inputs [
    A {from=MaterialExpressionNamedRerouteUsage_22}
    B {from=MaterialExpressionLinearInterpolate_1}
    Alpha {from=MaterialExpressionStep_3}
   ]
   properties {desc=ddxA}
  }
  MaterialExpressionLinearInterpolate_60 {
   class=MaterialExpressionLinearInterpolate
   inputs [
    A {from=MaterialExpressionConvert_42}
    B {from=MaterialExpressionConvert_42}
    Alpha {from=MaterialExpressionStep_15}
   ]
  }
  MaterialExpressionLinearInterpolate_61 {
   class=MaterialExpressionLinearInterpolate
   inputs [
    A {from=MaterialExpressionConvert_42}
    B {from=MaterialExpressionLinearInterpolate_60}
    Alpha {from=MaterialExpressionReroute_28}
   ]
  }
  MaterialExpressionLinearInterpolate_62 {
   class=MaterialExpressionLinearInterpolate
   inputs [
    A {from=MaterialExpressionConvert_44}
    B {from=MaterialExpressionLinearInterpolate_63}
    Alpha {from=MaterialExpressionReroute_108}
   ]
  }
  MaterialExpressionLinearInterpolate_63 {
   class=MaterialExpressionLinearInterpolate
   inputs [
    A {from=MaterialExpressionConvert_44}
    B {from=MaterialExpressionConvert_44}
    Alpha {from=MaterialExpressionStep_57}
   ]
  }
  MaterialExpressionLinearInterpolate_64 {
   class=MaterialExpressionLinearInterpolate
   inputs [
    A {from=MaterialExpressionReroute_110}
    B {from=MaterialExpressionLinearInterpolate_65}
    Alpha {from=MaterialExpressionReroute_109}
   ]
  }
  MaterialExpressionLinearInterpolate_65 {
   class=MaterialExpressionLinearInterpolate
   inputs [
    A {from=MaterialExpressionReroute_111}
    B {from=MaterialExpressionReroute_112}
    Alpha {from=MaterialExpressionStep_59}
   ]
  }
  MaterialExpressionLinearInterpolate_66 {
   class=MaterialExpressionLinearInterpolate
   inputs [
    A {from=MaterialExpressionMultiply_42}
    B {from=MaterialExpressionAppendVector_10}
    Alpha {from=MaterialExpressionOneMinus_15}
   ]
  }
  MaterialExpressionLinearInterpolate_67 {
   class=MaterialExpressionLinearInterpolate
   inputs [
    A {from=MaterialExpressionLinearInterpolate_66}
    B {from=MaterialExpressionConvert_45}
    Alpha {from=MaterialExpressionSaturate_11}
   ]
  }
  MaterialExpressionLinearInterpolate_68 {
   class=MaterialExpressionLinearInterpolate
   inputs [
    A {from=MaterialExpressionConvert_47}
    B {from=MaterialExpressionConvert_47}
    Alpha {from=MaterialExpressionStep_61}
   ]
  }
  MaterialExpressionLinearInterpolate_69 {
   class=MaterialExpressionLinearInterpolate
   inputs [
    A {from=MaterialExpressionConvert_47}
    B {from=MaterialExpressionLinearInterpolate_68}
    Alpha {from=MaterialExpressionReroute_113}
   ]
  }
  MaterialExpressionLinearInterpolate_7 {
   class=MaterialExpressionLinearInterpolate
   inputs [
    A {from=MaterialExpressionNamedRerouteUsage_26}
    B {from=MaterialExpressionNamedRerouteUsage_27}
    Alpha {from=MaterialExpressionStep_2}
   ]
  }
  MaterialExpressionLinearInterpolate_70 {
   class=MaterialExpressionLinearInterpolate
   inputs [
    A {from=MaterialExpressionConstant3Vector_18}
    B {from=MaterialExpressionConstant3Vector_19}
    Alpha {from=MaterialExpressionStep_62}
   ]
   properties {desc=tanDir}
  }
  MaterialExpressionLinearInterpolate_71 {
   class=MaterialExpressionLinearInterpolate
   inputs [
    A {from=MaterialExpressionConvert_46}
    B {from=MaterialExpressionConvert_46}
    Alpha {from=MaterialExpressionStep_61}
   ]
  }
  MaterialExpressionLinearInterpolate_72 {
   class=MaterialExpressionLinearInterpolate
   inputs [
    A {from=MaterialExpressionConvert_46}
    B {from=MaterialExpressionLinearInterpolate_71}
    Alpha {from=MaterialExpressionReroute_113}
   ]
  }
  MaterialExpressionLinearInterpolate_73 {
   class=MaterialExpressionLinearInterpolate
   inputs [
    A {from=MaterialExpressionConvert_48}
    B {from=MaterialExpressionConvert_48}
    Alpha {from=MaterialExpressionStep_61}
   ]
  }
  MaterialExpressionLinearInterpolate_74 {
   class=MaterialExpressionLinearInterpolate
   inputs [
    A {from=MaterialExpressionConvert_48}
    B {from=MaterialExpressionLinearInterpolate_73}
    Alpha {from=MaterialExpressionReroute_113}
   ]
  }
  MaterialExpressionLinearInterpolate_8 {
   class=MaterialExpressionLinearInterpolate
   inputs [
    A {from=MaterialExpressionConvert_18}
    B {from=MaterialExpressionLinearInterpolate_9}
    Alpha {from=MaterialExpressionReroute_18}
   ]
  }
  MaterialExpressionLinearInterpolate_9 {
   class=MaterialExpressionLinearInterpolate
   inputs [
    A {from=MaterialExpressionConvert_18}
    B {from=MaterialExpressionConvert_18}
    Alpha {from=MaterialExpressionStep_10}
   ]
  }
  MaterialExpressionMaterialFunctionCall_0 {
   class=MaterialExpressionMaterialFunctionCall
   inputs [
    RotationAmount {from=MaterialExpressionNamedRerouteUsage_5}
    Coords {from=MaterialExpressionMultiply_44}
    Pivot {from=MaterialExpressionConstant2Vector_1}
   ]
   properties {materialFunction=&/MF_Rotate2D}
  }
  MaterialExpressionMaterialFunctionCall_1 {
   class=MaterialExpressionMaterialFunctionCall
   inputs [
    RotationAmount {from=MaterialExpressionNamedRerouteUsage_4}
    Coords {from=MaterialExpressionMultiply_21}
    Pivot {from=MaterialExpressionConstant2Vector_0}
   ]
   properties {materialFunction=&MF_Rotate2D}
  }
  MaterialExpressionMultiply_0 {
   class=MaterialExpressionMultiply
   inputs [
    A {from=MaterialExpressionMultiply_5}
    B {from=""}
   ]
   properties {constB=-1}
  }
  MaterialExpressionMultiply_1 {
   class=MaterialExpressionMultiply
   inputs [
    A {from=MaterialExpressionLinearInterpolate_16}
    B {from=MaterialExpressionComponentMask_0}
   ]
  }
  MaterialExpressionMultiply_10 {
   class=MaterialExpressionMultiply
   inputs [
    A {from=MaterialExpressionLinearInterpolate_18}
    B {from=MaterialExpressionComponentMask_4}
   ]
  }
  MaterialExpressionMultiply_19 {
   class=MaterialExpressionMultiply
   inputs [
    A {from=MaterialExpressionAppendVector_2}
    B {from=""}
   ]
   properties {constB=-1}
  }
  MaterialExpressionMultiply_2 {
   class=MaterialExpressionMultiply
   inputs [
    A {from=MaterialExpressionLinearInterpolate_6}
    B {from=MaterialExpressionComponentMask_0}
   ]
  }
  MaterialExpressionMultiply_20 {
   class=MaterialExpressionMultiply
   inputs [
    A {from=MaterialExpressionConvert_20}
    B {from=""}
   ]
   properties {constB=-1}
  }
  MaterialExpressionMultiply_21 {
   class=MaterialExpressionMultiply
   inputs [
    A {from=MaterialExpressionLinearInterpolate_13}
    B {from=MaterialExpressionReroute_36}
   ]
  }
  MaterialExpressionMultiply_31 {
   class=MaterialExpressionMultiply
   inputs [
    A {from=MaterialExpressionLinearInterpolate_42}
    B {from=MaterialExpressionLinearInterpolate_44}
   ]
   properties {desc=Tangent}
  }
  MaterialExpressionMultiply_32 {
   class=MaterialExpressionMultiply
   inputs [
    A {from=MaterialExpressionDotProduct_0}
    B {from=""}
   ]
   properties {constB=-1}
  }
  MaterialExpressionMultiply_33 {
   class=MaterialExpressionMultiply
   inputs [
    A {from=MaterialExpressionReroute_75}
    B {from=MaterialExpressionMultiply_32}
   ]
  }
  MaterialExpressionMultiply_34 {
   class=MaterialExpressionMultiply
   inputs [
    A {from=MaterialExpressionCrossProduct_0}
    B {from=MaterialExpressionLinearInterpolate_44}
   ]
  }
  MaterialExpressionMultiply_42 {
   class=MaterialExpressionMultiply
   inputs [
    A {from=MaterialExpressionAppendVector_10}
    B {from=""}
   ]
   properties {constB=-1}
  }
  MaterialExpressionMultiply_43 {
   class=MaterialExpressionMultiply
   inputs [
    A {from=MaterialExpressionConvert_46}
    B {from=""}
   ]
   properties {constB=-1}
  }
  MaterialExpressionMultiply_44 {
   class=MaterialExpressionMultiply
   inputs [
    A {from=MaterialExpressionLinearInterpolate_67}
    B {from=MaterialExpressionReroute_114}
   ]
  }
  MaterialExpressionMultiply_45 {
   class=MaterialExpressionMultiply
   inputs [
    A {from=MaterialExpressionLinearInterpolate_70}
    B {from=MaterialExpressionLinearInterpolate_72}
   ]
   properties {desc=Tangent}
  }
  MaterialExpressionMultiply_46 {
   class=MaterialExpressionMultiply
   inputs [
    A {from=MaterialExpressionDotProduct_2}
    B {from=""}
   ]
   properties {constB=-1}
  }
  MaterialExpressionMultiply_47 {
   class=MaterialExpressionMultiply
   inputs [
    A {from=MaterialExpressionReroute_119}
    B {from=MaterialExpressionMultiply_46}
   ]
  }
  MaterialExpressionMultiply_48 {
   class=MaterialExpressionMultiply
   inputs [
    A {from=MaterialExpressionCrossProduct_2}
    B {from=MaterialExpressionLinearInterpolate_72}
   ]
  }
  MaterialExpressionMultiply_49 {
   class=MaterialExpressionMultiply
   inputs [
    A {from=MaterialExpressionFunctionInput_2}
    B {from=""}
   ]
   properties {constB=2}
  }
  MaterialExpressionMultiply_5 {
   class=MaterialExpressionMultiply
   inputs [
    A {from=MaterialExpressionFunctionInput_4}
    B {from=MaterialExpressionFunctionInput_0}
   ]
  }
  MaterialExpressionMultiply_6 {
   class=MaterialExpressionMultiply
   inputs [
    A {from=MaterialExpressionLinearInterpolate_20}
    B {from=MaterialExpressionComponentMask_4}
   ]
  }
  MaterialExpressionMultiply_7 {
   class=MaterialExpressionMultiply
   inputs [
    A {from=MaterialExpressionConvert_6}
    B {from=MaterialExpressionLength_1}
   ]
  }
  MaterialExpressionMultiply_8 {
   class=MaterialExpressionMultiply
   inputs [
    A {from=MaterialExpressionConvert_6}
    B {from=MaterialExpressionLength_2}
   ]
  }
  MaterialExpressionMultiply_9 {
   class=MaterialExpressionMultiply
   inputs [
    A {from=MaterialExpressionConvert_6}
    B {from=MaterialExpressionLength_3}
   ]
  }
  MaterialExpressionNamedRerouteDeclaration_0 {
   class=MaterialExpressionNamedRerouteDeclaration
   inputs [None {from=MaterialExpressionAppendVector_5}]
   properties {name=majA, nodeColor=<rgba 0 1 0.1328 1>}
  }
  MaterialExpressionNamedRerouteDeclaration_1 {
   class=MaterialExpressionNamedRerouteDeclaration
   inputs [None {from=MaterialExpressionAppendVector_11}]
   properties {name=medA, nodeColor=<rgba 0 1 0.1328 1>}
  }
  MaterialExpressionNamedRerouteDeclaration_10 {
   class=MaterialExpressionNamedRerouteDeclaration
   inputs [None {from=MaterialExpressionFunctionInput_3}]
   properties {name=turns, nodeColor=<rgba 0 1 0.7422 1>}
  }
  MaterialExpressionNamedRerouteDeclaration_11 {
   class=MaterialExpressionNamedRerouteDeclaration
   inputs [None {from=MaterialExpressionConvert_7}]
   properties {name=VTXx, nodeColor=<rgba 0.5938 0 1 1>}
  }
  MaterialExpressionNamedRerouteDeclaration_12 {
   class=MaterialExpressionNamedRerouteDeclaration
   inputs [None {from=MaterialExpressionConvert_10}]
   properties {name=VTXy, nodeColor=<rgba 0.5938 0 1 1>}
  }
  MaterialExpressionNamedRerouteDeclaration_13 {
   class=MaterialExpressionNamedRerouteDeclaration
   inputs [None {from=MaterialExpressionConvert_11}]
   properties {name=VTXz, nodeColor=<rgba 0.5938 0 1 1>}
  }
  MaterialExpressionNamedRerouteDeclaration_14 {
   class=MaterialExpressionNamedRerouteDeclaration
   inputs [None {from=MaterialExpressionConvert_0}]
   properties {name=xdx, nodeColor=<rgba 0.8516 1 0 1>}
  }
  MaterialExpressionNamedRerouteDeclaration_15 {
   class=MaterialExpressionNamedRerouteDeclaration
   inputs [None {from=MaterialExpressionConvert_0}]
   properties {name=ydx, nodeColor=<rgba 0.8516 1 0 1>}
  }
  MaterialExpressionNamedRerouteDeclaration_16 {
   class=MaterialExpressionNamedRerouteDeclaration
   inputs [None {from=MaterialExpressionConvert_0}]
   properties {name=zdx, nodeColor=<rgba 0.8516 1 0 1>}
  }
  MaterialExpressionNamedRerouteDeclaration_17 {
   class=MaterialExpressionNamedRerouteDeclaration
   inputs [None {from=MaterialExpressionConvert_1}]
   properties {name=xdy, nodeColor=<rgba 0.8516 1 0 1>}
  }
  MaterialExpressionNamedRerouteDeclaration_18 {
   class=MaterialExpressionNamedRerouteDeclaration
   inputs [None {from=MaterialExpressionConvert_1}]
   properties {name=ydy, nodeColor=<rgba 0.8516 1 0 1>}
  }
  MaterialExpressionNamedRerouteDeclaration_19 {
   class=MaterialExpressionNamedRerouteDeclaration
   inputs [None {from=MaterialExpressionConvert_1}]
   properties {name=zdy, nodeColor=<rgba 0.8516 1 0 1>}
  }
  MaterialExpressionNamedRerouteDeclaration_2 {
   class=MaterialExpressionNamedRerouteDeclaration
   inputs [None {from=MaterialExpressionAdd_2}]
   properties {name=working Norm, nodeColor=<rgba 0 0.9766 1 1>}
  }
  MaterialExpressionNamedRerouteDeclaration_3 {
   class=MaterialExpressionNamedRerouteDeclaration
   inputs [None {from=MaterialExpressionIf_0}]
   properties {name=majorAxis, desc=ma, nodeColor=<rgba 1 0.7031 0 1>}
  }
  MaterialExpressionNamedRerouteDeclaration_4 {
   class=MaterialExpressionNamedRerouteDeclaration
   inputs [None {from=MaterialExpressionSubtract_1}]
   properties {name=medianAxis, desc=me, nodeColor=<rgba 1 0.7031 0 1>}
  }
  MaterialExpressionNamedRerouteDeclaration_5 {
   class=MaterialExpressionNamedRerouteDeclaration
   inputs [None {from=MaterialExpressionIf_3}]
   properties {name=minorAxis, desc=mi, nodeColor=<rgba 1 0.7031 0 1>}
  }
  MaterialExpressionNamedRerouteDeclaration_6 {
   class=MaterialExpressionNamedRerouteDeclaration
   inputs [None {from=MaterialExpressionMultiply_0}]
   properties {name=inPosition, nodeColor=<rgba 1 0 0.2578 1>}
  }
  MaterialExpressionNamedRerouteDeclaration_7 {
   class=MaterialExpressionNamedRerouteDeclaration
   inputs [None {from=MaterialExpressionSign_0}]
   properties {name=normSign, nodeColor=<rgba 0 0.5781 1 1>}
  }
  MaterialExpressionNamedRerouteDeclaration_8 {
   class=MaterialExpressionNamedRerouteDeclaration
   inputs [None {from=MaterialExpressionLinearInterpolate_61}]
   properties {name=w.a, nodeColor=<rgba 0.7109 1 0 1>}
  }
  MaterialExpressionNamedRerouteDeclaration_9 {
   class=MaterialExpressionNamedRerouteDeclaration
   inputs [None {from=MaterialExpressionLinearInterpolate_74}]
   properties {name=w.e, nodeColor=<rgba 0.7109 1 0 1>}
  }
  MaterialExpressionNamedRerouteUsage_0 {class=MaterialExpressionNamedRerouteUsage}
  MaterialExpressionNamedRerouteUsage_1 {class=MaterialExpressionNamedRerouteUsage}
  MaterialExpressionNamedRerouteUsage_10 {class=MaterialExpressionNamedRerouteUsage}
  MaterialExpressionNamedRerouteUsage_11 {class=MaterialExpressionNamedRerouteUsage}
  MaterialExpressionNamedRerouteUsage_12 {class=MaterialExpressionNamedRerouteUsage}
  MaterialExpressionNamedRerouteUsage_13 {class=MaterialExpressionNamedRerouteUsage}
  MaterialExpressionNamedRerouteUsage_14 {class=MaterialExpressionNamedRerouteUsage}
  MaterialExpressionNamedRerouteUsage_15 {class=MaterialExpressionNamedRerouteUsage}
  MaterialExpressionNamedRerouteUsage_16 {class=MaterialExpressionNamedRerouteUsage}
  MaterialExpressionNamedRerouteUsage_17 {class=MaterialExpressionNamedRerouteUsage}
  MaterialExpressionNamedRerouteUsage_18 {class=MaterialExpressionNamedRerouteUsage}
  MaterialExpressionNamedRerouteUsage_19 {class=MaterialExpressionNamedRerouteUsage}
  MaterialExpressionNamedRerouteUsage_2 {class=MaterialExpressionNamedRerouteUsage}
  MaterialExpressionNamedRerouteUsage_20 {class=MaterialExpressionNamedRerouteUsage}
  MaterialExpressionNamedRerouteUsage_21 {class=MaterialExpressionNamedRerouteUsage}
  MaterialExpressionNamedRerouteUsage_22 {class=MaterialExpressionNamedRerouteUsage}
  MaterialExpressionNamedRerouteUsage_23 {class=MaterialExpressionNamedRerouteUsage}
  MaterialExpressionNamedRerouteUsage_24 {class=MaterialExpressionNamedRerouteUsage}
  MaterialExpressionNamedRerouteUsage_25 {class=MaterialExpressionNamedRerouteUsage}
  MaterialExpressionNamedRerouteUsage_26 {class=MaterialExpressionNamedRerouteUsage}
  MaterialExpressionNamedRerouteUsage_27 {class=MaterialExpressionNamedRerouteUsage}
  MaterialExpressionNamedRerouteUsage_29 {class=MaterialExpressionNamedRerouteUsage}
  MaterialExpressionNamedRerouteUsage_3 {class=MaterialExpressionNamedRerouteUsage}
  MaterialExpressionNamedRerouteUsage_30 {class=MaterialExpressionNamedRerouteUsage}
  MaterialExpressionNamedRerouteUsage_31 {class=MaterialExpressionNamedRerouteUsage}
  MaterialExpressionNamedRerouteUsage_32 {class=MaterialExpressionNamedRerouteUsage}
  MaterialExpressionNamedRerouteUsage_33 {class=MaterialExpressionNamedRerouteUsage}
  MaterialExpressionNamedRerouteUsage_34 {class=MaterialExpressionNamedRerouteUsage}
  MaterialExpressionNamedRerouteUsage_4 {class=MaterialExpressionNamedRerouteUsage}
  MaterialExpressionNamedRerouteUsage_5 {class=MaterialExpressionNamedRerouteUsage}
  MaterialExpressionNamedRerouteUsage_6 {class=MaterialExpressionNamedRerouteUsage}
  MaterialExpressionNamedRerouteUsage_7 {class=MaterialExpressionNamedRerouteUsage}
  MaterialExpressionNamedRerouteUsage_8 {class=MaterialExpressionNamedRerouteUsage}
  MaterialExpressionNamedRerouteUsage_9 {class=MaterialExpressionNamedRerouteUsage}
  MaterialExpressionNormalize_0 {class=MaterialExpressionNormalize, inputs [VectorInput {from=MaterialExpressionAdd_13, output=""}]}
  MaterialExpressionNormalize_1 {class=MaterialExpressionNormalize, inputs [VectorInput {from=MaterialExpressionMultiply_34, output=""}]}
  MaterialExpressionNormalize_4 {class=MaterialExpressionNormalize, inputs [VectorInput {from=MaterialExpressionAdd_15, output=""}]}
  MaterialExpressionNormalize_5 {class=MaterialExpressionNormalize, inputs [VectorInput {from=MaterialExpressionMultiply_48, output=""}]}
  MaterialExpressionOneMinus_15 {class=MaterialExpressionOneMinus, inputs [None {from=MaterialExpressionStep_61}]}
  MaterialExpressionOneMinus_16 {class=MaterialExpressionOneMinus, inputs [None {from=MaterialExpressionSubtract_13}]}
  MaterialExpressionOneMinus_6 {class=MaterialExpressionOneMinus, inputs [None {from=MaterialExpressionStep_15}]}
  MaterialExpressionPower_0 {
   class=MaterialExpressionPower
   inputs [
    Base {from=MaterialExpressionSaturate_12}
    Exp {from=MaterialExpressionMultiply_49}
   ]
  }
  MaterialExpressionReroute_0 {class=MaterialExpressionReroute, inputs [None {from=MaterialExpressionStep_5}]}
  MaterialExpressionReroute_108 {class=MaterialExpressionReroute, inputs [None {from=MaterialExpressionStep_58}]}
  MaterialExpressionReroute_109 {class=MaterialExpressionReroute, inputs [None {from=MaterialExpressionStep_60}]}
  MaterialExpressionReroute_110 {class=MaterialExpressionReroute, inputs [None {from=MaterialExpressionConvert_44}]}
  MaterialExpressionReroute_111 {class=MaterialExpressionReroute, inputs [None {from=MaterialExpressionConvert_44}]}
  MaterialExpressionReroute_112 {class=MaterialExpressionReroute, inputs [None {from=MaterialExpressionConvert_44}]}
  MaterialExpressionReroute_113 {class=MaterialExpressionReroute, inputs [None {from=MaterialExpressionStep_62}]}
  MaterialExpressionReroute_114 {class=MaterialExpressionReroute, inputs [None {from=MaterialExpressionLinearInterpolate_69}]}
  MaterialExpressionReroute_115 {class=MaterialExpressionReroute, inputs [None {from=MaterialExpressionConvert_43}]}
  MaterialExpressionReroute_116 {class=MaterialExpressionReroute, inputs [None {from=MaterialExpressionReroute_115}]}
  MaterialExpressionReroute_117 {class=MaterialExpressionReroute, inputs [None {from=MaterialExpressionNamedRerouteUsage_15}]}
  MaterialExpressionReroute_118 {class=MaterialExpressionReroute, inputs [None {from=MaterialExpressionReroute_117}]}
  MaterialExpressionReroute_119 {class=MaterialExpressionReroute, inputs [None {from=MaterialExpressionReroute_118}]}
  MaterialExpressionReroute_12 {class=MaterialExpressionReroute, inputs [None {from=MaterialExpressionStep_3}]}
  MaterialExpressionReroute_120 {class=MaterialExpressionReroute, inputs [None {from=MaterialExpressionReroute_119}]}
  MaterialExpressionReroute_18 {class=MaterialExpressionReroute, inputs [None {from=MaterialExpressionStep_9}]}
  MaterialExpressionReroute_19 {class=MaterialExpressionReroute, inputs [None {from=MaterialExpressionStep_12}]}
  MaterialExpressionReroute_24 {class=MaterialExpressionReroute, inputs [None {from=MaterialExpressionConvert_18}]}
  MaterialExpressionReroute_25 {class=MaterialExpressionReroute, inputs [None {from=MaterialExpressionConvert_18}]}
  MaterialExpressionReroute_26 {class=MaterialExpressionReroute, inputs [None {from=MaterialExpressionConvert_18}]}
  MaterialExpressionReroute_28 {class=MaterialExpressionReroute, inputs [None {from=MaterialExpressionStep_16}]}
  MaterialExpressionReroute_36 {class=MaterialExpressionReroute, inputs [None {from=MaterialExpressionLinearInterpolate_15}]}
  MaterialExpressionReroute_52 {class=MaterialExpressionReroute, inputs [None {from=MaterialExpressionConvert_17}]}
  MaterialExpressionReroute_53 {class=MaterialExpressionReroute, inputs [None {from=MaterialExpressionReroute_52}]}
  MaterialExpressionReroute_73 {class=MaterialExpressionReroute, inputs [None {from=MaterialExpressionNamedRerouteUsage_14}]}
  MaterialExpressionReroute_74 {class=MaterialExpressionReroute, inputs [None {from=MaterialExpressionReroute_73}]}
  MaterialExpressionReroute_75 {class=MaterialExpressionReroute, inputs [None {from=MaterialExpressionReroute_74}]}
  MaterialExpressionReroute_76 {class=MaterialExpressionReroute, inputs [None {from=MaterialExpressionReroute_75}]}
  MaterialExpressionSaturate_11 {class=MaterialExpressionSaturate, inputs [None {from=MaterialExpressionSubtract_12}]}
  MaterialExpressionSaturate_12 {class=MaterialExpressionSaturate, inputs [None {from=MaterialExpressionDivide_2}]}
  MaterialExpressionSaturate_6 {class=MaterialExpressionSaturate, inputs [None {from=MaterialExpressionSubtract_6}]}
  MaterialExpressionShaderStageSwitch_0 {
   class=MaterialExpressionShaderStageSwitch
   inputs [
    PixelShader {from=MaterialExpressionConvert_2, output=""}
    VertexShader {from=MaterialExpressionConstant4Vector_0, output=RGBA}
   ]
   properties {}
  }
  MaterialExpressionShaderStageSwitch_1 {
   class=MaterialExpressionShaderStageSwitch
   inputs [
    PixelShader {from=MaterialExpressionConvert_5, output=""}
    VertexShader {from=MaterialExpressionConstant4Vector_0, output=RGBA}
   ]
   properties {}
  }
  MaterialExpressionSign_0 {class=MaterialExpressionSign, inputs [None {from=MaterialExpressionFunctionInput_1, output=""}], properties {}}
  MaterialExpressionSquareRoot_0 {class=MaterialExpressionSquareRoot, inputs [None {from=MaterialExpressionConstant_10, output=""}], properties {}}
  MaterialExpressionStep_0 {
   class=MaterialExpressionStep
   inputs [
    Y {from="", output=""}
    X {from=MaterialExpressionTruncate_0, output=""}
   ]
   properties {constY=1}
  }
  MaterialExpressionStep_1 {
   class=MaterialExpressionStep
   inputs [
    Y {from="", output=""}
    X {from=MaterialExpressionTruncate_0, output=""}
   ]
   properties {constX=2, constY=2}
  }
  MaterialExpressionStep_10 {
   class=MaterialExpressionStep
   inputs [
    Y {from="", output=""}
    X {from=MaterialExpressionConvert_17, output=""}
   ]
   properties {constX=2, constY=2}
  }
  MaterialExpressionStep_11 {
   class=MaterialExpressionStep
   inputs [
    Y {from="", output=""}
    X {from=MaterialExpressionConvert_17, output=""}
   ]
   properties {constX=2, constY=2}
  }
  MaterialExpressionStep_12 {
   class=MaterialExpressionStep
   inputs [
    Y {from="", output=""}
    X {from=MaterialExpressionConvert_17, output=""}
   ]
   properties {constY=1}
  }
  MaterialExpressionStep_15 {
   class=MaterialExpressionStep
   inputs [
    Y {from="", output=""}
    X {from=MaterialExpressionConvert_17, output=""}
   ]
   properties {constX=2, constY=2}
  }
  MaterialExpressionStep_16 {
   class=MaterialExpressionStep
   inputs [
    Y {from="", output=""}
    X {from=MaterialExpressionConvert_17, output=""}
   ]
   properties {constY=1}
  }
  MaterialExpressionStep_2 {
   class=MaterialExpressionStep
   inputs [
    Y {from="", output=""}
    X {from=MaterialExpressionTruncate_1, output=""}
   ]
   properties {constX=2, constY=2}
  }
  MaterialExpressionStep_3 {
   class=MaterialExpressionStep
   inputs [
    Y {from="", output=""}
    X {from=MaterialExpressionTruncate_1, output=""}
   ]
   properties {constY=1}
  }
  MaterialExpressionStep_4 {
   class=MaterialExpressionStep
   inputs [
    Y {from="", output=""}
    X {from=MaterialExpressionTruncate_0, output=""}
   ]
   properties {constX=2, constY=2}
  }
  MaterialExpressionStep_5 {
   class=MaterialExpressionStep
   inputs [
    Y {from="", output=""}
    X {from=MaterialExpressionTruncate_0, output=""}
   ]
   properties {constY=1}
  }
  MaterialExpressionStep_57 {
   class=MaterialExpressionStep
   inputs [
    Y {from="", output=""}
    X {from=MaterialExpressionConvert_43, output=""}
   ]
   properties {constX=2, constY=2}
  }
  MaterialExpressionStep_58 {
   class=MaterialExpressionStep
   inputs [
    Y {from="", output=""}
    X {from=MaterialExpressionConvert_43, output=""}
   ]
   properties {constY=1}
  }
  MaterialExpressionStep_59 {
   class=MaterialExpressionStep
   inputs [
    Y {from="", output=""}
    X {from=MaterialExpressionConvert_43, output=""}
   ]
   properties {constX=2, constY=2}
  }
  MaterialExpressionStep_60 {
   class=MaterialExpressionStep
   inputs [
    Y {from="", output=""}
    X {from=MaterialExpressionConvert_43, output=""}
   ]
   properties {constY=1}
  }
  MaterialExpressionStep_61 {
   class=MaterialExpressionStep
   inputs [
    Y {from="", output=""}
    X {from=MaterialExpressionConvert_43, output=""}
   ]
   properties {constX=2, constY=2}
  }
  MaterialExpressionStep_62 {
   class=MaterialExpressionStep
   inputs [
    Y {from="", output=""}
    X {from=MaterialExpressionConvert_43, output=""}
   ]
   properties {constY=1}
  }
  MaterialExpressionStep_9 {
   class=MaterialExpressionStep
   inputs [
    Y {from="", output=""}
    X {from=MaterialExpressionConvert_17, output=""}
   ]
   properties {constY=1}
  }
  MaterialExpressionSubtract_0 {
   class=MaterialExpressionSubtract
   inputs [
    A {from=MaterialExpressionConstant3Vector_7}
    B {from=MaterialExpressionIf_3}
   ]
  }
  MaterialExpressionSubtract_1 {
   class=MaterialExpressionSubtract
   inputs [
    A {from=MaterialExpressionSubtract_0}
    B {from=MaterialExpressionIf_0}
   ]
   properties {desc=me}
  }
  MaterialExpressionSubtract_12 {
   class=MaterialExpressionSubtract
   inputs [
    A {from=MaterialExpressionReroute_113}
    B {from=MaterialExpressionStep_61}
   ]
  }
  MaterialExpressionSubtract_13 {
   class=MaterialExpressionSubtract
   inputs [
    A {from=MaterialExpressionDivide_0}
    B {from=""}
   ]
   properties {constB=0.0033}
  }
  MaterialExpressionSubtract_14 {
   class=MaterialExpressionSubtract
   inputs [
    A {from=MaterialExpressionAppendVector_12}
    B {from=MaterialExpressionSubtract_13}
   ]
  }
  MaterialExpressionSubtract_6 {
   class=MaterialExpressionSubtract
   inputs [
    A {from=MaterialExpressionReroute_28}
    B {from=MaterialExpressionStep_15}
   ]
  }
  MaterialExpressionTransformPosition_0 {
   class=MaterialExpressionTransformPosition
   inputs [None {from=MaterialExpressionConstant3Vector_8, output=""}]
   properties {transformSourceType=TRANSFORMPOSSOURCE_Instance, transformType=TRANSFORMPOSSOURCE_World}
  }
  MaterialExpressionTransformPosition_1 {
   class=MaterialExpressionTransformPosition
   inputs [None {from=MaterialExpressionConstant3Vector_9, output=""}]
   properties {transformSourceType=TRANSFORMPOSSOURCE_Instance, transformType=TRANSFORMPOSSOURCE_World}
  }
  MaterialExpressionTransformPosition_2 {
   class=MaterialExpressionTransformPosition
   inputs [None {from=MaterialExpressionConstant3Vector_10, output=""}]
   properties {transformSourceType=TRANSFORMPOSSOURCE_Instance, transformType=TRANSFORMPOSSOURCE_World}
  }
  MaterialExpressionTruncate_0 {class=MaterialExpressionTruncate, inputs [None {from=MaterialExpressionAdd_0, output=""}], properties {}}
  MaterialExpressionTruncate_1 {class=MaterialExpressionTruncate, inputs [None {from=MaterialExpressionAdd_1, output=""}], properties {}}
  MaterialExpressionVertexNormalWS_0 {class=MaterialExpressionVertexNormalWS, inputs [], properties {}}
  MaterialExpressionWorldPosition_1 {class=MaterialExpressionWorldPosition, inputs [], properties {}}
 }
 properties {bExposeToLibrary=true, libraryCategoriesText []}
}
