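# Material function that prepares world position for large world coordinates, taking a PeriodXYZ input.
# Max Axis Period output converts camera-relative position to periodic world space, tiled by a maximum over the broken-out period.
# Per-Axis Period output currently passes absolute world position straight through.
# A three-transform per-axis path is left unhooked over cost, and an LWC truncate is unused, noted as failing in big worlds.
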MaterialFunction /BaseMaterial/Materials/Functions/MF_PrepareLWC
graph {expressions {
  MaterialExpressionComponentMask_0 {class=MaterialExpressionComponentMask, inputs [None {from=MaterialExpressionTransformPosition_3}], properties {r=true}}
  MaterialExpressionComponentMask_1 {class=MaterialExpressionComponentMask, inputs [None {from=MaterialExpressionTransformPosition_4}], properties {g=true}}
  MaterialExpressionComponentMask_2 {class=MaterialExpressionComponentMask, inputs [None {from=MaterialExpressionTransformPosition_5}], properties {b=true}}
  MaterialExpressionConvert_0 {
   class=MaterialExpressionConvert
   inputs [None {from=MaterialExpressionFunctionInput_0}]
   properties {
    nodeName=Break Float3
    convertInputs [{
      type=Vector3
      expressionInput {
       expression=&self:MaterialExpressionFunctionInput_0
       inputName=None
       mask=0
       maskA=0
       maskB=0
       maskG=0
       maskR=0
       outputIndex=0
      }
     }]
    convertOutputs [{type=Scalar}, {type=Scalar}, {type=Scalar}]
   }
  }
  MaterialExpressionConvert_1 {
   class=MaterialExpressionConvert
   inputs [
    None {from=MaterialExpressionComponentMask_0}
    None {from=MaterialExpressionComponentMask_1}
    None {from=MaterialExpressionComponentMask_2}
   ]
   properties {
    desc="could swap this for the max axis period - but I don't really like the triple transform expense... so I'm leaving this unhooked for now "
    nodeName=Make Vector3
    convertInputs [
     {
      type=Scalar
      expressionInput {
       expression=&self:MaterialExpressionComponentMask_0
       inputName=None
       mask=0
       maskA=0
       maskB=0
       maskG=0
       maskR=0
       outputIndex=0
      }
     }
     {
      type=Scalar
      expressionInput {
       expression=&self:MaterialExpressionComponentMask_1
       inputName=None
       mask=0
       maskA=0
       maskB=0
       maskG=0
       maskR=0
       outputIndex=0
      }
     }
     {
      type=Scalar
      expressionInput {
       expression=&self:MaterialExpressionComponentMask_2
       inputName=None
       mask=0
       maskA=0
       maskB=0
       maskG=0
       maskR=0
       outputIndex=0
      }
     }
    ]
    convertOutputs [{type=Vector3}]
   }
  }
  MaterialExpressionFunctionInput_0 {
   class=MaterialExpressionFunctionInput
   inputs [Preview {from="", output=""}]
   properties {bUsePreviewValueAsDefault=true, inputName=PeriodXYZ, previewValue=<100 100 100 1>}
  }
  MaterialExpressionFunctionOutput_1 {
   class=MaterialExpressionFunctionOutput
   inputs [None {from=MaterialExpressionWorldPosition_1, output=XYZ}]
   properties {outputName=Per-Axis Period}
  }
  MaterialExpressionFunctionOutput_2 {
   class=MaterialExpressionFunctionOutput
   inputs [None {from=MaterialExpressionTransformPosition_1, output=""}]
   properties {outputName=Max Axis Period, sortPriority=10}
  }
  MaterialExpressionMax_2 {
   class=MaterialExpressionMax
   inputs [
    A {from=MaterialExpressionConvert_0}
    B {from=MaterialExpressionConvert_0}
   ]
  }
  MaterialExpressionMax_3 {
   class=MaterialExpressionMax
   inputs [
    A {from=MaterialExpressionConvert_0}
    B {from=MaterialExpressionMax_2}
   ]
  }
  MaterialExpressionTransformPosition_1 {
   class=MaterialExpressionTransformPosition
   inputs [
    None {from=MaterialExpressionWorldPosition_0, output=XYZ}
    Periodic World Tile Size {from=MaterialExpressionMax_3, output=""}
   ]
   properties {constPeriodicWorldTileSize=4096, transformSourceType=TRANSFORMPOSSOURCE_TranslatedWorld, transformType=TRANSFORMPOSSOURCE_PeriodicWorld}
  }
  MaterialExpressionTransformPosition_3 {
   class=MaterialExpressionTransformPosition
   inputs [
    None {from=MaterialExpressionWorldPosition_0, output=XYZ}
    Periodic World Tile Size {from=MaterialExpressionConvert_0, output=""}
   ]
   properties {constPeriodicWorldTileSize=4096, transformSourceType=TRANSFORMPOSSOURCE_TranslatedWorld, transformType=TRANSFORMPOSSOURCE_PeriodicWorld}
  }
  MaterialExpressionTransformPosition_4 {
   class=MaterialExpressionTransformPosition
   inputs [
    None {from=MaterialExpressionWorldPosition_0, output=XYZ}
    Periodic World Tile Size {from=MaterialExpressionConvert_0, output=""}
   ]
   properties {constPeriodicWorldTileSize=4096, transformSourceType=TRANSFORMPOSSOURCE_TranslatedWorld, transformType=TRANSFORMPOSSOURCE_PeriodicWorld}
  }
  MaterialExpressionTransformPosition_5 {
   class=MaterialExpressionTransformPosition
   inputs [
    None {from=MaterialExpressionWorldPosition_0, output=XYZ}
    Periodic World Tile Size {from=MaterialExpressionConvert_0, output=""}
   ]
   properties {constPeriodicWorldTileSize=4096, transformSourceType=TRANSFORMPOSSOURCE_TranslatedWorld, transformType=TRANSFORMPOSSOURCE_PeriodicWorld}
  }
  MaterialExpressionTruncateLWC_0 {
   class=MaterialExpressionTruncateLWC
   inputs [None {from=MaterialExpressionWorldPosition_1, output=XYZ}]
   properties {desc="tempting - but fails with big worlds, and introduces difficult to debug failure cases"}
  }
  MaterialExpressionWorldPosition_0 {class=MaterialExpressionWorldPosition, inputs [], properties {worldPositionShaderOffset=WPT_CameraRelative}}
  MaterialExpressionWorldPosition_1 {class=MaterialExpressionWorldPosition, inputs [], properties {}}
 }}
