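# Material function "CoordinateBasis: Biplanar (Standard)" that produces two-axis projection UVs, packed derivatives, blend weights and world-space tangent bases for texture sampling and normal reorientation.
# A CoordinateSpace selector picks position and normal from World, Local, Periodic or Preskinned sources, with world positions wrapped through MF_PrepareLWC.
# Offset position, normal, repeat, rotation and blend tolerance feed MF_Coordinate_Biplanar.
# Per-space switches transform tangents and bitangents to world space for output.
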MaterialFunction /BaseMaterial/Materials/Functions/MF_ResolveCoordinateFrame_Biplanar_Basic
graph {
 expressions {
  MaterialExpressionAdd_10 {
   class=MaterialExpressionAdd
   inputs [
    A {from=MaterialExpressionMaterialFunctionCall_2, output=Selected Vector}
    B {from=MaterialExpressionFunctionInput_13}
   ]
  }
  MaterialExpressionConstant3Vector_2 {class=MaterialExpressionConstant3Vector, properties {constant=<rgba 0 0 1 0>}}
  MaterialExpressionConstant3Vector_3 {class=MaterialExpressionConstant3Vector, properties {constant=<rgba 0 1 0 0>}}
  MaterialExpressionConstant3Vector_4 {class=MaterialExpressionConstant3Vector, properties {constant=<rgba 1 0 0 0>}}
  MaterialExpressionConstant_0 {class=MaterialExpressionConstant, properties {r=1}}
  MaterialExpressionConvert_13 {
   class=MaterialExpressionConvert
   inputs [
    None {from=MaterialExpressionLength_3}
    None {from=MaterialExpressionLength_2}
    None {from=MaterialExpressionLength_0}
   ]
   properties {
    desc=maintain scale
    convertInputs [
     {
      type=Scalar
      expressionInput {
       expression=&self:MaterialExpressionLength_3
       inputName=None
       mask=0
       maskA=0
       maskB=0
       maskG=0
       maskR=0
       outputIndex=0
      }
     }
     {
      type=Scalar
      expressionInput {
       expression=&self:MaterialExpressionLength_2
       inputName=None
       mask=0
       maskA=0
       maskB=0
       maskG=0
       maskR=0
       outputIndex=0
      }
     }
     {
      type=Scalar
      expressionInput {
       expression=&self:MaterialExpressionLength_0
       inputName=None
       mask=0
       maskA=0
       maskB=0
       maskG=0
       maskR=0
       outputIndex=0
      }
     }
    ]
    convertOutputs [{type=Vector3}]
   }
  }
  MaterialExpressionDivide_0 {
   class=MaterialExpressionDivide
   inputs [
    A {from=""}
    B {from=MaterialExpressionNamedRerouteUsage_3}
   ]
  }
  MaterialExpressionFunctionInput_0 {
   class=MaterialExpressionFunctionInput
   inputs [Preview {from="", output=""}]
   properties {
    bUsePreviewValueAsDefault=true
    description="Sets how many times the projected coordinates repeat along each axis. Use values above 1 for denser tiling, negative values for mirroring, and avoid zero."
    inputName=CoordinateRepeat
    sortPriority=-10
    previewValue=<0.01 0.01 0.01 1>
   }
  }
  MaterialExpressionFunctionInput_1 {
   class=MaterialExpressionFunctionInput
   inputs [Preview {from="", output=""}]
   properties {
    bUsePreviewValueAsDefault=true
    description="Rotates the projected UVs around the selected axis. Units are in turns, where 1.0 = 360° and 0.25 = 90°."
    inputName="Rotation (Turns)"
    inputType=FunctionInput_Scalar
    sortPriority=-8
   }
  }
  MaterialExpressionFunctionInput_13 {
   class=MaterialExpressionFunctionInput
   inputs [Preview {from="", output=""}]
   properties {
    bUsePreviewValueAsDefault=true
    description=Applies an additional per-axis offset to the coordinates before projection. Useful for correctly positioning a texture.
    inputName=CoordinateTransform
    sortPriority=-9
   }
  }
  MaterialExpressionFunctionInput_14 {
   class=MaterialExpressionFunctionInput
   inputs [Preview {from="", output=""}]
   properties {
    bUsePreviewValueAsDefault=true
    description="Defines the space used for projection: World (0) keeps textures fixed in world space, Local (1) moves with the object, Periodic (2) enforces a repeating pattern, and Preskinned (3) locks to preskinned positions for stable results on animated meshes."
    inputName=CoordinateSpace
    inputType=FunctionInput_Scalar
    sortPriority=-11
   }
  }
  MaterialExpressionFunctionInput_16 {
   class=MaterialExpressionFunctionInput
   inputs [Preview {from="", output=""}]
   properties {
    bUsePreviewValueAsDefault=true
    description="Controls how sharply the biplanar projection blends between its two axes. Implemented as a power function, with 0.5 giving a linear (pow1) blend. Lower values soften the blend between axes; higher values create sharper transitions."
    inputName=BlendTolerance
    inputType=FunctionInput_Scalar
    sortPriority=8
    previewValue=<0.5 0 0 1>
   }
  }
  MaterialExpressionFunctionOutput_1 {
   class=MaterialExpressionFunctionOutput
   inputs [None {from=MaterialExpressionMaterialFunctionCall_10, output=UV1}]
   properties {
    description=2D coordinates from the major axis of the biplanar projection. Use these to sample textures aligned to the major axis.
    outputName=UV1
   }
  }
  MaterialExpressionFunctionOutput_10 {
   class=MaterialExpressionFunctionOutput
   inputs [None {from=MaterialExpressionMaterialFunctionCall_5, output=Selected Vector}]
   properties {
    description="World-space tangent vector for UV2 (median axis). Used with Bitangent2WS for tangent-space normal reorientation."
    outputName=Tangent2
    sortPriority=6
   }
  }
  MaterialExpressionFunctionOutput_11 {
   class=MaterialExpressionFunctionOutput
   inputs [None {from=MaterialExpressionMaterialFunctionCall_6, output=Selected Vector}]
   properties {
    description="World-space bitangent vector for UV2 (median axis). Complements Tangent2WS for normal reorientation."
    outputName=Bitangent2
    sortPriority=7
   }
  }
  MaterialExpressionFunctionOutput_14 {
   class=MaterialExpressionFunctionOutput
   inputs [None {from=MaterialExpressionFunctionInput_1, output=""}]
   properties {
    description="The rotation applied around the projection axis, expressed in turns (1.0 = 360°). Used along with Tangent/Bitangent for reorienting normal maps via MF_TangentNormal_OrientWorld."
    outputName="Rotation (Turns)"
    sortPriority=-5
   }
  }
  MaterialExpressionFunctionOutput_2 {
   class=MaterialExpressionFunctionOutput
   inputs [None {from=MaterialExpressionMaterialFunctionCall_10, output=UV1_derivs}]
   properties {
    description="Packed derivatives (ddx/ddy) of UV1 (major axis). Use with SampleGrad for correct mip selection during texture sampling."
    outputName=DXY1
    sortPriority=3
   }
  }
  MaterialExpressionFunctionOutput_3 {
   class=MaterialExpressionFunctionOutput
   inputs [None {from=MaterialExpressionMaterialFunctionCall_10, output=UV2}]
   properties {
    description=2D coordinates from the median axis of the biplanar projection. Use these to sample textures aligned to the median axis.
    outputName=UV2
    sortPriority=5
   }
  }
  MaterialExpressionFunctionOutput_4 {
   class=MaterialExpressionFunctionOutput
   inputs [None {from=MaterialExpressionMaterialFunctionCall_10, output=UV2_derivs}]
   properties {
    description="Packed derivatives (ddx/ddy) of UV2 (median axis). Use with SampleGrad for correct mip selection during texture sampling."
    outputName=DXY2
    sortPriority=8
   }
  }
  MaterialExpressionFunctionOutput_7 {
   class=MaterialExpressionFunctionOutput
   inputs [None {from=MaterialExpressionMaterialFunctionCall_10, output=Blend}]
   properties {
    description="RG mask values used to blend between the UV1 (major) and UV2 (median) samples. Typically used to drive post-sample blending for smooth transitions between the two projections."
    outputName=BlendMask
    sortPriority=15
   }
  }
  MaterialExpressionFunctionOutput_8 {
   class=MaterialExpressionFunctionOutput
   inputs [None {from=MaterialExpressionMaterialFunctionCall_3, output=Selected Vector}]
   properties {
    description="World-space tangent vector for UV1 (major axis). Used with Bitangent1WS for tangent-space normal reorientation."
    outputName=Tangent1
    sortPriority=1
   }
  }
  MaterialExpressionFunctionOutput_9 {
   class=MaterialExpressionFunctionOutput
   inputs [None {from=MaterialExpressionMaterialFunctionCall_4, output=Selected Vector}]
   properties {
    description="World-space bitangent vector for UV1 (major axis). Complements Tangent1WS for normal reorientation."
    outputName=Bitangent1
    sortPriority=2
   }
  }
  MaterialExpressionLength_0 {class=MaterialExpressionLength, inputs [None {from=MaterialExpressionTransform_0, output=""}], properties {}}
  MaterialExpressionLength_2 {class=MaterialExpressionLength, inputs [None {from=MaterialExpressionTransform_12, output=""}], properties {}}
  MaterialExpressionLength_3 {class=MaterialExpressionLength, inputs [None {from=MaterialExpressionTransform_13, output=""}], properties {}}
  MaterialExpressionLocalPosition_1 {class=MaterialExpressionLocalPosition, inputs [], properties {includedOffsets=ExcludeOffsets, localOrigin=InstancePreSkinning}}
  MaterialExpressionLocalPosition_2 {class=MaterialExpressionLocalPosition, inputs [], properties {}}
  MaterialExpressionMaterialFunctionCall_0 {
   class=MaterialExpressionMaterialFunctionCall
   inputs [
    Selector {from=MaterialExpressionNamedRerouteUsage_41}
    VectorA {from=MaterialExpressionReroute_0}
    VectorB {from=MaterialExpressionNormalize_1}
    VectorC {from=MaterialExpressionReroute_0}
    VectorD {from=MaterialExpressionShaderStageSwitch_0}
   ]
   properties {materialFunction=&/MF_Switch4_Vec3}
  }
  MaterialExpressionMaterialFunctionCall_1 {
   class=MaterialExpressionMaterialFunctionCall
   inputs [PeriodXYZ {from=MaterialExpressionDivide_0}]
   properties {materialFunction=&/MF_PrepareLWC}
  }
  MaterialExpressionMaterialFunctionCall_10 {
   class=MaterialExpressionMaterialFunctionCall
   inputs [
    Position {from=MaterialExpressionNamedRerouteUsage_40}
    Tiling {from=MaterialExpressionNamedRerouteDeclaration_2}
    RefNorm {from=MaterialExpressionNamedRerouteUsage_36}
    "Rotation (Turns)" {from=MaterialExpressionNamedRerouteDeclaration_11}
    BlendFactor {from=MaterialExpressionFunctionInput_16}
   ]
   properties {materialFunction=&/MF_Coordinate_Biplanar}
  }
  MaterialExpressionMaterialFunctionCall_2 {
   class=MaterialExpressionMaterialFunctionCall
   inputs [
    Selector {from=MaterialExpressionFunctionInput_14}
    VectorA {from=MaterialExpressionReroute_6}
    VectorB {from=MaterialExpressionMultiply_0}
    VectorC {from=MaterialExpressionReroute_7}
    VectorD {from=MaterialExpressionMultiply_1}
   ]
   properties {materialFunction=&MF_Switch4_Vec3}
  }
  MaterialExpressionMaterialFunctionCall_3 {
   class=MaterialExpressionMaterialFunctionCall
   inputs [
    Selector {from=MaterialExpressionNamedRerouteUsage_42}
    VectorA {from=MaterialExpressionReroute_2}
    VectorB {from=MaterialExpressionTransform_4}
    VectorC {from=MaterialExpressionReroute_2}
    VectorD {from=MaterialExpressionTransform_6}
   ]
   properties {materialFunction=&MF_Switch4_Vec3}
  }
  MaterialExpressionMaterialFunctionCall_4 {
   class=MaterialExpressionMaterialFunctionCall
   inputs [
    Selector {from=MaterialExpressionNamedRerouteUsage_0}
    VectorA {from=MaterialExpressionReroute_3}
    VectorB {from=MaterialExpressionTransform_1}
    VectorC {from=MaterialExpressionReroute_3}
    VectorD {from=MaterialExpressionTransform_2}
   ]
   properties {materialFunction=&MF_Switch4_Vec3}
  }
  MaterialExpressionMaterialFunctionCall_5 {
   class=MaterialExpressionMaterialFunctionCall
   inputs [
    Selector {from=MaterialExpressionNamedRerouteUsage_2}
    VectorA {from=MaterialExpressionReroute_4}
    VectorB {from=MaterialExpressionTransform_3}
    VectorC {from=MaterialExpressionReroute_4}
    VectorD {from=MaterialExpressionTransform_5}
   ]
   properties {materialFunction=&MF_Switch4_Vec3}
  }
  MaterialExpressionMaterialFunctionCall_6 {
   class=MaterialExpressionMaterialFunctionCall
   inputs [
    Selector {from=MaterialExpressionNamedRerouteUsage_6}
    VectorA {from=MaterialExpressionReroute_5}
    VectorB {from=MaterialExpressionTransform_15}
    VectorC {from=MaterialExpressionReroute_5}
    VectorD {from=MaterialExpressionTransform_16}
   ]
   properties {materialFunction=&MF_Switch4_Vec3}
  }
  MaterialExpressionMultiply_0 {
   class=MaterialExpressionMultiply
   inputs [
    A {from=MaterialExpressionNamedRerouteUsage_35}
    B {from=MaterialExpressionLocalPosition_2, output=XYZ}
   ]
  }
  MaterialExpressionMultiply_1 {
   class=MaterialExpressionMultiply
   inputs [
    A {from=MaterialExpressionNamedRerouteUsage_1}
    B {from=MaterialExpressionShaderStageSwitch_2}
   ]
  }
  MaterialExpressionNamedRerouteDeclaration_0 {
   class=MaterialExpressionNamedRerouteDeclaration
   inputs [None {from=MaterialExpressionFunctionOutput_1}]
   properties {name=UV1, nodeColor=<rgba 0.9 0.009 0.9 1>}
  }
  MaterialExpressionNamedRerouteDeclaration_1 {
   class=MaterialExpressionNamedRerouteDeclaration
   inputs [None {from=MaterialExpressionFunctionOutput_3}]
   properties {name=UV2, nodeColor=<rgba 0.009 0.9 0.9 1>}
  }
  MaterialExpressionNamedRerouteDeclaration_10 {
   class=MaterialExpressionNamedRerouteDeclaration
   inputs [None {from=MaterialExpressionFunctionOutput_7}]
   properties {name=Blendi, nodeColor=<rgba 1 1 1 1>}
  }
  MaterialExpressionNamedRerouteDeclaration_11 {
   class=MaterialExpressionNamedRerouteDeclaration
   inputs [None {from=MaterialExpressionFunctionOutput_14}]
   properties {name=turns, nodeColor=<rgba 0.8963 0.8963 0.0091 1>}
  }
  MaterialExpressionNamedRerouteDeclaration_12 {
   class=MaterialExpressionNamedRerouteDeclaration
   inputs [None {from=MaterialExpressionMaterialFunctionCall_10, output=UV1_bitangent}]
   properties {name=UV1B, nodeColor=<rgba 0.9 0.009 0.9 1>}
  }
  MaterialExpressionNamedRerouteDeclaration_13 {
   class=MaterialExpressionNamedRerouteDeclaration
   inputs [None {from=MaterialExpressionFunctionOutput_9}]
   properties {name=B1, nodeColor=<rgba 0.9 0.009 0.009 1>}
  }
  MaterialExpressionNamedRerouteDeclaration_14 {
   class=MaterialExpressionNamedRerouteDeclaration
   inputs [None {from=MaterialExpressionFunctionOutput_10}]
   properties {name=T2, nodeColor=<rgba 0.009 0.9 0.009 1>}
  }
  MaterialExpressionNamedRerouteDeclaration_15 {
   class=MaterialExpressionNamedRerouteDeclaration
   inputs [None {from=MaterialExpressionFunctionOutput_11}]
   properties {name=B2, nodeColor=<rgba 0.009 0.9 0.009 1>}
  }
  MaterialExpressionNamedRerouteDeclaration_18 {
   class=MaterialExpressionNamedRerouteDeclaration
   inputs [None {from=MaterialExpressionMaterialFunctionCall_0, output=Selected Vector}]
   properties {name=RefNorm, nodeColor=<rgba 0.8963 0.8963 0.0091 1>}
  }
  MaterialExpressionNamedRerouteDeclaration_19 {
   class=MaterialExpressionNamedRerouteDeclaration
   inputs [None {from=MaterialExpressionAdd_10}]
   properties {name=RefPos, nodeColor=<rgba 0.8963 0.8963 0.0091 1>}
  }
  MaterialExpressionNamedRerouteDeclaration_2 {
   class=MaterialExpressionNamedRerouteDeclaration
   inputs [None {from=MaterialExpressionFunctionInput_0}]
   properties {name=Repeat, nodeColor=<rgba 0.7812 0 1 1>}
  }
  MaterialExpressionNamedRerouteDeclaration_20 {
   class=MaterialExpressionNamedRerouteDeclaration
   inputs [None {from=MaterialExpressionFunctionInput_14}]
   properties {name=Selector, nodeColor=<rgba 0 0.3437 1 1>}
  }
  MaterialExpressionNamedRerouteDeclaration_23 {
   class=MaterialExpressionNamedRerouteDeclaration
   inputs [None {from=MaterialExpressionMaterialFunctionCall_10, output=UV2_bitangent}]
   properties {name=UV2B, nodeColor=<rgba 0.009 0.9 0.9 1>}
  }
  MaterialExpressionNamedRerouteDeclaration_26 {
   class=MaterialExpressionNamedRerouteDeclaration
   inputs [None {from=MaterialExpressionMaterialFunctionCall_10, output=UV2_tangent}]
   properties {name=UV2T, nodeColor=<rgba 0.009 0.9 0.9 1>}
  }
  MaterialExpressionNamedRerouteDeclaration_27 {
   class=MaterialExpressionNamedRerouteDeclaration
   inputs [None {from=MaterialExpressionMaterialFunctionCall_10, output=UV1_tangent}]
   properties {name=UV1T, nodeColor=<rgba 0.9 0.009 0.9 1>}
  }
  MaterialExpressionNamedRerouteDeclaration_32 {
   class=MaterialExpressionNamedRerouteDeclaration
   inputs [None {from=MaterialExpressionConstant_0}]
   properties {name=MeshVecScale, nodeColor=<rgba 0.6641 0 1 1>}
  }
  MaterialExpressionNamedRerouteDeclaration_6 {
   class=MaterialExpressionNamedRerouteDeclaration
   inputs [None {from=MaterialExpressionFunctionOutput_8}]
   properties {name=T1, nodeColor=<rgba 0.9 0.009 0.009 1>}
  }
  MaterialExpressionNamedRerouteDeclaration_7 {
   class=MaterialExpressionNamedRerouteDeclaration
   inputs [None {from=MaterialExpressionFunctionOutput_2}]
   properties {name=DXY1, nodeColor=<rgba 0.9 0.009 0.9 1>}
  }
  MaterialExpressionNamedRerouteDeclaration_8 {
   class=MaterialExpressionNamedRerouteDeclaration
   inputs [None {from=MaterialExpressionFunctionOutput_4}]
   properties {name=DXY2, nodeColor=<rgba 0.009 0.9 0.9 1>}
  }
  MaterialExpressionNamedRerouteUsage_0 {class=MaterialExpressionNamedRerouteUsage}
  MaterialExpressionNamedRerouteUsage_1 {class=MaterialExpressionNamedRerouteUsage}
  MaterialExpressionNamedRerouteUsage_2 {class=MaterialExpressionNamedRerouteUsage}
  MaterialExpressionNamedRerouteUsage_3 {class=MaterialExpressionNamedRerouteUsage}
  MaterialExpressionNamedRerouteUsage_35 {class=MaterialExpressionNamedRerouteUsage}
  MaterialExpressionNamedRerouteUsage_36 {class=MaterialExpressionNamedRerouteUsage}
  MaterialExpressionNamedRerouteUsage_37 {class=MaterialExpressionNamedRerouteUsage}
  MaterialExpressionNamedRerouteUsage_38 {class=MaterialExpressionNamedRerouteUsage}
  MaterialExpressionNamedRerouteUsage_4 {class=MaterialExpressionNamedRerouteUsage}
  MaterialExpressionNamedRerouteUsage_40 {class=MaterialExpressionNamedRerouteUsage}
  MaterialExpressionNamedRerouteUsage_41 {class=MaterialExpressionNamedRerouteUsage}
  MaterialExpressionNamedRerouteUsage_42 {class=MaterialExpressionNamedRerouteUsage}
  MaterialExpressionNamedRerouteUsage_6 {class=MaterialExpressionNamedRerouteUsage}
  MaterialExpressionNamedRerouteUsage_8 {class=MaterialExpressionNamedRerouteUsage}
  MaterialExpressionNormalize_1 {class=MaterialExpressionNormalize, inputs [VectorInput {from=MaterialExpressionTransform_14, output=""}]}
  MaterialExpressionPreSkinnedNormal_0 {class=MaterialExpressionPreSkinnedNormal, inputs [], properties {}}
  MaterialExpressionReroute_0 {class=MaterialExpressionReroute, inputs [None {from=MaterialExpressionVertexNormalWS_1}]}
  MaterialExpressionReroute_2 {class=MaterialExpressionReroute, inputs [None {from=MaterialExpressionNamedRerouteUsage_37}]}
  MaterialExpressionReroute_3 {class=MaterialExpressionReroute, inputs [None {from=MaterialExpressionNamedRerouteUsage_38}]}
  MaterialExpressionReroute_4 {class=MaterialExpressionReroute, inputs [None {from=MaterialExpressionNamedRerouteUsage_4}]}
  MaterialExpressionReroute_5 {class=MaterialExpressionReroute, inputs [None {from=MaterialExpressionNamedRerouteUsage_8}]}
  MaterialExpressionReroute_6 {class=MaterialExpressionReroute, inputs [None {from=MaterialExpressionMaterialFunctionCall_1, output=Per-Axis Period}]}
  MaterialExpressionReroute_7 {class=MaterialExpressionReroute, inputs [None {from=MaterialExpressionReroute_8}]}
  MaterialExpressionReroute_8 {class=MaterialExpressionReroute, inputs [None {from=MaterialExpressionMaterialFunctionCall_1, output=Max Axis Period}]}
  MaterialExpressionShaderStageSwitch_0 {
   class=MaterialExpressionShaderStageSwitch
   inputs [
    PixelShader {from=MaterialExpressionVertexInterpolator_0, output=PS}
    VertexShader {from=MaterialExpressionPreSkinnedNormal_0, output=""}
   ]
   properties {}
  }
  MaterialExpressionShaderStageSwitch_2 {
   class=MaterialExpressionShaderStageSwitch
   inputs [
    PixelShader {from=MaterialExpressionVertexInterpolator_2, output=PS}
    VertexShader {from=MaterialExpressionLocalPosition_1, output=XYZ}
   ]
   properties {}
  }
  MaterialExpressionTransform_0 {
   class=MaterialExpressionTransform
   inputs [None {from=MaterialExpressionConstant3Vector_2}]
   properties {transformSourceType=TRANSFORMSOURCE_Instance}
  }
  MaterialExpressionTransform_1 {
   class=MaterialExpressionTransform
   inputs [None {from=MaterialExpressionNamedRerouteUsage_38}]
   properties {transformSourceType=TRANSFORMSOURCE_Instance}
  }
  MaterialExpressionTransform_12 {
   class=MaterialExpressionTransform
   inputs [None {from=MaterialExpressionConstant3Vector_3}]
   properties {transformSourceType=TRANSFORMSOURCE_Instance}
  }
  MaterialExpressionTransform_13 {
   class=MaterialExpressionTransform
   inputs [None {from=MaterialExpressionConstant3Vector_4}]
   properties {transformSourceType=TRANSFORMSOURCE_Instance}
  }
  MaterialExpressionTransform_14 {
   class=MaterialExpressionTransform
   inputs [None {from=MaterialExpressionVertexNormalWS_1}]
   properties {transformSourceType=TRANSFORMSOURCE_World, transformType=TRANSFORM_Instance}
  }
  MaterialExpressionTransform_15 {
   class=MaterialExpressionTransform
   inputs [None {from=MaterialExpressionNamedRerouteUsage_8}]
   properties {transformSourceType=TRANSFORMSOURCE_Instance}
  }
  MaterialExpressionTransform_16 {
   class=MaterialExpressionTransform
   inputs [None {from=MaterialExpressionNamedRerouteUsage_8}]
   properties {transformSourceType=TRANSFORMSOURCE_Local}
  }
  MaterialExpressionTransform_2 {
   class=MaterialExpressionTransform
   inputs [None {from=MaterialExpressionNamedRerouteUsage_38}]
   properties {transformSourceType=TRANSFORMSOURCE_Local}
  }
  MaterialExpressionTransform_3 {
   class=MaterialExpressionTransform
   inputs [None {from=MaterialExpressionNamedRerouteUsage_4}]
   properties {transformSourceType=TRANSFORMSOURCE_Instance}
  }
  MaterialExpressionTransform_4 {
   class=MaterialExpressionTransform
   inputs [None {from=MaterialExpressionNamedRerouteUsage_37}]
   properties {transformSourceType=TRANSFORMSOURCE_Instance}
  }
  MaterialExpressionTransform_5 {
   class=MaterialExpressionTransform
   inputs [None {from=MaterialExpressionNamedRerouteUsage_4}]
   properties {transformSourceType=TRANSFORMSOURCE_Local}
  }
  MaterialExpressionTransform_6 {
   class=MaterialExpressionTransform
   inputs [None {from=MaterialExpressionNamedRerouteUsage_37}]
   properties {transformSourceType=TRANSFORMSOURCE_Local}
  }
  MaterialExpressionVertexInterpolator_0 {class=MaterialExpressionVertexInterpolator, inputs [VS {from=MaterialExpressionPreSkinnedNormal_0, output=""}], properties {}}
  MaterialExpressionVertexInterpolator_2 {class=MaterialExpressionVertexInterpolator, inputs [VS {from=MaterialExpressionLocalPosition_1, output=XYZ}], properties {}}
  MaterialExpressionVertexNormalWS_1 {class=MaterialExpressionVertexNormalWS, inputs [], properties {}}
 }
 properties {
  bExposeToLibrary=true
  description="Generates biplanar projection UVs, basis vectors, and packed derivatives for texture sampling and normal reorientation. Provides two complete output sets: UV1 (major axis) and UV2 (median axis), each with its own tangent, bitangent, and derivatives. These represent the two strongest projection axes for the current surface orientation. A BlendMask output is also provided (RG), which gives weights for blending the two samples post-texture lookup, allowing smooth transitions between axes and supporting height- or albedo-driven weight modulation."
  userExposedCaption="CoordinateBasis: Biplanar (Standard)"
  libraryCategoriesText [Base Material]
 }
}
