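# Material function in the Base Material library that applies a 2D rotate, tile and offset transform to UV coordinates.
# MF_Rotate2D rotates the input coordinates about a pivot, then the result is multiplied by UVTile and shifted by UVOffset.
# It outputs the transformed UV plus its DDX and DDY packed into one vector.
# A shader stage switch supplies a constant vector for those derivatives in the vertex shader.
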MaterialFunction /BaseMaterial/Materials/Functions/MF_CoordinateTransform_2D
graph {
 properties {bExposeToLibrary=true, libraryCategoriesText [Base Material]}
 nodes [
  n1 = FunctionInput() {bUsePreviewValueAsDefault=true, inputName=UVRotation, inputType=FunctionInput_Scalar, sortPriority=10}
  n2 = FunctionInput(Preview=TextureCoordinate()) {bUsePreviewValueAsDefault=true, inputName=Coords, inputType=FunctionInput_Vector2, sortPriority=-10}
  n3 = FunctionInput(Preview=Constant2Vector() {g=0.5, r=0.5}) {bUsePreviewValueAsDefault=true, inputName=RotationPivot, inputType=FunctionInput_Vector2, sortPriority=15, previewValue=<0.5 0 0 1>}
  n4 = FunctionInput() {bUsePreviewValueAsDefault=true, inputName=UVTile, inputType=FunctionInput_Vector2, sortPriority=20, previewValue=<1 1 0 1>}
  n5 = Add(
   A=Multiply(
    A=MaterialFunctionCall!(RotationAmount=@n1, Coords=@n2, Pivot=@n3) {materialFunction=&/MF_Rotate2D}.Result
    B=@n4
   )
   B=FunctionInput() {bUsePreviewValueAsDefault=true, inputName=UVOffset, inputType=FunctionInput_Vector2}
  )
  FunctionOutput(None=@n5) {outputName=geometry_UV}
  n6 = DDX(Value=@n5)
  n7 = DDY(Value=@n5)
  FunctionOutput(None=ShaderStageSwitch(PixelShader=AppendVector(A=@n6, B=@n7), VertexShader=Constant4Vector().RGBA)) {outputName=geometry_packedDXY, sortPriority=10}
 ]
}
