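# Library material function that builds capsule-projected UVs, a tangent/bitangent basis and packed derivatives for texture sampling and normal reorientation.
# MF_Coordinate_AnalyticalCapsule generates the projection from height, radius and hardness.
# MF_CoordinateTransform_2D then applies offset, rotation in turns around a fixed pivot, and tiling to the UVs only.
# Tangent, bitangent and DXY pass through untransformed, while the rotation output is one minus the input turns.
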MaterialFunction /BaseMaterial/Materials/Functions/MF_ResolveCoordinateFrame_CapsuleUV_Basic
graph {
 properties {
  bExposeToLibrary=true
  description="Generates capsule projection UVs, basis vectors, and packed derivatives for texture sampling and normal reorientation. Provides UVs for texture lookups, tangent/bitangent vectors with rotation for tangent-space normal blending, and derivatives for stable mip selection via SampleGrad. The capsule shape is defined by Height, Radius, and Hardness, allowing you to control both the cylindrical body and the cap transitions."
  userExposedCaption="CoordinateBasis: Capsule (Standard)"
  libraryCategoriesText [Base Material]
 }
 nodes [
  n1 = FunctionInput() {
   bUsePreviewValueAsDefault=true
   description=Applies an additional per-axis offset to the coordinates before projection. Useful for correctly positioning a texture.
   inputName=CoordinateTransform
   inputType=FunctionInput_Vector2
   sortPriority=-9
  }
  n2 = FunctionInput() {
   bUsePreviewValueAsDefault=true
   description="Rotates the projected UVs around the selected axis. Units are in turns, where 1.0 = 360° and 0.25 = 90°."
   inputName="Rotation (Turns)"
   inputType=FunctionInput_Scalar
   sortPriority=-8
  }
  n3 = NamedRerouteDeclaration(None=@n2) {name=turns, nodeColor=<rgba 0.8963 0.8963 0.0091 1>}
  n4 = FunctionInput() {
   bUsePreviewValueAsDefault=true
   description="Sets how many times the projected coordinates repeat along each axis. Use values above 1 for denser tiling, negative values for mirroring, and avoid zero."
   inputName=CoordinateRepeat
   inputType=FunctionInput_Vector2
   sortPriority=-10
   previewValue=<1 1 1 1>
  }
  MaterialExpressionFunctionInput_7 = FunctionInput() {
   bUsePreviewValueAsDefault=true
   description="Sets capsule shape: X = Height, Y = Radius, Z = Hardness (transition between cap and cylinder)."
   inputName="CapsuleControls(Height/Radius/Hardness)"
   previewValue=<1 0.2 0.01 1>
  }
  n5 = Convert(None=@MaterialExpressionFunctionInput_7) {
   nodeName=Break Float3
   convertInputs [{
     type=Vector3
     expressionInput {
      expression=&self:MaterialExpressionFunctionInput_7
      inputName=None
      mask=0
      maskA=0
      maskB=0
      maskG=0
      maskR=0
      outputIndex=0
     }
    }]
   convertOutputs [{type=Scalar}, {type=Scalar}, {type=Scalar}]
  }
  n6 = MaterialFunctionCall!(CapsuleHeight=@n5, CapsuleRadius=@n5, BlendHardness=@n5) {materialFunction=&/MF_Coordinate_AnalyticalCapsule}
  FunctionOutput(None=MaterialFunctionCall!(Coords=@n6.UV, UVOffset=@n1, UVRotation=@n3, RotationPivot=Constant2Vector() {g=0.5}, UVTile=@n4) {materialFunction=&/MF_CoordinateTransform_2D}.geometry_UV) {description=The primary 2D coordinates generated by the projection. Feed these into texture samplers for surface detail., outputName=UV}
  FunctionOutput(None=OneMinus(None=@n3)) {
   description="The rotation applied around the projection axis, expressed in turns (1.0 = 360°). Used along with Tangent/Bitangent for reorienting normal maps via MF_TangentNormal_OrientWorld."
   outputName="Rotation (Turns)"
   sortPriority=-5
  }
  FunctionOutput(None=@n6.packedDXY) {
   description="Packed derivatives of the projected UVs (ddx/ddy). Use with SampleGrad to ensure textures sample with correct mip levels and anisotropic filtering."
   outputName=DXY
   sortPriority=3
  }
  FunctionOutput(None=@n6.Tangent) {
   description=The world-space tangent vector of the projection frame. Use with BitangentWS and Rotation to correctly reorient tangent-space normals.
   outputName=Tangent
   sortPriority=1
  }
  FunctionOutput(None=@n6.Bitangent) {
   description=The world-space bitangent vector of the projection frame. Complements TangentWS to form the basis for tangent-space normal reorientation.
   outputName=Bitangent
   sortPriority=2
  }
 ]
}
