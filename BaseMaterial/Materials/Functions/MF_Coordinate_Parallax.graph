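# Material function for parallax occlusion mapping, exposed in the Base Material library.
# It projects the camera ray into the tangent basis, rotates and scales it by height strength, and varies step count with view angle.
# A custom HLSL node marches the height texture using supplied derivatives and interpolates the hit.
# It outputs offset UVs, intersection mask, pixel depth offset and world position; ray tracing skips the marched offset.
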MaterialFunction /BaseMaterial/Materials/Functions/MF_Coordinate_Parallax
graph {
 expressions {
  MaterialExpressionAbs_0 {class=MaterialExpressionAbs, inputs [None {from=MaterialExpressionDotProduct_0, output=""}], properties {}}
  MaterialExpressionAbs_2 {class=MaterialExpressionAbs, inputs [None {from=MaterialExpressionDotProduct_4, output=""}], properties {}}
  MaterialExpressionAdd_0 {
   class=MaterialExpressionAdd
   inputs [
    A {from=MaterialExpressionFunctionInput_1}
    B {from=MaterialExpressionMultiply_5}
   ]
  }
  MaterialExpressionAdd_1 {
   class=MaterialExpressionAdd
   inputs [
    A {from=MaterialExpressionFunctionInput_11}
    B {from=MaterialExpressionAbs_0}
   ]
  }
  MaterialExpressionAdd_3 {
   class=MaterialExpressionAdd
   inputs [
    A {from=MaterialExpressionConvert_1}
    B {from=MaterialExpressionReroute_1}
   ]
  }
  MaterialExpressionCameraVectorWS_2 {class=MaterialExpressionCameraVectorWS, inputs [], properties {}}
  MaterialExpressionComponentMask_10 {class=MaterialExpressionComponentMask, inputs [None {from=MaterialExpressionReroute_6}], properties {g=true, r=true}}
  MaterialExpressionComponentMask_11 {class=MaterialExpressionComponentMask, inputs [None {from=MaterialExpressionReroute_6}], properties {b=true}}
  MaterialExpressionConstant2Vector_0 {class=MaterialExpressionConstant2Vector}
  MaterialExpressionConstant3Vector_0 {class=MaterialExpressionConstant3Vector, properties {constant=<rgba 0 0 1 0>}}
  MaterialExpressionConstant_1 {class=MaterialExpressionConstant, properties {desc=MinStep, r=1}}
  MaterialExpressionConstant_14 {class=MaterialExpressionConstant}
  MaterialExpressionConstant_2 {class=MaterialExpressionConstant, properties {r=32}}
  MaterialExpressionConvert_1 {
   class=MaterialExpressionConvert
   inputs [None {from=MaterialExpressionCustom_9}]
   properties {
    convertInputs [{
      type=Vector4
      defaultValue=<rgba 0 0 0 1>
      expressionInput {
       expression=&self:MaterialExpressionCustom_9
       inputName=None
       mask=0
       maskA=0
       maskB=0
       maskG=0
       maskR=0
       outputIndex=0
      }
     }]
    convertOutputs [{type=Vector2}, {type=Scalar}, {type=Scalar}]
   }
  }
  MaterialExpressionConvert_3 {
   class=MaterialExpressionConvert
   inputs [
    None {from=MaterialExpressionConvert_1}
    None {from=MaterialExpressionMultiply_19}
   ]
   properties {
    convertInputs [
     {
      type=Vector2
      expressionInput {
       expression=&self:MaterialExpressionConvert_1
       inputName=None
       mask=0
       maskA=0
       maskB=0
       maskG=0
       maskR=0
       outputIndex=0
      }
     }
     {
      type=Scalar
      expressionInput {
       expression=&self:MaterialExpressionMultiply_19
       inputName=None
       mask=0
       maskA=0
       maskB=0
       maskG=0
       maskR=0
       outputIndex=0
      }
     }
    ]
    convertOutputs [{type=Vector3}]
   }
  }
  MaterialExpressionConvert_7 {
   class=MaterialExpressionConvert
   inputs [None {from=MaterialExpressionFunctionInput_4}]
   properties {
    convertInputs [{
      type=Vector4
      defaultValue=<rgba 0 0 0 1>
      expressionInput {
       expression=&self:MaterialExpressionFunctionInput_4
       inputName=None
       mask=0
       maskA=0
       maskB=0
       maskG=0
       maskR=0
       outputIndex=0
      }
     }]
    convertOutputs [{type=Vector2}, {type=Vector2}]
   }
  }
  MaterialExpressionCustom_9 {
   class=MaterialExpressionCustom
   inputs [
    Tex {from=MaterialExpressionFunctionInput_2, output=""}
    UV {from=MaterialExpressionMultiply_1, output=""}
    UVDist {from=MaterialExpressionNamedRerouteUsage_3, output=""}
    MaxSteps {from=MaterialExpressionNamedRerouteUsage_4, output=""}
    stepsize {from=MaterialExpressionNamedRerouteUsage_5, output=""}
    InDDX {from=MaterialExpressionMultiply_2, output=""}
    InDDY {from=MaterialExpressionMultiply_7, output=""}
    HeightChannel {from=MaterialExpressionFunctionInput_6, output=""}
   ]
   properties {
    code="float rayheight = 1;\r\nfloat oldray = 1;\r\nfloat2 offset = 0;\r\nfloat oldtex = 1;\r\nfloat texAtRay;\r\nfloat yintersect;\r\nint i = 0;\r\n\r\nwhile (i < MaxSteps + 2)\r\n{\r\n\tfloat2 curUV = UV + offset;\r\n\r\n    float texAtRay = Tex.SampleGrad( GetMaterialSharedSampler(TexSampler,View.MaterialTextureBilinearWrapedSampler), curUV, \r\n                     InDDX * (MaterialFloat2)View.MaterialTextureDerivativeMultiply, \r\n                     InDDY * (MaterialFloat2)View.MaterialTextureDerivativeMultiply ) [int(HeightChannel)];\r\n\r\n\tif (rayheight < texAtRay)\r\n\t{\r\n\t\tfloat xintersect = (oldray - oldtex) + (texAtRay - rayheight);\r\n\t\txintersect = (texAtRay - rayheight) / xintersect;\r\n\t\tyintersect = ( oldray * (xintersect) ) + ( rayheight * (1 - xintersect) );\r\n\t\toffset -= (xintersect * UVDist);\r\n        \r\n        float4 output;\r\n        output.xy = offset;\r\n        output.z  = yintersect;\r\n        output.a  = 1.f;\r\n        return output;\r\n\t}\r\n\r\n\toldray = rayheight;\r\n\trayheight -= stepsize;\r\n\toffset += UVDist;\r\n\toldtex = texAtRay;\r\n\r\n\ti++;\r\n}\r\n\r\nreturn 0;"
    description=ParallaxMapping
    outputType=CMOT_Float4
    showCode=true
    inputs [
     {
      inputName=Tex
      input {
       expression=&self:MaterialExpressionFunctionInput_2
       inputName=None
       mask=0
       maskA=0
       maskB=0
       maskG=0
       maskR=0
       outputIndex=0
      }
     }
     {
      inputName=UV
      input {
       expression=&self:MaterialExpressionMultiply_1
       inputName=None
       mask=0
       maskA=0
       maskB=0
       maskG=0
       maskR=0
       outputIndex=0
      }
     }
     {
      inputName=UVDist
      input {
       expression=&self:MaterialExpressionNamedRerouteUsage_3
       inputName=None
       mask=0
       maskA=0
       maskB=0
       maskG=0
       maskR=0
       outputIndex=0
      }
     }
     {
      inputName=MaxSteps
      input {
       expression=&self:MaterialExpressionNamedRerouteUsage_4
       inputName=None
       mask=0
       maskA=0
       maskB=0
       maskG=0
       maskR=0
       outputIndex=0
      }
     }
     {
      inputName=stepsize
      input {
       expression=&self:MaterialExpressionNamedRerouteUsage_5
       inputName=None
       mask=0
       maskA=0
       maskB=0
       maskG=0
       maskR=0
       outputIndex=0
      }
     }
     {
      inputName=InDDX
      input {
       expression=&self:MaterialExpressionMultiply_2
       inputName=None
       mask=0
       maskA=0
       maskB=0
       maskG=0
       maskR=0
       outputIndex=0
      }
     }
     {
      inputName=InDDY
      input {
       expression=&self:MaterialExpressionMultiply_7
       inputName=None
       mask=0
       maskA=0
       maskB=0
       maskG=0
       maskR=0
       outputIndex=0
      }
     }
     {
      inputName=HeightChannel
      input {
       expression=&self:MaterialExpressionFunctionInput_6
       inputName=None
       mask=0
       maskA=0
       maskB=0
       maskG=0
       maskR=0
       outputIndex=0
      }
     }
    ]
   }
  }
  MaterialExpressionDistance_9 {
   class=MaterialExpressionDistance
   inputs [
    A {from=MaterialExpressionConvert_3, output=""}
    B {from=MaterialExpressionConstant_14, output=""}
   ]
   properties {}
  }
  MaterialExpressionDivide_49 {
   class=MaterialExpressionDivide
   inputs [
    A {from=""}
    B {from=MaterialExpressionLinearInterpolate_4}
   ]
  }
  MaterialExpressionDivide_9 {
   class=MaterialExpressionDivide
   inputs [
    A {from=MaterialExpressionMultiply_18}
    B {from=MaterialExpressionComponentMask_11}
   ]
   properties {constB=1}
  }
  MaterialExpressionDotProduct_0 {
   class=MaterialExpressionDotProduct
   inputs [
    A {from=MaterialExpressionVertexNormalWS_0, output=""}
    B {from=MaterialExpressionNamedRerouteDeclaration_7, output=""}
   ]
   properties {}
  }
  MaterialExpressionDotProduct_4 {
   class=MaterialExpressionDotProduct
   inputs [
    A {from=MaterialExpressionNamedRerouteUsage_7, output=""}
    B {from=MaterialExpressionMaterialFunctionCall_9, output=Result}
   ]
   properties {}
  }
  MaterialExpressionFloor_0 {class=MaterialExpressionFloor, inputs [None {from=MaterialExpressionLinearInterpolate_4, output=""}], properties {}}
  MaterialExpressionFunctionInput_0 {
   class=MaterialExpressionFunctionInput
   inputs [Preview {from="", output=""}]
   properties {bUsePreviewValueAsDefault=true, inputName=Height Tile, inputType=FunctionInput_Vector2, sortPriority=12, previewValue=<1 1 0 1>}
  }
  MaterialExpressionFunctionInput_1 {
   class=MaterialExpressionFunctionInput
   inputs [Preview {from=MaterialExpressionTextureCoordinate_0, output=""}]
   properties {bUsePreviewValueAsDefault=true, inputName=Coordinates, inputType=FunctionInput_Vector2, sortPriority=1}
  }
  MaterialExpressionFunctionInput_10 {
   class=MaterialExpressionFunctionInput
   inputs [Preview {from="", output=""}]
   properties {bUsePreviewValueAsDefault=true, inputName=Reference Plane, inputType=FunctionInput_Scalar, sortPriority=4, previewValue=<1 0 0 1>}
  }
  MaterialExpressionFunctionInput_11 {
   class=MaterialExpressionFunctionInput
   inputs [Preview {from="", output=""}]
   properties {bUsePreviewValueAsDefault=true, inputName=QualityBias, inputType=FunctionInput_Scalar, sortPriority=10}
  }
  MaterialExpressionFunctionInput_12 {
   class=MaterialExpressionFunctionInput
   inputs [Preview {from=MaterialExpressionConstant_2, output=""}]
   properties {bUsePreviewValueAsDefault=true, inputName=MaxSteps, inputType=FunctionInput_Scalar, sortPriority=5}
  }
  MaterialExpressionFunctionInput_13 {
   class=MaterialExpressionFunctionInput
   inputs [Preview {from=MaterialExpressionCameraVectorWS_2, output=""}]
   properties {bUsePreviewValueAsDefault=true, inputName=CameraVector, sortPriority=11, previewValue=<1 1 1 0>}
  }
  MaterialExpressionFunctionInput_2 {
   class=MaterialExpressionFunctionInput
   inputs [Preview {from=MaterialExpressionTextureObject_0, output=""}]
   properties {bUsePreviewValueAsDefault=true, inputName=POM_HeightTexture, inputType=FunctionInput_Texture2D}
  }
  MaterialExpressionFunctionInput_3 {
   class=MaterialExpressionFunctionInput
   inputs [Preview {from="", output=""}]
   properties {bUsePreviewValueAsDefault=true, inputName="Rotation (Turns)", inputType=FunctionInput_Scalar, sortPriority=8}
  }
  MaterialExpressionFunctionInput_4 {
   class=MaterialExpressionFunctionInput
   inputs [Preview {from="", output=""}]
   properties {bUsePreviewValueAsDefault=true, inputName=coordinateDXY, inputType=FunctionInput_Vector4, sortPriority=2}
  }
  MaterialExpressionFunctionInput_5 {
   class=MaterialExpressionFunctionInput
   inputs [Preview {from="", output=""}]
   properties {bUsePreviewValueAsDefault=true, inputName=Bitangent, sortPriority=7, previewValue=<0 1 0 1>}
  }
  MaterialExpressionFunctionInput_6 {
   class=MaterialExpressionFunctionInput
   inputs [Preview {from="", output=""}]
   properties {bUsePreviewValueAsDefault=true, inputName=HeightChannel, inputType=FunctionInput_Scalar, sortPriority=9, previewValue=<3 0 0 1>}
  }
  MaterialExpressionFunctionInput_7 {
   class=MaterialExpressionFunctionInput
   inputs [Preview {from="", output=""}]
   properties {bUsePreviewValueAsDefault=true, inputName=HeightStrength, inputType=FunctionInput_Scalar, sortPriority=3, previewValue=<0.5 0 0 1>}
  }
  MaterialExpressionFunctionInput_8 {
   class=MaterialExpressionFunctionInput
   inputs [Preview {from="", output=""}]
   properties {bUsePreviewValueAsDefault=true, inputName=Tangent, sortPriority=6, previewValue=<1 0 0 1>}
  }
  MaterialExpressionFunctionOutput_0 {
   class=MaterialExpressionFunctionOutput
   inputs [None {from=MaterialExpressionRayTracingQualitySwitch_0, output=""}]
   properties {description="This returns the input UVs with Parallax Offset. ", outputName=POM UVs}
  }
  MaterialExpressionFunctionOutput_1 {
   class=MaterialExpressionFunctionOutput
   inputs [None {from=MaterialExpressionConvert_1, output=""}]
   properties {description=Outputs a mask for alpha test materials, outputName=Intersection Mask, sortPriority=1}
  }
  MaterialExpressionFunctionOutput_2 {
   class=MaterialExpressionFunctionOutput
   inputs [None {from=MaterialExpressionMultiply_20, output=""}]
   properties {
    description=Pixel Depth Offset is used to give accurate intersection with other meshes as well as shadows cast by other meshes. Will cause a problem with Dynamic Shadows since the original depths will shadow the new offset depths.
    outputName=Pixel Depth Offset
    sortPriority=4
   }
  }
  MaterialExpressionFunctionOutput_3 {
   class=MaterialExpressionFunctionOutput
   inputs [None {from=MaterialExpressionSubtract_1, output=""}]
   properties {
    description="World Position accounting for offset. Can be used to map virtual worldposition textures by subtracting this Z value as the vertex normal (or 0,0,1 for a floor)."
    outputName=World Position
    sortPriority=5
   }
  }
  MaterialExpressionLinearInterpolate_4 {
   class=MaterialExpressionLinearInterpolate
   inputs [
    A {from=MaterialExpressionFunctionInput_12}
    B {from=MaterialExpressionConstant_1}
    Alpha {from=MaterialExpressionSaturate_0}
   ]
  }
  MaterialExpressionMaterialFunctionCall_1 {
   class=MaterialExpressionMaterialFunctionCall
   inputs [
    RotationAmount {from=MaterialExpressionFunctionInput_3}
    Coords {from=MaterialExpressionDivide_9}
    Pivot {from=MaterialExpressionConstant2Vector_0}
   ]
   properties {materialFunction=&/MF_Rotate2D}
  }
  MaterialExpressionMaterialFunctionCall_3 {
   class=MaterialExpressionMaterialFunctionCall
   inputs [
    VectorToTransform {from=MaterialExpressionNamedRerouteDeclaration_7}
    BasisX {from=MaterialExpressionFunctionInput_8}
    BasisY {from=MaterialExpressionFunctionInput_5}
    BasisZ {from=MaterialExpressionVertexNormalWS_3}
   ]
   properties {materialFunction=/Engine/Functions/Engine_MaterialFunctions02/Math/InverseTransformMatrix}
  }
  MaterialExpressionMaterialFunctionCall_9 {
   class=MaterialExpressionMaterialFunctionCall
   inputs []
   properties {materialFunction=&CameraDirectionVector}
  }
  MaterialExpressionMultiply_0 {
   class=MaterialExpressionMultiply
   inputs [
    A {from=MaterialExpressionFunctionInput_7}
    B {from=""}
   ]
   properties {constB=0.01, desc=could measure vs world size instead of fixed ratio?}
  }
  MaterialExpressionMultiply_1 {
   class=MaterialExpressionMultiply
   inputs [
    A {from=MaterialExpressionAdd_0}
    B {from=MaterialExpressionFunctionInput_0}
   ]
  }
  MaterialExpressionMultiply_18 {
   class=MaterialExpressionMultiply
   inputs [
    A {from=MaterialExpressionComponentMask_10}
    B {from=""}
   ]
   properties {constB=-1}
  }
  MaterialExpressionMultiply_19 {
   class=MaterialExpressionMultiply
   inputs [
    A {from=MaterialExpressionOneMinus_0}
    B {from=MaterialExpressionNamedRerouteUsage_6}
   ]
   properties {constB=0.01}
  }
  MaterialExpressionMultiply_2 {
   class=MaterialExpressionMultiply
   inputs [
    A {from=MaterialExpressionFunctionInput_0}
    B {from=MaterialExpressionConvert_7}
   ]
  }
  MaterialExpressionMultiply_20 {
   class=MaterialExpressionMultiply
   inputs [
    A {from=MaterialExpressionDistance_9}
    B {from=MaterialExpressionAbs_2}
   ]
  }
  MaterialExpressionMultiply_21 {
   class=MaterialExpressionMultiply
   inputs [
    A {from=MaterialExpressionDistance_9}
    B {from=MaterialExpressionNamedRerouteUsage_7}
   ]
  }
  MaterialExpressionMultiply_3 {
   class=MaterialExpressionMultiply
   inputs [
    A {from=MaterialExpressionMaterialFunctionCall_1, output=Result}
    B {from=MaterialExpressionMultiply_0}
   ]
  }
  MaterialExpressionMultiply_4 {
   class=MaterialExpressionMultiply
   inputs [
    A {from=MaterialExpressionMultiply_3}
    B {from=MaterialExpressionNamedRerouteUsage_2}
   ]
  }
  MaterialExpressionMultiply_5 {
   class=MaterialExpressionMultiply
   inputs [
    A {from=MaterialExpressionNamedRerouteUsage_1}
    B {from=MaterialExpressionMultiply_6}
   ]
  }
  MaterialExpressionMultiply_6 {
   class=MaterialExpressionMultiply
   inputs [
    A {from=MaterialExpressionOneMinus_1}
    B {from=""}
   ]
   properties {constB=-1}
  }
  MaterialExpressionMultiply_7 {
   class=MaterialExpressionMultiply
   inputs [
    A {from=MaterialExpressionFunctionInput_0}
    B {from=MaterialExpressionConvert_7}
   ]
  }
  MaterialExpressionNamedRerouteDeclaration_0 {
   class=MaterialExpressionNamedRerouteDeclaration
   inputs [None {from=MaterialExpressionMaterialFunctionCall_1, output=Result}]
   properties {name=ViewTrace_ST, nodeColor=<rgba 0.5703 0 1 1>}
  }
  MaterialExpressionNamedRerouteDeclaration_1 {
   class=MaterialExpressionNamedRerouteDeclaration
   inputs [None {from=MaterialExpressionMultiply_3}]
   properties {name=ViewTraceScaled, nodeColor=<rgba 1 0.2109 0 1>}
  }
  MaterialExpressionNamedRerouteDeclaration_2 {
   class=MaterialExpressionNamedRerouteDeclaration
   inputs [None {from=MaterialExpressionDivide_49}]
   properties {name=StepSize, nodeColor=<rgba 1 0.7266 0 1>}
  }
  MaterialExpressionNamedRerouteDeclaration_3 {
   class=MaterialExpressionNamedRerouteDeclaration
   inputs [None {from=MaterialExpressionFloor_0}]
   properties {name=MaxSteps, nodeColor=<rgba 0.875 1 0 1>}
  }
  MaterialExpressionNamedRerouteDeclaration_4 {
   class=MaterialExpressionNamedRerouteDeclaration
   inputs [None {from=MaterialExpressionMultiply_4}]
   properties {name=UVDist, nodeColor=<rgba 1 0 0.75 1>}
  }
  MaterialExpressionNamedRerouteDeclaration_6 {
   class=MaterialExpressionNamedRerouteDeclaration
   inputs [None {from=MaterialExpressionFunctionInput_7}]
   properties {name=HeightStr, nodeColor=<rgba 1 0 0 1>}
  }
  MaterialExpressionNamedRerouteDeclaration_7 {
   class=MaterialExpressionNamedRerouteDeclaration
   inputs [None {from=MaterialExpressionNormalize_2}]
   properties {name=RayDir, nodeColor=<rgba 1 0.7031 0 1>}
  }
  MaterialExpressionNamedRerouteUsage_1 {class=MaterialExpressionNamedRerouteUsage}
  MaterialExpressionNamedRerouteUsage_2 {class=MaterialExpressionNamedRerouteUsage}
  MaterialExpressionNamedRerouteUsage_3 {class=MaterialExpressionNamedRerouteUsage}
  MaterialExpressionNamedRerouteUsage_4 {class=MaterialExpressionNamedRerouteUsage}
  MaterialExpressionNamedRerouteUsage_5 {class=MaterialExpressionNamedRerouteUsage}
  MaterialExpressionNamedRerouteUsage_6 {class=MaterialExpressionNamedRerouteUsage}
  MaterialExpressionNamedRerouteUsage_7 {class=MaterialExpressionNamedRerouteUsage}
  MaterialExpressionNormalize_2 {class=MaterialExpressionNormalize, inputs [VectorInput {from=MaterialExpressionShaderStageSwitch_1, output=""}]}
  MaterialExpressionOneMinus_0 {class=MaterialExpressionOneMinus, inputs [None {from=MaterialExpressionConvert_1}]}
  MaterialExpressionOneMinus_1 {class=MaterialExpressionOneMinus, inputs [None {from=MaterialExpressionFunctionInput_10}]}
  MaterialExpressionRayTracingQualitySwitch_0 {
   class=MaterialExpressionRayTracingQualitySwitch
   inputs [
    Normal {from=MaterialExpressionAdd_3, output=""}
    RayTraced {from=MaterialExpressionReroute_1, output=""}
   ]
   properties {}
  }
  MaterialExpressionReroute_1 {class=MaterialExpressionReroute, inputs [None {from=MaterialExpressionAdd_0}]}
  MaterialExpressionReroute_6 {class=MaterialExpressionReroute, inputs [None {from=MaterialExpressionMaterialFunctionCall_3, output=Transformed Vector 3}]}
  MaterialExpressionSaturate_0 {class=MaterialExpressionSaturate, inputs [None {from=MaterialExpressionAdd_1}]}
  MaterialExpressionShaderStageSwitch_1 {
   class=MaterialExpressionShaderStageSwitch
   inputs [
    PixelShader {from=MaterialExpressionFunctionInput_13, output=""}
    VertexShader {from=MaterialExpressionTransform_1, output=""}
   ]
   properties {}
  }
  MaterialExpressionSubtract_1 {
   class=MaterialExpressionSubtract
   inputs [
    A {from=MaterialExpressionWorldPosition_0, output=XYZ}
    B {from=MaterialExpressionMultiply_21}
   ]
  }
  MaterialExpressionTextureCoordinate_0 {class=MaterialExpressionTextureCoordinate}
  MaterialExpressionTextureObject_0 {
   class=MaterialExpressionTextureObject
   inputs []
   properties {samplerType=SAMPLERTYPE_LinearColor, texture=/Engine/Functions/Engine_MaterialFunctions02/ExampleContent/Textures/SphereRenderHeightMap}
  }
  MaterialExpressionTransform_1 {
   class=MaterialExpressionTransform
   inputs [None {from=MaterialExpressionConstant3Vector_0}]
   properties {transformSourceType=TRANSFORMSOURCE_View}
  }
  MaterialExpressionVertexNormalWS_0 {class=MaterialExpressionVertexNormalWS, inputs [], properties {}}
  MaterialExpressionVertexNormalWS_3 {class=MaterialExpressionVertexNormalWS, inputs [], properties {}}
  MaterialExpressionWorldPosition_0 {class=MaterialExpressionWorldPosition, inputs [], properties {}}
 }
 properties {bExposeToLibrary=true, libraryCategoriesText [Base Material]}
}
