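# Material function for the Base Material library that produces capsule-projected UVs with parallax occlusion mapping.
# MF_Coordinate_AnalyticalCapsule builds UVs, tangent frame and packed derivatives from height, radius and hardness.
# MF_CoordinateTransform_2D offsets, rotates and tiles the UVs, then MF_Coordinate_Parallax ray-marches a selectable heightmap channel with quality-scaled steps.
# Outputs are displaced UVs, depth, one-minus rotation, plus the capsule's tangent, bitangent and untransformed derivatives for SampleGrad and normal reorientation.
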MaterialFunction /BaseMaterial/Materials/Functions/MF_ResolveCoordinateFrame_CapsuleUV_POM
graph {
 properties {
  bExposeToLibrary=true
  description="Generates capsule projection UVs, basis vectors, and packed derivatives for texture sampling and normal reorientation. Provides UVs for texture lookups, tangent/bitangent vectors with rotation for tangent-space normal blending, and derivatives for stable mip selection via SampleGrad. Applies Parallax Occlusion Mapping (POM) to the capsule projection for added depth and self-occlusion. The capsule is defined by Height, Radius, and Hardness, giving control over both cylindrical and cap regions. "
  userExposedCaption="CoordinateBasis: Capsule (Standard POM)"
  libraryCategoriesText [Base Material]
 }
 nodes [
  n1 = FunctionInput(Preview=TextureObject() {texture=/BaseMaterial/Textures/Default/T_Gray_sRGB}) {
   bUsePreviewValueAsDefault=true
   description="The heightmap texture used for parallax occlusion mapping. Grayscale values define surface depth, with brighter values typically pushed outward and darker values pushed inward."
   inputName=HeightTex
   inputType=FunctionInput_Texture2D
   sortPriority=-20
  }
  MaterialExpressionNamedRerouteUsage_2 = NamedRerouteUsage()
  n2 = Convert(None=@MaterialExpressionNamedRerouteUsage_2) {
   nodeName=Break Float3
   convertInputs [{
     type=Vector3
     expressionInput {
      expression=&self:MaterialExpressionNamedRerouteUsage_2
      inputName=None
      mask=0
      maskA=0
      maskB=0
      maskG=0
      maskR=0
      outputIndex=0
     }
    }]
   convertOutputs [{type=Scalar}, {type=Scalar}, {type=Scalar}]
  }
  n3 = MaterialFunctionCall!(CapsuleHeight=@n2, CapsuleRadius=@n2, BlendHardness=@n2) {materialFunction=&/MF_Coordinate_AnalyticalCapsule}
  n4 = FunctionInput() {
   bUsePreviewValueAsDefault=true
   description=Applies an additional per-axis offset to the coordinates before projection. Useful for correctly positioning a texture.
   inputName=CoordinateTransform
   inputType=FunctionInput_Vector2
   sortPriority=-9
  }
  n5 = FunctionInput() {
   bUsePreviewValueAsDefault=true
   description="Sets how many times the projected coordinates repeat along each axis. Use values above 1 for denser tiling, negative values for mirroring, and avoid zero."
   inputName=CoordinateRepeat
   inputType=FunctionInput_Vector2
   sortPriority=-10
   previewValue=<1 1 1 1>
  }
  n6 = FunctionOutput(None=@n3.packedDXY) {
   description="Packed derivatives of the projected UVs (ddx/ddy). Use with SampleGrad to ensure textures sample with correct mip levels and anisotropic filtering.\r\n"
   outputName=DXY
   sortPriority=3
  }
  n7 = FunctionOutput(None=@n3.Tangent) {
   description="The world-space tangent vector of the projection frame. Use with BitangentWS and Rotation to correctly reorient tangent-space normals.\r\n"
   outputName=Tangent
   sortPriority=1
  }
  n8 = FunctionOutput(None=@n3.Bitangent) {
   description="The world-space bitangent vector of the projection frame. Complements TangentWS to form the basis for tangent-space normal reorientation.\r\n"
   outputName=Bitangent
   sortPriority=2
  }
  n9 = FunctionInput() {
   bUsePreviewValueAsDefault=true
   description=Control the U/V Tiling of the height texture relative to the input coordinates.
   inputName=POM HeightTile
   inputType=FunctionInput_Vector2
   sortPriority=12
   previewValue=<1 1 0 1>
  }
  n10 = FunctionInput() {
   bUsePreviewValueAsDefault=true
   description="Specifies which channel of the heightmap texture to use for depth information. Assumes the Alpha channel is used, but any channel can be selected if your texture packs data differently."
   inputName=POM HeightChannel
   inputType=FunctionInput_Scalar
   sortPriority=7
   previewValue=<3 0 0 1>
  }
  n11 = MaterialFunctionCall!(
   POM_HeightTexture=@n1
   Coordinates=MaterialFunctionCall!(Coords=@n3.UV, UVOffset=@n4, UVRotation=NamedRerouteUsage(), RotationPivot=Constant2Vector() {g=0.5}, UVTile=@n5) {materialFunction=&/MF_CoordinateTransform_2D}.geometry_UV
   coordinateDXY=@n6
   HeightStrength=NamedRerouteUsage()
   Reference Plane=NamedRerouteUsage()
   MaxSteps=NamedRerouteUsage()
   Tangent=@n7
   Bitangent=@n8
   "Rotation (Turns)"=NamedRerouteUsage()
   HeightChannel=@n10
   QualityBias=NamedRerouteUsage()
   CameraVector
   Height Tile=@n9
  ) {materialFunction=&/MF_Coordinate_Parallax}
  FunctionOutput(None=@n11.POM UVs) {
   description="The primary 2D coordinates generated by the projection. Feed these into texture samplers for surface detail.\r\n"
   outputName=UV
  }
  n12 = FunctionInput() {
   bUsePreviewValueAsDefault=true
   description="Rotates the projected UVs around the selected axis. Units are in turns, where 1.0 = 360° and 0.25 = 90°."
   inputName="Rotation (Turns)"
   inputType=FunctionInput_Scalar
   sortPriority=-8
  }
  FunctionOutput(None=OneMinus(None=NamedRerouteDeclaration(None=@n12) {name=turns, nodeColor=<rgba 0.8963 0.8963 0.0091 1>})) {
   description="The rotation applied around the projection axis, expressed in turns (1.0 = 360°). Used along with Tangent/Bitangent for reorienting normal maps via MF_TangentNormal_OrientWorld.\r\n"
   outputName="Rotation (Turns)"
   sortPriority=-5
  }
  FunctionOutput(None=@n11.Pixel Depth Offset) {
   description="Depth at the POM intersection (after Intensity/RefPlane). Higher (whiter) = deeper—the farther the ray travels into the material, the larger the value."
   outputName=Depth
   sortPriority=4
  }
  n13 = FunctionInput() {
   bUsePreviewValueAsDefault=true
   description="Sets capsule shape: X = Height, Y = Radius, Z = Hardness (transition between cap and cylinder)."
   inputName="CapsuleControls(Height/Radius/Hardness)"
   previewValue=<1 0.2 0.01 1>
  }
  NamedRerouteDeclaration(None=@n13) {name=Capsule, nodeColor=<rgba 0.8047 0 1 1>}
  n14 = FunctionInput() {
   bUsePreviewValueAsDefault=true
   description="Controls the strength of the parallax effect. Higher values exaggerate depth; lower values make surfaces appear flatter."
   inputName=POM Intensity
   inputType=FunctionInput_Scalar
   sortPriority=8
   previewValue=<1 0 0 1>
  }
  NamedRerouteDeclaration(None=@n14) {name=POM_Height, nodeColor=<rgba 0.9 0.306 0.009 1>}
  n15 = FunctionInput() {
   bUsePreviewValueAsDefault=true
   description="Sets the maximum number of depth sampling steps. This is the upper bound, with actual steps scaled by the Quality setting."
   inputName=POM MaxSteps
   inputType=FunctionInput_Scalar
   sortPriority=9
   previewValue=<16 0 0 1>
  }
  NamedRerouteDeclaration(None=@n15) {name=POM_MaxSteps, nodeColor=<rgba 0.9 0.306 0.009 1>}
  n16 = FunctionInput() {
   bUsePreviewValueAsDefault=true
   description="Scales step count by view angle. At 1, MaxSteps are always used; at 0, only one step is used when looking straight on, with more steps added at grazing angles."
   inputName=POM Quality
   inputType=FunctionInput_Scalar
   sortPriority=10
   previewValue=<0.25 0 0 1>
  }
  NamedRerouteDeclaration(None=@n16) {name=POM_Noise, nodeColor=<rgba 0.9 0.306 0.009 1>}
  n17 = FunctionInput() {
   bUsePreviewValueAsDefault=true
   description=Defines the reference plane for parallax. This value shifts where the heightmap appears anchored relative to the surface.
   inputName=POM RefPlane
   inputType=FunctionInput_Scalar
   sortPriority=11
   previewValue=<1 0 0 1>
  }
  NamedRerouteDeclaration(None=@n17) {name=POM_RefPlane, nodeColor=<rgba 0.9 0.306 0.009 1>}
 ]
}
