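# Material function that cell-bombs texture coordinates, giving each cell of a perturbed Voronoi noise texture random scale, offset and rotation to vary texture sampling.
# The texture's alpha and BlendFactor form a sample mask, thresholded against Dither to switch hard between cell-randomized and original coordinates.
# It outputs rotated UVs, rotation, mask, unblended cell variants, and original-coordinate derivatives for sampling, zeroed in the vertex shader to force mip 0.
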MaterialFunction /BaseMaterial/Materials/Functions/MF_Coordinate_CellBombing
graph {
 properties {bExposeToLibrary=true, libraryCategoriesText [Base Material]}
 nodes [
  n1 = FunctionInput(Preview=TextureCoordinate()) {bUsePreviewValueAsDefault=true, inputName=Coordinates, inputType=FunctionInput_Vector2}
  n2 = FunctionInput() {bUsePreviewValueAsDefault=true, inputName=CellSize, inputType=FunctionInput_Scalar, sortPriority=5, previewValue=<0.1 1 0 1>}
  n3 = Multiply(A=@n1, B=@n2) {constB=0.005}
  n4 = TextureSample!(UVs=Frac(None=@n3), Tex, "DDX(UVs)"=DDX(Value=@n3), "DDY(UVs)"=DDY(Value=@n3), Apply View MipBias) {
   mipValueMode=TMVM_Derivative
   samplerSource=SSM_Wrap_WorldGroupSettings
   samplerType=SAMPLERTYPE_LinearColor
   texture=/BaseMaterial/Textures/Noises/T_Voronoi_Perturbed_1k_RGBA8_nonVT
  }
  n5 = FunctionInput() {bUsePreviewValueAsDefault=true, inputName=BlendFactor, inputType=FunctionInput_Scalar, sortPriority=35, previewValue=<0.3 1 0 1>}
  n6 = Multiply(A=OneMinus(None=@n4.A), B=Multiply(A=@n5) {constB=10}) {constB=3}
  FunctionOutput(None=OneMinus(None=Saturate(None=@n6))) {outputName=SampleMask, sortPriority=30}
  n7 = DDX(Value=@n1)
  n8 = DDY(Value=@n1)
  FunctionOutput(None=ShaderStageSwitch(PixelShader=AppendVector(A=@n7, B=@n8), VertexShader=Constant4Vector() {desc=force mip 0 in vs}.RGBA)) {outputName=cellBomb_DXY, sortPriority=10}
  n9 = ConstantBiasScale(None=@n4.R) {bias=-0.5, scale=10}
  MaterialExpressionFunctionInput_2 = FunctionInput() {
   bUsePreviewValueAsDefault=true
   description="RandomAmount in R channel, existing rotation in G channel"
   inputName="Rotation (PerCellRand/Base)"
   inputType=FunctionInput_Vector2
   sortPriority=30
  }
  n10 = Convert(None=@MaterialExpressionFunctionInput_2) {
   nodeName=Break Float2
   convertInputs [{
     type=Vector2
     expressionInput {
      expression=&self:MaterialExpressionFunctionInput_2
      inputName=None
      mask=0
      maskA=0
      maskB=0
      maskG=0
      maskR=0
      outputIndex=0
     }
    }]
   convertOutputs [{type=Scalar}, {type=Scalar}]
  }
  n11 = Add(A=Frac(None=Multiply(A=@n9, B=@n10)), B=@n10)
  n12 = Subtract(A=@n6, B=FunctionInput() {bUsePreviewValueAsDefault=true, inputName=Dither, inputType=FunctionInput_Scalar, sortPriority=40})
  n13 = Saturate(None=Ceil(None=Max(A=@n12) {constB=0}))
  n14 = LinearInterpolate(A=@n11, B=@n10, Alpha=@n13)
  FunctionOutput(None=@n14) {outputName=cellBomb_Rotation, sortPriority=20}
  MaterialExpressionFunctionInput_1 = FunctionInput() {bUsePreviewValueAsDefault=true, inputName=ScaleRandom, inputType=FunctionInput_Vector2, sortPriority=10, previewValue=<1 1 0 1>}
  n15 = Convert(None=@MaterialExpressionFunctionInput_1) {
   nodeName=Break Float2
   convertInputs [{
     type=Vector2
     expressionInput {
      expression=&self:MaterialExpressionFunctionInput_1
      inputName=None
      mask=0
      maskA=0
      maskB=0
      maskG=0
      maskR=0
      outputIndex=0
     }
    }]
   convertOutputs [{type=Scalar}, {type=Scalar}]
  }
  n16 = ComponentMask(None=@n4.RGB) {g=true, r=true}
  n17 = Add(
   A=Multiply(A=LinearInterpolate(A=@n15, B=@n15, Alpha=@n4.G) {constA=0.5, constB=1.5, desc=random scale}, B=@n1)
   B=Multiply(
    A=FunctionInput() {bUsePreviewValueAsDefault=true, inputName=OffsetRandom, inputType=FunctionInput_Vector2, sortPriority=20}
    B=Multiply(A=@n16, B=@n4.B)
   ) {constA=10, constB=100, desc=random offset}
  )
  n18 = Constant2Vector() {g=0.5, r=0.5}
  FunctionOutput(None=MaterialFunctionCall!(RotationAmount=@n14, Coords=LinearInterpolate(A=@n17, B=@n1, Alpha=@n13), Pivot=@n18) {materialFunction=&/MF_Rotate2D}.Result) {outputName=cellBomb_UV_Rot, sortPriority=1}
  FunctionOutput(None=MaterialFunctionCall(RotationAmount=@n11, Coords=@n17, Pivot=@n18) {materialFunction=&MF_Rotate2D}.Result) {outputName=cellBomb_CellUVOnly, sortPriority=40}
  FunctionOutput(None=@n11) {outputName=cellBomb_CellTurnsOnly, sortPriority=41}
 ]
}
