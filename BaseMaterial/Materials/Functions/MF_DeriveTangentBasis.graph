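# Material function, exposed to the library, that derives a per-pixel world-space Tangent and Bitangent from screen-space derivatives, giving a tangent basis for a chosen UV set.
# Cross products of world position derivatives (excluding shader offsets) with the vertex normal divided by object scale form two perpendicular vectors.
# The packedDXY input, defaulting to packed UV1 DDX/DDY, weights and sums these vectors, and each result is normalized separately.
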MaterialFunction /BaseMaterial/Materials/Functions/MF_DeriveTangentBasis
graph {
 expressions {
  MaterialExpressionAdd_10 {
   class=MaterialExpressionAdd
   inputs [
    A {from=MaterialExpressionMultiply_37}
    B {from=MaterialExpressionMultiply_38}
   ]
   properties {desc=T}
  }
  MaterialExpressionAdd_11 {
   class=MaterialExpressionAdd
   inputs [
    A {from=MaterialExpressionMultiply_39}
    B {from=MaterialExpressionMultiply_40}
   ]
   properties {desc=B}
  }
  MaterialExpressionConvert_2 {
   class=MaterialExpressionConvert
   inputs [None {from=MaterialExpressionFunctionInput_0}]
   properties {
    nodeName=Break Float4
    convertInputs [{
      type=Vector4
      defaultValue=<rgba 0 0 0 1>
      expressionInput {
       expression=&self:MaterialExpressionFunctionInput_0
       inputName=None
       mask=0
       maskA=0
       maskB=0
       maskG=0
       maskR=0
       outputIndex=0
      }
     }]
    convertOutputs [{type=Scalar}, {type=Scalar}, {type=Scalar}, {type=Scalar}]
   }
  }
  MaterialExpressionConvert_3 {
   class=MaterialExpressionConvert
   inputs [
    None {from=MaterialExpressionDDX_12}
    None {from=MaterialExpressionDDY_13}
   ]
   properties {
    convertInputs [
     {
      type=Vector2
      expressionInput {
       expression=&self:MaterialExpressionDDX_12
       inputName=None
       mask=0
       maskA=0
       maskB=0
       maskG=0
       maskR=0
       outputIndex=0
      }
     }
     {
      type=Vector2
      expressionInput {
       expression=&self:MaterialExpressionDDY_13
       inputName=None
       mask=0
       maskA=0
       maskB=0
       maskG=0
       maskR=0
       outputIndex=0
      }
     }
    ]
    convertOutputs [{type=Vector4, defaultValue=<rgba 0 0 0 1>}]
   }
  }
  MaterialExpressionCrossProduct_12 {
   class=MaterialExpressionCrossProduct
   inputs [
    A {from=MaterialExpressionDDY_12, output=""}
    B {from=MaterialExpressionDivide_0, output=""}
   ]
   properties {desc=dp2perp}
  }
  MaterialExpressionCrossProduct_13 {
   class=MaterialExpressionCrossProduct
   inputs [
    A {from=MaterialExpressionDivide_0, output=""}
    B {from=MaterialExpressionDDX_13, output=""}
   ]
   properties {desc=dp1perp}
  }
  MaterialExpressionDDX_12 {class=MaterialExpressionDDX, inputs [Value {from=MaterialExpressionTextureCoordinate_0, output=""}], properties {desc=duv1}}
  MaterialExpressionDDX_13 {class=MaterialExpressionDDX, inputs [Value {from=MaterialExpressionWorldPosition_5, output=XYZ}], properties {desc=dp1}}
  MaterialExpressionDDY_12 {class=MaterialExpressionDDY, inputs [Value {from=MaterialExpressionWorldPosition_5, output=XYZ}], properties {desc=dp2}}
  MaterialExpressionDDY_13 {class=MaterialExpressionDDY, inputs [Value {from=MaterialExpressionTextureCoordinate_0, output=""}], properties {desc=duv2}}
  MaterialExpressionDivide_0 {
   class=MaterialExpressionDivide
   inputs [
    A {from=MaterialExpressionVertexNormalWS_2}
    B {from=MaterialExpressionMaterialFunctionCall_3, output=Scale XYZ}
   ]
  }
  MaterialExpressionFunctionInput_0 {
   class=MaterialExpressionFunctionInput
   inputs [Preview {from=MaterialExpressionConvert_3, output=""}]
   properties {
    bUsePreviewValueAsDefault=true
    description=Which UV set to use when creating new tangent basis.
    inputName=packedDXY
    inputType=FunctionInput_Vector4
    sortPriority=1
   }
  }
  MaterialExpressionFunctionOutput_8 {
   class=MaterialExpressionFunctionOutput
   inputs [None {from=MaterialExpressionNormalize_0, output=""}]
   properties {outputName=Tangent, sortPriority=2}
  }
  MaterialExpressionFunctionOutput_9 {
   class=MaterialExpressionFunctionOutput
   inputs [None {from=MaterialExpressionNormalize_1, output=""}]
   properties {outputName=Bitangent, sortPriority=3}
  }
  MaterialExpressionMaterialFunctionCall_3 {
   class=MaterialExpressionMaterialFunctionCall
   inputs []
   properties {materialFunction=&ObjectScale}
  }
  MaterialExpressionMultiply_37 {
   class=MaterialExpressionMultiply
   inputs [
    A {from=MaterialExpressionCrossProduct_12}
    B {from=MaterialExpressionConvert_2}
   ]
  }
  MaterialExpressionMultiply_38 {
   class=MaterialExpressionMultiply
   inputs [
    A {from=MaterialExpressionCrossProduct_13}
    B {from=MaterialExpressionConvert_2}
   ]
  }
  MaterialExpressionMultiply_39 {
   class=MaterialExpressionMultiply
   inputs [
    A {from=MaterialExpressionCrossProduct_12}
    B {from=MaterialExpressionConvert_2}
   ]
  }
  MaterialExpressionMultiply_40 {
   class=MaterialExpressionMultiply
   inputs [
    A {from=MaterialExpressionCrossProduct_13}
    B {from=MaterialExpressionConvert_2}
   ]
  }
  MaterialExpressionNormalize_0 {class=MaterialExpressionNormalize, inputs [VectorInput {from=MaterialExpressionAdd_10, output=""}]}
  MaterialExpressionNormalize_1 {class=MaterialExpressionNormalize, inputs [VectorInput {from=MaterialExpressionAdd_11, output=""}]}
  MaterialExpressionTextureCoordinate_0 {class=MaterialExpressionTextureCoordinate, properties {coordinateIndex=1}}
  MaterialExpressionVertexNormalWS_2 {class=MaterialExpressionVertexNormalWS, inputs [], properties {}}
  MaterialExpressionWorldPosition_5 {class=MaterialExpressionWorldPosition, inputs [], properties {worldPositionShaderOffset=WPT_ExcludeAllShaderOffsets}}
 }
 properties {bExposeToLibrary=true, description=Derive a tangent basis from derivatives., libraryCategoriesText [Base Material]}
}
