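# Material function that selects one of four Vector2 inputs (VectorA to VectorD) with a scalar Selector.
# Three nested If nodes compare Selector to thresholds 1, 2 and 3, so zero-based ranges map to A, B, C and D in order.
# The equals pins are unconnected, so a value on a threshold follows the greater-than branch.
# All inputs use their preview values as defaults, making each one optional.
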MaterialFunction /BaseMaterial/Materials/Functions/MF_Switch4_Vec2
graph {nodes [
  n1 = FunctionInput() {bUsePreviewValueAsDefault=true, inputName=Selector, inputType=FunctionInput_Scalar}
  n2 = FunctionInput() {bUsePreviewValueAsDefault=true, inputName=VectorD, inputType=FunctionInput_Vector2, sortPriority=4, previewValue=<1 1 0 1>}
  n3 = If(
   A=@n1
   "A > B"=@n2
   "A < B"=FunctionInput() {bUsePreviewValueAsDefault=true, inputName=VectorC, inputType=FunctionInput_Vector2, sortPriority=3}
  ) {constB=3}
  n4 = FunctionInput() {bUsePreviewValueAsDefault=true, inputName=VectorB, inputType=FunctionInput_Vector2, sortPriority=2, previewValue=<0 1 0 1>}
  n5 = FunctionInput() {bUsePreviewValueAsDefault=true, inputName=VectorA, inputType=FunctionInput_Vector2, sortPriority=1, previewValue=<1 0 0 1>}
  FunctionOutput(None=If(A=@n1, "A > B"=If(A=@n1, "A > B"=@n3, "A < B"=@n4) {constB=2}, "A < B"=@n5) {constB=1}) {outputName=Selected Vector}
 ]}
