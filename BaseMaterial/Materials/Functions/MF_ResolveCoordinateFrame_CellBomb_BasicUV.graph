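# Library material function that turns a selected mesh UV channel into two coordinate sets, base and cell-bombed, for breaking up texture tiling.
# MF_CoordinateTransform_2D applies repeat and offset, then MF_Coordinate_CellBombing adds per-cell random scale, offset and rotation.
# Both sets share the cell-bomb derivatives for SampleGrad and one tangent basis derived from them, with separate rotation outputs.
# An RG BlendMask built from the sample mask and its inverse supports post-sample blending.
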MaterialFunction /BaseMaterial/Materials/Functions/MF_ResolveCoordinateFrame_CellBomb_BasicUV
graph {
 expressions {
  MaterialExpressionConvert_0 {
   class=MaterialExpressionConvert
   inputs [
    None {from=MaterialExpressionFunctionInput_6}
    None {from=MaterialExpressionNamedRerouteDeclaration_2}
   ]
   properties {
    nodeName=Make Float2
    convertInputs [
     {
      type=Scalar
      expressionInput {
       expression=&self:MaterialExpressionFunctionInput_6
       inputName=None
       mask=0
       maskA=0
       maskB=0
       maskG=0
       maskR=0
       outputIndex=0
      }
     }
     {
      type=Scalar
      expressionInput {
       expression=&self:MaterialExpressionNamedRerouteDeclaration_2
       inputName=None
       mask=0
       maskA=0
       maskB=0
       maskG=0
       maskR=0
       outputIndex=0
      }
     }
    ]
    convertOutputs [{type=Vector2}]
   }
  }
  MaterialExpressionConvert_1 {
   class=MaterialExpressionConvert
   inputs [
    None {from=MaterialExpressionOneMinus_0}
    None {from=MaterialExpressionNamedRerouteUsage_32}
   ]
   properties {
    nodeName=Make Float2
    convertInputs [
     {
      type=Scalar
      expressionInput {
       expression=&self:MaterialExpressionOneMinus_0
       inputName=None
       mask=0
       maskA=0
       maskB=0
       maskG=0
       maskR=0
       outputIndex=0
      }
     }
     {
      type=Scalar
      expressionInput {
       expression=&self:MaterialExpressionNamedRerouteUsage_32
       inputName=None
       mask=0
       maskA=0
       maskB=0
       maskG=0
       maskR=0
       outputIndex=0
      }
     }
    ]
    convertOutputs [{type=Vector2}]
   }
  }
  MaterialExpressionFunctionInput_0 {
   class=MaterialExpressionFunctionInput
   inputs [Preview {from="", output=""}]
   properties {
    bUsePreviewValueAsDefault=true
    description="Sets how many times the projected coordinates repeat along each axis. Use values above 1 for denser tiling, negative values for mirroring, and avoid zero."
    inputName=CoordinateRepeat
    inputType=FunctionInput_Vector2
    sortPriority=-10
    previewValue=<1 1 1 1>
   }
  }
  MaterialExpressionFunctionInput_1 {
   class=MaterialExpressionFunctionInput
   inputs [Preview {from="", output=""}]
   properties {
    bUsePreviewValueAsDefault=true
    description="Rotates the projected UVs around the selected axis. Units are in turns, where 1.0 = 360° and 0.25 = 90°."
    inputName="Rotation (Turns)"
    inputType=FunctionInput_Scalar
    sortPriority=-8
   }
  }
  MaterialExpressionFunctionInput_13 {
   class=MaterialExpressionFunctionInput
   inputs [Preview {from="", output=""}]
   properties {
    bUsePreviewValueAsDefault=true
    description=Applies an additional per-axis offset to the coordinates before projection. Useful for correctly positioning a texture.
    inputName=CoordinateTransform
    inputType=FunctionInput_Vector2
    sortPriority=-9
   }
  }
  MaterialExpressionFunctionInput_22 {
   class=MaterialExpressionFunctionInput
   inputs [Preview {from="", output=""}]
   properties {
    bUsePreviewValueAsDefault=true
    description=Controls the sharpness of the transition between coordinate sets.
    inputName=BlendTolerance
    inputType=FunctionInput_Scalar
    sortPriority=8
    previewValue=<0.5 0 0 1>
   }
  }
  MaterialExpressionFunctionInput_3 {
   class=MaterialExpressionFunctionInput
   inputs [Preview {from="", output=""}]
   properties {
    bUsePreviewValueAsDefault=true
    description="Controls tiling of the cells texture in UV space. Higher values create smaller, more numerous cells."
    inputName=Cell PatternSize
    inputType=FunctionInput_Scalar
    sortPriority=2
    previewValue=<0.01 0 0 1>
   }
  }
  MaterialExpressionFunctionInput_4 {
   class=MaterialExpressionFunctionInput
   inputs [Preview {from="", output=""}]
   properties {
    bUsePreviewValueAsDefault=true
    description="Per-cell random scale range (min/max). Values below 1 enlarge, above 1 shrink. Applied uniformly per cell."
    inputName=Cell Min/MaxRand Scale
    inputType=FunctionInput_Vector2
    sortPriority=3
    previewValue=<1 1 0 1>
   }
  }
  MaterialExpressionFunctionInput_5 {
   class=MaterialExpressionFunctionInput
   inputs [Preview {from="", output=""}]
   properties {
    bUsePreviewValueAsDefault=true
    description="Per-cell random UV offset range (min/max). Shifts sampling position within each cell."
    inputName=Cell Min/MaxRand Offset
    inputType=FunctionInput_Vector2
    sortPriority=4
   }
  }
  MaterialExpressionFunctionInput_6 {
   class=MaterialExpressionFunctionInput
   inputs [Preview {from="", output=""}]
   properties {
    bUsePreviewValueAsDefault=true
    description="Maximum per-cell random rotation (in turns, 1.0 = 360°). The actual rotation per cell is drawn within this magnitude."
    inputName="Cell Rand Rotation (turns)"
    inputType=FunctionInput_Scalar
    sortPriority=5
   }
  }
  MaterialExpressionFunctionInput_7 {
   class=MaterialExpressionFunctionInput
   inputs [Preview {from="", output=""}]
   properties {
    bUsePreviewValueAsDefault=true
    description="Selects which mesh UV channel (0–3) to use as the base coordinate source."
    inputName=UVIndex
    inputType=FunctionInput_Scalar
    sortPriority=-5
   }
  }
  MaterialExpressionFunctionOutput_1 {
   class=MaterialExpressionFunctionOutput
   inputs [None {from=MaterialExpressionMaterialFunctionCall_3, output=geometry_UV}]
   properties {description="Base UVs sampled in the same coordinate frame as the cells texture (unrandomized reference).", outputName=UV1}
  }
  MaterialExpressionFunctionOutput_10 {
   class=MaterialExpressionFunctionOutput
   inputs [None {from=MaterialExpressionMaterialFunctionCall_0, output=Tangent}]
   properties {description=World-space tangent for UV2., outputName=Tangent2, sortPriority=6}
  }
  MaterialExpressionFunctionOutput_11 {
   class=MaterialExpressionFunctionOutput
   inputs [None {from=MaterialExpressionMaterialFunctionCall_0, output=Bitangent}]
   properties {description=World-space bitangent for UV2., outputName=Bitangent2, sortPriority=7}
  }
  MaterialExpressionFunctionOutput_14 {
   class=MaterialExpressionFunctionOutput
   inputs [None {from=MaterialExpressionMultiply_1, output=""}]
   properties {description="Rotation for UV1 in turns (1.0 = 360°).", outputName="Rotation1 (Turns)", sortPriority=-6}
  }
  MaterialExpressionFunctionOutput_2 {
   class=MaterialExpressionFunctionOutput
   inputs [None {from=MaterialExpressionMaterialFunctionCall_1, output=cellBomb_DXY}]
   properties {description="Packed derivatives for UV1 (ddx/ddy) for SampleGrad.", outputName=DXY1, sortPriority=3}
  }
  MaterialExpressionFunctionOutput_3 {
   class=MaterialExpressionFunctionOutput
   inputs [None {from=MaterialExpressionMaterialFunctionCall_1, output=cellBomb_CellUVOnly}]
   properties {description="UVs randomized per cell (offset/scale/rotation applied).", outputName=UV2, sortPriority=5}
  }
  MaterialExpressionFunctionOutput_4 {
   class=MaterialExpressionFunctionOutput
   inputs [None {from=MaterialExpressionMaterialFunctionCall_1, output=cellBomb_DXY}]
   properties {description="Packed derivatives for UV2 (ddx/ddy) for SampleGrad.", outputName=DXY2, sortPriority=8}
  }
  MaterialExpressionFunctionOutput_5 {
   class=MaterialExpressionFunctionOutput
   inputs [None {from=MaterialExpressionMaterialFunctionCall_1, output=cellBomb_CellTurnsOnly}]
   properties {description="Rotation for UV2 in turns (1.0 = 360°).", outputName="Rotation2 (Turns)", sortPriority=-5}
  }
  MaterialExpressionFunctionOutput_7 {
   class=MaterialExpressionFunctionOutput
   inputs [None {from=MaterialExpressionConvert_1, output=""}]
   properties {
    description="RG weights for blending UV1 (base) and UV2 (randomized) post-sample; transition sharpness is set by BlendFactor. Also supports albedo/height-driven modulation."
    outputName=BlendMask
    sortPriority=15
   }
  }
  MaterialExpressionFunctionOutput_8 {
   class=MaterialExpressionFunctionOutput
   inputs [None {from=MaterialExpressionMaterialFunctionCall_0, output=Tangent}]
   properties {description=World-space tangent for UV1., outputName=Tangent1, sortPriority=1}
  }
  MaterialExpressionFunctionOutput_9 {
   class=MaterialExpressionFunctionOutput
   inputs [None {from=MaterialExpressionMaterialFunctionCall_0, output=Bitangent}]
   properties {description=World-space bitangent for UV1., outputName=Bitangent1, sortPriority=2}
  }
  MaterialExpressionMaterialFunctionCall_0 {
   class=MaterialExpressionMaterialFunctionCall
   inputs [packedDXY {from=MaterialExpressionFunctionOutput_4}]
   properties {materialFunction=&/MF_DeriveTangentBasis}
  }
  MaterialExpressionMaterialFunctionCall_1 {
   class=MaterialExpressionMaterialFunctionCall
   inputs [
    Coordinates {from=MaterialExpressionNamedRerouteUsage_5}
    CellSize {from=MaterialExpressionFunctionInput_3}
    ScaleRandom {from=MaterialExpressionFunctionInput_4}
    OffsetRandom {from=MaterialExpressionFunctionInput_5}
    "Rotation (PerCellRand/Base)" {from=MaterialExpressionConvert_0}
    BlendFactor {from=MaterialExpressionFunctionInput_22}
    Dither {from=""}
   ]
   properties {materialFunction=&/MF_Coordinate_CellBombing}
  }
  MaterialExpressionMaterialFunctionCall_2 {
   class=MaterialExpressionMaterialFunctionCall
   inputs [
    Selector {from=MaterialExpressionFunctionInput_7}
    VectorA {from=MaterialExpressionTextureCoordinate_0}
    VectorB {from=MaterialExpressionTextureCoordinate_1}
    VectorC {from=MaterialExpressionTextureCoordinate_2}
    VectorD {from=MaterialExpressionTextureCoordinate_3}
   ]
   properties {materialFunction=&/MF_Switch4_Vec2}
  }
  MaterialExpressionMaterialFunctionCall_3 {
   class=MaterialExpressionMaterialFunctionCall
   inputs [
    Coords {from=MaterialExpressionMaterialFunctionCall_2, output=Selected Vector}
    UVOffset {from=MaterialExpressionFunctionInput_13}
    UVRotation {from=""}
    RotationPivot {from=""}
    UVTile {from=MaterialExpressionFunctionInput_0}
   ]
   properties {materialFunction=&/MF_CoordinateTransform_2D}
  }
  MaterialExpressionMultiply_1 {
   class=MaterialExpressionMultiply
   inputs [
    A {from=MaterialExpressionNamedRerouteDeclaration_2}
    B {from=""}
   ]
   properties {constB=-1}
  }
  MaterialExpressionNamedRerouteDeclaration_1 {
   class=MaterialExpressionNamedRerouteDeclaration
   inputs [None {from=MaterialExpressionMaterialFunctionCall_1, output=SampleMask}]
   properties {name=Mask, nodeColor=<rgba 0 1 0.3906 1>}
  }
  MaterialExpressionNamedRerouteDeclaration_2 {
   class=MaterialExpressionNamedRerouteDeclaration
   inputs [None {from=MaterialExpressionFunctionInput_1}]
   properties {name=turnsi, nodeColor=<rgba 1 0.7031 0 1>}
  }
  MaterialExpressionNamedRerouteDeclaration_6 {
   class=MaterialExpressionNamedRerouteDeclaration
   inputs [None {from=MaterialExpressionMaterialFunctionCall_3, output=geometry_UV}]
   properties {name=UVi, nodeColor=<rgba 0.9 0.009 0.009 1>}
  }
  MaterialExpressionNamedRerouteUsage_32 {class=MaterialExpressionNamedRerouteUsage}
  MaterialExpressionNamedRerouteUsage_5 {class=MaterialExpressionNamedRerouteUsage}
  MaterialExpressionNamedRerouteUsage_8 {class=MaterialExpressionNamedRerouteUsage}
  MaterialExpressionOneMinus_0 {class=MaterialExpressionOneMinus, inputs [None {from=MaterialExpressionNamedRerouteUsage_8}]}
  MaterialExpressionTextureCoordinate_0 {class=MaterialExpressionTextureCoordinate}
  MaterialExpressionTextureCoordinate_1 {class=MaterialExpressionTextureCoordinate, properties {coordinateIndex=1}}
  MaterialExpressionTextureCoordinate_2 {class=MaterialExpressionTextureCoordinate, properties {coordinateIndex=2}}
  MaterialExpressionTextureCoordinate_3 {class=MaterialExpressionTextureCoordinate, properties {coordinateIndex=3}}
 }
 properties {
  bExposeToLibrary=true
  description="Randomizes UVs per cell using a cells texture (per-cell offset, scale, rotation) and outputs two full projection sets: UV1 (base coordinates) and UV2 (randomized per cell), each with tangent/bitangent/rotation/derivatives. A BlendMask (RG) is also provided for blending the two sets, with transition sharpness controlled by BlendFactor. Because blending is done post-sample, weights can also be modulated by albedo or height, enabling perceptually smoother blends across cell edges."
  userExposedCaption="CoordinateBasis: Cell Bomb (Standard)"
  libraryCategoriesText [Base Material]
 }
}
