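# Material function that outputs per-pixel, temporally varying scalar noise in the 0 to 1 range.
# It computes interleaved gradient noise from PixelPosition, which defaults to screen pixel position.
# TemporalSampleIndex, defaulting to the frame index modulo 64, times the golden ratio offsets that noise before a final frac, giving a low-discrepancy sequence over time.
# The math is built from nodes, with an unconnected Custom node holding reference HLSL from RandomInterleavedGradientNoise.ush.
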MaterialFunction /BaseMaterial/Materials/Functions/MF_Noise_InterleavedGradientGolden_1d
graph {
 properties {bExposeToLibrary=true, libraryCategoriesText [Base Material]}
 nodes [
  Custom!(frame, pixel) {
   code="//HLSL for Ref\r\n//From RandomInterleavedGradientNoise.ush\r\nconst float3 magic = float3( 0.06711056f, 0.00583715f, 52.9829189f ); \r\nfloat ign = frac(magic.z * frac(dot(pixel, magic.xy)));\r\n//Add golden ratio for low discrepancy\r\nfloat gr = frame * (1.0f + sqrt(5.0f)) / 2.0f;\r\n\r\nreturn frac(ign + gr);"
   description=InterleavedGradientNoise
   outputType=CMOT_Float1
   showCode=true
   inputs [{inputName=frame, input {expression=null}}, {inputName=pixel, input {expression=null}}]
  }
  n1 = Constant() {r=1}
  n2 = SquareRoot(None=Constant() {r=5})
  n3 = Multiply(
   A=FunctionInput(Preview=Custom!() {code="View.StateFrameIndex%64;", description=ViewStateFrameIndexMod64, outputType=CMOT_Float1, inputs []}) {bUsePreviewValueAsDefault=true, inputName=TemporalSampleIndex, inputType=FunctionInput_Scalar}
   B=Multiply(A=Add(A=@n1, B=@n2)) {constB=0.5}
  ) {desc=gr}
  n4 = Constant3Vector() {constant=<rgba 0.0671 0.0058 52.9829 0>}
  n5 = DotProduct(
   A=FunctionInput(Preview=ScreenPosition().PixelPosition) {bUsePreviewValueAsDefault=true, inputName=PixelPosition, inputType=FunctionInput_Vector2, sortPriority=1}
   B=ComponentMask(None=@n4) {g=true, r=true}
  )
  n6 = Frac(None=Multiply(A=Frac(None=@n5), B=@n4.B)) {desc=ign}
  FunctionOutput(None=Frac(None=Add(A=@n3, B=@n6))) {outputName=Unsigned}
 ]
}
