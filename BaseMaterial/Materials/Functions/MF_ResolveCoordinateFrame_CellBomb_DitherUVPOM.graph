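# Material function that builds a UV-based coordinate frame with dithered per-cell randomization (cell bombing) and parallax occlusion mapping.
# A selected mesh UV channel is offset and tiled, then cell-bombed with random scale, offset and rotation plus base rotation.
# Dither, defaulting to golden-ratio interleaved gradient noise, approximates blending cheaply.
# The cell derivatives yield a tangent basis for parallax, and the function outputs UV, depth, derivatives, tangent, bitangent and rotation.
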MaterialFunction /BaseMaterial/Materials/Functions/MF_ResolveCoordinateFrame_CellBomb_DitherUVPOM
graph {
 expressions {
  MaterialExpressionConvert_0 {
   class=MaterialExpressionConvert
   inputs [
    None {from=MaterialExpressionFunctionInput_6}
    None {from=MaterialExpressionFunctionInput_1}
   ]
   properties {
    nodeName=Make Float2
    convertInputs [
     {
      type=Scalar
      expressionInput {
       expression=&self:MaterialExpressionFunctionInput_6
       inputName=None
       mask=0
       maskA=0
       maskB=0
       maskG=0
       maskR=0
       outputIndex=0
      }
     }
     {
      type=Scalar
      expressionInput {
       expression=&self:MaterialExpressionFunctionInput_1
       inputName=None
       mask=0
       maskA=0
       maskB=0
       maskG=0
       maskR=0
       outputIndex=0
      }
     }
    ]
    convertOutputs [{type=Vector2}]
   }
  }
  MaterialExpressionFunctionInput_0 {
   class=MaterialExpressionFunctionInput
   inputs [Preview {from="", output=""}]
   properties {
    bUsePreviewValueAsDefault=true
    description="Sets how many times the projected coordinates repeat along each axis. Use values above 1 for denser tiling, negative values for mirroring, and avoid zero."
    inputName=CoordinateRepeat
    inputType=FunctionInput_Vector2
    sortPriority=-10
    previewValue=<1 1 1 1>
   }
  }
  MaterialExpressionFunctionInput_1 {
   class=MaterialExpressionFunctionInput
   inputs [Preview {from="", output=""}]
   properties {
    bUsePreviewValueAsDefault=true
    description="Rotates the projected UVs around the selected axis. Units are in turns, where 1.0 = 360° and 0.25 = 90°."
    inputName="Rotation (Turns)"
    inputType=FunctionInput_Scalar
    sortPriority=-8
   }
  }
  MaterialExpressionFunctionInput_11 {
   class=MaterialExpressionFunctionInput
   inputs [Preview {from="", output=""}]
   properties {
    bUsePreviewValueAsDefault=true
    description="Specifies which channel of the heightmap texture to use for depth information. Assumes the Alpha channel is used, but any channel can be selected if your texture packs data differently."
    inputName=POM HeightChannel
    inputType=FunctionInput_Scalar
    sortPriority=7
    previewValue=<3 0 0 1>
   }
  }
  MaterialExpressionFunctionInput_12 {
   class=MaterialExpressionFunctionInput
   inputs [Preview {from="", output=""}]
   properties {
    bUsePreviewValueAsDefault=true
    description="Sets the maximum number of depth sampling steps. This is the upper bound, with actual steps scaled by the Quality setting."
    inputName=POM MaxSteps
    inputType=FunctionInput_Scalar
    sortPriority=9
    previewValue=<16 0 0 1>
   }
  }
  MaterialExpressionFunctionInput_13 {
   class=MaterialExpressionFunctionInput
   inputs [Preview {from="", output=""}]
   properties {
    bUsePreviewValueAsDefault=true
    description=Applies an additional per-axis offset to the coordinates before projection. Useful for correctly positioning a texture.
    inputName=CoordinateTransform
    inputType=FunctionInput_Vector2
    sortPriority=-9
   }
  }
  MaterialExpressionFunctionInput_14 {
   class=MaterialExpressionFunctionInput
   inputs [Preview {from="", output=""}]
   properties {
    bUsePreviewValueAsDefault=true
    description="Controls the strength of the parallax effect. Higher values exaggerate depth; lower values make surfaces appear flatter."
    inputName=POM Intensity
    inputType=FunctionInput_Scalar
    sortPriority=8
    previewValue=<1 0 0 1>
   }
  }
  MaterialExpressionFunctionInput_15 {
   class=MaterialExpressionFunctionInput
   inputs [Preview {from="", output=""}]
   properties {
    bUsePreviewValueAsDefault=true
    description="Scales step count by view angle. At 1, MaxSteps are always used; at 0, only one step is used when looking straight on, with more steps added at grazing angles."
    inputName=POM Quality
    inputType=FunctionInput_Scalar
    sortPriority=10
    previewValue=<0.25 0 0 1>
   }
  }
  MaterialExpressionFunctionInput_16 {
   class=MaterialExpressionFunctionInput
   inputs [Preview {from="", output=""}]
   properties {
    bUsePreviewValueAsDefault=true
    description=Defines the reference plane for parallax. This value shifts where the heightmap appears anchored relative to the surface.
    inputName=POM RefPlane
    inputType=FunctionInput_Scalar
    sortPriority=11
    previewValue=<1 0 0 1>
   }
  }
  MaterialExpressionFunctionInput_17 {
   class=MaterialExpressionFunctionInput
   inputs [Preview {from=MaterialExpressionMaterialFunctionCall_7, output=Unsigned}]
   properties {
    bUsePreviewValueAsDefault=true
    description="Dithering sequence input; defaults to golden-ratio interleaved gradient noise (slow-converging), but can be overridden with custom patterns."
    inputName=Dither
    inputType=FunctionInput_Scalar
   }
  }
  MaterialExpressionFunctionInput_2 {
   class=MaterialExpressionFunctionInput
   inputs [Preview {from="", output=""}]
   properties {
    bUsePreviewValueAsDefault=true
    description=Control the U/V Tiling of the height texture relative to the input coordinates.
    inputName=POM HeightTile
    inputType=FunctionInput_Vector2
    sortPriority=12
    previewValue=<1 1 0 1>
   }
  }
  MaterialExpressionFunctionInput_25 {
   class=MaterialExpressionFunctionInput
   inputs [Preview {from=MaterialExpressionTextureObject_3, output=""}]
   properties {
    bUsePreviewValueAsDefault=true
    description="The heightmap texture used for parallax occlusion mapping. Grayscale values define surface depth, with brighter values typically pushed outward and darker values pushed inward."
    inputName=HeightTex
    inputType=FunctionInput_Texture2D
    sortPriority=-20
   }
  }
  MaterialExpressionFunctionInput_3 {
   class=MaterialExpressionFunctionInput
   inputs [Preview {from="", output=""}]
   properties {
    bUsePreviewValueAsDefault=true
    description="Controls tiling of the cells texture in UV space. Higher values create smaller, more numerous cells."
    inputName=Cell PatternSize
    inputType=FunctionInput_Scalar
    sortPriority=3
    previewValue=<0.01 0 0 1>
   }
  }
  MaterialExpressionFunctionInput_4 {
   class=MaterialExpressionFunctionInput
   inputs [Preview {from="", output=""}]
   properties {
    bUsePreviewValueAsDefault=true
    description="Per-cell random scale range (min/max). Values below 1 enlarge, above 1 shrink. Applied uniformly per cell."
    inputName=Cell Min/MaxRand Scale
    inputType=FunctionInput_Vector2
    sortPriority=4
    previewValue=<1 1 0 1>
   }
  }
  MaterialExpressionFunctionInput_5 {
   class=MaterialExpressionFunctionInput
   inputs [Preview {from="", output=""}]
   properties {
    bUsePreviewValueAsDefault=true
    description="Per-cell random UV offset range (min/max). Shifts sampling position within each cell."
    inputName=Cell Min/MaxRand Offset
    inputType=FunctionInput_Vector2
    sortPriority=5
   }
  }
  MaterialExpressionFunctionInput_6 {
   class=MaterialExpressionFunctionInput
   inputs [Preview {from="", output=""}]
   properties {
    bUsePreviewValueAsDefault=true
    description="Maximum per-cell random rotation (in turns, 1.0 = 360°). The actual rotation per cell is drawn within this magnitude."
    inputName="Cell Rand Rotation (turns)"
    inputType=FunctionInput_Scalar
    sortPriority=6
   }
  }
  MaterialExpressionFunctionInput_7 {
   class=MaterialExpressionFunctionInput
   inputs [Preview {from="", output=""}]
   properties {
    bUsePreviewValueAsDefault=true
    description="Selects which mesh UV channel (0–3) to use as the base coordinate source."
    inputName=UVIndex
    inputType=FunctionInput_Scalar
   }
  }
  MaterialExpressionFunctionOutput_1 {
   class=MaterialExpressionFunctionOutput
   inputs [None {from=MaterialExpressionMaterialFunctionCall_5, output=POM UVs}]
   properties {
    description="The primary 2D coordinates generated by the projection. Feed these into texture samplers for surface detail.\r\n"
    outputName=UV
   }
  }
  MaterialExpressionFunctionOutput_14 {
   class=MaterialExpressionFunctionOutput
   inputs [None {from=MaterialExpressionReroute_3, output=""}]
   properties {
    description="The rotation applied around the projection axis, expressed in turns (1.0 = 360°). Used along with Tangent/Bitangent for reorienting normal maps via MF_TangentNormal_OrientWorld.\r\n"
    outputName="Rotation (Turns)"
    sortPriority=-5
   }
  }
  MaterialExpressionFunctionOutput_2 {
   class=MaterialExpressionFunctionOutput
   inputs [None {from=MaterialExpressionMaterialFunctionCall_1, output=cellBomb_DXY}]
   properties {
    description="Packed derivatives of the projected UVs (ddx/ddy). Use with SampleGrad to ensure textures sample with correct mip levels and anisotropic filtering.\r\n"
    outputName=DXY
    sortPriority=3
   }
  }
  MaterialExpressionFunctionOutput_7 {
   class=MaterialExpressionFunctionOutput
   inputs [None {from=MaterialExpressionMaterialFunctionCall_5, output=Pixel Depth Offset}]
   properties {
    description="Depth at the POM intersection (after Intensity/RefPlane). Higher (whiter) = deeper—the farther the ray travels into the material, the larger the value."
    outputName=Depth
    sortPriority=4
   }
  }
  MaterialExpressionFunctionOutput_8 {
   class=MaterialExpressionFunctionOutput
   inputs [None {from=MaterialExpressionMaterialFunctionCall_6, output=Tangent}]
   properties {
    description="The world-space tangent vector of the projection frame. Use with BitangentWS and Rotation to correctly reorient tangent-space normals.\r\n"
    outputName=Tangent
    sortPriority=1
   }
  }
  MaterialExpressionFunctionOutput_9 {
   class=MaterialExpressionFunctionOutput
   inputs [None {from=MaterialExpressionMaterialFunctionCall_6, output=Bitangent}]
   properties {
    description=The world-space bitangent vector of the projection frame. Complements TangentWS to form the basis for tangent-space normal reorientation.
    outputName=Bitangent
    sortPriority=2
   }
  }
  MaterialExpressionMaterialFunctionCall_0 {
   class=MaterialExpressionMaterialFunctionCall
   inputs [
    Selector {from=MaterialExpressionFunctionInput_7}
    VectorA {from=MaterialExpressionTextureCoordinate_4}
    VectorB {from=MaterialExpressionTextureCoordinate_5}
    VectorC {from=MaterialExpressionTextureCoordinate_6}
    VectorD {from=MaterialExpressionTextureCoordinate_7}
   ]
   properties {materialFunction=&/MF_Switch4_Vec2}
  }
  MaterialExpressionMaterialFunctionCall_1 {
   class=MaterialExpressionMaterialFunctionCall
   inputs [
    Coordinates {from=MaterialExpressionNamedRerouteUsage_5}
    CellSize {from=MaterialExpressionFunctionInput_3}
    ScaleRandom {from=MaterialExpressionFunctionInput_4}
    OffsetRandom {from=MaterialExpressionFunctionInput_5}
    "Rotation (PerCellRand/Base)" {from=MaterialExpressionConvert_0}
    BlendFactor {from=""}
    Dither {from=MaterialExpressionFunctionInput_17}
   ]
   properties {materialFunction=&/MF_Coordinate_CellBombing}
  }
  MaterialExpressionMaterialFunctionCall_3 {
   class=MaterialExpressionMaterialFunctionCall
   inputs [
    Coords {from=MaterialExpressionMaterialFunctionCall_0, output=Selected Vector}
    UVOffset {from=MaterialExpressionFunctionInput_13}
    UVRotation {from=""}
    RotationPivot {from=""}
    UVTile {from=MaterialExpressionFunctionInput_0}
   ]
   properties {materialFunction=&/MF_CoordinateTransform_2D}
  }
  MaterialExpressionMaterialFunctionCall_5 {
   class=MaterialExpressionMaterialFunctionCall
   inputs [
    POM_HeightTexture {from=MaterialExpressionNamedRerouteUsage_7}
    Coordinates {from=MaterialExpressionReroute_2}
    coordinateDXY {from=MaterialExpressionFunctionOutput_2}
    HeightStrength {from=MaterialExpressionFunctionInput_14}
    Reference Plane {from=MaterialExpressionFunctionInput_16}
    MaxSteps {from=MaterialExpressionFunctionInput_12}
    Tangent {from=MaterialExpressionFunctionOutput_8}
    Bitangent {from=MaterialExpressionFunctionOutput_9}
    "Rotation (Turns)" {from=MaterialExpressionFunctionOutput_14}
    HeightChannel {from=MaterialExpressionReroute_0}
    QualityBias {from=MaterialExpressionFunctionInput_15}
    CameraVector {from=""}
    Height Tile {from=MaterialExpressionFunctionInput_2}
   ]
   properties {materialFunction=&/MF_Coordinate_Parallax}
  }
  MaterialExpressionMaterialFunctionCall_6 {
   class=MaterialExpressionMaterialFunctionCall
   inputs [packedDXY {from=MaterialExpressionMaterialFunctionCall_1, output=cellBomb_DXY}]
   properties {materialFunction=&/MF_DeriveTangentBasis}
  }
  MaterialExpressionMaterialFunctionCall_7 {
   class=MaterialExpressionMaterialFunctionCall
   inputs [
    TemporalSampleIndex {from=""}
    PixelPosition {from=""}
   ]
   properties {materialFunction=&/MF_Noise_InterleavedGradientGolden_1d}
  }
  MaterialExpressionNamedRerouteDeclaration_19 {
   class=MaterialExpressionNamedRerouteDeclaration
   inputs [None {from=MaterialExpressionFunctionInput_25}]
   properties {name=Height, nodeColor=<rgba 1 0 0.4922 1>}
  }
  MaterialExpressionNamedRerouteDeclaration_6 {
   class=MaterialExpressionNamedRerouteDeclaration
   inputs [None {from=MaterialExpressionMaterialFunctionCall_3, output=geometry_UV}]
   properties {name=UVi, nodeColor=<rgba 0.9 0.009 0.009 1>}
  }
  MaterialExpressionNamedRerouteUsage_5 {class=MaterialExpressionNamedRerouteUsage}
  MaterialExpressionNamedRerouteUsage_7 {class=MaterialExpressionNamedRerouteUsage}
  MaterialExpressionReroute_0 {class=MaterialExpressionReroute, inputs [None {from=MaterialExpressionFunctionInput_11}]}
  MaterialExpressionReroute_1 {class=MaterialExpressionReroute, inputs [None {from=MaterialExpressionMaterialFunctionCall_1, output=cellBomb_UV_Rot}]}
  MaterialExpressionReroute_2 {class=MaterialExpressionReroute, inputs [None {from=MaterialExpressionReroute_1}]}
  MaterialExpressionReroute_3 {class=MaterialExpressionReroute, inputs [None {from=MaterialExpressionMaterialFunctionCall_1, output=cellBomb_Rotation}]}
  MaterialExpressionTextureCoordinate_4 {class=MaterialExpressionTextureCoordinate}
  MaterialExpressionTextureCoordinate_5 {class=MaterialExpressionTextureCoordinate, properties {coordinateIndex=1}}
  MaterialExpressionTextureCoordinate_6 {class=MaterialExpressionTextureCoordinate, properties {coordinateIndex=2}}
  MaterialExpressionTextureCoordinate_7 {class=MaterialExpressionTextureCoordinate, properties {coordinateIndex=3}}
  MaterialExpressionTextureObject_3 {class=MaterialExpressionTextureObject, inputs [], properties {texture=/BaseMaterial/Textures/Default/T_Gray_sRGB}}
 }
 properties {
  bExposeToLibrary=true
  description="Randomizes UVs per cell using a cells texture (per-cell offset, scale, rotation) and generates UVs, basis vectors, and packed derivatives for sampling and normal reorientation. Uses dithered sampling to approximate blending at very low cost (single sample per axis). This function applies Parallax Occlusion Mapping (POM) with dithered sampling. Provides low-cost per-cell depth cues, with the same dithering tradeoffs and no post-sample weight modulation."
  userExposedCaption="CoordinateBasis: Cell Bomb (Dithered POM)"
  libraryCategoriesText [Base Material]
 }
}
