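# Material function that resolves a UV-based coordinate frame with cell bombing and parallax occlusion mapping, breaking up texture tiling while adding depth.
# A selected, transformed mesh UV channel feeds cell bombing, which randomizes scale, offset and rotation per cell.
# Base and randomized coordinates each get their own parallax pass, with shared height settings.
# It outputs both UV sets with tangent bases, packed derivatives, rotations, depths and a two-weight mask for post-sample blending.
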MaterialFunction /BaseMaterial/Materials/Functions/MF_ResolveCoordinateFrame_CellBomb_UVPOM
graph {
 expressions {
  MaterialExpressionConvert_0 {
   class=MaterialExpressionConvert
   inputs [
    None {from=MaterialExpressionFunctionInput_6}
    None {from=MaterialExpressionNamedRerouteDeclaration_2}
   ]
   properties {
    nodeName=Make Float2
    convertInputs [
     {
      type=Scalar
      expressionInput {
       expression=&self:MaterialExpressionFunctionInput_6
       inputName=None
       mask=0
       maskA=0
       maskB=0
       maskG=0
       maskR=0
       outputIndex=0
      }
     }
     {
      type=Scalar
      expressionInput {
       expression=&self:MaterialExpressionNamedRerouteDeclaration_2
       inputName=None
       mask=0
       maskA=0
       maskB=0
       maskG=0
       maskR=0
       outputIndex=0
      }
     }
    ]
    convertOutputs [{type=Vector2}]
   }
  }
  MaterialExpressionConvert_1 {
   class=MaterialExpressionConvert
   inputs [
    None {from=MaterialExpressionNamedRerouteUsage_30}
    None {from=MaterialExpressionNamedRerouteUsage_5}
   ]
   properties {
    nodeName=Make Float2
    convertInputs [
     {
      type=Scalar
      expressionInput {
       expression=&self:MaterialExpressionNamedRerouteUsage_30
       inputName=None
       mask=0
       maskA=0
       maskB=0
       maskG=0
       maskR=0
       outputIndex=0
      }
     }
     {
      type=Scalar
      expressionInput {
       expression=&self:MaterialExpressionNamedRerouteUsage_5
       inputName=None
       mask=0
       maskA=0
       maskB=0
       maskG=0
       maskR=0
       outputIndex=0
      }
     }
    ]
    convertOutputs [{type=Vector2}]
   }
  }
  MaterialExpressionFunctionInput_0 {
   class=MaterialExpressionFunctionInput
   inputs [Preview {from="", output=""}]
   properties {
    bUsePreviewValueAsDefault=true
    description="Sets how many times the projected coordinates repeat along each axis. Use values above 1 for denser tiling, negative values for mirroring, and avoid zero."
    inputName=CoordinateRepeat
    inputType=FunctionInput_Vector2
    sortPriority=-10
    previewValue=<1 1 1 1>
   }
  }
  MaterialExpressionFunctionInput_1 {
   class=MaterialExpressionFunctionInput
   inputs [Preview {from="", output=""}]
   properties {
    bUsePreviewValueAsDefault=true
    description="Rotates the projected UVs around the selected axis. Units are in turns, where 1.0 = 360° and 0.25 = 90°."
    inputName="Rotation (Turns)"
    inputType=FunctionInput_Scalar
    sortPriority=-8
   }
  }
  MaterialExpressionFunctionInput_11 {
   class=MaterialExpressionFunctionInput
   inputs [Preview {from="", output=""}]
   properties {
    bUsePreviewValueAsDefault=true
    description="Specifies which channel of the heightmap texture to use for depth information. Assumes the Alpha channel is used, but any channel can be selected if your texture packs data differently."
    inputName=POM HeightChannel
    inputType=FunctionInput_Scalar
    sortPriority=10
    previewValue=<3 0 0 1>
   }
  }
  MaterialExpressionFunctionInput_12 {
   class=MaterialExpressionFunctionInput
   inputs [Preview {from="", output=""}]
   properties {
    bUsePreviewValueAsDefault=true
    description="Sets the maximum number of depth sampling steps. This is the upper bound, with actual steps scaled by the Quality setting."
    inputName=POM MaxSteps
    inputType=FunctionInput_Scalar
    sortPriority=12
    previewValue=<16 0 0 1>
   }
  }
  MaterialExpressionFunctionInput_13 {
   class=MaterialExpressionFunctionInput
   inputs [Preview {from="", output=""}]
   properties {
    bUsePreviewValueAsDefault=true
    description=Applies an additional per-axis offset to the coordinates before projection. Useful for correctly positioning a texture.
    inputName=CoordinateTransform
    inputType=FunctionInput_Vector2
    sortPriority=-9
   }
  }
  MaterialExpressionFunctionInput_14 {
   class=MaterialExpressionFunctionInput
   inputs [Preview {from="", output=""}]
   properties {
    bUsePreviewValueAsDefault=true
    description="Controls the strength of the parallax effect. Higher values exaggerate depth; lower values make surfaces appear flatter."
    inputName=POM Intensity
    inputType=FunctionInput_Scalar
    sortPriority=11
    previewValue=<1 0 0 1>
   }
  }
  MaterialExpressionFunctionInput_15 {
   class=MaterialExpressionFunctionInput
   inputs [Preview {from="", output=""}]
   properties {
    bUsePreviewValueAsDefault=true
    description="Scales step count by view angle. At 1, MaxSteps are always used; at 0, only one step is used when looking straight on, with more steps added at grazing angles."
    inputName=POM Quality
    inputType=FunctionInput_Scalar
    sortPriority=13
    previewValue=<0.25 0 0 1>
   }
  }
  MaterialExpressionFunctionInput_16 {
   class=MaterialExpressionFunctionInput
   inputs [Preview {from="", output=""}]
   properties {
    bUsePreviewValueAsDefault=true
    description=Defines the reference plane for parallax. This value shifts where the heightmap appears anchored relative to the surface.
    inputName=POM RefPlane
    inputType=FunctionInput_Scalar
    sortPriority=14
    previewValue=<1 0 0 1>
   }
  }
  MaterialExpressionFunctionInput_2 {
   class=MaterialExpressionFunctionInput
   inputs [Preview {from="", output=""}]
   properties {
    bUsePreviewValueAsDefault=true
    description=Control the U/V Tiling of the height texture relative to the input coordinates.
    inputName=POM HeightTile
    inputType=FunctionInput_Vector2
    sortPriority=12
    previewValue=<1 1 0 1>
   }
  }
  MaterialExpressionFunctionInput_22 {
   class=MaterialExpressionFunctionInput
   inputs [Preview {from="", output=""}]
   properties {
    bUsePreviewValueAsDefault=true
    description=Controls the sharpness of the transition between coordinate sets.
    inputName=BlendTolerance
    inputType=FunctionInput_Scalar
    sortPriority=8
    previewValue=<0.5 0 0 1>
   }
  }
  MaterialExpressionFunctionInput_25 {
   class=MaterialExpressionFunctionInput
   inputs [Preview {from=MaterialExpressionTextureObject_3, output=""}]
   properties {
    bUsePreviewValueAsDefault=true
    description="The heightmap texture used for parallax occlusion mapping. Grayscale values define surface depth, with brighter values typically pushed outward and darker values pushed inward."
    inputName=HeightTex
    inputType=FunctionInput_Texture2D
    sortPriority=-20
   }
  }
  MaterialExpressionFunctionInput_3 {
   class=MaterialExpressionFunctionInput
   inputs [Preview {from="", output=""}]
   properties {
    bUsePreviewValueAsDefault=true
    description="Controls tiling of the cells texture in UV space. Higher values create smaller, more numerous cells."
    inputName=Cell PatternSize
    inputType=FunctionInput_Scalar
    sortPriority=5
    previewValue=<0.01 0 0 1>
   }
  }
  MaterialExpressionFunctionInput_4 {
   class=MaterialExpressionFunctionInput
   inputs [Preview {from="", output=""}]
   properties {
    bUsePreviewValueAsDefault=true
    description="Per-cell random scale range (min/max). Values below 1 enlarge, above 1 shrink. Applied uniformly per cell."
    inputName=Cell Min/MaxRand Scale
    inputType=FunctionInput_Vector2
    sortPriority=6
    previewValue=<1 1 0 1>
   }
  }
  MaterialExpressionFunctionInput_5 {
   class=MaterialExpressionFunctionInput
   inputs [Preview {from="", output=""}]
   properties {
    bUsePreviewValueAsDefault=true
    description="Per-cell random UV offset range (min/max). Shifts sampling position within each cell."
    inputName=Cell Min/MaxRand Offset
    inputType=FunctionInput_Vector2
    sortPriority=7
   }
  }
  MaterialExpressionFunctionInput_6 {
   class=MaterialExpressionFunctionInput
   inputs [Preview {from="", output=""}]
   properties {
    bUsePreviewValueAsDefault=true
    description="Maximum per-cell random rotation (in turns, 1.0 = 360°). The actual rotation per cell is drawn within this magnitude."
    inputName="Cell Rand Rotation (turns)"
    inputType=FunctionInput_Scalar
    sortPriority=8
   }
  }
  MaterialExpressionFunctionInput_7 {
   class=MaterialExpressionFunctionInput
   inputs [Preview {from="", output=""}]
   properties {
    bUsePreviewValueAsDefault=true
    description="Selects which mesh UV channel (0–3) to use as the base coordinate source."
    inputName=UVIndex
    inputType=FunctionInput_Scalar
    sortPriority=-5
   }
  }
  MaterialExpressionFunctionOutput_1 {
   class=MaterialExpressionFunctionOutput
   inputs [None {from=MaterialExpressionMaterialFunctionCall_10, output=POM UVs}]
   properties {description="Base UVs sampled in the same coordinate frame as the cells texture (unrandomized reference).", outputName=UV1}
  }
  MaterialExpressionFunctionOutput_10 {
   class=MaterialExpressionFunctionOutput
   inputs [None {from=MaterialExpressionFunctionOutput_8, output=""}]
   properties {description=World-space tangent for UV2., outputName=Tangent2, sortPriority=6}
  }
  MaterialExpressionFunctionOutput_11 {
   class=MaterialExpressionFunctionOutput
   inputs [None {from=MaterialExpressionFunctionOutput_9, output=""}]
   properties {description=World-space bitangent for UV2., outputName=Bitangent2, sortPriority=7}
  }
  MaterialExpressionFunctionOutput_12 {
   class=MaterialExpressionFunctionOutput
   inputs [None {from=MaterialExpressionMaterialFunctionCall_10, output=Pixel Depth Offset}]
   properties {
    description="Depth at the POM intersection (after Intensity/RefPlane). Higher (whiter) = deeper—the farther the ray travels into the material, the larger the value."
    outputName=Depth1
    sortPriority=4
   }
  }
  MaterialExpressionFunctionOutput_14 {
   class=MaterialExpressionFunctionOutput
   inputs [None {from=MaterialExpressionMultiply_1, output=""}]
   properties {description="Rotation for UV1 in turns (1.0 = 360°).", outputName="Rotation1 (Turns)", sortPriority=-6}
  }
  MaterialExpressionFunctionOutput_15 {
   class=MaterialExpressionFunctionOutput
   inputs [None {from=MaterialExpressionMaterialFunctionCall_7, output=Pixel Depth Offset}]
   properties {
    description="Depth at the POM intersection (after Intensity/RefPlane). Higher (whiter) = deeper—the farther the ray travels into the material, the larger the value."
    outputName=Depth2
    sortPriority=9
   }
  }
  MaterialExpressionFunctionOutput_2 {
   class=MaterialExpressionFunctionOutput
   inputs [None {from=MaterialExpressionMaterialFunctionCall_1, output=cellBomb_DXY}]
   properties {description="Packed derivatives for UV1 (ddx/ddy) for SampleGrad.", outputName=DXY1, sortPriority=3}
  }
  MaterialExpressionFunctionOutput_3 {
   class=MaterialExpressionFunctionOutput
   inputs [None {from=MaterialExpressionMaterialFunctionCall_7, output=POM UVs}]
   properties {description="UVs randomized per cell (offset/scale/rotation applied).", outputName=UV2, sortPriority=5}
  }
  MaterialExpressionFunctionOutput_4 {
   class=MaterialExpressionFunctionOutput
   inputs [None {from=MaterialExpressionMaterialFunctionCall_1, output=cellBomb_DXY}]
   properties {description="Packed derivatives for UV2 (ddx/ddy) for SampleGrad.", outputName=DXY2, sortPriority=8}
  }
  MaterialExpressionFunctionOutput_5 {
   class=MaterialExpressionFunctionOutput
   inputs [None {from=MaterialExpressionMaterialFunctionCall_1, output=cellBomb_CellTurnsOnly}]
   properties {description="Rotation for UV2 in turns (1.0 = 360°).", outputName="Rotation2 (Turns)", sortPriority=-5}
  }
  MaterialExpressionFunctionOutput_7 {
   class=MaterialExpressionFunctionOutput
   inputs [None {from=MaterialExpressionConvert_1, output=""}]
   properties {
    description="RG weights for blending UV1 (base) and UV2 (randomized) post-sample; transition sharpness is set by BlendFactor. Also supports albedo/height-driven modulation.\r\n"
    outputName=BlendMask
    sortPriority=15
   }
  }
  MaterialExpressionFunctionOutput_8 {
   class=MaterialExpressionFunctionOutput
   inputs [None {from=MaterialExpressionMaterialFunctionCall_0, output=Tangent}]
   properties {description="World-space tangent for UV1.\r\n", outputName=Tangent1, sortPriority=1}
  }
  MaterialExpressionFunctionOutput_9 {
   class=MaterialExpressionFunctionOutput
   inputs [None {from=MaterialExpressionMaterialFunctionCall_0, output=Bitangent}]
   properties {description=World-space bitangent for UV1., outputName=Bitangent1, sortPriority=2}
  }
  MaterialExpressionMaterialFunctionCall_0 {
   class=MaterialExpressionMaterialFunctionCall
   inputs [packedDXY {from=MaterialExpressionNamedRerouteUsage_0}]
   properties {materialFunction=&/MF_DeriveTangentBasis}
  }
  MaterialExpressionMaterialFunctionCall_1 {
   class=MaterialExpressionMaterialFunctionCall
   inputs [
    Coordinates {from=MaterialExpressionNamedRerouteUsage_38}
    CellSize {from=MaterialExpressionFunctionInput_3}
    ScaleRandom {from=MaterialExpressionFunctionInput_4}
    OffsetRandom {from=MaterialExpressionFunctionInput_5}
    "Rotation (PerCellRand/Base)" {from=MaterialExpressionConvert_0}
    BlendFactor {from=MaterialExpressionFunctionInput_22}
    Dither {from=""}
   ]
   properties {materialFunction=&/MF_Coordinate_CellBombing}
  }
  MaterialExpressionMaterialFunctionCall_10 {
   class=MaterialExpressionMaterialFunctionCall
   inputs [
    POM_HeightTexture {from=MaterialExpressionNamedRerouteUsage_18}
    Coordinates {from=MaterialExpressionNamedRerouteDeclaration_15}
    coordinateDXY {from=MaterialExpressionNamedRerouteUsage_44}
    HeightStrength {from=MaterialExpressionNamedRerouteUsage_42}
    Reference Plane {from=MaterialExpressionNamedRerouteUsage_43}
    MaxSteps {from=MaterialExpressionNamedRerouteUsage_41}
    Tangent {from=MaterialExpressionNamedRerouteUsage_3}
    Bitangent {from=MaterialExpressionReroute_5}
    "Rotation (Turns)" {from=MaterialExpressionFunctionOutput_14}
    HeightChannel {from=MaterialExpressionNamedRerouteUsage_47}
    QualityBias {from=MaterialExpressionNamedRerouteUsage_46}
    CameraVector {from=""}
    Height Tile {from=MaterialExpressionNamedRerouteDeclaration_0}
   ]
   properties {materialFunction=&/MF_Coordinate_Parallax}
  }
  MaterialExpressionMaterialFunctionCall_2 {
   class=MaterialExpressionMaterialFunctionCall
   inputs [
    Selector {from=MaterialExpressionFunctionInput_7}
    VectorA {from=MaterialExpressionTextureCoordinate_4}
    VectorB {from=MaterialExpressionTextureCoordinate_5}
    VectorC {from=MaterialExpressionTextureCoordinate_6}
    VectorD {from=MaterialExpressionTextureCoordinate_7}
   ]
   properties {materialFunction=&/MF_Switch4_Vec2}
  }
  MaterialExpressionMaterialFunctionCall_3 {
   class=MaterialExpressionMaterialFunctionCall
   inputs [
    Coords {from=MaterialExpressionMaterialFunctionCall_2, output=Selected Vector}
    UVOffset {from=MaterialExpressionFunctionInput_13}
    UVRotation {from=""}
    RotationPivot {from=""}
    UVTile {from=MaterialExpressionFunctionInput_0}
   ]
   properties {materialFunction=&/MF_CoordinateTransform_2D}
  }
  MaterialExpressionMaterialFunctionCall_7 {
   class=MaterialExpressionMaterialFunctionCall
   inputs [
    POM_HeightTexture {from=MaterialExpressionNamedRerouteUsage_9}
    Coordinates {from=MaterialExpressionReroute_0}
    coordinateDXY {from=MaterialExpressionNamedRerouteDeclaration_7}
    HeightStrength {from=MaterialExpressionNamedRerouteUsage_35}
    Reference Plane {from=MaterialExpressionNamedRerouteUsage_36}
    MaxSteps {from=MaterialExpressionNamedRerouteUsage_34}
    Tangent {from=MaterialExpressionNamedRerouteUsage_1}
    Bitangent {from=MaterialExpressionReroute_3}
    "Rotation (Turns)" {from=MaterialExpressionReroute_4}
    HeightChannel {from=MaterialExpressionNamedRerouteUsage_39}
    QualityBias {from=MaterialExpressionNamedRerouteUsage_37}
    CameraVector {from=""}
    Height Tile {from=MaterialExpressionNamedRerouteUsage_6}
   ]
   properties {materialFunction=&MF_Coordinate_Parallax}
  }
  MaterialExpressionMultiply_1 {
   class=MaterialExpressionMultiply
   inputs [
    A {from=MaterialExpressionNamedRerouteUsage_45}
    B {from=""}
   ]
   properties {constB=-1}
  }
  MaterialExpressionNamedRerouteDeclaration_0 {
   class=MaterialExpressionNamedRerouteDeclaration
   inputs [None {from=MaterialExpressionFunctionInput_2}]
   properties {name=POM_HeightTile, nodeColor=<rgba 0.9 0.306 0.009 1>}
  }
  MaterialExpressionNamedRerouteDeclaration_1 {
   class=MaterialExpressionNamedRerouteDeclaration
   inputs [None {from=MaterialExpressionMaterialFunctionCall_1, output=SampleMask}]
   properties {name=Mask, nodeColor=<rgba 0 1 0.3906 1>}
  }
  MaterialExpressionNamedRerouteDeclaration_15 {
   class=MaterialExpressionNamedRerouteDeclaration
   inputs [None {from=MaterialExpressionMaterialFunctionCall_3, output=geometry_UV}]
   properties {name=UVt, nodeColor=<rgba 0.9 0.009 0.009 1>}
  }
  MaterialExpressionNamedRerouteDeclaration_16 {
   class=MaterialExpressionNamedRerouteDeclaration
   inputs [None {from=MaterialExpressionFunctionInput_11}]
   properties {name=POM_Hchan, nodeColor=<rgba 0.9 0.306 0.009 1>}
  }
  MaterialExpressionNamedRerouteDeclaration_17 {
   class=MaterialExpressionNamedRerouteDeclaration
   inputs [None {from=MaterialExpressionOneMinus_0}]
   properties {name=MaskCmp, nodeColor=<rgba 0 1 0.3906 1>}
  }
  MaterialExpressionNamedRerouteDeclaration_19 {
   class=MaterialExpressionNamedRerouteDeclaration
   inputs [None {from=MaterialExpressionFunctionInput_25}]
   properties {name=Height, nodeColor=<rgba 1 0 0.4922 1>}
  }
  MaterialExpressionNamedRerouteDeclaration_2 {
   class=MaterialExpressionNamedRerouteDeclaration
   inputs [None {from=MaterialExpressionFunctionInput_1}]
   properties {name=turnsi, nodeColor=<rgba 1 0.7031 0 1>}
  }
  MaterialExpressionNamedRerouteDeclaration_3 {
   class=MaterialExpressionNamedRerouteDeclaration
   inputs [None {from=MaterialExpressionFunctionOutput_10}]
   properties {name=T, nodeColor=<rgba 0.9 0.009 0.009 1>}
  }
  MaterialExpressionNamedRerouteDeclaration_33 {
   class=MaterialExpressionNamedRerouteDeclaration
   inputs [None {from=MaterialExpressionFunctionInput_14}]
   properties {name=POM_Height, nodeColor=<rgba 0.9 0.306 0.009 1>}
  }
  MaterialExpressionNamedRerouteDeclaration_34 {
   class=MaterialExpressionNamedRerouteDeclaration
   inputs [None {from=MaterialExpressionFunctionInput_12}]
   properties {name=POM_MaxSteps, nodeColor=<rgba 0.9 0.306 0.009 1>}
  }
  MaterialExpressionNamedRerouteDeclaration_35 {
   class=MaterialExpressionNamedRerouteDeclaration
   inputs [None {from=MaterialExpressionFunctionInput_15}]
   properties {name=POM_Noise, nodeColor=<rgba 0.9 0.306 0.009 1>}
  }
  MaterialExpressionNamedRerouteDeclaration_36 {
   class=MaterialExpressionNamedRerouteDeclaration
   inputs [None {from=MaterialExpressionFunctionInput_16}]
   properties {name=POM_RefPlane, nodeColor=<rgba 0.9 0.306 0.009 1>}
  }
  MaterialExpressionNamedRerouteDeclaration_4 {
   class=MaterialExpressionNamedRerouteDeclaration
   inputs [None {from=MaterialExpressionFunctionOutput_11}]
   properties {name=B, nodeColor=<rgba 0.9 0.009 0.009 1>}
  }
  MaterialExpressionNamedRerouteDeclaration_6 {
   class=MaterialExpressionNamedRerouteDeclaration
   inputs [None {from=MaterialExpressionFunctionOutput_1}]
   properties {name=UVi, nodeColor=<rgba 0.9 0.009 0.009 1>}
  }
  MaterialExpressionNamedRerouteDeclaration_7 {
   class=MaterialExpressionNamedRerouteDeclaration
   inputs [None {from=MaterialExpressionFunctionOutput_2}]
   properties {name=DXY, nodeColor=<rgba 0.9 0.009 0.009 1>}
  }
  MaterialExpressionNamedRerouteUsage_0 {class=MaterialExpressionNamedRerouteUsage}
  MaterialExpressionNamedRerouteUsage_1 {class=MaterialExpressionNamedRerouteUsage}
  MaterialExpressionNamedRerouteUsage_18 {class=MaterialExpressionNamedRerouteUsage}
  MaterialExpressionNamedRerouteUsage_2 {class=MaterialExpressionNamedRerouteUsage}
  MaterialExpressionNamedRerouteUsage_3 {class=MaterialExpressionNamedRerouteUsage}
  MaterialExpressionNamedRerouteUsage_30 {class=MaterialExpressionNamedRerouteUsage}
  MaterialExpressionNamedRerouteUsage_34 {class=MaterialExpressionNamedRerouteUsage}
  MaterialExpressionNamedRerouteUsage_35 {class=MaterialExpressionNamedRerouteUsage}
  MaterialExpressionNamedRerouteUsage_36 {class=MaterialExpressionNamedRerouteUsage}
  MaterialExpressionNamedRerouteUsage_37 {class=MaterialExpressionNamedRerouteUsage}
  MaterialExpressionNamedRerouteUsage_38 {class=MaterialExpressionNamedRerouteUsage}
  MaterialExpressionNamedRerouteUsage_39 {class=MaterialExpressionNamedRerouteUsage}
  MaterialExpressionNamedRerouteUsage_4 {class=MaterialExpressionNamedRerouteUsage}
  MaterialExpressionNamedRerouteUsage_41 {class=MaterialExpressionNamedRerouteUsage}
  MaterialExpressionNamedRerouteUsage_42 {class=MaterialExpressionNamedRerouteUsage}
  MaterialExpressionNamedRerouteUsage_43 {class=MaterialExpressionNamedRerouteUsage}
  MaterialExpressionNamedRerouteUsage_44 {class=MaterialExpressionNamedRerouteUsage}
  MaterialExpressionNamedRerouteUsage_45 {class=MaterialExpressionNamedRerouteUsage}
  MaterialExpressionNamedRerouteUsage_46 {class=MaterialExpressionNamedRerouteUsage}
  MaterialExpressionNamedRerouteUsage_47 {class=MaterialExpressionNamedRerouteUsage}
  MaterialExpressionNamedRerouteUsage_48 {class=MaterialExpressionNamedRerouteUsage}
  MaterialExpressionNamedRerouteUsage_5 {class=MaterialExpressionNamedRerouteUsage}
  MaterialExpressionNamedRerouteUsage_6 {class=MaterialExpressionNamedRerouteUsage}
  MaterialExpressionNamedRerouteUsage_9 {class=MaterialExpressionNamedRerouteUsage}
  MaterialExpressionOneMinus_0 {class=MaterialExpressionOneMinus, inputs [None {from=MaterialExpressionNamedRerouteUsage_48}]}
  MaterialExpressionReroute_0 {class=MaterialExpressionReroute, inputs [None {from=MaterialExpressionReroute_2}]}
  MaterialExpressionReroute_2 {class=MaterialExpressionReroute, inputs [None {from=MaterialExpressionMaterialFunctionCall_1, output=cellBomb_CellUVOnly}]}
  MaterialExpressionReroute_3 {class=MaterialExpressionReroute, inputs [None {from=MaterialExpressionNamedRerouteUsage_2}]}
  MaterialExpressionReroute_4 {class=MaterialExpressionReroute, inputs [None {from=MaterialExpressionFunctionOutput_5}]}
  MaterialExpressionReroute_5 {class=MaterialExpressionReroute, inputs [None {from=MaterialExpressionNamedRerouteUsage_4}]}
  MaterialExpressionTextureCoordinate_4 {class=MaterialExpressionTextureCoordinate}
  MaterialExpressionTextureCoordinate_5 {class=MaterialExpressionTextureCoordinate, properties {coordinateIndex=1}}
  MaterialExpressionTextureCoordinate_6 {class=MaterialExpressionTextureCoordinate, properties {coordinateIndex=2}}
  MaterialExpressionTextureCoordinate_7 {class=MaterialExpressionTextureCoordinate, properties {coordinateIndex=3}}
  MaterialExpressionTextureObject_3 {class=MaterialExpressionTextureObject, inputs [], properties {texture=/BaseMaterial/Textures/Default/T_Gray_sRGB}}
 }
 properties {
  bExposeToLibrary=true
  description="Standard Cell Bomb with Parallax Occlusion Mapping (POM). Outputs UV1 (base coordinates) and UV2 (randomized per cell) with full bases and derivatives, plus a BlendMask (RG) whose transition sharpness is controlled by BlendFactor. Supports post-sample modulation of blend weights by albedo or height, producing smoother, more perceptual blending between cell regions while also adding depth and self-occlusion through POM."
  userExposedCaption="CoordinateBasis: Cell Bomb (Standard POM)"
  libraryCategoriesText [Base Material]
 }
}
