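# Material function in the Base Material library that turns coordinates into a two-channel circle-splat blend mask for a grid of overlapping circles.
# For each UV cell it picks the nearer corner on each diagonal, outputting both as CircleID1 and CircleID2.
# Squared distances to those corners become clamped falloffs, normalized by their sum with a zero guard to form PatternMask.
# An unconnected Custom node holds the same algorithm as HLSL.
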MaterialFunction /BaseMaterial/Materials/Functions/MF_Blend_CircleMask
graph {
 properties {bExposeToLibrary=true, libraryCategoriesText [Base Material]}
 nodes [
  Custom!(inUV) {
   code="float2 baseID = floor(inUV);\r\nfloat2 tmp = frac(inUV);\r\n\r\ncircID1 = tmp.x < tmp.y       ? float2(0., 1.) : float2(1., 0.);\r\ncircID1 += baseID;\r\ncircID2 = tmp.x < 1.f - tmp.y ? float2(0., 0.) : float2(1., 1.);\r\ncircID2 += baseID;\r\n\r\nfloat2 outMask = float2( dot(inUV - circID1, inUV - circID1), \r\n                         dot(inUV - circID2, inUV - circID2) );\r\n\r\noutMask = max( 1.f - 2.f * outMask, 0.f );\r\noutMask /= (outMask.x + outMask.y) > 0.f ? (outMask.x + outMask.y) : 1.f;\r\n\r\nreturn outMask;"
   description=CircleSplats
   outputType=CMOT_Float2
   showCode=true
   additionalOutputs [{outputName=circID1, outputType=CMOT_Float2}, {outputName=circID2, outputType=CMOT_Float2}]
   inputs [{inputName=inUV, input {expression=null}}]
  }
  n1 = FunctionInput(Preview=TextureCoordinate()) {bUsePreviewValueAsDefault=true, inputName=Coordinates, inputType=FunctionInput_Vector2}
  MaterialExpressionFrac_0 = Frac(None=@n1) {desc=tmp}
  n2 = Convert(None=@MaterialExpressionFrac_0) {
   nodeName=Break Float2
   convertInputs [{
     type=Vector2
     expressionInput {
      expression=&self:MaterialExpressionFrac_0
      inputName=None
      mask=0
      maskA=0
      maskB=0
      maskG=0
      maskR=0
      outputIndex=0
     }
    }]
   convertOutputs [{type=Scalar}, {type=Scalar}]
  }
  n3 = If(A=@n2, B=@n2, "A > B"=Constant2Vector() {r=1}, "A < B"=Constant2Vector() {g=1}) {desc=circleID1}
  n4 = Floor(None=@n1) {desc=baseID}
  n5 = Subtract(A=@n1, B=FunctionOutput(None=Add(A=@n4, B=@n3)) {outputName=CircleID1, sortPriority=5})
  n6 = If(A=@n2, B=OneMinus(None=@n2), "A > B"=Constant2Vector() {g=1, r=1}, "A < B"=Constant2Vector()) {desc=circleID2}
  n7 = Subtract(A=@n1, B=FunctionOutput(None=Add(A=@n4, B=@n6)) {outputName=CircleID2, sortPriority=6})
  n8 = AppendVector(A=DotProduct(A=@n5, B=@n5), B=DotProduct(A=@n7, B=@n7))
  MaterialExpressionMax_0 = Max(A=OneMinus(None=Multiply(A=@n8) {constB=2})) {constB=0}
  n9 = Convert(None=@MaterialExpressionMax_0) {
   nodeName=Break Float2
   convertInputs [{
     type=Vector2
     expressionInput {
      expression=&self:MaterialExpressionMax_0
      inputName=None
      mask=0
      maskA=0
      maskB=0
      maskG=0
      maskR=0
      outputIndex=0
     }
    }]
   convertOutputs [{type=Scalar}, {type=Scalar}]
  }
  n10 = Add(A=@n9, B=@n9)
  n11 = Constant() {r=1}
  FunctionOutput(None=Divide(A=@MaterialExpressionMax_0, B=If(A=@n10, "A > B"=@n10, "A < B"=@n11))) {outputName=PatternMask, sortPriority=10}
 ]
}
