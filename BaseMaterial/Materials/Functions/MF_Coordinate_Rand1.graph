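# Material function, exposed in the Base Material library, that randomizes 2D texture coordinates from a four-component seed.
# The seed splits into separate components that, scaled by strength inputs, drive min/max lerps for offset and tiling.
# A seed component remapped to a signed range and passed through Frac gives rotation in turns.
# MF_CoordinateTransform_2D applies all three to Coords, and the function outputs UV, packedDXY and SampleRotation.
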MaterialFunction /BaseMaterial/Materials/Functions/MF_Coordinate_Rand1
graph {
 properties {bExposeToLibrary=true, libraryCategoriesText [Base Material]}
 nodes [
  n1 = NamedRerouteUsage()
  MaterialExpressionFunctionInput_0 = FunctionInput() {bUsePreviewValueAsDefault=true, inputName=SeedA, inputType=FunctionInput_Vector4, sortPriority=10, previewValue=<1 1 1 1>}
  n2 = Convert(None=@MaterialExpressionFunctionInput_0) {
   convertInputs [{
     type=Vector4
     defaultValue=<rgba 0 0 0 1>
     expressionInput {
      expression=&self:MaterialExpressionFunctionInput_0
      inputName=None
      mask=0
      maskA=0
      maskB=0
      maskG=0
      maskR=0
      outputIndex=0
     }
    }]
   convertOutputs [{type=Vector2}, {type=Scalar}, {type=Scalar}]
  }
  n3 = Multiply(A=ConstantBiasScale(None=@n2) {bias=-0.5, scale=2}, B=NamedRerouteUsage())
  n4 = NamedRerouteUsage()
  n5 = MaterialFunctionCall!(
   Coords=FunctionInput(Preview=TextureCoordinate()) {bUsePreviewValueAsDefault=true, inputName=Coords, inputType=FunctionInput_Vector2, sortPriority=-10}
   UVOffset=LinearInterpolate(A=NamedRerouteUsage(), B=NamedRerouteUsage(), Alpha=Multiply(A=@n1, B=@n2))
   UVRotation=NamedRerouteDeclaration(None=Frac(None=@n3)) {name=turnsa, nodeColor=<rgba 0.8963 0.8963 0.0091 1>}
   RotationPivot
   UVTile=LinearInterpolate(A=NamedRerouteUsage(), B=NamedRerouteUsage(), Alpha=Multiply(A=@n2, B=@n4))
  ) {materialFunction=&/MF_CoordinateTransform_2D}
  FunctionOutput(None=@n5.geometry_UV) {outputName=UV}
  FunctionOutput(None=@n5.geometry_packedDXY) {outputName=packedDXY, sortPriority=10}
  FunctionOutput(None=NamedRerouteUsage()) {outputName=SampleRotation, sortPriority=20}
  n6 = FunctionInput() {bUsePreviewValueAsDefault=true, inputName=OffsetStrengthRand, inputType=FunctionInput_Scalar, sortPriority=6, previewValue=<1 1 1 1>}
  NamedRerouteDeclaration(None=@n6) {name=OffA, nodeColor=<rgba 0.6875 1 0 1>}
  MaterialExpressionFunctionInput_7 = FunctionInput() {bUsePreviewValueAsDefault=true, inputName="TileRand (Min/Max)", inputType=FunctionInput_Vector2, previewValue=<1 1 0 1>}
  n7 = Convert(None=@MaterialExpressionFunctionInput_7) {
   nodeName=Break Float2
   convertInputs [{
     type=Vector2
     expressionInput {
      expression=&self:MaterialExpressionFunctionInput_7
      inputName=None
      mask=0
      maskA=0
      maskB=0
      maskG=0
      maskR=0
      outputIndex=0
     }
    }]
   convertOutputs [{type=Scalar}, {type=Scalar}]
  }
  NamedRerouteDeclaration(None=@n7) {name=MinTile, nodeColor=<rgba 0.6875 1 0 1>}
  MaterialExpressionFunctionInput_8 = FunctionInput() {bUsePreviewValueAsDefault=true, inputName="OffsetRand (Min/Max)", inputType=FunctionInput_Vector2, sortPriority=1}
  n8 = Convert(None=@MaterialExpressionFunctionInput_8) {
   nodeName=Break Float2
   convertInputs [{
     type=Vector2
     expressionInput {
      expression=&self:MaterialExpressionFunctionInput_8
      inputName=None
      mask=0
      maskA=0
      maskB=0
      maskG=0
      maskR=0
      outputIndex=0
     }
    }]
   convertOutputs [{type=Scalar}, {type=Scalar}]
  }
  NamedRerouteDeclaration(None=@n8) {name=MaxOffset, nodeColor=<rgba 0.6875 1 0 1>}
  n9 = FunctionInput() {bUsePreviewValueAsDefault=true, inputName=TileStrengthRand, inputType=FunctionInput_Scalar, sortPriority=5, previewValue=<1 1 1 1>}
  NamedRerouteDeclaration(None=@n9) {name=TileA, nodeColor=<rgba 0.6875 1 0 1>}
  NamedRerouteDeclaration(None=FunctionInput() {bUsePreviewValueAsDefault=true, inputName=RotateStrengthRand, inputType=FunctionInput_Scalar, sortPriority=7}) {name=RotA, nodeColor=<rgba 0.6875 1 0 1>}
  NamedRerouteDeclaration(None=@n7) {name=MaxTile, nodeColor=<rgba 0.6875 1 0 1>}
  NamedRerouteDeclaration(None=@n8) {name=MinOffset, nodeColor=<rgba 0.6875 1 0 1>}
 ]
}
