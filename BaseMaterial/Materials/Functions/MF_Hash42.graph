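# Material function that hashes a Vector2 seed into four pseudo-random values in the 0-1 range, output as Hash4D.
# It builds the Hash42 algorithm from native nodes, using Convert nodes for vector widening and swizzles around multiply, dot product and frac steps.
# An unconnected Custom HLSL node keeps the Shadertoy-sourced code as a reference only.
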MaterialFunction /BaseMaterial/Materials/Functions/MF_Hash42
graph {nodes [
  Custom!(inSeed) {
   code="//HLSL for Ref\r\n//ref https://www.shadertoy.com/view/4djSRW\r\n\r\nfloat4 p4 = frac( float4(inSeed.xyxy) * float4(0.1031, 0.1030, 0.0973, 0.1099) );\r\n       p4 += dot(p4, p4.wzxy + 33.33);\r\n\r\nreturn frac((p4.xxyz + p4.yzzw) * p4.zywx);"
   description=hash42
   outputType=CMOT_Float4
   showCode=true
   inputs [{inputName=inSeed, input {expression=null}}]
  }
  MaterialExpressionFunctionInput_0 = FunctionInput(Preview=TextureCoordinate()) {inputName=seed, inputType=FunctionInput_Vector2}
  n1 = Convert(None=@MaterialExpressionFunctionInput_0) {
   convertInputs [{
     type=Vector2
     expressionInput {
      expression=&self:MaterialExpressionFunctionInput_0
      inputName=None
      mask=0
      maskA=0
      maskB=0
      maskG=0
      maskR=0
      outputIndex=0
     }
    }]
   convertOutputs [{type=Vector4, defaultValue=<rgba 0 0 0 1>}]
  }
  MaterialExpressionFrac_0 = Frac(None=Multiply(A=@n1, B=Constant4Vector() {constant=<rgba 0.1031 0.103 0.0973 0.1099>}.RGBA)) {desc=p4}
  n2 = Convert(None=@MaterialExpressionFrac_0) {
   convertInputs [{
     type=Vector4
     defaultValue=<rgba 0 0 0 1>
     expressionInput {
      expression=&self:MaterialExpressionFrac_0
      inputName=None
      mask=0
      maskA=0
      maskB=0
      maskG=0
      maskR=0
      outputIndex=0
     }
    }]
   convertOutputs [{type=Vector4, defaultValue=<rgba 0 0 0 1>}]
  }
  MaterialExpressionAdd_1 = Add(A=@MaterialExpressionFrac_0, B=DotProduct(A=@MaterialExpressionFrac_0, B=Add(A=@n2) {constB=33.33}))
  n3 = Convert(None=@MaterialExpressionAdd_1) {
   convertInputs [{
     type=Vector4
     defaultValue=<rgba 0 0 0 1>
     expressionInput {
      expression=&self:MaterialExpressionAdd_1
      inputName=None
      mask=0
      maskA=0
      maskB=0
      maskG=0
      maskR=0
      outputIndex=0
     }
    }]
   convertOutputs [{type=Vector4, defaultValue=<rgba 0 0 0 1>}]
  }
  n4 = Convert(None=@MaterialExpressionAdd_1) {
   convertInputs [{
     type=Vector4
     defaultValue=<rgba 0 0 0 1>
     expressionInput {
      expression=&self:MaterialExpressionAdd_1
      inputName=None
      mask=0
      maskA=0
      maskB=0
      maskG=0
      maskR=0
      outputIndex=0
     }
    }]
   convertOutputs [{type=Vector4, defaultValue=<rgba 0 0 0 1>}]
  }
  n5 = Add(A=@n3, B=@n4)
  n6 = Convert(None=@MaterialExpressionAdd_1) {
   convertInputs [{
     type=Vector4
     defaultValue=<rgba 0 0 0 1>
     expressionInput {
      expression=&self:MaterialExpressionAdd_1
      inputName=None
      mask=0
      maskA=0
      maskB=0
      maskG=0
      maskR=0
      outputIndex=0
     }
    }]
   convertOutputs [{type=Vector4, defaultValue=<rgba 0 0 0 1>}]
  }
  FunctionOutput(None=Frac(None=Multiply(A=@n5, B=@n6))) {outputName=Hash4D}
 ]}
