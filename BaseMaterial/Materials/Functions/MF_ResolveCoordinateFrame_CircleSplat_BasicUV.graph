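# Material function for the Base Material library that resolves a UV-based coordinate frame broken into randomized circular splats to reduce tiling repetition.
# A selected UV channel is transformed, then scaled into a circle mask pattern yielding a blend mask and two cell IDs.
# Each hashed ID randomizes scale, offset and rotation for one of two UV sets.
# Both sets share derivatives and a tangent basis from the unrandomized coordinates, while rotations are output separately.
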MaterialFunction /BaseMaterial/Materials/Functions/MF_ResolveCoordinateFrame_CircleSplat_BasicUV
graph {
 properties {
  bExposeToLibrary=true
  description="Generates randomized circular splats across the surface using a tiled pattern driven by the selected UV channel. Each splat can vary in scale, offset, and rotation. Outputs two randomized coordinate sets (UV1 and UV2), each with tangent/bitangent/rotation/derivatives, plus a BlendMask (RG) for post-sample blending. Supports blend weight modulation by albedo or height for smoother, perceptual transitions between randomized splats."
  userExposedCaption="CoordinateBasis: Circle Splat (Standard)"
  libraryCategoriesText [Base Material]
 }
 nodes [
  n1 = MaterialFunctionCall!(
   Coords=NamedRerouteUsage()
   "TileRand (Min/Max)"=NamedRerouteUsage()
   "OffsetRand (Min/Max)"=NamedRerouteUsage()
   TileStrengthRand
   OffsetStrengthRand
   RotateStrengthRand=NamedRerouteUsage()
   SeedA=NamedRerouteUsage()
  ) {materialFunction=&/MF_Coordinate_Rand1}
  FunctionOutput(None=@n1.UV) {description="First randomized UV coordinate set generated by the function.\r\n", outputName=UV1}
  n2 = FunctionInput() {
   bUsePreviewValueAsDefault=true
   description="Selects which mesh UV channel (0–3) to use as the base coordinate source."
   inputName=UVIndex
   inputType=FunctionInput_Scalar
   sortPriority=-5
  }
  n3 = FunctionInput() {
   bUsePreviewValueAsDefault=true
   description=Applies an additional per-axis offset to the coordinates before projection. Useful for correctly positioning a texture.
   inputName=CoordinateTransform
   inputType=FunctionInput_Vector2
   sortPriority=-9
  }
  n4 = FunctionInput() {
   bUsePreviewValueAsDefault=true
   description="Rotates the projected UVs around the selected axis. Units are in turns, where 1.0 = 360° and 0.25 = 90°."
   inputName="Rotation (Turns)"
   inputType=FunctionInput_Scalar
   sortPriority=-8
  }
  n5 = FunctionInput() {
   bUsePreviewValueAsDefault=true
   description="Sets how many times the projected coordinates repeat along each axis. Use values above 1 for denser tiling, negative values for mirroring, and avoid zero."
   inputName=CoordinateRepeat
   inputType=FunctionInput_Vector2
   sortPriority=-10
   previewValue=<1 1 1 1>
  }
  n6 = MaterialFunctionCall!(
   Coords=MaterialFunctionCall!(
    Selector=@n2
    VectorA=TextureCoordinate()
    VectorB=TextureCoordinate() {coordinateIndex=1}
    VectorC=TextureCoordinate() {coordinateIndex=2}
    VectorD=TextureCoordinate() {coordinateIndex=3}
   ) {materialFunction=&/MF_Switch4_Vec2}.Selected Vector
   UVOffset=@n3
   UVRotation=@n4
   RotationPivot
   UVTile=@n5
  ) {materialFunction=&/MF_CoordinateTransform_2D}
  n7 = MaterialFunctionCall!(packedDXY=@n6.geometry_packedDXY) {materialFunction=&/MF_DeriveTangentBasis}
  FunctionOutput(None=@n7.Tangent) {description="World-space tangent vector for UV2. Use with Bitangent2WS for normal reorientation.\r\n", outputName=Tangent2, sortPriority=6}
  FunctionOutput(None=@n7.Bitangent) {
   description="World-space bitangent vector for UV2. Complements Tangent2WS to form the basis for tangent-space normal reorientation.\r\n"
   outputName=Bitangent2
   sortPriority=7
  }
  FunctionOutput(None=@n1.SampleRotation) {
   description="Rotation applied to UV1, expressed in turns (1.0 = 360°). Used for correct tangent-space normal reorientation.\r\n"
   outputName="Rotation1 (Turns)"
   sortPriority=-6
  }
  FunctionOutput(None=@n6.geometry_packedDXY) {
   description="Packed derivatives (ddx/ddy) for UV1. Use with SampleGrad for correct mip level selection during texture sampling."
   outputName=DXY1
   sortPriority=3
  }
  n8 = MaterialFunctionCall(
   Coords=NamedRerouteUsage()
   "TileRand (Min/Max)"=NamedRerouteUsage()
   "OffsetRand (Min/Max)"=NamedRerouteUsage()
   RotateStrengthRand=NamedRerouteUsage()
   SeedA=NamedRerouteUsage()
  ) {materialFunction=&MF_Coordinate_Rand1}
  FunctionOutput(None=@n8.UV) {description="Second randomized UV coordinate set generated by the function.\r\n", outputName=UV2, sortPriority=5}
  FunctionOutput(None=@n6.geometry_packedDXY) {
   description="Packed derivatives (ddx/ddy) for UV2. Use with SampleGrad for correct mip level selection during texture sampling.\r\n"
   outputName=DXY2
   sortPriority=8
  }
  FunctionOutput(None=@n8.SampleRotation) {
   description="Rotation applied to UV2, expressed in turns (1.0 = 360°). Used for correct tangent-space normal reorientation."
   outputName="Rotation2 (Turns)"
   sortPriority=-5
  }
  n9 = FunctionInput() {
   bUsePreviewValueAsDefault=true
   description="Controls the tiling of the splat pattern. Larger values create more, smaller splats across the surface; smaller values create fewer, larger splats."
   inputName=Splat PatternSize
   inputType=FunctionInput_Scalar
   sortPriority=2
   previewValue=<1 0 0 1>
  }
  n10 = MaterialFunctionCall!(Coordinates=Multiply(A=@n9, B=@n6.geometry_UV)) {materialFunction=&/MF_Blend_CircleMask}
  FunctionOutput(None=@n10.PatternMask) {
   description="Outputs the radial circle splat pattern used to blend between the two randomized coordinate sets (UV1 and UV2), where the RG channels provide the weights for each sample. "
   outputName=BlendMask
   sortPriority=15
  }
  FunctionOutput(None=@n7.Tangent) {description="World-space tangent vector for UV1. Use with Bitangent1WS for normal reorientation.\r\n", outputName=Tangent1, sortPriority=1}
  FunctionOutput(None=@n7.Bitangent) {
   description="World-space bitangent vector for UV1. Complements Tangent1WS to form the basis for tangent-space normal reorientation.\r\n"
   outputName=Bitangent1
   sortPriority=2
  }
  n11 = FunctionInput() {
   bUsePreviewValueAsDefault=true
   description="Defines the minimum and maximum random UV offset applied to each splat. Offsets shift the splat’s position within its cell, breaking up repetition."
   inputName=Splat Min/MaxRand Offset
   inputType=FunctionInput_Vector2
   sortPriority=4
  }
  NamedRerouteDeclaration(None=@n11) {name=CircleOffset, nodeColor=<rgba 0.9 0.3505 0.009 1>}
  n12 = FunctionInput() {
   bUsePreviewValueAsDefault=true
   description="Controls random per-splat rotation. Expressed in turns (1.0 = 360°). Each splat is rotated by a random amount up to this value. "
   inputName="Splat Rand Rotation (turns)"
   inputType=FunctionInput_Scalar
   sortPriority=5
  }
  NamedRerouteDeclaration(None=@n12) {name=CircleRot, nodeColor=<rgba 0.9 0.3505 0.009 1>}
  n13 = FunctionInput() {
   bUsePreviewValueAsDefault=true
   description="Defines the minimum and maximum random scale applied to each splat. Values below 1 enlarge splats, values above 1 shrink them. A random value within this range is chosen per splat."
   inputName=Splat Min/MaxRand Scale
   inputType=FunctionInput_Vector2
   sortPriority=3
   previewValue=<1 1 0 1>
  }
  NamedRerouteDeclaration(None=@n13) {name=CircleScale, nodeColor=<rgba 0.9 0.3505 0.009 1>}
  NamedRerouteDeclaration(None=MaterialFunctionCall!(seed=@n10.CircleID1) {materialFunction=&/MF_Hash42}.Hash4D) {name=RandA, nodeColor=<rgba 1 0 0.75 1>}
  NamedRerouteDeclaration(None=@n6.geometry_UV) {name=UVi, nodeColor=<rgba 0.9 0.009 0.009 1>}
  NamedRerouteDeclaration(None=MaterialFunctionCall(seed=@n10.CircleID2) {materialFunction=&MF_Hash42}.Hash4D) {name=RandB, nodeColor=<rgba 1 0 0.75 1>}
 ]
}
