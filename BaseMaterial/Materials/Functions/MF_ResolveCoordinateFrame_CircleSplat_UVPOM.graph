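# Material function that builds two randomized, parallax-occluded UV sets in a circle splat pattern, for breaking up texture repetition.
# A selected, transformed UV channel drives a circle mask, whose two cell IDs are hashed to randomize coordinates.
# Each set gets its own height-map POM pass, outputting UV, depth and rotation.
# Tangent bases and packed derivatives come from the pre-randomized UVs; the mask is output for post-sample blending.
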MaterialFunction /BaseMaterial/Materials/Functions/MF_ResolveCoordinateFrame_CircleSplat_UVPOM
graph {
 properties {
  bExposeToLibrary=true
  description="Standard Circle Splat with Parallax Occlusion Mapping (POM) for added depth and occlusion. Generates the splat pattern from the chosen UV channel and outputs two randomized coordinate sets (UV1 and UV2) with full bases and derivatives, plus a BlendMask (RG) for post-sample blending. Supports post-sample modulation of blend weights by albedo or height, yielding more perceptually correct blends while also adding parallax depth to the splats."
  userExposedCaption="CoordinateBasis: Circle Splat (Standard POM)"
  libraryCategoriesText [Base Material]
 }
 nodes [
  n1 = MaterialFunctionCall!(
   Coords=NamedRerouteUsage()
   "TileRand (Min/Max)"=NamedRerouteUsage()
   "OffsetRand (Min/Max)"=NamedRerouteUsage()
   TileStrengthRand
   OffsetStrengthRand
   RotateStrengthRand=NamedRerouteUsage()
   SeedA=NamedRerouteUsage()
  ) {materialFunction=&/MF_Coordinate_Rand1}
  n2 = FunctionInput() {
   bUsePreviewValueAsDefault=true
   description=Control the U/V Tiling of the height texture relative to the input coordinates.
   inputName=POM HeightTile
   inputType=FunctionInput_Vector2
   sortPriority=12
   previewValue=<1 1 0 1>
  }
  n3 = FunctionOutput(None=@n1.SampleRotation) {
   description="Rotation applied to UV1, expressed in turns (1.0 = 360°). Used for correct tangent-space normal reorientation.\r\n"
   outputName="Rotation1 (Turns)"
   sortPriority=-6
  }
  n4 = MaterialFunctionCall!(
   POM_HeightTexture=NamedRerouteUsage()
   Coordinates=@n1.UV
   coordinateDXY=NamedRerouteUsage()
   HeightStrength=NamedRerouteUsage()
   Reference Plane=NamedRerouteUsage()
   MaxSteps=NamedRerouteUsage()
   Tangent=NamedRerouteUsage()
   Bitangent=NamedRerouteUsage()
   "Rotation (Turns)"=@n3
   HeightChannel=NamedRerouteUsage()
   QualityBias=NamedRerouteUsage()
   CameraVector
   Height Tile=@n2
  ) {materialFunction=&/MF_Coordinate_Parallax}
  FunctionOutput(None=@n4.POM UVs) {description="First randomized UV coordinate set generated by the function.\r\n", outputName=UV1}
  n5 = FunctionInput() {
   bUsePreviewValueAsDefault=true
   description="Selects which mesh UV channel (0–3) to use as the base coordinate source."
   inputName=UVIndex
   inputType=FunctionInput_Scalar
   sortPriority=-5
  }
  n6 = FunctionInput() {
   bUsePreviewValueAsDefault=true
   description=Applies an additional per-axis offset to the coordinates before projection. Useful for correctly positioning a texture.
   inputName=CoordinateTransform
   inputType=FunctionInput_Vector2
   sortPriority=-9
  }
  n7 = FunctionInput() {
   bUsePreviewValueAsDefault=true
   description="Rotates the projected UVs around the selected axis. Units are in turns, where 1.0 = 360° and 0.25 = 90°."
   inputName="Rotation (Turns)"
   inputType=FunctionInput_Scalar
   sortPriority=-8
  }
  n8 = FunctionInput() {
   bUsePreviewValueAsDefault=true
   description="Sets how many times the projected coordinates repeat along each axis. Use values above 1 for denser tiling, negative values for mirroring, and avoid zero."
   inputName=CoordinateRepeat
   inputType=FunctionInput_Vector2
   sortPriority=-10
   previewValue=<1 1 1 1>
  }
  n9 = MaterialFunctionCall!(
   Coords=MaterialFunctionCall!(
    Selector=@n5
    VectorA=TextureCoordinate()
    VectorB=TextureCoordinate() {coordinateIndex=1}
    VectorC=TextureCoordinate() {coordinateIndex=2}
    VectorD=TextureCoordinate() {coordinateIndex=3}
   ) {materialFunction=&/MF_Switch4_Vec2}.Selected Vector
   UVOffset=@n6
   UVRotation=@n7
   RotationPivot
   UVTile=@n8
  ) {materialFunction=&/MF_CoordinateTransform_2D}
  n10 = NamedRerouteDeclaration(None=@n9.geometry_packedDXY) {name=DXY1, nodeColor=<rgba 0.9 0.009 0.009 1>}
  n11 = MaterialFunctionCall!(packedDXY=@n10) {materialFunction=&/MF_DeriveTangentBasis}
  FunctionOutput(None=NamedRerouteDeclaration(None=@n11.Tangent) {name=T2, nodeColor=<rgba 0.009 0.9 0.009 1>}) {description="World-space tangent vector for UV2. Use with Bitangent2WS for normal reorientation.\r\n", outputName=Tangent2, sortPriority=6}
  FunctionOutput(None=NamedRerouteDeclaration(None=@n11.Bitangent) {name=B2, nodeColor=<rgba 0.009 0.9 0.009 1>}) {
   description="World-space bitangent vector for UV2. Complements Tangent2WS to form the basis for tangent-space normal reorientation.\r\n"
   outputName=Bitangent2
   sortPriority=7
  }
  FunctionOutput(None=@n4.Pixel Depth Offset) {
   description="Depth at the POM intersection (after Intensity/RefPlane). Higher (whiter) = deeper—the farther the ray travels into the material, the larger the value."
   outputName=Depth1
   sortPriority=4
  }
  n12 = MaterialFunctionCall(
   Coords=NamedRerouteUsage()
   "TileRand (Min/Max)"=NamedRerouteUsage()
   "OffsetRand (Min/Max)"=NamedRerouteUsage()
   RotateStrengthRand=NamedRerouteUsage()
   SeedA=NamedRerouteUsage()
  ) {materialFunction=&MF_Coordinate_Rand1}
  n13 = FunctionOutput(None=@n12.SampleRotation) {
   description="Rotation applied to UV2, expressed in turns (1.0 = 360°). Used for correct tangent-space normal reorientation.\r\n"
   outputName="Rotation2 (Turns)"
   sortPriority=-5
  }
  n14 = MaterialFunctionCall(
   POM_HeightTexture=NamedRerouteUsage()
   Coordinates=@n12.UV
   coordinateDXY=NamedRerouteUsage()
   HeightStrength=NamedRerouteUsage()
   Reference Plane=NamedRerouteUsage()
   MaxSteps=NamedRerouteUsage()
   Tangent=NamedRerouteUsage()
   Bitangent=NamedRerouteUsage()
   "Rotation (Turns)"=@n13
   HeightChannel=NamedRerouteUsage()
   QualityBias=NamedRerouteUsage()
   Height Tile=@n2
  ) {materialFunction=&MF_Coordinate_Parallax}
  FunctionOutput(None=@n14.Pixel Depth Offset) {
   description="Depth at the POM intersection (after Intensity/RefPlane). Higher (whiter) = deeper—the farther the ray travels into the material, the larger the value."
   outputName=Depth2
   sortPriority=9
  }
  FunctionOutput(None=@n10) {
   description="Packed derivatives (ddx/ddy) for UV1. Use with SampleGrad for correct mip level selection during texture sampling.\r\n"
   outputName=DXY1
   sortPriority=3
  }
  FunctionOutput(None=@n14.POM UVs) {description="Second randomized UV coordinate set generated by the function.\r\n", outputName=UV2, sortPriority=5}
  FunctionOutput(None=NamedRerouteDeclaration(None=@n9.geometry_packedDXY) {name=DXY2, nodeColor=<rgba 0.009 0.9 0.009 1>}) {
   description="Packed derivatives (ddx/ddy) for UV2. Use with SampleGrad for correct mip level selection during texture sampling.\r\n"
   outputName=DXY2
   sortPriority=8
  }
  n15 = FunctionInput() {
   bUsePreviewValueAsDefault=true
   description="Controls the tiling of the splat pattern. Larger values create more, smaller splats across the surface; smaller values create fewer, larger splats."
   inputName=Splat PatternSize
   inputType=FunctionInput_Scalar
   sortPriority=2
   previewValue=<1 0 0 1>
  }
  n16 = MaterialFunctionCall!(Coordinates=Multiply(A=@n15, B=@n9.geometry_UV)) {materialFunction=&/MF_Blend_CircleMask}
  FunctionOutput(None=@n16.PatternMask) {
   description="Outputs the radial circle splat pattern used to blend between the two randomized coordinate sets (UV1 and UV2), where the RG channels provide the weights for each sample. \r\n"
   outputName=BlendMask
   sortPriority=15
  }
  FunctionOutput(None=NamedRerouteDeclaration(None=@n11.Tangent) {name=T1, nodeColor=<rgba 0.9 0.009 0.009 1>}) {description="World-space tangent vector for UV1. Use with Bitangent1WS for normal reorientation.\r\n", outputName=Tangent1, sortPriority=1}
  FunctionOutput(None=NamedRerouteDeclaration(None=@n11.Bitangent) {name=B1, nodeColor=<rgba 0.9 0.009 0.009 1>}) {
   description="World-space bitangent vector for UV1. Complements Tangent1WS to form the basis for tangent-space normal reorientation.\r\n"
   outputName=Bitangent1
   sortPriority=2
  }
  NamedRerouteDeclaration(None=@n16.PatternMask) {name=Blendi, nodeColor=<rgba 1 1 1 1>}
  n17 = FunctionInput() {
   bUsePreviewValueAsDefault=true
   description="Defines the minimum and maximum random UV offset applied to each splat. Offsets shift the splat’s position within its cell, breaking up repetition."
   inputName=Splat Min/MaxRand Offset
   inputType=FunctionInput_Vector2
   sortPriority=4
  }
  NamedRerouteDeclaration(None=@n17) {name=CircleOffset, nodeColor=<rgba 0.9 0.3505 0.009 1>}
  n18 = FunctionInput() {
   bUsePreviewValueAsDefault=true
   description="Controls random per-splat rotation. Expressed in turns (1.0 = 360°). Each splat is rotated by a random amount up to this value. "
   inputName="Splat Rand Rotation (turns)"
   inputType=FunctionInput_Scalar
   sortPriority=5
  }
  NamedRerouteDeclaration(None=@n18) {name=CircleRot, nodeColor=<rgba 0.9 0.3505 0.009 1>}
  n19 = FunctionInput() {
   bUsePreviewValueAsDefault=true
   description="Defines the minimum and maximum random scale applied to each splat. Values below 1 enlarge splats, values above 1 shrink them. A random value within this range is chosen per splat."
   inputName=Splat Min/MaxRand Scale
   inputType=FunctionInput_Vector2
   sortPriority=3
   previewValue=<1 1 0 1>
  }
  NamedRerouteDeclaration(None=@n19) {name=CircleScale, nodeColor=<rgba 0.9 0.3505 0.009 1>}
  NamedRerouteDeclaration(None=MaterialFunctionCall!(seed=@n16.CircleID1) {materialFunction=&/MF_Hash42}.Hash4D) {name=RandA, nodeColor=<rgba 1 0 0.75 1>}
  n20 = FunctionInput(Preview=TextureObject() {texture=/BaseMaterial/Textures/Default/T_Gray_sRGB}) {
   bUsePreviewValueAsDefault=true
   description="The heightmap texture used for parallax occlusion mapping. Grayscale values define surface depth, with brighter values typically pushed outward and darker values pushed inward."
   inputName=HeightTex
   inputType=FunctionInput_Texture2D
   sortPriority=-20
  }
  NamedRerouteDeclaration(None=@n20) {name=Height, nodeColor=<rgba 1 0 0.4922 1>}
  n21 = FunctionInput() {
   bUsePreviewValueAsDefault=true
   description="Specifies which channel of the heightmap texture to use for depth information. Assumes the Alpha channel is used, but any channel can be selected if your texture packs data differently."
   inputName=POM HeightChannel
   inputType=FunctionInput_Scalar
   sortPriority=10
   previewValue=<3 0 0 1>
  }
  NamedRerouteDeclaration(None=@n21) {name=POM_Hchan, nodeColor=<rgba 0.9 0.306 0.009 1>}
  n22 = FunctionInput() {
   bUsePreviewValueAsDefault=true
   description="Controls the strength of the parallax effect. Higher values exaggerate depth; lower values make surfaces appear flatter."
   inputName=POM Intensity
   inputType=FunctionInput_Scalar
   sortPriority=11
   previewValue=<1 0 0 1>
  }
  NamedRerouteDeclaration(None=@n22) {name=POM_Height, nodeColor=<rgba 0.9 0.306 0.009 1>}
  n23 = FunctionInput() {
   bUsePreviewValueAsDefault=true
   description="Sets the maximum number of depth sampling steps. This is the upper bound, with actual steps scaled by the Quality setting."
   inputName=POM MaxSteps
   inputType=FunctionInput_Scalar
   sortPriority=12
   previewValue=<16 0 0 1>
  }
  NamedRerouteDeclaration(None=@n23) {name=POM_MaxSteps, nodeColor=<rgba 0.9 0.306 0.009 1>}
  n24 = FunctionInput() {
   bUsePreviewValueAsDefault=true
   description="Scales step count by view angle. At 1, MaxSteps are always used; at 0, only one step is used when looking straight on, with more steps added at grazing angles."
   inputName=POM Quality
   inputType=FunctionInput_Scalar
   sortPriority=13
   previewValue=<0.25 0 0 1>
  }
  NamedRerouteDeclaration(None=@n24) {name=POM_Noise, nodeColor=<rgba 0.9 0.306 0.009 1>}
  n25 = FunctionInput() {
   bUsePreviewValueAsDefault=true
   description=Defines the reference plane for parallax. This value shifts where the heightmap appears anchored relative to the surface.
   inputName=POM RefPlane
   inputType=FunctionInput_Scalar
   sortPriority=14
   previewValue=<1 0 0 1>
  }
  NamedRerouteDeclaration(None=@n25) {name=POM_RefPlane, nodeColor=<rgba 0.9 0.306 0.009 1>}
  NamedRerouteDeclaration(None=@n9.geometry_UV) {name=UVi, nodeColor=<rgba 0.9 0.009 0.009 1>}
  NamedRerouteDeclaration(None=MaterialFunctionCall(seed=@n16.CircleID2) {materialFunction=&MF_Hash42}.Hash4D) {name=RandB, nodeColor=<rgba 1 0 0.75 1>}
 ]
}
