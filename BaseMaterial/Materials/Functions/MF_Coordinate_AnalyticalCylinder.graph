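# Material function that generates cylindrical UVs around an object's local Z axis, with matching tangent vectors and analytically derived UV gradients.
# A custom HLSL node takes instance-space position times object scale, normalized position as the normal, and object radius for sizing.
# It chains hand-written UV partials with screen-space position derivatives, packing the gradients into one float4.
# Tangent and bitangent come from cross products and are transformed out of instance space.
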MaterialFunction /BaseMaterial/Materials/Functions/MF_Coordinate_AnalyticalCylinder
graph {nodes [
  n1 = TransformPosition(None=WorldPosition() {worldPositionShaderOffset=WPT_CameraRelative}.XYZ) {transformSourceType=TRANSFORMPOSSOURCE_TranslatedWorld, transformType=TRANSFORMPOSSOURCE_Instance}
  MaterialExpressionMultiply_1 = Multiply(A=MaterialFunctionCall() {materialFunction=&ObjectScale}.Scale XYZ, B=@n1)
  MaterialExpressionNormalize_0 = Normalize(VectorInput=@n1)
  MaterialExpressionObjectRadius_0 = ObjectRadius()
  n2 = Custom!(p=@MaterialExpressionMultiply_1, n=@MaterialExpressionNormalize_0, size=@MaterialExpressionObjectRadius_0) {
   code="float3 dpdx = ddx(p);\r\nfloat3 dpdy = ddy(p);\r\n\r\n//Cylindrical UV Mapping\r\nfloat2 uv = float2( atan2(p.y, p.x)/(PI), -p.z/(2.f * size) + 0.5 );\r\n\r\n//analytical derivatives\r\nfloat denom =  2.f * PI * (p.x*p.x + p.y*p.y + 1e-6);\r\nfloat3 dudp = float3( -p.y/denom, -p.x/denom, 0.f);\r\nfloat3 dvdp = float3(  0., 0., 0.5);\r\n\r\nfloat3 tangent = cross(float3(0,0,1), n);\r\nfloat3 bitangent = normalize( cross(tangent, n) );\r\n\r\n//transform derivatives to screen space\r\nfloat4 uv_derivs = float4( dot(dudp,dpdx), dot(dvdp, dpdx),\r\n                           dot(dudp,dpdy), dot(dvdp, dpdy) );\r\n\r\n//scale derivatives\r\nfloat2 derivSize = (MaterialFloat2)View.MaterialTextureDerivativeMultiply;\r\nuv_derivs *= rcp(size);\r\nuv_derivs *= derivSize.xyxy;\r\n\r\n// Outputs\r\nout_tangent = tangent;\r\nout_bitangent = bitangent;\r\nout_derivs = uv_derivs;\r\n\r\nreturn uv;"
   description=MakeCylinder
   outputType=CMOT_Float2
   showCode=true
   additionalOutputs [
    {outputName=out_tangent, outputType=CMOT_Float3}
    {outputName=out_bitangent, outputType=CMOT_Float3}
    {outputName=out_derivs, outputType=CMOT_Float4}
   ]
   inputs [
    {inputName=p, input {expression=&self:MaterialExpressionMultiply_1}}
    {inputName=n, input {expression=&self:MaterialExpressionNormalize_0}}
    {inputName=size, input {expression=&self:MaterialExpressionObjectRadius_0}}
   ]
  }
  FunctionOutput(None=@n2.return) {outputName=UV}
  FunctionOutput(None=@n2.out_derivs) {outputName=packedDXY, sortPriority=3}
  FunctionOutput(None=Transform(None=@n2.out_bitangent) {transformSourceType=TRANSFORMSOURCE_Instance}) {outputName=Bitangent, sortPriority=2}
  FunctionOutput(None=Transform(None=@n2.out_tangent) {transformSourceType=TRANSFORMSOURCE_Instance}) {outputName=Tangent, sortPriority=1}
 ]}
