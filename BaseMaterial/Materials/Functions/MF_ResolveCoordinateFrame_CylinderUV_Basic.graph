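# Material function in the Base Material library that resolves a cylindrical UV coordinate frame for materials.
# It takes UVs from MF_Coordinate_AnalyticalCylinder and runs them through MF_CoordinateTransform_2D with repeat, offset and rotation-in-turns inputs, pivoting rotation at U=0, V=0.5.
# It outputs the transformed UV and one minus the rotation, and passes the cylinder's tangent, bitangent and packed derivatives (DXY) through unmodified.
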MaterialFunction /BaseMaterial/Materials/Functions/MF_ResolveCoordinateFrame_CylinderUV_Basic
graph {
 properties {bExposeToLibrary=true, libraryCategoriesText [Base Material]}
 nodes [
  n1 = MaterialFunctionCall() {materialFunction=&/MF_Coordinate_AnalyticalCylinder}
  n2 = NamedRerouteDeclaration(None=FunctionInput() {bUsePreviewValueAsDefault=true, inputName="Rotation (Turns)", inputType=FunctionInput_Scalar, sortPriority=-8}) {name=turns, nodeColor=<rgba 0.8963 0.8963 0.0091 1>}
  n3 = FunctionInput() {bUsePreviewValueAsDefault=true, inputName=CoordinateRepeat, inputType=FunctionInput_Vector2, sortPriority=-10, previewValue=<1 1 1 1>}
  FunctionOutput(None=MaterialFunctionCall!(
    Coords=@n1.UV
    UVOffset=FunctionInput() {bUsePreviewValueAsDefault=true, inputName=CoordinateTransform, inputType=FunctionInput_Vector2, sortPriority=-9}
    UVRotation=@n2
    RotationPivot=Constant2Vector() {g=0.5}
    UVTile=@n3
   ) {materialFunction=&/MF_CoordinateTransform_2D}.geometry_UV) {outputName=UV}
  FunctionOutput(None=OneMinus(None=@n2)) {outputName="Rotation (Turns)", sortPriority=-5}
  FunctionOutput(None=@n1.packedDXY) {outputName=DXY, sortPriority=3}
  FunctionOutput(None=@n1.Tangent) {outputName=Tangent, sortPriority=1}
  FunctionOutput(None=@n1.Bitangent) {outputName=Bitangent, sortPriority=2}
 ]
}
