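# Material function that resolves an analytical cylinder coordinate frame with parallax occlusion mapped UVs for Base Material shaders.
# Cylinder UVs pass through a 2D repeat, offset and rotation transform pivoted on the seam-edge midpoint (0, 0.5).
# The parallax function then displaces them using a height texture, channel, POM settings, camera vector and the cylinder's derivatives and tangent basis.
# It outputs UV, DXY, Tangent, Bitangent and a one-minus rotation.
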MaterialFunction /BaseMaterial/Materials/Functions/MF_ResolveCoordinateFrame_CylinderUV_POM
graph {
 properties {bExposeToLibrary=true, libraryCategoriesText [Base Material]}
 nodes [
  n1 = TextureObject() {texture=/BaseMaterial/Textures/Default/T_Gray_sRGB}
  n2 = MaterialFunctionCall() {materialFunction=&/MF_Coordinate_AnalyticalCylinder}
  n3 = FunctionInput() {bUsePreviewValueAsDefault=true, inputName=CoordinateTransform, inputType=FunctionInput_Vector2, sortPriority=-9}
  n4 = NamedRerouteDeclaration(None=FunctionInput() {bUsePreviewValueAsDefault=true, inputName="Rotation (Turns)", inputType=FunctionInput_Scalar, sortPriority=-8}) {name=turns, nodeColor=<rgba 0.8963 0.8963 0.0091 1>}
  n5 = Constant2Vector() {g=0.5}
  n6 = FunctionInput() {bUsePreviewValueAsDefault=true, inputName=CoordinateRepeat, inputType=FunctionInput_Vector2, sortPriority=-10, previewValue=<1 1 1 1>}
  n7 = NamedRerouteUsage()
  FunctionOutput(None=MaterialFunctionCall!(
    POM_HeightTexture=FunctionInput(Preview=@n1) {bUsePreviewValueAsDefault=true, inputName=HeightTex, inputType=FunctionInput_Texture2D, sortPriority=-20}
    Coordinates=MaterialFunctionCall!(Coords=@n2.UV, UVOffset=@n3, UVRotation=@n4, RotationPivot=@n5, UVTile=@n6) {materialFunction=&/MF_CoordinateTransform_2D}.geometry_UV
    coordinateDXY=FunctionOutput(None=@n2.packedDXY) {outputName=DXY, sortPriority=3}
    HeightStrength=NamedRerouteUsage()
    Reference Plane=NamedRerouteUsage()
    MaxSteps=NamedRerouteUsage()
    Tangent=FunctionOutput(None=@n2.Tangent) {outputName=Tangent, sortPriority=1}
    Bitangent=FunctionOutput(None=@n2.Bitangent) {outputName=Bitangent, sortPriority=2}
    "Rotation (Turns)"=@n7
    HeightChannel=FunctionInput() {bUsePreviewValueAsDefault=true, inputName=POM HeightChannel, inputType=FunctionInput_Scalar, sortPriority=7}
    QualityBias=NamedRerouteUsage()
    CameraVector
   ) {materialFunction=&/MF_Coordinate_Parallax}.POM UVs) {outputName=UV}
  FunctionOutput(None=OneMinus(None=@n7)) {outputName="Rotation (Turns)", sortPriority=-5}
  n8 = FunctionInput() {bUsePreviewValueAsDefault=true, inputName=POM Intensity, inputType=FunctionInput_Scalar, sortPriority=8, previewValue=<1 0 0 1>}
  NamedRerouteDeclaration(None=@n8) {name=POM_Height, nodeColor=<rgba 0.9 0.306 0.009 1>}
  n9 = FunctionInput() {bUsePreviewValueAsDefault=true, inputName=POM MaxSteps, inputType=FunctionInput_Scalar, sortPriority=9, previewValue=<16 0 0 1>}
  NamedRerouteDeclaration(None=@n9) {name=POM_MaxSteps, nodeColor=<rgba 0.9 0.306 0.009 1>}
  n10 = FunctionInput() {bUsePreviewValueAsDefault=true, inputName=POM Quality, inputType=FunctionInput_Scalar, sortPriority=10, previewValue=<0.25 0 0 1>}
  NamedRerouteDeclaration(None=@n10) {name=POM_Noise, nodeColor=<rgba 0.9 0.306 0.009 1>}
  n11 = FunctionInput() {bUsePreviewValueAsDefault=true, inputName=POM RefPlane, inputType=FunctionInput_Scalar, sortPriority=11, previewValue=<1 0 0 1>}
  NamedRerouteDeclaration(None=@n11) {name=POM_RefPlane, nodeColor=<rgba 0.9 0.306 0.009 1>}
 ]
}
