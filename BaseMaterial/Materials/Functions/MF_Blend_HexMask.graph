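# Material function that builds a hex-tile blend mask and three per-hex IDs from UVs, for hex tiling with randomized coordinates.
# It skews UVs into a triangle grid, derives HexID1-3 from floored cell coordinates and grid/triangle masks, and builds three gradients from the fractional part.
# MaskContrast sharpens the gradients through a power, then they are normalized into PatternMask.
# An unconnected Custom node keeps the original HLSL, citing Mikkelsen's hex-tiling paper.
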MaterialFunction /BaseMaterial/Materials/Functions/MF_Blend_HexMask
graph {
 expressions {
  MaterialExpressionAbs_0 {class=MaterialExpressionAbs, inputs [None {from=MaterialExpressionSubtract_1, output=""}], properties {}}
  MaterialExpressionAdd_1 {
   class=MaterialExpressionAdd
   inputs [
    A {from=MaterialExpressionConvert_0}
    B {from=MaterialExpressionConvert_0}
   ]
  }
  MaterialExpressionAdd_10 {
   class=MaterialExpressionAdd
   inputs [
    A {from=MaterialExpressionReroute_0}
    B {from=MaterialExpressionAdd_5}
   ]
  }
  MaterialExpressionAdd_11 {
   class=MaterialExpressionAdd
   inputs [
    A {from=MaterialExpressionConvert_11}
    B {from=MaterialExpressionConvert_11}
   ]
  }
  MaterialExpressionAdd_12 {
   class=MaterialExpressionAdd
   inputs [
    A {from=MaterialExpressionConvert_11}
    B {from=MaterialExpressionAdd_11}
   ]
  }
  MaterialExpressionAdd_2 {
   class=MaterialExpressionAdd
   inputs [
    A {from=MaterialExpressionSubtract_0}
    B {from=MaterialExpressionConstant3Vector_0}
   ]
  }
  MaterialExpressionAdd_3 {
   class=MaterialExpressionAdd
   inputs [
    A {from=MaterialExpressionMultiply_5}
    B {from=MaterialExpressionDivide_1}
   ]
  }
  MaterialExpressionAdd_4 {
   class=MaterialExpressionAdd
   inputs [
    A {from=MaterialExpressionAdd_3}
    B {from=MaterialExpressionConstant_7}
   ]
  }
  MaterialExpressionAdd_5 {
   class=MaterialExpressionAdd
   inputs [
    A {from=MaterialExpressionConvert_3}
    B {from=MaterialExpressionConvert_2}
   ]
  }
  MaterialExpressionAdd_6 {
   class=MaterialExpressionAdd
   inputs [
    A {from=MaterialExpressionConvert_3}
    B {from=MaterialExpressionConvert_2}
   ]
  }
  MaterialExpressionAdd_7 {
   class=MaterialExpressionAdd
   inputs [
    A {from=MaterialExpressionConvert_3}
    B {from=MaterialExpressionConvert_2}
   ]
  }
  MaterialExpressionAdd_8 {
   class=MaterialExpressionAdd
   inputs [
    A {from=MaterialExpressionReroute_0}
    B {from=MaterialExpressionAdd_7}
   ]
  }
  MaterialExpressionAdd_9 {
   class=MaterialExpressionAdd
   inputs [
    A {from=MaterialExpressionReroute_0}
    B {from=MaterialExpressionAdd_6}
   ]
  }
  MaterialExpressionAppendVector_0 {
   class=MaterialExpressionAppendVector
   inputs [
    A {from=MaterialExpressionDotProduct_0, output=""}
    B {from=MaterialExpressionDotProduct_1, output=""}
   ]
  }
  MaterialExpressionConstant2Vector_0 {class=MaterialExpressionConstant2Vector, properties {g=-0.5774, r=1}}
  MaterialExpressionConstant2Vector_1 {class=MaterialExpressionConstant2Vector, properties {g=1.1547}}
  MaterialExpressionConstant3Vector_0 {class=MaterialExpressionConstant3Vector, properties {constant=<rgba 0 1 2 0>}}
  MaterialExpressionConstant_0 {class=MaterialExpressionConstant, properties {r=3}}
  MaterialExpressionConstant_3 {class=MaterialExpressionConstant, properties {r=1}}
  MaterialExpressionConstant_4 {class=MaterialExpressionConstant, properties {r=3}}
  MaterialExpressionConstant_5 {class=MaterialExpressionConstant, properties {r=5}}
  MaterialExpressionConstant_6 {class=MaterialExpressionConstant, properties {r=3}}
  MaterialExpressionConstant_7 {class=MaterialExpressionConstant, properties {r=0}}
  MaterialExpressionConvert_0 {
   class=MaterialExpressionConvert
   inputs [None {from=MaterialExpressionFrac_1}]
   properties {
    nodeName=Break Float2
    convertInputs [{
      type=Vector2
      expressionInput {
       expression=&self:MaterialExpressionFrac_1
       inputName=None
       mask=0
       maskA=0
       maskB=0
       maskG=0
       maskR=0
       outputIndex=0
      }
     }]
    convertOutputs [{type=Scalar}, {type=Scalar}]
   }
  }
  MaterialExpressionConvert_1 {
   class=MaterialExpressionConvert
   inputs [None {from=MaterialExpressionFloor_0}]
   properties {
    nodeName=Break Float2
    convertInputs [{
      type=Vector2
      expressionInput {
       expression=&self:MaterialExpressionFloor_0
       inputName=None
       mask=0
       maskA=0
       maskB=0
       maskG=0
       maskR=0
       outputIndex=0
      }
     }]
    convertOutputs [{type=Scalar}, {type=Scalar}]
   }
  }
  MaterialExpressionConvert_10 {
   class=MaterialExpressionConvert
   inputs [
    None {from=MaterialExpressionDotProduct_2}
    None {from=MaterialExpressionDotProduct_3}
    None {from=MaterialExpressionDotProduct_4}
   ]
   properties {
    desc=hexGrad
    nodeName=Make Vector3
    convertInputs [
     {
      type=Scalar
      expressionInput {
       expression=&self:MaterialExpressionDotProduct_2
       inputName=None
       mask=0
       maskA=0
       maskB=0
       maskG=0
       maskR=0
       outputIndex=0
      }
     }
     {
      type=Scalar
      expressionInput {
       expression=&self:MaterialExpressionDotProduct_3
       inputName=None
       mask=0
       maskA=0
       maskB=0
       maskG=0
       maskR=0
       outputIndex=0
      }
     }
     {
      type=Scalar
      expressionInput {
       expression=&self:MaterialExpressionDotProduct_4
       inputName=None
       mask=0
       maskA=0
       maskB=0
       maskG=0
       maskR=0
       outputIndex=0
      }
     }
    ]
    convertOutputs [{type=Vector3}]
   }
  }
  MaterialExpressionConvert_11 {
   class=MaterialExpressionConvert
   inputs [None {from=MaterialExpressionPower_0}]
   properties {
    convertInputs [{
      type=Vector3
      expressionInput {
       expression=&self:MaterialExpressionPower_0
       inputName=None
       mask=0
       maskA=0
       maskB=0
       maskG=0
       maskR=0
       outputIndex=0
      }
     }]
    convertOutputs [{type=Scalar}, {type=Scalar}, {type=Scalar}]
   }
  }
  MaterialExpressionConvert_2 {
   class=MaterialExpressionConvert
   inputs [None {from=MaterialExpressionRound_0}]
   properties {
    convertInputs [{
      type=Vector3
      expressionInput {
       expression=&self:MaterialExpressionRound_0
       inputName=None
       mask=0
       maskA=0
       maskB=0
       maskG=0
       maskR=0
       outputIndex=0
      }
     }]
    convertOutputs [{type=Vector2}, {type=Vector2}, {type=Vector2}]
   }
  }
  MaterialExpressionConvert_3 {
   class=MaterialExpressionConvert
   inputs [None {from=MaterialExpressionMultiply_6}]
   properties {
    nodeName=Break Float3
    convertInputs [{
      type=Vector3
      expressionInput {
       expression=&self:MaterialExpressionMultiply_6
       inputName=None
       mask=0
       maskA=0
       maskB=0
       maskG=0
       maskR=0
       outputIndex=0
      }
     }]
    convertOutputs [{type=Scalar}, {type=Scalar}, {type=Scalar}]
   }
  }
  MaterialExpressionConvert_5 {
   class=MaterialExpressionConvert
   inputs [None {from=MaterialExpressionReroute_1}]
   properties {
    convertInputs [{
      type=Vector2
      expressionInput {
       expression=&self:MaterialExpressionReroute_1
       inputName=None
       mask=0
       maskA=0
       maskB=0
       maskG=0
       maskR=0
       outputIndex=0
      }
     }]
    convertOutputs [{type=Vector2}]
   }
  }
  MaterialExpressionConvert_6 {
   class=MaterialExpressionConvert
   inputs [
    None {from=MaterialExpressionAbs_0}
    None {from=MaterialExpressionIf_0}
   ]
   properties {
    desc=gradMask
    convertInputs [
     {
      type=Scalar
      expressionInput {
       expression=&self:MaterialExpressionAbs_0
       inputName=None
       mask=0
       maskA=0
       maskB=0
       maskG=0
       maskR=0
       outputIndex=0
      }
     }
     {
      type=Vector2
      expressionInput {
       expression=&self:MaterialExpressionIf_0
       inputName=None
       mask=0
       maskA=0
       maskB=0
       maskG=0
       maskR=0
       outputIndex=0
      }
     }
    ]
    convertOutputs [{type=Vector3}]
   }
  }
  MaterialExpressionConvert_7 {
   class=MaterialExpressionConvert
   inputs [None {from=MaterialExpressionReroute_3}]
   properties {
    convertInputs [{
      type=Vector3
      expressionInput {
       expression=&self:MaterialExpressionReroute_3
       inputName=None
       mask=0
       maskA=0
       maskB=0
       maskG=0
       maskR=0
       outputIndex=0
      }
     }]
    convertOutputs [{type=Vector3}]
   }
  }
  MaterialExpressionConvert_8 {
   class=MaterialExpressionConvert
   inputs [None {from=MaterialExpressionReroute_3}]
   properties {
    convertInputs [{
      type=Vector3
      expressionInput {
       expression=&self:MaterialExpressionReroute_3
       inputName=None
       mask=0
       maskA=0
       maskB=0
       maskG=0
       maskR=0
       outputIndex=0
      }
     }]
    convertOutputs [{type=Vector3}]
   }
  }
  MaterialExpressionConvert_9 {
   class=MaterialExpressionConvert
   inputs [None {from=MaterialExpressionReroute_3}]
   properties {
    convertInputs [{
      type=Vector3
      expressionInput {
       expression=&self:MaterialExpressionReroute_3
       inputName=None
       mask=0
       maskA=0
       maskB=0
       maskG=0
       maskR=0
       outputIndex=0
      }
     }]
    convertOutputs [{type=Vector3}]
   }
  }
  MaterialExpressionCustom_3 {
   class=MaterialExpressionCustom
   inputs [
    inUV {from="", output=""}
    MaskContrast {from="", output=""}
   ]
   properties {
    code="//Original HLSL left for reference\r\n//for ref: https://github.com/mmikk/mmikk.github.io/blob/master/papers3d/mm_hex_compressed.pdf\r\n//https://www.youtube.com/watch?v=hc6msdFcnA4\r\n\r\n//skew coordinates to make triagles\r\ninUV *= 2.f * sqrt(3); \r\nconst float2x2 skewGrid = float2x2( 1.f, -0.57735027, \r\n                                    0.f,  1.15470054 );\r\ninUV = mul(skewGrid, inUV);\r\n\r\n\r\n//break coordinates into frac and integer parts + store info in z channels for later\r\nint3 baseID = int3(floor(inUV), 0u);\r\nfloat3 tmp = float3(frac(inUV), 0.f);\r\ntmp.z = tmp.x + tmp.y - 1.f;\r\nbaseID.z = baseID.x - baseID.y;\r\n\r\n//make inital mask for the three samples \r\nfloat3 gridMask = float3(baseID.zzz) + float3(0.f,1.f,2.f);\r\ngridMask *= rcp(3.f);\r\ngridMask += (5.f/3.f);\r\ngridMask = frac(gridMask + 0.000001);\r\ngridMask = round(gridMask);\r\n\r\n//generate per-hex ID for coordinate randomization \r\nfloat3 triMask = step(0.f, tmp.z) * gridMask;\r\nhexID1 = baseID + gridMask.xy + triMask.zz;\r\nhexID2 = baseID + gridMask.zx + triMask.yy;\r\nhexID3 = baseID + gridMask.yz + triMask.xx; \r\n\r\n//generate hex mask pattern\r\nfloat2 swizGrad = (tmp.z > 0) ? 1.f - tmp.yx : tmp.xy;\r\nfloat3 gradMask = float3(abs(tmp.z), swizGrad);\r\nfloat3 hexGrad = float3( dot(gradMask, gridMask.zxy), \r\n                         dot(gradMask, gridMask.yzx),\r\n                         dot(gradMask, gridMask.xyz) );\r\n\r\n//normalize output masks\r\nhexGrad = pow(hexGrad, MaskContrast * 8.f);\r\nhexGrad /= (hexGrad.x + hexGrad.y + hexGrad.z);\r\n\r\nreturn hexGrad;"
    description=HexMask
    showCode=true
    additionalOutputs [{outputName=hexID1, outputType=CMOT_Float2}, {outputName=hexID2, outputType=CMOT_Float2}, {outputName=hexID3, outputType=CMOT_Float2}]
    inputs [
     {inputName=inUV, input {expression=null, inputName=None, mask=0, maskA=0, maskB=0, maskG=0, maskR=0, outputIndex=0}}
     {inputName=MaskContrast, input {expression=null, inputName=None, mask=0, maskA=0, maskB=0, maskG=0, maskR=0, outputIndex=0}}
    ]
   }
  }
  MaterialExpressionDivide_0 {
   class=MaterialExpressionDivide
   inputs [
    A {from=MaterialExpressionConstant_3}
    B {from=MaterialExpressionConstant_4}
   ]
  }
  MaterialExpressionDivide_1 {
   class=MaterialExpressionDivide
   inputs [
    A {from=MaterialExpressionConstant_5}
    B {from=MaterialExpressionConstant_6}
   ]
  }
  MaterialExpressionDivide_2 {
   class=MaterialExpressionDivide
   inputs [
    A {from=MaterialExpressionPower_0}
    B {from=MaterialExpressionAdd_12}
   ]
  }
  MaterialExpressionDotProduct_0 {
   class=MaterialExpressionDotProduct
   inputs [
    A {from=MaterialExpressionMultiply_0, output=""}
    B {from=MaterialExpressionConstant2Vector_0, output=""}
   ]
   properties {}
  }
  MaterialExpressionDotProduct_1 {
   class=MaterialExpressionDotProduct
   inputs [
    A {from=MaterialExpressionMultiply_0, output=""}
    B {from=MaterialExpressionConstant2Vector_1, output=""}
   ]
   properties {}
  }
  MaterialExpressionDotProduct_2 {
   class=MaterialExpressionDotProduct
   inputs [
    A {from=MaterialExpressionConvert_7, output=""}
    B {from=MaterialExpressionReroute_2, output=""}
   ]
   properties {}
  }
  MaterialExpressionDotProduct_3 {
   class=MaterialExpressionDotProduct
   inputs [
    A {from=MaterialExpressionConvert_9, output=""}
    B {from=MaterialExpressionReroute_2, output=""}
   ]
   properties {}
  }
  MaterialExpressionDotProduct_4 {
   class=MaterialExpressionDotProduct
   inputs [
    A {from=MaterialExpressionReroute_2, output=""}
    B {from=MaterialExpressionConvert_8, output=""}
   ]
   properties {}
  }
  MaterialExpressionFloor_0 {class=MaterialExpressionFloor, inputs [None {from=MaterialExpressionAppendVector_0, output=""}], properties {desc=baseID}}
  MaterialExpressionFrac_1 {class=MaterialExpressionFrac, inputs [None {from=MaterialExpressionAppendVector_0, output=""}], properties {desc=tmp}}
  MaterialExpressionFrac_2 {class=MaterialExpressionFrac, inputs [None {from=MaterialExpressionAdd_4, output=""}], properties {}}
  MaterialExpressionFunctionInput_0 {
   class=MaterialExpressionFunctionInput
   inputs [Preview {from=MaterialExpressionTextureCoordinate_0, output=""}]
   properties {bUsePreviewValueAsDefault=true, inputName=Coordinates, inputType=FunctionInput_Vector2}
  }
  MaterialExpressionFunctionInput_2 {
   class=MaterialExpressionFunctionInput
   inputs [Preview {from="", output=""}]
   properties {
    bUsePreviewValueAsDefault=true
    description=Contrast of the hex tile mask
    inputName=MaskContrast
    inputType=FunctionInput_Scalar
    sortPriority=2
    previewValue=<5 0 0 1>
   }
  }
  MaterialExpressionFunctionOutput_0 {
   class=MaterialExpressionFunctionOutput
   inputs [None {from=MaterialExpressionDivide_2, output=""}]
   properties {outputName=PatternMask, sortPriority=10}
  }
  MaterialExpressionFunctionOutput_1 {
   class=MaterialExpressionFunctionOutput
   inputs [None {from=MaterialExpressionAdd_10, output=""}]
   properties {outputName=HexID1, sortPriority=5}
  }
  MaterialExpressionFunctionOutput_2 {
   class=MaterialExpressionFunctionOutput
   inputs [None {from=MaterialExpressionAdd_9, output=""}]
   properties {outputName=HexID2, sortPriority=6}
  }
  MaterialExpressionFunctionOutput_3 {
   class=MaterialExpressionFunctionOutput
   inputs [None {from=MaterialExpressionAdd_8, output=""}]
   properties {outputName=HexID3, sortPriority=7}
  }
  MaterialExpressionIf_0 {
   class=MaterialExpressionIf
   inputs [
    A {from=MaterialExpressionSubtract_1, output=""}
    B {from="", output=""}
    "A > B" {from=MaterialExpressionOneMinus_0, output=""}
    "A == B" {from="", output=""}
    "A < B" {from=MaterialExpressionReroute_1, output=""}
   ]
   properties {desc=swizGrad}
  }
  MaterialExpressionMultiply_0 {
   class=MaterialExpressionMultiply
   inputs [
    A {from=MaterialExpressionFunctionInput_0}
    B {from=MaterialExpressionMultiply_4}
   ]
  }
  MaterialExpressionMultiply_4 {
   class=MaterialExpressionMultiply
   inputs [
    A {from=MaterialExpressionSquareRoot_0}
    B {from=""}
   ]
   properties {constB=2}
  }
  MaterialExpressionMultiply_5 {
   class=MaterialExpressionMultiply
   inputs [
    A {from=MaterialExpressionAdd_2}
    B {from=MaterialExpressionDivide_0}
   ]
  }
  MaterialExpressionMultiply_6 {
   class=MaterialExpressionMultiply
   inputs [
    A {from=MaterialExpressionStep_0}
    B {from=MaterialExpressionRound_0}
   ]
   properties {desc=triMask}
  }
  MaterialExpressionMultiply_7 {
   class=MaterialExpressionMultiply
   inputs [
    A {from=MaterialExpressionFunctionInput_2}
    B {from=""}
   ]
   properties {constB=8}
  }
  MaterialExpressionOneMinus_0 {class=MaterialExpressionOneMinus, inputs [None {from=MaterialExpressionConvert_5}]}
  MaterialExpressionPower_0 {
   class=MaterialExpressionPower
   inputs [
    Base {from=MaterialExpressionConvert_10}
    Exp {from=MaterialExpressionMultiply_7}
   ]
  }
  MaterialExpressionReroute_0 {class=MaterialExpressionReroute, inputs [None {from=MaterialExpressionFloor_0}]}
  MaterialExpressionReroute_1 {class=MaterialExpressionReroute, inputs [None {from=MaterialExpressionFrac_1}]}
  MaterialExpressionReroute_2 {class=MaterialExpressionReroute, inputs [None {from=MaterialExpressionConvert_6}]}
  MaterialExpressionReroute_3 {class=MaterialExpressionReroute, inputs [None {from=MaterialExpressionRound_0}]}
  MaterialExpressionRound_0 {class=MaterialExpressionRound, inputs [None {from=MaterialExpressionFrac_2, output=""}], properties {desc=gridMask}}
  MaterialExpressionSquareRoot_0 {class=MaterialExpressionSquareRoot, inputs [None {from=MaterialExpressionConstant_0, output=""}], properties {}}
  MaterialExpressionStep_0 {
   class=MaterialExpressionStep
   inputs [
    Y {from="", output=""}
    X {from=MaterialExpressionSubtract_1, output=""}
   ]
   properties {}
  }
  MaterialExpressionSubtract_0 {
   class=MaterialExpressionSubtract
   inputs [
    A {from=MaterialExpressionConvert_1}
    B {from=MaterialExpressionConvert_1}
   ]
  }
  MaterialExpressionSubtract_1 {
   class=MaterialExpressionSubtract
   inputs [
    A {from=MaterialExpressionAdd_1}
    B {from=""}
   ]
  }
  MaterialExpressionTextureCoordinate_0 {class=MaterialExpressionTextureCoordinate}
 }
 properties {bExposeToLibrary=true, libraryCategoriesText [Base Material]}
}
