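# Material function that builds hex-tiled coordinate frames from a selectable mesh UV channel, randomizing scale, offset and rotation per tile to break up texture repetition.
# Transformed UVs feed a hex mask, whose three cell IDs are hashed to seed three coordinate randomizers.
# It outputs x/y/z UVs, packed derivatives for SampleGrad, derivative-derived tangent bases and rotations, plus an RGB mask for blending the three samples afterwards.
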MaterialFunction /BaseMaterial/Materials/Functions/MF_ResolveCoordinateFrame_HexTile_BasicUV
graph {
 properties {
  bExposeToLibrary=true
  description="Generates randomized hexagonal tile UVs with per-tile variation in scale, offset, and rotation. Performs true multi-sample blending using three texture samples to resolve transitions across neighboring hex regions. Outputs xUV, yUV, and zUV coordinate sets, each with tangent/bitangent/rotation/derivatives, plus a BlendMask (RGB) for post-sample blending. Supports blend weight modulation by albedo or height for more perceptually correct transitions between tiles."
  userExposedCaption="CoordinateBasis: HexTile (Standard)"
  libraryCategoriesText [Base Material]
 }
 nodes [
  n1 = FunctionInput() {
   bUsePreviewValueAsDefault=true
   description="Controls the tiling of the hex pattern. Larger values create more, smaller hex tiles across the surface; smaller values create fewer, larger tiles."
   inputName=Hex PatternSize
   inputType=FunctionInput_Scalar
   sortPriority=2
   previewValue=<1 0 0 1>
  }
  n2 = FunctionInput() {
   bUsePreviewValueAsDefault=true
   description="Selects which mesh UV channel (0–3) to use as the base coordinate source."
   inputName=UVIndex
   inputType=FunctionInput_Scalar
   sortPriority=-5
  }
  n3 = FunctionInput() {
   bUsePreviewValueAsDefault=true
   description=Applies an additional per-axis offset to the coordinates before projection. Useful for correctly positioning a texture.
   inputName=CoordinateTransform
   inputType=FunctionInput_Vector2
   sortPriority=-9
  }
  n4 = FunctionInput() {
   bUsePreviewValueAsDefault=true
   description="Rotates the projected UVs around the selected axis. Units are in turns, where 1.0 = 360° and 0.25 = 90°."
   inputName="Rotation (Turns)"
   inputType=FunctionInput_Scalar
   sortPriority=-8
  }
  n5 = FunctionInput() {
   bUsePreviewValueAsDefault=true
   description="Sets how many times the projected coordinates repeat along each axis. Use values above 1 for denser tiling, negative values for mirroring, and avoid zero."
   inputName=CoordinateRepeat
   inputType=FunctionInput_Vector2
   sortPriority=-10
   previewValue=<1 1 1 1>
  }
  n6 = MaterialFunctionCall!(
   Coords=MaterialFunctionCall!(
    Selector=@n2
    VectorA=TextureCoordinate()
    VectorB=TextureCoordinate() {coordinateIndex=1}
    VectorC=TextureCoordinate() {coordinateIndex=2}
    VectorD=TextureCoordinate() {coordinateIndex=3}
   ) {materialFunction=&/MF_Switch4_Vec2}.Selected Vector
   UVOffset=@n3
   UVRotation=NamedRerouteDeclaration(None=@n4) {name=turnsi, nodeColor=<rgba 0.8963 0.8963 0.0091 1>}
   RotationPivot
   UVTile=@n5
  ) {materialFunction=&/MF_CoordinateTransform_2D}
  n7 = FunctionInput() {
   bUsePreviewValueAsDefault=true
   description="Controls blend contrast by raising the hex pattern to 8 × BlendFactor (default 0.5 = pow4) "
   inputName=BlendTolerance
   inputType=FunctionInput_Scalar
   sortPriority=8
   previewValue=<0.5 0 0 1>
  }
  n8 = MaterialFunctionCall!(Coordinates=Multiply(A=@n1, B=@n6.geometry_UV), MaskContrast=@n7) {materialFunction=&/MF_Blend_HexMask}
  FunctionOutput(None=@n8.PatternMask) {
   description="RGB mask giving blend weights for xUV, yUV, zUV. Used for post-sample blending. Weights can also be modulated by albedo or height for more perceptual transitions.\r\n"
   outputName=BlendMask
   sortPriority=15
  }
  n9 = MaterialFunctionCall!(
   Coords=NamedRerouteUsage()
   "TileRand (Min/Max)"=NamedRerouteUsage()
   "OffsetRand (Min/Max)"=NamedRerouteUsage()
   TileStrengthRand
   OffsetStrengthRand
   RotateStrengthRand=NamedRerouteUsage()
   SeedA=NamedRerouteUsage()
  ) {materialFunction=&/MF_Coordinate_Rand1}
  n10 = FunctionOutput(None=@n9.packedDXY) {description="Packed derivatives (ddx/ddy) for xUV. Use with SampleGrad for stable mip level selection.\r\n", outputName=xDXY, sortPriority=3}
  n11 = MaterialFunctionCall!(packedDXY=@n10) {materialFunction=&/MF_DeriveTangentBasis}
  FunctionOutput(None=@n11.Tangent) {description="World-space tangent for xUV. Use with xBitangent to reorient tangent-space normals.\r\n", outputName=xTangent, sortPriority=1}
  FunctionOutput(None=@n11.Bitangent) {description="World-space bitangent for xUV. Complements xTangent.\r\n", outputName=xBitangent, sortPriority=2}
  n12 = MaterialFunctionCall(
   Coords=NamedRerouteUsage()
   "TileRand (Min/Max)"=NamedRerouteUsage()
   "OffsetRand (Min/Max)"=NamedRerouteUsage()
   RotateStrengthRand=NamedRerouteUsage()
   SeedA=NamedRerouteUsage()
  ) {materialFunction=&MF_Coordinate_Rand1}
  n13 = FunctionOutput(None=@n12.packedDXY) {description="Packed derivatives (ddx/ddy) for yUV. Use with SampleGrad for stable mip level selection.\r\n", outputName=yDXY, sortPriority=8}
  n14 = MaterialFunctionCall(packedDXY=@n13) {materialFunction=&MF_DeriveTangentBasis}
  FunctionOutput(None=@n14.Bitangent) {description=World-space bitangent for yUV. Complements yTangent., outputName=yBitangent, sortPriority=7}
  FunctionOutput(None=@n14.Tangent) {description=World-space tangent for yUV. Use with yBitangent to reorient tangent-space normals., outputName=yTangent, sortPriority=6}
  n15 = MaterialFunctionCall(
   Coords=NamedRerouteUsage()
   "TileRand (Min/Max)"=NamedRerouteUsage()
   "OffsetRand (Min/Max)"=NamedRerouteUsage()
   RotateStrengthRand=NamedRerouteUsage()
   SeedA=NamedRerouteUsage()
  ) {materialFunction=&MF_Coordinate_Rand1}
  n16 = FunctionOutput(None=@n15.packedDXY) {
   description="Packed derivatives (ddx/ddy) for zUV. Use with SampleGrad for stable mip level selection.\r\n"
   outputName=zDXY
   sortPriority=13
  }
  n17 = MaterialFunctionCall(packedDXY=@n16) {materialFunction=&MF_DeriveTangentBasis}
  FunctionOutput(None=@n17.Tangent) {description="World-space tangent for zUV. Use with zBitangent to reorient tangent-space normals.\r\n", outputName=zTangent, sortPriority=11}
  FunctionOutput(None=@n17.Bitangent) {description="World-space bitangent for zUV. Complements zTangent.\r\n", outputName=zBitangent, sortPriority=12}
  n18 = NamedRerouteUsage()
  FunctionOutput(None=Multiply(A=Add(A=@n9.SampleRotation, B=@n18)) {constB=-1}) {description="Rotation applied to xUV, in turns (1.0 = 360°).\r\n", outputName="xRotation (Turns)", sortPriority=-7}
  n19 = NamedRerouteUsage()
  FunctionOutput(None=Multiply(A=Add(A=@n12.SampleRotation, B=@n19)) {constB=-1}) {description="Rotation applied to yUV, in turns (1.0 = 360°).\r\n", outputName="yRotation (Turns)", sortPriority=-6}
  n20 = NamedRerouteUsage()
  FunctionOutput(None=Multiply(A=Add(A=@n15.SampleRotation, B=@n20)) {constB=-1}) {description="Rotation applied to zUV, in turns (1.0 = 360°).\r\n", outputName="zRotation (Turns)", sortPriority=-5}
  FunctionOutput(None=@n9.UV) {description="First randomized UV coordinate set generated by the hex tiling projection.\r\n", outputName=xUV}
  FunctionOutput(None=@n12.UV) {description="Second randomized UV coordinate set generated by the hex tiling projection.\r\n", outputName=yUV, sortPriority=5}
  FunctionOutput(None=@n15.UV) {description="Third randomized UV coordinate set generated by the hex tiling projection.\r\n", outputName=zUV, sortPriority=10}
  n21 = FunctionInput() {
   bUsePreviewValueAsDefault=true
   description=Defines the minimum and maximum random UV offset applied to each hex tile. Shifts the sampling position of tiles within their cells to break up repetition.
   inputName=Hex Min/MaxRand Offset
   inputType=FunctionInput_Vector2
   sortPriority=4
  }
  NamedRerouteDeclaration(None=@n21) {name=HexOffset, nodeColor=<rgba 0.9 0.3505 0.009 1>}
  n22 = FunctionInput() {
   bUsePreviewValueAsDefault=true
   description="Maximum random per-tile rotation, expressed in turns (1.0 = 360°). Each hex tile rotates by a random amount up to this value."
   inputName="Hex Rand Rotation (turns)"
   inputType=FunctionInput_Scalar
   sortPriority=5
  }
  NamedRerouteDeclaration(None=@n22) {name=HexRot, nodeColor=<rgba 0.9 0.3505 0.009 1>}
  n23 = FunctionInput() {
   bUsePreviewValueAsDefault=true
   description="Defines the minimum and maximum random scale applied to each hex tile. Values below 1 enlarge tiles, values above 1 shrink them. Each tile chooses a random value within this range."
   inputName=Hex Min/MaxRand Scale
   inputType=FunctionInput_Vector2
   sortPriority=3
   previewValue=<1 1 0 1>
  }
  NamedRerouteDeclaration(None=@n23) {name=HexScale, nodeColor=<rgba 0.9 0.3505 0.009 1>}
  NamedRerouteDeclaration(None=MaterialFunctionCall!(seed=@n8.HexID1) {materialFunction=&/MF_Hash42}.Hash4D) {name=RandA, nodeColor=<rgba 1 0 0.75 1>}
  NamedRerouteDeclaration(None=@n6.geometry_UV) {name=UVi, nodeColor=<rgba 0.9 0.009 0.009 1>}
  NamedRerouteDeclaration(None=MaterialFunctionCall(seed=@n8.HexID2) {materialFunction=&MF_Hash42}.Hash4D) {name=RandB, nodeColor=<rgba 1 0 0.75 1>}
  NamedRerouteDeclaration(None=MaterialFunctionCall(seed=@n8.HexID3) {materialFunction=&MF_Hash42}.Hash4D) {name=RandC, nodeColor=<rgba 1 0 0.75 1>}
 ]
}
