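# Base Material function that converts three-channel triplanar blend weights and a dither pattern into a single axis index of 0, 1 or 2.
# It saturates and splits the weights, subtracts the dither from two of them, then picks the winner through Max and nested If comparisons.
# Slightly padded constants are truncated to give the integer AxisIndex.
# An unconnected Custom HLSL node keeps the original, adapted from Ryan Dowling-Soka's dithered biplanar article, for reference.
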MaterialFunction /BaseMaterial/Materials/Functions/MF_Blend_RGBToIndex
graph {
 properties {bExposeToLibrary=true, libraryCategoriesText [Base Material]}
 nodes [
  Custom!(blend, dither) {
   code="//original custom node left for ref\r\n//modified from https://ryandowlingsoka.com/unreal/triplanar-dither-biplanar/\r\nblend = saturate(blend);\r\nconst float ditherx = blend.x - dither;\r\nconst float ditherz = blend.z - dither; \r\n\r\nint index = 0;\r\nindex = max(ditherx, ditherz) > blend.y ? 0 : 1;\r\nindex = ditherz > max(blend.x, blend.y) ? 2 : index;\r\n\r\nreturn index;"
   description=createTriplanarSelection
   outputType=CMOT_Float1
   showCode=true
   inputs [{inputName=blend, input {expression=null}}, {inputName=dither, input {expression=null}}]
  }
  MaterialExpressionSaturate_0 = Saturate(None=FunctionInput() {bUsePreviewValueAsDefault=true, inputName=InputBlend})
  n1 = Convert(None=@MaterialExpressionSaturate_0) {
   nodeName=Break Float3
   convertInputs [{
     type=Vector3
     expressionInput {
      expression=&self:MaterialExpressionSaturate_0
      inputName=None
      mask=0
      maskA=0
      maskB=0
      maskG=0
      maskR=0
      outputIndex=0
     }
    }]
   convertOutputs [{type=Scalar}, {type=Scalar}, {type=Scalar}]
  }
  n2 = Max(A=@n1, B=@n1)
  n3 = Constant() {r=2.0033}
  n4 = FunctionInput() {bUsePreviewValueAsDefault=true, inputName=TriplanarDitherPattern, inputType=FunctionInput_Scalar, sortPriority=5}
  n5 = Subtract(A=@n1, B=@n4)
  n6 = If(A=Max(A=Subtract(A=@n1, B=@n4), B=@n5), B=@n1, "A > B"=Constant() {r=0.0033}, "A < B"=Constant() {r=1.0033})
  FunctionOutput(None=Truncate!(None=If(A=@n5, B=@n2, "A > B"=@n3, "A < B"=@n6)) {is=class MaterialExpressionTruncate}) {outputName=AxisIndex}
 ]
}
