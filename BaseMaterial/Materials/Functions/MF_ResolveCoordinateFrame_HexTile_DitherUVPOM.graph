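# Material function that generates randomized hex-tile UVs with dithered Parallax Occlusion Mapping to break up texture repetition cheaply.
# The transformed mesh UV drives a hex mask, whose three cell IDs are hashed for per-tile scale, offset and rotation.
# Dither noise selects one tile per pixel instead of blending samples.
# Parallax marches the height texture on that tile's frame, outputting UV, depth, packed derivatives, tangent basis and rotation.
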MaterialFunction /BaseMaterial/Materials/Functions/MF_ResolveCoordinateFrame_HexTile_DitherUVPOM
graph {
 properties {
  bExposeToLibrary=true
  description="Generates randomized hexagonal tile UVs for procedural mapping. Each tile can vary in scale, offset, and rotation. Uses dithered sampling to approximate blending at very low cost (single sample per tile). This variant applies Parallax Occlusion Mapping (POM) to the tiles using dithered sampling. Produces low-cost depth and occlusion cues, with the same dithering tradeoffs and no post-sample blend weight modulation."
  userExposedCaption="CoordinateBasis: HexTile (Dithered POM)"
  libraryCategoriesText [Base Material]
 }
 nodes [
  n1 = MaterialFunctionCall!(
   Coords=NamedRerouteUsage()
   "TileRand (Min/Max)"=NamedRerouteUsage()
   "OffsetRand (Min/Max)"=NamedRerouteUsage()
   TileStrengthRand
   OffsetStrengthRand
   RotateStrengthRand=NamedRerouteUsage()
   SeedA=NamedRerouteUsage()
   SeedB=NamedRerouteUsage()
   SeedC=NamedRerouteUsage()
   MaskIndex=NamedRerouteUsage()
  ) {materialFunction=&/MF_Coordinate_Rand3}
  n2 = FunctionOutput(None=@n1.packedDXY) {
   description="Packed derivatives of the projected UVs (ddx/ddy). Use with SampleGrad to ensure textures sample with correct mip levels and anisotropic filtering.\r\n"
   outputName=DXY
   sortPriority=3
  }
  n3 = MaterialFunctionCall!(packedDXY=@n1.packedDXY) {materialFunction=&/MF_DeriveTangentBasis}
  n4 = FunctionOutput(None=@n3.Tangent) {
   description="The world-space tangent vector of the projection frame. Use with BitangentWS and Rotation to correctly reorient tangent-space normals.\r\n"
   outputName=Tangent
   sortPriority=1
  }
  n5 = FunctionOutput(None=@n3.Bitangent) {
   description="The world-space bitangent vector of the projection frame. Complements TangentWS to form the basis for tangent-space normal reorientation.\r\n"
   outputName=Bitangent
   sortPriority=2
  }
  n6 = NamedRerouteUsage()
  n7 = FunctionOutput(None=Multiply(A=Add(A=@n1.SampleRotation, B=@n6)) {constB=-1}) {
   description="The rotation applied around the projection axis, expressed in turns (1.0 = 360°). Used along with Tangent/Bitangent for reorienting normal maps via MF_TangentNormal_OrientWorld.\r\n"
   outputName="Rotation (Turns)"
   sortPriority=-5
  }
  n8 = FunctionInput() {
   bUsePreviewValueAsDefault=true
   description=Control the U/V Tiling of the height texture relative to the input coordinates.
   inputName=POM HeightTile
   inputType=FunctionInput_Vector2
   sortPriority=12
   previewValue=<1 1 0 1>
  }
  n9 = MaterialFunctionCall!(
   POM_HeightTexture=NamedRerouteUsage()
   Coordinates=@n1.UV
   coordinateDXY=@n2
   HeightStrength=NamedRerouteUsage()
   Reference Plane=NamedRerouteUsage()
   MaxSteps=NamedRerouteUsage()
   Tangent=@n4
   Bitangent=@n5
   "Rotation (Turns)"=@n7
   HeightChannel=NamedRerouteUsage()
   QualityBias=NamedRerouteUsage()
   CameraVector
   Height Tile=@n8
  ) {materialFunction=&/MF_Coordinate_Parallax}
  FunctionOutput(None=@n9.POM UVs) {
   description="The primary 2D coordinates generated by the projection. Feed these into texture samplers for surface detail.\r\n"
   outputName=UV
  }
  FunctionOutput(None=@n9.Pixel Depth Offset) {
   description="Depth at the POM intersection (after Intensity/RefPlane). Higher (whiter) = deeper—the farther the ray travels into the material, the larger the value."
   outputName=Depth
   sortPriority=4
  }
  n10 = FunctionInput() {
   bUsePreviewValueAsDefault=true
   description="Controls the tiling of the hex pattern. Larger values create more, smaller hex tiles across the surface; smaller values create fewer, larger tiles."
   inputName=Hex PatternSize
   inputType=FunctionInput_Scalar
   sortPriority=3
   previewValue=<1 0 0 1>
  }
  n11 = FunctionInput() {
   bUsePreviewValueAsDefault=true
   description="Selects which mesh UV channel (0–3) to use as the base coordinate source."
   inputName=UVIndex
   inputType=FunctionInput_Scalar
  }
  n12 = FunctionInput() {
   bUsePreviewValueAsDefault=true
   description=Applies an additional per-axis offset to the coordinates before projection. Useful for correctly positioning a texture.
   inputName=CoordinateTransform
   inputType=FunctionInput_Vector2
   sortPriority=-9
  }
  n13 = FunctionInput() {
   bUsePreviewValueAsDefault=true
   description="Rotates the projected UVs around the selected axis. Units are in turns, where 1.0 = 360° and 0.25 = 90°."
   inputName="Rotation (Turns)"
   inputType=FunctionInput_Scalar
   sortPriority=-8
  }
  n14 = FunctionInput() {
   bUsePreviewValueAsDefault=true
   description="Sets how many times the projected coordinates repeat along each axis. Use values above 1 for denser tiling, negative values for mirroring, and avoid zero."
   inputName=CoordinateRepeat
   inputType=FunctionInput_Vector2
   sortPriority=-10
   previewValue=<1 1 1 1>
  }
  n15 = MaterialFunctionCall!(
   Coords=MaterialFunctionCall!(
    Selector=@n11
    VectorA=TextureCoordinate()
    VectorB=TextureCoordinate() {coordinateIndex=1}
    VectorC=TextureCoordinate() {coordinateIndex=2}
    VectorD=TextureCoordinate() {coordinateIndex=3}
   ) {materialFunction=&/MF_Switch4_Vec2}.Selected Vector
   UVOffset=@n12
   UVRotation=NamedRerouteDeclaration(None=@n13) {name=turnsi, nodeColor=<rgba 0.8963 0.8963 0.0091 1>}
   RotationPivot
   UVTile=@n14
  ) {materialFunction=&/MF_CoordinateTransform_2D}
  n16 = FunctionInput() {
   bUsePreviewValueAsDefault=true
   description="Controls blend contrast by raising the hex pattern to 8 × BlendFactor (default 0.5 = pow4) "
   inputName=BlendFactor
   inputType=FunctionInput_Scalar
   sortPriority=7
   previewValue=<0.5 0 0 1>
  }
  n17 = MaterialFunctionCall!(Coordinates=Multiply(A=@n10, B=@n15.geometry_UV), MaskContrast=@n16) {materialFunction=&/MF_Blend_HexMask}
  n18 = FunctionInput(Preview=Subtract(A=MaterialFunctionCall!(TemporalSampleIndex, PixelPosition) {materialFunction=&/MF_Noise_InterleavedGradientGolden_1d}.Unsigned) {constB=0.5}) {
   bUsePreviewValueAsDefault=true
   description="Provides the dithering sequence used for stochastic projection blending. By default, the function generates an interleaved gradient noise pattern based on the golden ratio, which produces stable, slow-converging results at low cost. Supplying a custom value lets you override this with your own dithering scheme, useful for synchronizing with other systems or experimenting with alternate noise sources."
   inputName=Dither
   inputType=FunctionInput_Scalar
  }
  NamedRerouteDeclaration(None=MaterialFunctionCall!(InputBlend=@n17.PatternMask, TriplanarDitherPattern=@n18) {materialFunction=&/MF_Blend_RGBToIndex}.AxisIndex) {name=Blend, nodeColor=<rgba 1 1 1 1>}
  n19 = FunctionInput() {
   bUsePreviewValueAsDefault=true
   description=Defines the minimum and maximum random UV offset applied to each hex tile. Shifts the sampling position of tiles within their cells to break up repetition.
   inputName=Hex Min/MaxRand Offset
   inputType=FunctionInput_Vector2
   sortPriority=5
  }
  NamedRerouteDeclaration(None=@n19) {name=HexOffset, nodeColor=<rgba 0.9 0.3505 0.009 1>}
  n20 = FunctionInput() {
   bUsePreviewValueAsDefault=true
   description="Maximum random per-tile rotation, expressed in turns (1.0 = 360°). Each hex tile rotates by a random amount up to this value."
   inputName="Hex Rand Rotation (turns)"
   inputType=FunctionInput_Scalar
   sortPriority=6
  }
  NamedRerouteDeclaration(None=@n20) {name=HexRot, nodeColor=<rgba 0.9 0.3505 0.009 1>}
  n21 = FunctionInput() {
   bUsePreviewValueAsDefault=true
   description="Defines the minimum and maximum random scale applied to each hex tile. Values below 1 enlarge tiles, values above 1 shrink them. Each tile chooses a random value within this range."
   inputName=Hex Min/MaxRand Scale
   inputType=FunctionInput_Vector2
   sortPriority=4
   previewValue=<1 1 0 1>
  }
  NamedRerouteDeclaration(None=@n21) {name=HexScale, nodeColor=<rgba 0.9 0.3505 0.009 1>}
  n22 = FunctionInput() {
   bUsePreviewValueAsDefault=true
   description="Specifies which channel of the heightmap texture to use for depth information. Assumes the Alpha channel is used, but any channel can be selected if your texture packs data differently."
   inputName=POM HeightChannel
   inputType=FunctionInput_Scalar
   sortPriority=8
   previewValue=<3 0 0 1>
  }
  NamedRerouteDeclaration(None=@n22) {name=POM_Hchan, nodeColor=<rgba 0.9 0.306 0.009 1>}
  n23 = FunctionInput(Preview=TextureObject() {texture=/BaseMaterial/Textures/Default/T_Gray_sRGB}) {
   bUsePreviewValueAsDefault=true
   description="The heightmap texture used for parallax occlusion mapping. Grayscale values define surface depth, with brighter values typically pushed outward and darker values pushed inward."
   inputName=HeightTex
   inputType=FunctionInput_Texture2D
   sortPriority=-20
  }
  NamedRerouteDeclaration(None=@n23) {name=Height, nodeColor=<rgba 1 0 0.4922 1>}
  NamedRerouteDeclaration(None=MaterialFunctionCall!(seed=@n17.HexID1) {materialFunction=&/MF_Hash42}.Hash4D) {name=RandA, nodeColor=<rgba 1 0 0.75 1>}
  n24 = FunctionInput() {
   bUsePreviewValueAsDefault=true
   description="Controls the strength of the parallax effect. Higher values exaggerate depth; lower values make surfaces appear flatter."
   inputName=POM Intensity
   inputType=FunctionInput_Scalar
   sortPriority=9
   previewValue=<1 0 0 1>
  }
  NamedRerouteDeclaration(None=@n24) {name=POM_Height, nodeColor=<rgba 0.9 0.306 0.009 1>}
  n25 = FunctionInput() {
   bUsePreviewValueAsDefault=true
   description="Sets the maximum number of depth sampling steps. This is the upper bound, with actual steps scaled by the Quality setting."
   inputName=POM MaxSteps
   inputType=FunctionInput_Scalar
   sortPriority=10
   previewValue=<16 0 0 1>
  }
  NamedRerouteDeclaration(None=@n25) {name=POM_MaxSteps, nodeColor=<rgba 0.9 0.306 0.009 1>}
  n26 = FunctionInput() {
   bUsePreviewValueAsDefault=true
   description="Scales step count by view angle. At 1, MaxSteps are always used; at 0, only one step is used when looking straight on, with more steps added at grazing angles."
   inputName=POM Quality
   inputType=FunctionInput_Scalar
   sortPriority=11
   previewValue=<0.25 0 0 1>
  }
  NamedRerouteDeclaration(None=@n26) {name=POM_Noise, nodeColor=<rgba 0.9 0.306 0.009 1>}
  n27 = FunctionInput() {
   bUsePreviewValueAsDefault=true
   description=Defines the reference plane for parallax. This value shifts where the heightmap appears anchored relative to the surface.
   inputName=POM RefPlane
   inputType=FunctionInput_Scalar
   sortPriority=12
   previewValue=<1 0 0 1>
  }
  NamedRerouteDeclaration(None=@n27) {name=POM_RefPlane, nodeColor=<rgba 0.9 0.306 0.009 1>}
  NamedRerouteDeclaration(None=@n15.geometry_UV) {name=UVi, nodeColor=<rgba 0.9 0.009 0.009 1>}
  NamedRerouteDeclaration(None=MaterialFunctionCall(seed=@n17.HexID2) {materialFunction=&MF_Hash42}.Hash4D) {name=RandB, nodeColor=<rgba 1 0 0.75 1>}
  NamedRerouteDeclaration(None=MaterialFunctionCall(seed=@n17.HexID3) {materialFunction=&MF_Hash42}.Hash4D) {name=RandC, nodeColor=<rgba 1 0 0.75 1>}
 ]
}
