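# Material function that builds hex-tiled UV coordinates with parallax occlusion mapping, breaking up texture repetition while adding depth.
# A transformed mesh UV drives a hex mask whose three cell IDs are hashed to randomize scale, offset and rotation.
# Each randomized set derives a tangent basis from its derivatives and runs its own height-texture parallax march.
# Outputs x/y/z UVs, derivatives, tangents, bitangents, rotations and depths, plus an RGB BlendMask for post-sample blending.
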MaterialFunction /BaseMaterial/Materials/Functions/MF_ResolveCoordinateFrame_HexTile_UVPOM
graph {
 properties {
  bExposeToLibrary=true
  description="Standard HexTile with Parallax Occlusion Mapping (POM) for added depth and occlusion. Performs three texture samples for smooth transitions across hex tile boundaries. Outputs xUV, yUV, and zUV coordinate sets, each with tangent/bitangent/rotation/derivatives, plus a BlendMask (RGB) derived from the hex pattern. Supports post-sample modulation of blend weights by albedo or height, producing smoother, perceptual transitions while adding parallax depth to the tiles."
  userExposedCaption="CoordinateBasis: HexTile (Standard POM)"
  libraryCategoriesText [Base Material]
 }
 nodes [
  n1 = MaterialFunctionCall!(
   Coords=NamedRerouteUsage()
   "TileRand (Min/Max)"=NamedRerouteUsage()
   "OffsetRand (Min/Max)"=NamedRerouteUsage()
   TileStrengthRand
   OffsetStrengthRand
   RotateStrengthRand=NamedRerouteUsage()
   SeedA=NamedRerouteUsage()
  ) {materialFunction=&/MF_Coordinate_Rand1}
  n2 = FunctionOutput(None=@n1.packedDXY) {description="Packed derivatives (ddx/ddy) for xUV. Use with SampleGrad for stable mip level selection.\r\n", outputName=xDXY, sortPriority=3}
  n3 = MaterialFunctionCall!(packedDXY=@n2) {materialFunction=&/MF_DeriveTangentBasis}
  n4 = FunctionOutput(None=@n3.Tangent) {description="World-space tangent for xUV. Use with xBitangent to reorient tangent-space normals.\r\n", outputName=xTangent, sortPriority=1}
  n5 = Add(A=@n1.SampleRotation, B=NamedRerouteUsage())
  n6 = MaterialFunctionCall!(
   POM_HeightTexture=NamedRerouteUsage()
   Coordinates=@n1.UV
   coordinateDXY=@n2
   HeightStrength=NamedRerouteUsage()
   Reference Plane=NamedRerouteUsage()
   MaxSteps=NamedRerouteUsage()
   Tangent=@n4
   Bitangent=FunctionOutput(None=@n3.Bitangent) {description="World-space bitangent for xUV. Complements xTangent.\r\n", outputName=xBitangent, sortPriority=2}
   "Rotation (Turns)"=FunctionOutput(None=Multiply(A=@n5) {constB=-1}) {description="Rotation applied to xUV, in turns (1.0 = 360°).\r\n", outputName="xRotation (Turns)", sortPriority=-7}
   HeightChannel=NamedRerouteUsage()
   QualityBias=NamedRerouteUsage()
   CameraVector
   Height Tile=NamedRerouteUsage()
  ) {materialFunction=&/MF_Coordinate_Parallax}
  FunctionOutput(None=@n6.Pixel Depth Offset) {
   description="Depth at the POM intersection (after Intensity/RefPlane). Higher (whiter) = deeper—the farther the ray travels into the material, the larger the value."
   outputName=xDepth
   sortPriority=4
  }
  n7 = MaterialFunctionCall(
   Coords=NamedRerouteUsage()
   "TileRand (Min/Max)"=NamedRerouteUsage()
   "OffsetRand (Min/Max)"=NamedRerouteUsage()
   RotateStrengthRand=NamedRerouteUsage()
   SeedA=NamedRerouteUsage()
  ) {materialFunction=&MF_Coordinate_Rand1}
  n8 = FunctionOutput(None=@n7.packedDXY) {description="Packed derivatives (ddx/ddy) for yUV. Use with SampleGrad for stable mip level selection.\r\n", outputName=yDXY, sortPriority=8}
  n9 = MaterialFunctionCall(packedDXY=@n8) {materialFunction=&MF_DeriveTangentBasis}
  n10 = FunctionOutput(None=@n9.Tangent) {description="World-space tangent for yUV. Use with yBitangent to reorient tangent-space normals.\r\n", outputName=yTangent, sortPriority=6}
  n11 = Add(A=@n7.SampleRotation, B=NamedRerouteUsage())
  n12 = MaterialFunctionCall(
   POM_HeightTexture=NamedRerouteUsage()
   Coordinates=@n7.UV
   coordinateDXY=@n8
   HeightStrength=NamedRerouteUsage()
   Reference Plane=NamedRerouteUsage()
   MaxSteps=NamedRerouteUsage()
   Tangent=@n10
   Bitangent=FunctionOutput(None=@n9.Bitangent) {description="World-space bitangent for yUV. Complements yTangent.\r\n", outputName=yBitangent, sortPriority=7}
   "Rotation (Turns)"=FunctionOutput(None=Multiply(A=@n11) {constB=-1}) {description="Rotation applied to yUV, in turns (1.0 = 360°).\r\n", outputName="yRotation (Turns)", sortPriority=-6}
   HeightChannel=NamedRerouteUsage()
   QualityBias=NamedRerouteUsage()
   Height Tile=NamedRerouteUsage()
  ) {materialFunction=&MF_Coordinate_Parallax}
  FunctionOutput(None=@n12.Pixel Depth Offset) {
   description="Depth at the POM intersection (after Intensity/RefPlane). Higher (whiter) = deeper—the farther the ray travels into the material, the larger the value."
   outputName=yDepth
   sortPriority=9
  }
  n13 = FunctionInput() {
   bUsePreviewValueAsDefault=true
   description="Controls the tiling of the hex pattern. Larger values create more, smaller hex tiles across the surface; smaller values create fewer, larger tiles."
   inputName=Hex PatternSize
   inputType=FunctionInput_Scalar
   sortPriority=2
   previewValue=<1 0 0 1>
  }
  n14 = FunctionInput() {
   bUsePreviewValueAsDefault=true
   description="Selects which mesh UV channel (0–3) to use as the base coordinate source."
   inputName=UVIndex
   inputType=FunctionInput_Scalar
   sortPriority=-5
  }
  n15 = FunctionInput() {
   bUsePreviewValueAsDefault=true
   description=Applies an additional per-axis offset to the coordinates before projection. Useful for correctly positioning a texture.
   inputName=CoordinateTransform
   inputType=FunctionInput_Vector2
   sortPriority=-9
  }
  n16 = FunctionInput() {
   bUsePreviewValueAsDefault=true
   description="Rotates the projected UVs around the selected axis. Units are in turns, where 1.0 = 360° and 0.25 = 90°."
   inputName="Rotation (Turns)"
   inputType=FunctionInput_Scalar
   sortPriority=-8
  }
  n17 = FunctionInput() {
   bUsePreviewValueAsDefault=true
   description="Sets how many times the projected coordinates repeat along each axis. Use values above 1 for denser tiling, negative values for mirroring, and avoid zero."
   inputName=CoordinateRepeat
   inputType=FunctionInput_Vector2
   sortPriority=-10
   previewValue=<1 1 1 1>
  }
  n18 = MaterialFunctionCall!(
   Coords=MaterialFunctionCall!(
    Selector=@n14
    VectorA=TextureCoordinate()
    VectorB=TextureCoordinate() {coordinateIndex=1}
    VectorC=TextureCoordinate() {coordinateIndex=2}
    VectorD=TextureCoordinate() {coordinateIndex=3}
   ) {materialFunction=&/MF_Switch4_Vec2}.Selected Vector
   UVOffset=@n15
   UVRotation=NamedRerouteDeclaration(None=@n16) {name=turnsi, nodeColor=<rgba 0.8963 0.8963 0.0091 1>}
   RotationPivot
   UVTile=@n17
  ) {materialFunction=&/MF_CoordinateTransform_2D}
  n19 = FunctionInput() {
   bUsePreviewValueAsDefault=true
   description="Controls blend contrast by raising the hex pattern to 8 × BlendFactor (default 0.5 = pow4) "
   inputName=BlendTolerance
   inputType=FunctionInput_Scalar
   sortPriority=8
   previewValue=<0.5 0 0 1>
  }
  n20 = MaterialFunctionCall!(Coordinates=Multiply(A=@n13, B=@n18.geometry_UV), MaskContrast=@n19) {materialFunction=&/MF_Blend_HexMask}
  FunctionOutput(None=@n20.PatternMask) {
   description="RGB mask giving blend weights for xUV, yUV, zUV. Used for post-sample blending. Weights can also be modulated by albedo or height for more perceptual transitions.\r\n"
   outputName=BlendMask
   sortPriority=15
  }
  n21 = MaterialFunctionCall(
   Coords=NamedRerouteUsage()
   "TileRand (Min/Max)"=NamedRerouteUsage()
   "OffsetRand (Min/Max)"=NamedRerouteUsage()
   RotateStrengthRand=NamedRerouteUsage()
   SeedA=NamedRerouteUsage()
  ) {materialFunction=&MF_Coordinate_Rand1}
  n22 = FunctionOutput(None=@n21.packedDXY) {
   description="Packed derivatives (ddx/ddy) for zUV. Use with SampleGrad for stable mip level selection.\r\n"
   outputName=zDXY
   sortPriority=13
  }
  n23 = MaterialFunctionCall(packedDXY=@n22) {materialFunction=&MF_DeriveTangentBasis}
  n24 = FunctionOutput(None=@n23.Tangent) {description="World-space tangent for zUV. Use with zBitangent to reorient tangent-space normals.\r\n", outputName=zTangent, sortPriority=11}
  n25 = Add(A=@n21.SampleRotation, B=NamedRerouteUsage())
  n26 = MaterialFunctionCall(
   POM_HeightTexture=NamedRerouteUsage()
   Coordinates=@n21.UV
   coordinateDXY=@n22
   HeightStrength=NamedRerouteUsage()
   Reference Plane=NamedRerouteUsage()
   MaxSteps=NamedRerouteUsage()
   Tangent=@n24
   Bitangent=FunctionOutput(None=@n23.Bitangent) {description="World-space bitangent for zUV. Complements zTangent.\r\n", outputName=zBitangent, sortPriority=12}
   "Rotation (Turns)"=FunctionOutput(None=Multiply(A=@n25) {constB=-1}) {description="Rotation applied to zUV, in turns (1.0 = 360°).\r\n", outputName="zRotation (Turns)", sortPriority=-5}
   HeightChannel=NamedRerouteUsage()
   QualityBias=NamedRerouteUsage()
   Height Tile=NamedRerouteUsage()
  ) {materialFunction=&MF_Coordinate_Parallax}
  FunctionOutput(None=@n26.Pixel Depth Offset) {
   description="Depth at the POM intersection (after Intensity/RefPlane). Higher (whiter) = deeper—the farther the ray travels into the material, the larger the value."
   outputName=zDepth
   sortPriority=14
  }
  FunctionOutput(None=@n6.POM UVs) {description="First randomized UV coordinate set generated by the hex tiling projection.\r\n", outputName=xUV}
  FunctionOutput(None=@n12.POM UVs) {description="Second randomized UV coordinate set generated by the hex tiling projection.\r\n", outputName=yUV, sortPriority=5}
  FunctionOutput(None=@n26.POM UVs) {description="Third randomized UV coordinate set generated by the hex tiling projection.\r\n", outputName=zUV, sortPriority=10}
  n27 = FunctionInput() {
   bUsePreviewValueAsDefault=true
   description=Control the U/V Tiling of the height texture relative to the input coordinates.
   inputName=POM HeightTile
   inputType=FunctionInput_Vector2
   sortPriority=12
   previewValue=<1 1 0 1>
  }
  NamedRerouteDeclaration(None=@n27) {name=POM_HeightTile, nodeColor=<rgba 0.9 0.306 0.009 1>}
  n28 = FunctionInput() {
   bUsePreviewValueAsDefault=true
   description=Defines the minimum and maximum random UV offset applied to each hex tile. Shifts the sampling position of tiles within their cells to break up repetition.
   inputName=Hex Min/MaxRand Offset
   inputType=FunctionInput_Vector2
   sortPriority=4
  }
  NamedRerouteDeclaration(None=@n28) {name=HexOffset, nodeColor=<rgba 0.9 0.3505 0.009 1>}
  n29 = FunctionInput() {
   bUsePreviewValueAsDefault=true
   description="Maximum random per-tile rotation, expressed in turns (1.0 = 360°). Each hex tile rotates by a random amount up to this value."
   inputName="Hex Rand Rotation (turns)"
   inputType=FunctionInput_Scalar
   sortPriority=5
  }
  NamedRerouteDeclaration(None=@n29) {name=HexRot, nodeColor=<rgba 0.9 0.3505 0.009 1>}
  n30 = FunctionInput() {
   bUsePreviewValueAsDefault=true
   description="Defines the minimum and maximum random scale applied to each hex tile. Values below 1 enlarge tiles, values above 1 shrink them. Each tile chooses a random value within this range."
   inputName=Hex Min/MaxRand Scale
   inputType=FunctionInput_Vector2
   sortPriority=3
   previewValue=<1 1 0 1>
  }
  NamedRerouteDeclaration(None=@n30) {name=HexScale, nodeColor=<rgba 0.9 0.3505 0.009 1>}
  NamedRerouteDeclaration(None=MaterialFunctionCall!(seed=@n20.HexID1) {materialFunction=&/MF_Hash42}.Hash4D) {name=RandA, nodeColor=<rgba 1 0 0.75 1>}
  n31 = FunctionInput() {
   bUsePreviewValueAsDefault=true
   description="Specifies which channel of the heightmap texture to use for depth information. Assumes the Alpha channel is used, but any channel can be selected if your texture packs data differently."
   inputName=POM HeightChannel
   inputType=FunctionInput_Scalar
   sortPriority=10
   previewValue=<3 0 0 1>
  }
  NamedRerouteDeclaration(None=@n31) {name=POM_Hchan, nodeColor=<rgba 0.9 0.306 0.009 1>}
  n32 = FunctionInput(Preview=TextureObject() {texture=/BaseMaterial/Textures/Default/T_Gray_sRGB}) {
   bUsePreviewValueAsDefault=true
   description="The heightmap texture used for parallax occlusion mapping. Grayscale values define surface depth, with brighter values typically pushed outward and darker values pushed inward."
   inputName=HeightTex
   inputType=FunctionInput_Texture2D
   sortPriority=-20
  }
  NamedRerouteDeclaration(None=@n32) {name=Height, nodeColor=<rgba 1 0 0.4922 1>}
  n33 = FunctionInput() {
   bUsePreviewValueAsDefault=true
   description="Controls the strength of the parallax effect. Higher values exaggerate depth; lower values make surfaces appear flatter."
   inputName=POM Intensity
   inputType=FunctionInput_Scalar
   sortPriority=11
   previewValue=<1 0 0 1>
  }
  NamedRerouteDeclaration(None=@n33) {name=POM_Height, nodeColor=<rgba 0.9 0.306 0.009 1>}
  n34 = FunctionInput() {
   bUsePreviewValueAsDefault=true
   description="Sets the maximum number of depth sampling steps. This is the upper bound, with actual steps scaled by the Quality setting."
   inputName=POM MaxSteps
   inputType=FunctionInput_Scalar
   sortPriority=12
   previewValue=<16 0 0 1>
  }
  NamedRerouteDeclaration(None=@n34) {name=POM_MaxSteps, nodeColor=<rgba 0.9 0.306 0.009 1>}
  n35 = FunctionInput() {
   bUsePreviewValueAsDefault=true
   description="Scales step count by view angle. At 1, MaxSteps are always used; at 0, only one step is used when looking straight on, with more steps added at grazing angles."
   inputName=POM Quality
   inputType=FunctionInput_Scalar
   sortPriority=13
   previewValue=<0.25 0 0 1>
  }
  NamedRerouteDeclaration(None=@n35) {name=POM_Noise, nodeColor=<rgba 0.9 0.306 0.009 1>}
  n36 = FunctionInput() {
   bUsePreviewValueAsDefault=true
   description=Defines the reference plane for parallax. This value shifts where the heightmap appears anchored relative to the surface.
   inputName=POM RefPlane
   inputType=FunctionInput_Scalar
   sortPriority=14
   previewValue=<1 0 0 1>
  }
  NamedRerouteDeclaration(None=@n36) {name=POM_RefPlane, nodeColor=<rgba 0.9 0.306 0.009 1>}
  NamedRerouteDeclaration(None=@n18.geometry_UV) {name=UVi, nodeColor=<rgba 0.9 0.009 0.009 1>}
  NamedRerouteDeclaration(None=MaterialFunctionCall(seed=@n20.HexID2) {materialFunction=&MF_Hash42}.Hash4D) {name=RandB, nodeColor=<rgba 1 0 0.75 1>}
  NamedRerouteDeclaration(None=MaterialFunctionCall(seed=@n20.HexID3) {materialFunction=&MF_Hash42}.Hash4D) {name=RandC, nodeColor=<rgba 1 0 0.75 1>}
 ]
}
