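# Material function that builds two keystone-tilted side projection UV sets from position and reference normal, with a blend between them for biplanar-style mapping.
# Each set picks its dominant X or Y side axis, tilted by KeystoneAngle; the second works 45 degrees around Z, giving octagonal coverage.
# It outputs rotated UVs, tangents, bitangents and shader-stage-switched derivatives.
# An unwired Custom node preserves the iquilez-derived HLSL as reference for the node version.
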MaterialFunction /BaseMaterial/Materials/Functions/MF_Coordinate_Keystone
graph {
 expressions {
  MaterialExpressionAbs_1 {class=MaterialExpressionAbs, inputs [None {from=MaterialExpressionReroute_6, output=""}], properties {}}
  MaterialExpressionAbs_4 {class=MaterialExpressionAbs, inputs [None {from=MaterialExpressionReroute_39, output=""}], properties {}}
  MaterialExpressionAbs_5 {class=MaterialExpressionAbs, inputs [None {from=MaterialExpressionReroute_63, output=""}], properties {}}
  MaterialExpressionAdd_0 {
   class=MaterialExpressionAdd
   inputs [
    A {from=MaterialExpressionMultiply_6}
    B {from=MaterialExpressionMultiply_4}
   ]
  }
  MaterialExpressionAdd_2 {
   class=MaterialExpressionAdd
   inputs [
    A {from=MaterialExpressionMultiply_22}
    B {from=MaterialExpressionMultiply_24}
   ]
  }
  MaterialExpressionAdd_3 {
   class=MaterialExpressionAdd
   inputs [
    A {from=MaterialExpressionMultiply_29}
    B {from=MaterialExpressionMultiply_30}
   ]
  }
  MaterialExpressionAdd_4 {
   class=MaterialExpressionAdd
   inputs [
    A {from=MaterialExpressionMultiply_33}
    B {from=MaterialExpressionMultiply_34}
   ]
  }
  MaterialExpressionAdd_5 {
   class=MaterialExpressionAdd
   inputs [
    A {from=MaterialExpressionMultiply_42}
    B {from=MaterialExpressionMultiply_43}
   ]
  }
  MaterialExpressionAdd_6 {
   class=MaterialExpressionAdd
   inputs [
    A {from=MaterialExpressionMultiply_46}
    B {from=MaterialExpressionMultiply_47}
   ]
  }
  MaterialExpressionAppendVector_0 {
   class=MaterialExpressionAppendVector
   inputs [
    A {from=MaterialExpressionDotProduct_9, output=""}
    B {from=MaterialExpressionMultiply_7, output=""}
   ]
  }
  MaterialExpressionAppendVector_1 {
   class=MaterialExpressionAppendVector
   inputs [
    A {from=MaterialExpressionDotProduct_11, output=""}
    B {from=MaterialExpressionMultiply_8, output=""}
   ]
  }
  MaterialExpressionAppendVector_10 {
   class=MaterialExpressionAppendVector
   inputs [
    A {from=MaterialExpressionDotProduct_29, output=""}
    B {from=MaterialExpressionMultiply_51, output=""}
   ]
  }
  MaterialExpressionAppendVector_4 {
   class=MaterialExpressionAppendVector
   inputs [
    A {from=MaterialExpressionDotProduct_19, output=""}
    B {from=MaterialExpressionMultiply_25, output=""}
   ]
  }
  MaterialExpressionAppendVector_5 {
   class=MaterialExpressionAppendVector
   inputs [
    A {from=MaterialExpressionDotProduct_21, output=""}
    B {from=MaterialExpressionMultiply_26, output=""}
   ]
  }
  MaterialExpressionAppendVector_6 {
   class=MaterialExpressionAppendVector
   inputs [
    A {from=MaterialExpressionReroute_12, output=""}
    B {from=MaterialExpressionReroute_11, output=""}
   ]
  }
  MaterialExpressionAppendVector_7 {
   class=MaterialExpressionAppendVector
   inputs [
    A {from=MaterialExpressionDotProduct_23, output=""}
    B {from=MaterialExpressionMultiply_40, output=""}
   ]
  }
  MaterialExpressionAppendVector_8 {
   class=MaterialExpressionAppendVector
   inputs [
    A {from=MaterialExpressionDotProduct_25, output=""}
    B {from=MaterialExpressionMultiply_41, output=""}
   ]
  }
  MaterialExpressionAppendVector_9 {
   class=MaterialExpressionAppendVector
   inputs [
    A {from=MaterialExpressionDotProduct_27, output=""}
    B {from=MaterialExpressionMultiply_50, output=""}
   ]
  }
  MaterialExpressionConstant2Vector_0 {class=MaterialExpressionConstant2Vector}
  MaterialExpressionConstant3Vector_10 {class=MaterialExpressionConstant3Vector, properties {constant=<rgba 0 0 1 0>}}
  MaterialExpressionConstant3Vector_4 {class=MaterialExpressionConstant3Vector, properties {constant=<rgba 0 0 1 0>}}
  MaterialExpressionConstant3Vector_8 {class=MaterialExpressionConstant3Vector, properties {constant=<rgba 1 0 0 0>}}
  MaterialExpressionConstant3Vector_9 {class=MaterialExpressionConstant3Vector, properties {constant=<rgba 0 1 0 0>}}
  MaterialExpressionConstant4Vector_0 {class=MaterialExpressionConstant4Vector, inputs [], properties {desc="force mip 0 in VS - might be possible to do better (see vd1 and vd2)"}}
  MaterialExpressionConstant_2 {class=MaterialExpressionConstant, properties {desc=-1/180, r=-0.0056}}
  MaterialExpressionConstant_4 {class=MaterialExpressionConstant, properties {r=0.25}}
  MaterialExpressionConstant_5 {class=MaterialExpressionConstant, properties {r=0.25}}
  MaterialExpressionConvert_1 {
   class=MaterialExpressionConvert
   inputs [None {from=MaterialExpressionAbs_4}]
   properties {
    nodeName=Break Float3
    convertInputs [{
      type=Vector3
      expressionInput {
       expression=&self:MaterialExpressionAbs_4
       inputName=None
       mask=0
       maskA=0
       maskB=0
       maskG=0
       maskR=0
       outputIndex=0
      }
     }]
    convertOutputs [{type=Scalar}, {type=Scalar}, {type=Scalar}]
   }
  }
  MaterialExpressionConvert_15 {
   class=MaterialExpressionConvert
   inputs [None {from=MaterialExpressionSaturate_2}]
   properties {
    nodeName=Break Float3
    convertInputs [{
      type=Vector3
      expressionInput {
       expression=&self:MaterialExpressionSaturate_2
       inputName=None
       mask=0
       maskA=0
       maskB=0
       maskG=0
       maskR=0
       outputIndex=0
      }
     }]
    convertOutputs [{type=Scalar}, {type=Scalar}, {type=Scalar}]
   }
  }
  MaterialExpressionConvert_16 {
   class=MaterialExpressionConvert
   inputs [
    None {from=MaterialExpressionCosine_2}
    None {from=""}
    None {from=MaterialExpressionSine_2}
   ]
   properties {
    nodeName=plusX
    convertInputs [
     {
      type=Scalar
      expressionInput {
       expression=&self:MaterialExpressionCosine_2
       inputName=None
       mask=0
       maskA=0
       maskB=0
       maskG=0
       maskR=0
       outputIndex=0
      }
     }
     {type=Scalar, expressionInput {expression=null, inputName=None, mask=0, maskA=0, maskB=0, maskG=0, maskR=0, outputIndex=0}}
     {
      type=Scalar
      expressionInput {
       expression=&self:MaterialExpressionSine_2
       inputName=None
       mask=0
       maskA=0
       maskB=0
       maskG=0
       maskR=0
       outputIndex=0
      }
     }
    ]
    convertOutputs [{type=Vector3}]
   }
  }
  MaterialExpressionConvert_17 {
   class=MaterialExpressionConvert
   inputs [None {from=MaterialExpressionConvert_16}]
   properties {
    nodeName=plusY
    convertInputs [{
      type=Vector3
      expressionInput {
       expression=&self:MaterialExpressionConvert_16
       inputName=None
       mask=0
       maskA=0
       maskB=0
       maskG=0
       maskR=0
       outputIndex=0
      }
     }]
    convertOutputs [{type=Vector3}]
   }
  }
  MaterialExpressionConvert_18 {
   class=MaterialExpressionConvert
   inputs [
    None {from=MaterialExpressionMultiply_18}
    None {from=""}
    None {from=MaterialExpressionSine_2}
   ]
   properties {
    nodeName=minusX
    convertInputs [
     {
      type=Scalar
      expressionInput {
       expression=&self:MaterialExpressionMultiply_18
       inputName=None
       mask=0
       maskA=0
       maskB=0
       maskG=0
       maskR=0
       outputIndex=0
      }
     }
     {type=Scalar, expressionInput {expression=null, inputName=None, mask=0, maskA=0, maskB=0, maskG=0, maskR=0, outputIndex=0}}
     {
      type=Scalar
      expressionInput {
       expression=&self:MaterialExpressionSine_2
       inputName=None
       mask=0
       maskA=0
       maskB=0
       maskG=0
       maskR=0
       outputIndex=0
      }
     }
    ]
    convertOutputs [{type=Vector3}]
   }
  }
  MaterialExpressionConvert_19 {
   class=MaterialExpressionConvert
   inputs [None {from=MaterialExpressionConvert_18}]
   properties {
    nodeName=minusY
    convertInputs [{
      type=Vector3
      expressionInput {
       expression=&self:MaterialExpressionConvert_18
       inputName=None
       mask=0
       maskA=0
       maskB=0
       maskG=0
       maskR=0
       outputIndex=0
      }
     }]
    convertOutputs [{type=Vector3}]
   }
  }
  MaterialExpressionConvert_25 {
   class=MaterialExpressionConvert
   inputs [
    None {from=MaterialExpressionAdd_3}
    None {from=MaterialExpressionSubtract_2}
    None {from=MaterialExpressionReroute_47}
   ]
   properties {
    nodeName=Make Vector3
    convertInputs [
     {
      type=Scalar
      expressionInput {
       expression=&self:MaterialExpressionAdd_3
       inputName=None
       mask=0
       maskA=0
       maskB=0
       maskG=0
       maskR=0
       outputIndex=0
      }
     }
     {
      type=Scalar
      expressionInput {
       expression=&self:MaterialExpressionSubtract_2
       inputName=None
       mask=0
       maskA=0
       maskB=0
       maskG=0
       maskR=0
       outputIndex=0
      }
     }
     {
      type=Scalar
      expressionInput {
       expression=&self:MaterialExpressionReroute_47
       inputName=None
       mask=0
       maskA=0
       maskB=0
       maskG=0
       maskR=0
       outputIndex=0
      }
     }
    ]
    convertOutputs [{type=Vector3}]
   }
  }
  MaterialExpressionConvert_26 {
   class=MaterialExpressionConvert
   inputs [None {from=MaterialExpressionNamedRerouteUsage_19}]
   properties {
    nodeName=Break Float3
    convertInputs [{
      type=Vector3
      expressionInput {
       expression=&self:MaterialExpressionNamedRerouteUsage_19
       inputName=None
       mask=0
       maskA=0
       maskB=0
       maskG=0
       maskR=0
       outputIndex=0
      }
     }]
    convertOutputs [{type=Scalar}, {type=Scalar}, {type=Scalar}]
   }
  }
  MaterialExpressionConvert_27 {
   class=MaterialExpressionConvert
   inputs [
    None {from=MaterialExpressionAdd_4}
    None {from=MaterialExpressionSubtract_3}
    None {from=MaterialExpressionReroute_48}
   ]
   properties {
    nodeName=Make Vector3
    convertInputs [
     {
      type=Scalar
      expressionInput {
       expression=&self:MaterialExpressionAdd_4
       inputName=None
       mask=0
       maskA=0
       maskB=0
       maskG=0
       maskR=0
       outputIndex=0
      }
     }
     {
      type=Scalar
      expressionInput {
       expression=&self:MaterialExpressionSubtract_3
       inputName=None
       mask=0
       maskA=0
       maskB=0
       maskG=0
       maskR=0
       outputIndex=0
      }
     }
     {
      type=Scalar
      expressionInput {
       expression=&self:MaterialExpressionReroute_48
       inputName=None
       mask=0
       maskA=0
       maskB=0
       maskG=0
       maskR=0
       outputIndex=0
      }
     }
    ]
    convertOutputs [{type=Vector3}]
   }
  }
  MaterialExpressionConvert_28 {
   class=MaterialExpressionConvert
   inputs [None {from=MaterialExpressionNamedRerouteUsage_0}]
   properties {
    nodeName=Break Float3
    convertInputs [{
      type=Vector3
      expressionInput {
       expression=&self:MaterialExpressionNamedRerouteUsage_0
       inputName=None
       mask=0
       maskA=0
       maskB=0
       maskG=0
       maskR=0
       outputIndex=0
      }
     }]
    convertOutputs [{type=Scalar}, {type=Scalar}, {type=Scalar}]
   }
  }
  MaterialExpressionConvert_33 {
   class=MaterialExpressionConvert
   inputs [
    None {from=MaterialExpressionLength_1}
    None {from=MaterialExpressionLength_2}
    None {from=MaterialExpressionLength_3}
   ]
   properties {
    nodeName=Make Vector3
    convertInputs [
     {
      type=Scalar
      expressionInput {
       expression=&self:MaterialExpressionLength_1
       inputName=None
       mask=0
       maskA=0
       maskB=0
       maskG=0
       maskR=0
       outputIndex=0
      }
     }
     {
      type=Scalar
      expressionInput {
       expression=&self:MaterialExpressionLength_2
       inputName=None
       mask=0
       maskA=0
       maskB=0
       maskG=0
       maskR=0
       outputIndex=0
      }
     }
     {
      type=Scalar
      expressionInput {
       expression=&self:MaterialExpressionLength_3
       inputName=None
       mask=0
       maskA=0
       maskB=0
       maskG=0
       maskR=0
       outputIndex=0
      }
     }
    ]
    convertOutputs [{type=Vector3}]
   }
  }
  MaterialExpressionConvert_34 {
   class=MaterialExpressionConvert
   inputs [None {from=MaterialExpressionAbs_5}]
   properties {
    nodeName=Break Float3
    convertInputs [{
      type=Vector3
      expressionInput {
       expression=&self:MaterialExpressionAbs_5
       inputName=None
       mask=0
       maskA=0
       maskB=0
       maskG=0
       maskR=0
       outputIndex=0
      }
     }]
    convertOutputs [{type=Scalar}, {type=Scalar}, {type=Scalar}]
   }
  }
  MaterialExpressionConvert_35 {
   class=MaterialExpressionConvert
   inputs [None {from=MaterialExpressionIf_7}]
   properties {
    convertInputs [{
      type=Vector2
      expressionInput {
       expression=&self:MaterialExpressionIf_7
       inputName=None
       mask=0
       maskA=0
       maskB=0
       maskG=0
       maskR=0
       outputIndex=0
      }
     }]
    convertOutputs [{type=Vector4, defaultValue=<rgba 0 0 0 1>}]
   }
  }
  MaterialExpressionConvert_36 {
   class=MaterialExpressionConvert
   inputs [
    None {from=MaterialExpressionAdd_5}
    None {from=MaterialExpressionSubtract_7}
    None {from=MaterialExpressionReroute_67}
   ]
   properties {
    nodeName=Make Vector3
    convertInputs [
     {
      type=Scalar
      expressionInput {
       expression=&self:MaterialExpressionAdd_5
       inputName=None
       mask=0
       maskA=0
       maskB=0
       maskG=0
       maskR=0
       outputIndex=0
      }
     }
     {
      type=Scalar
      expressionInput {
       expression=&self:MaterialExpressionSubtract_7
       inputName=None
       mask=0
       maskA=0
       maskB=0
       maskG=0
       maskR=0
       outputIndex=0
      }
     }
     {
      type=Scalar
      expressionInput {
       expression=&self:MaterialExpressionReroute_67
       inputName=None
       mask=0
       maskA=0
       maskB=0
       maskG=0
       maskR=0
       outputIndex=0
      }
     }
    ]
    convertOutputs [{type=Vector3}]
   }
  }
  MaterialExpressionConvert_37 {
   class=MaterialExpressionConvert
   inputs [None {from=MaterialExpressionMultiply_39}]
   properties {
    nodeName=Break Float3
    convertInputs [{
      type=Vector3
      expressionInput {
       expression=&self:MaterialExpressionMultiply_39
       inputName=None
       mask=0
       maskA=0
       maskB=0
       maskG=0
       maskR=0
       outputIndex=0
      }
     }]
    convertOutputs [{type=Scalar}, {type=Scalar}, {type=Scalar}]
   }
  }
  MaterialExpressionConvert_39 {
   class=MaterialExpressionConvert
   inputs [None {from=MaterialExpressionReroute_63}]
   properties {
    nodeName=Break Float3
    convertInputs [{
      type=Vector3
      expressionInput {
       expression=&self:MaterialExpressionReroute_63
       inputName=None
       mask=0
       maskA=0
       maskB=0
       maskG=0
       maskR=0
       outputIndex=0
      }
     }]
    convertOutputs [{type=Scalar}, {type=Scalar}, {type=Scalar}]
   }
  }
  MaterialExpressionConvert_40 {
   class=MaterialExpressionConvert
   inputs [None {from=MaterialExpressionIf_8}]
   properties {
    convertInputs [{
      type=Vector2
      expressionInput {
       expression=&self:MaterialExpressionIf_8
       inputName=None
       mask=0
       maskA=0
       maskB=0
       maskG=0
       maskR=0
       outputIndex=0
      }
     }]
    convertOutputs [{type=Vector4, defaultValue=<rgba 0 0 0 1>}]
   }
  }
  MaterialExpressionConvert_6 {
   class=MaterialExpressionConvert
   inputs [None {from=MaterialExpressionAbs_1}]
   properties {
    nodeName=Break Float3
    convertInputs [{
      type=Vector3
      expressionInput {
       expression=&self:MaterialExpressionAbs_1
       inputName=None
       mask=0
       maskA=0
       maskB=0
       maskG=0
       maskR=0
       outputIndex=0
      }
     }]
    convertOutputs [{type=Scalar}, {type=Scalar}, {type=Scalar}]
   }
  }
  MaterialExpressionConvert_8 {
   class=MaterialExpressionConvert
   inputs [
    None {from=MaterialExpressionDotProduct_0}
    None {from=MaterialExpressionDotProduct_1}
    None {from=MaterialExpressionDotProduct_2}
    None {from=MaterialExpressionDotProduct_3}
   ]
   properties {
    convertInputs [
     {
      type=Scalar
      expressionInput {
       expression=&self:MaterialExpressionDotProduct_0
       inputName=None
       mask=0
       maskA=0
       maskB=0
       maskG=0
       maskR=0
       outputIndex=0
      }
     }
     {
      type=Scalar
      expressionInput {
       expression=&self:MaterialExpressionDotProduct_1
       inputName=None
       mask=0
       maskA=0
       maskB=0
       maskG=0
       maskR=0
       outputIndex=0
      }
     }
     {
      type=Scalar
      expressionInput {
       expression=&self:MaterialExpressionDotProduct_2
       inputName=None
       mask=0
       maskA=0
       maskB=0
       maskG=0
       maskR=0
       outputIndex=0
      }
     }
     {
      type=Scalar
      expressionInput {
       expression=&self:MaterialExpressionDotProduct_3
       inputName=None
       mask=0
       maskA=0
       maskB=0
       maskG=0
       maskR=0
       outputIndex=0
      }
     }
    ]
    convertOutputs [{type=Vector4, defaultValue=<rgba 0 0 0 1>}]
   }
  }
  MaterialExpressionConvert_9 {
   class=MaterialExpressionConvert
   inputs [
    None {from=MaterialExpressionDotProduct_4}
    None {from=MaterialExpressionDotProduct_5}
    None {from=MaterialExpressionDotProduct_6}
    None {from=MaterialExpressionDotProduct_7}
   ]
   properties {
    convertInputs [
     {
      type=Scalar
      expressionInput {
       expression=&self:MaterialExpressionDotProduct_4
       inputName=None
       mask=0
       maskA=0
       maskB=0
       maskG=0
       maskR=0
       outputIndex=0
      }
     }
     {
      type=Scalar
      expressionInput {
       expression=&self:MaterialExpressionDotProduct_5
       inputName=None
       mask=0
       maskA=0
       maskB=0
       maskG=0
       maskR=0
       outputIndex=0
      }
     }
     {
      type=Scalar
      expressionInput {
       expression=&self:MaterialExpressionDotProduct_6
       inputName=None
       mask=0
       maskA=0
       maskB=0
       maskG=0
       maskR=0
       outputIndex=0
      }
     }
     {
      type=Scalar
      expressionInput {
       expression=&self:MaterialExpressionDotProduct_7
       inputName=None
       mask=0
       maskA=0
       maskB=0
       maskG=0
       maskR=0
       outputIndex=0
      }
     }
    ]
    convertOutputs [{type=Vector4, defaultValue=<rgba 0 0 0 1>}]
   }
  }
  MaterialExpressionCosine_2 {class=MaterialExpressionCosine, inputs [None {from=MaterialExpressionMultiply_17, output=""}], properties {}}
  MaterialExpressionCosine_4 {class=MaterialExpressionCosine, inputs [None {from=MaterialExpressionMaterialFunctionCall_6, output=Result}], properties {period=6.2832}}
  MaterialExpressionCosine_5 {class=MaterialExpressionCosine, inputs [None {from=MaterialExpressionMaterialFunctionCall_9, output=Result}], properties {period=6.2832}}
  MaterialExpressionCrossProduct_0 {
   class=MaterialExpressionCrossProduct
   inputs [
    A {from=MaterialExpressionReroute_38, output=""}
    B {from=MaterialExpressionNormalize_12, output=""}
   ]
   properties {}
  }
  MaterialExpressionCrossProduct_10 {
   class=MaterialExpressionCrossProduct
   inputs [
    A {from=MaterialExpressionLinearInterpolate_4, output=""}
    B {from=MaterialExpressionNormalize_8, output=""}
   ]
   properties {desc=xac}
  }
  MaterialExpressionCrossProduct_11 {
   class=MaterialExpressionCrossProduct
   inputs [
    A {from=MaterialExpressionLinearInterpolate_4, output=""}
    B {from=MaterialExpressionConstant3Vector_4, output=""}
   ]
   properties {}
  }
  MaterialExpressionCrossProduct_12 {
   class=MaterialExpressionCrossProduct
   inputs [
    A {from=MaterialExpressionLinearInterpolate_5, output=""}
    B {from=MaterialExpressionNormalize_9, output=""}
   ]
   properties {desc=yac}
  }
  MaterialExpressionCrossProduct_13 {
   class=MaterialExpressionCrossProduct
   inputs [
    A {from=MaterialExpressionLinearInterpolate_5, output=""}
    B {from=MaterialExpressionConstant3Vector_4, output=""}
   ]
   properties {}
  }
  MaterialExpressionCrossProduct_4 {
   class=MaterialExpressionCrossProduct
   inputs [
    A {from=MaterialExpressionReroute_1, output=""}
    B {from=MaterialExpressionNormalize_2, output=""}
   ]
   properties {}
  }
  MaterialExpressionCustom_0 {
   class=MaterialExpressionCustom
   inputs [
    p {from="", output=""}
    dpx {from="", output=""}
    dpy {from="", output=""}
    n {from="", output=""}
    sn {from="", output=""}
    k {from="", output=""}
    a {from="", output=""}
   ]
   properties {
    code="// HLSL preserved for Reference\r\n// modified for UE from iquilez https://iquilezles.org/articles/biplanar/\r\n// more reading here https://ryandowlingsoka.com/unreal/triplanar-dither-biplanar/\r\n// last bit of ref https://www.youtube.com/watch?v=Cq5H59G-DHI\r\n\r\nfloat3 dpdx = dpx;\r\nfloat3 dpdy = dpy;\r\nconst float3 signs = sn;\r\n\r\nconst float theta = radians(-a);\r\nfloat3 plusX  = float3(  cos(theta), 0.f, sin(theta) );\r\nfloat3 minusX = float3( -cos(theta), 0.f, sin(theta) );\r\nfloat3 plusY  = plusX.yxz;\r\nfloat3 minusY = minusX.yxz;\r\n\r\n//new x/y axes\r\nfloat3 xaa = saturate(signs.x) == 0 ? plusX : minusX;\r\nfloat3 yaa = saturate(signs.y) == 0 ? plusY : minusY;\r\nfloat3 xab = normalize(cross(xaa, float3(0,0,1)));\r\nfloat3 yab = normalize(cross(yaa, float3(0,0,1)));\r\nfloat3 xac = cross(xaa, xab);\r\nfloat3 yac = cross(yaa, yab);\r\n\r\n//Output UV1\r\nuv1 = (n.x*signs.x >= n.y*signs.y) ? float2( dot(p, xab), -dot(p, xac) ):\r\n\t\t\t\t                     float2( dot(p, yab), -dot(p, yac) );\r\n\r\n//Output UV1 Tangent + Bitangent\r\nfloat3 Tangent1 = n.x * signs.x >= n.y * signs.y ? -xab : -yab;\r\nTangent1 += n * ( -1.f * dot(Tangent1, n) );\r\nuv1_tangent = normalize(Tangent1); \r\nuv1_bitangent = normalize( cross(n,Tangent1) );\r\n\r\n//Make UV2 by rotating 45 degrees around Z\r\nstatic const float3x3 rot45 = {  cos(PI/4), sin(PI/4), 0.f,\r\n                                -sin(PI/4), cos(PI/4), 0.f,\r\n                                       0.f,       0.f, 1.f };\r\n\r\nfloat3 nn  = mul(rot45, n);\r\nfloat3 np  = mul(rot45, p);\r\n\r\n//Get UV2 Sign\r\nint3 nsign = sign(nn);\r\n\r\n//Output UV2\r\nuv2 = (nn.x*nsign.x >= nn.y*nsign.y) ? float2( dot(np, xab), -dot(np, xac) ):\r\n\t\t\t\t                       float2( dot(np, yab), -dot(np, yac) );\r\n\r\n//Output UV2 Tangent + Bitangent\r\nfloat3 Tangent2 = nn.x * nsign.x >= nn.y * nsign.y ? -xab : -yab;\r\nTangent2 += nn * ( -1.f * dot(Tangent2, nn) );\r\nuv2_tangent =   normalize(Tangent2);\r\nuv2_bitangent = normalize( cross(nn,Tangent2) );\r\n\r\n//Output UV Masks\r\nuv1_index =  n.x * signs.x >=  n.y * signs.y ? 0 : 1;\r\nuv2_index = nn.x * nsign.x >= nn.y * nsign.y ? 0 : 1;\r\n\r\n// blend factors\r\nfloat2 w = float2( signs[uv1_index]*n[uv1_index], nsign[uv2_index] * nn[uv2_index]);\r\nfloat2 wmax = max(w.x, w.y);\r\nw = wmax - w;\r\nw = saturate( ((k/3.f) - w) / (k/3.f));\r\n\r\nreturn w;"
    description=OctaKeystone
    outputType=CMOT_Float2
    showCode=true
    additionalOutputs [
     {outputName=uv1, outputType=CMOT_Float2}
     {outputName=uv1_index, outputType=CMOT_Float1}
     {outputName=uv1_tangent, outputType=CMOT_Float3}
     {outputName=uv1_bitangent, outputType=CMOT_Float3}
     {outputName=uv2, outputType=CMOT_Float2}
     {outputName=uv2_index, outputType=CMOT_Float1}
     {outputName=uv2_tangent, outputType=CMOT_Float3}
     {outputName=uv2_bitangent, outputType=CMOT_Float3}
    ]
    inputs [
     {inputName=p, input {expression=null, inputName=None, mask=0, maskA=0, maskB=0, maskG=0, maskR=0, outputIndex=0}}
     {inputName=dpx, input {expression=null, inputName=None, mask=0, maskA=0, maskB=0, maskG=0, maskR=0, outputIndex=0}}
     {inputName=dpy, input {expression=null, inputName=None, mask=0, maskA=0, maskB=0, maskG=0, maskR=0, outputIndex=0}}
     {inputName=n, input {expression=null, inputName=None, mask=0, maskA=0, maskB=0, maskG=0, maskR=0, outputIndex=0}}
     {inputName=sn, input {expression=null, inputName=None, mask=0, maskA=0, maskB=0, maskG=0, maskR=0, outputIndex=0}}
     {inputName=k, input {expression=null, inputName=None, mask=0, maskA=0, maskB=0, maskG=0, maskR=0, outputIndex=0}}
     {inputName=a, input {expression=null, inputName=None, mask=0, maskA=0, maskB=0, maskG=0, maskR=0, outputIndex=0}}
    ]
   }
  }
  MaterialExpressionDDX_1 {class=MaterialExpressionDDX, inputs [Value {from=MaterialExpressionNamedRerouteUsage_1, output=""}], properties {}}
  MaterialExpressionDDY_1 {class=MaterialExpressionDDY, inputs [Value {from=MaterialExpressionNamedRerouteUsage_2, output=""}], properties {}}
  MaterialExpressionDivide_0 {
   class=MaterialExpressionDivide
   inputs [
    A {from=MaterialExpressionFunctionInput_2}
    B {from=""}
   ]
   properties {constB=3}
  }
  MaterialExpressionDivide_1 {
   class=MaterialExpressionDivide
   inputs [
    A {from=MaterialExpressionWorldPosition_1, output=XYZ}
    B {from=""}
   ]
   properties {constB=128}
  }
  MaterialExpressionDivide_2 {
   class=MaterialExpressionDivide
   inputs [
    A {from=MaterialExpressionSubtract_6}
    B {from=MaterialExpressionDivide_0}
   ]
  }
  MaterialExpressionDotProduct_0 {
   class=MaterialExpressionDotProduct
   inputs [
    A {from=MaterialExpressionReroute_2, output=""}
    B {from=MaterialExpressionFunctionOutput_3, output=""}
   ]
   properties {}
  }
  MaterialExpressionDotProduct_1 {
   class=MaterialExpressionDotProduct
   inputs [
    A {from=MaterialExpressionReroute_2, output=""}
    B {from=MaterialExpressionFunctionOutput_0, output=""}
   ]
   properties {}
  }
  MaterialExpressionDotProduct_10 {
   class=MaterialExpressionDotProduct
   inputs [
    A {from=MaterialExpressionReroute_5, output=""}
    B {from=MaterialExpressionNamedRerouteUsage_17, output=""}
   ]
   properties {}
  }
  MaterialExpressionDotProduct_11 {
   class=MaterialExpressionDotProduct
   inputs [
    A {from=MaterialExpressionReroute_5, output=""}
    B {from=MaterialExpressionNamedRerouteUsage_23, output=""}
   ]
   properties {}
  }
  MaterialExpressionDotProduct_12 {
   class=MaterialExpressionDotProduct
   inputs [
    A {from=MaterialExpressionReroute_5, output=""}
    B {from=MaterialExpressionNamedRerouteUsage_14, output=""}
   ]
   properties {}
  }
  MaterialExpressionDotProduct_18 {
   class=MaterialExpressionDotProduct
   inputs [
    A {from=MaterialExpressionReroute_38, output=""}
    B {from=MaterialExpressionMultiply_24, output=""}
   ]
   properties {}
  }
  MaterialExpressionDotProduct_19 {
   class=MaterialExpressionDotProduct
   inputs [
    A {from=MaterialExpressionReroute_40, output=""}
    B {from=MaterialExpressionNamedRerouteUsage_6, output=""}
   ]
   properties {}
  }
  MaterialExpressionDotProduct_2 {
   class=MaterialExpressionDotProduct
   inputs [
    A {from=MaterialExpressionReroute_3, output=""}
    B {from=MaterialExpressionFunctionOutput_3, output=""}
   ]
   properties {}
  }
  MaterialExpressionDotProduct_20 {
   class=MaterialExpressionDotProduct
   inputs [
    A {from=MaterialExpressionReroute_40, output=""}
    B {from=MaterialExpressionNamedRerouteUsage_16, output=""}
   ]
   properties {}
  }
  MaterialExpressionDotProduct_21 {
   class=MaterialExpressionDotProduct
   inputs [
    A {from=MaterialExpressionReroute_40, output=""}
    B {from=MaterialExpressionNamedRerouteUsage_10, output=""}
   ]
   properties {}
  }
  MaterialExpressionDotProduct_22 {
   class=MaterialExpressionDotProduct
   inputs [
    A {from=MaterialExpressionReroute_40, output=""}
    B {from=MaterialExpressionNamedRerouteUsage_13, output=""}
   ]
   properties {}
  }
  MaterialExpressionDotProduct_23 {
   class=MaterialExpressionDotProduct
   inputs [
    A {from=MaterialExpressionReroute_55, output=""}
    B {from=MaterialExpressionNamedRerouteUsage_4, output=""}
   ]
   properties {}
  }
  MaterialExpressionDotProduct_24 {
   class=MaterialExpressionDotProduct
   inputs [
    A {from=MaterialExpressionReroute_55, output=""}
    B {from=MaterialExpressionNamedRerouteUsage_15, output=""}
   ]
   properties {}
  }
  MaterialExpressionDotProduct_25 {
   class=MaterialExpressionDotProduct
   inputs [
    A {from=MaterialExpressionReroute_55, output=""}
    B {from=MaterialExpressionNamedRerouteUsage_8, output=""}
   ]
   properties {}
  }
  MaterialExpressionDotProduct_26 {
   class=MaterialExpressionDotProduct
   inputs [
    A {from=MaterialExpressionReroute_55, output=""}
    B {from=MaterialExpressionNamedRerouteUsage_12, output=""}
   ]
   properties {}
  }
  MaterialExpressionDotProduct_27 {
   class=MaterialExpressionDotProduct
   inputs [
    A {from=MaterialExpressionReroute_69, output=""}
    B {from="", output=""}
   ]
   properties {}
  }
  MaterialExpressionDotProduct_28 {
   class=MaterialExpressionDotProduct
   inputs [
    A {from=MaterialExpressionReroute_69, output=""}
    B {from="", output=""}
   ]
   properties {}
  }
  MaterialExpressionDotProduct_29 {
   class=MaterialExpressionDotProduct
   inputs [
    A {from=MaterialExpressionReroute_69, output=""}
    B {from="", output=""}
   ]
   properties {}
  }
  MaterialExpressionDotProduct_3 {
   class=MaterialExpressionDotProduct
   inputs [
    A {from=MaterialExpressionReroute_3, output=""}
    B {from=MaterialExpressionFunctionOutput_0, output=""}
   ]
   properties {}
  }
  MaterialExpressionDotProduct_30 {
   class=MaterialExpressionDotProduct
   inputs [
    A {from=MaterialExpressionReroute_69, output=""}
    B {from="", output=""}
   ]
   properties {}
  }
  MaterialExpressionDotProduct_4 {
   class=MaterialExpressionDotProduct
   inputs [
    A {from=MaterialExpressionReroute_2, output=""}
    B {from=MaterialExpressionFunctionOutput_6, output=""}
   ]
   properties {}
  }
  MaterialExpressionDotProduct_5 {
   class=MaterialExpressionDotProduct
   inputs [
    A {from=MaterialExpressionReroute_2, output=""}
    B {from=MaterialExpressionFunctionOutput_8, output=""}
   ]
   properties {}
  }
  MaterialExpressionDotProduct_6 {
   class=MaterialExpressionDotProduct
   inputs [
    A {from=MaterialExpressionReroute_3, output=""}
    B {from=MaterialExpressionFunctionOutput_6, output=""}
   ]
   properties {}
  }
  MaterialExpressionDotProduct_7 {
   class=MaterialExpressionDotProduct
   inputs [
    A {from=MaterialExpressionReroute_3, output=""}
    B {from=MaterialExpressionFunctionOutput_8, output=""}
   ]
   properties {}
  }
  MaterialExpressionDotProduct_8 {
   class=MaterialExpressionDotProduct
   inputs [
    A {from=MaterialExpressionReroute_1, output=""}
    B {from=MaterialExpressionMultiply_4, output=""}
   ]
   properties {}
  }
  MaterialExpressionDotProduct_9 {
   class=MaterialExpressionDotProduct
   inputs [
    A {from=MaterialExpressionReroute_5, output=""}
    B {from=MaterialExpressionNamedRerouteUsage_22, output=""}
   ]
   properties {}
  }
  MaterialExpressionFunctionInput_0 {
   class=MaterialExpressionFunctionInput
   inputs [Preview {from=MaterialExpressionDivide_1, output=""}]
   properties {bUsePreviewValueAsDefault=true, inputName=Position, sortPriority=-1}
  }
  MaterialExpressionFunctionInput_1 {
   class=MaterialExpressionFunctionInput
   inputs [Preview {from=MaterialExpressionVertexNormalWS_0, output=""}]
   properties {bUsePreviewValueAsDefault=true, inputName=RefNorm, sortPriority=1}
  }
  MaterialExpressionFunctionInput_2 {
   class=MaterialExpressionFunctionInput
   inputs [Preview {from="", output=""}]
   properties {bUsePreviewValueAsDefault=true, inputName=BlendFactor, inputType=FunctionInput_Scalar, sortPriority=2, previewValue=<0.5 0 0 1>}
  }
  MaterialExpressionFunctionInput_3 {
   class=MaterialExpressionFunctionInput
   inputs [Preview {from="", output=""}]
   properties {bUsePreviewValueAsDefault=true, inputName="Rotation (Turns)", inputType=FunctionInput_Scalar, sortPriority=2}
  }
  MaterialExpressionFunctionInput_4 {
   class=MaterialExpressionFunctionInput
   inputs [Preview {from="", output=""}]
   properties {bUsePreviewValueAsDefault=true, inputName=KeystoneAngle, inputType=FunctionInput_Scalar, sortPriority=3}
  }
  MaterialExpressionFunctionInput_5 {
   class=MaterialExpressionFunctionInput
   inputs [Preview {from="", output=""}]
   properties {bUsePreviewValueAsDefault=true, inputName=Tiling, previewValue=<1 1 1 1>}
  }
  MaterialExpressionFunctionOutput_0 {
   class=MaterialExpressionFunctionOutput
   inputs [None {from=MaterialExpressionNormalize_3, output=""}]
   properties {outputName=UV1_bitangent, sortPriority=3}
  }
  MaterialExpressionFunctionOutput_1 {
   class=MaterialExpressionFunctionOutput
   inputs [None {from=MaterialExpressionMaterialFunctionCall_1, output=Result}]
   properties {outputName=UV1}
  }
  MaterialExpressionFunctionOutput_2 {
   class=MaterialExpressionFunctionOutput
   inputs [None {from=MaterialExpressionShaderStageSwitch_1, output=""}]
   properties {outputName=UV1_derivs, sortPriority=1}
  }
  MaterialExpressionFunctionOutput_3 {
   class=MaterialExpressionFunctionOutput
   inputs [None {from=MaterialExpressionNormalize_2, output=""}]
   properties {outputName=UV1_tangent, sortPriority=2}
  }
  MaterialExpressionFunctionOutput_4 {
   class=MaterialExpressionFunctionOutput
   inputs [None {from=MaterialExpressionMaterialFunctionCall_0, output=Result}]
   properties {outputName=UV2, sortPriority=5}
  }
  MaterialExpressionFunctionOutput_5 {
   class=MaterialExpressionFunctionOutput
   inputs [None {from=MaterialExpressionShaderStageSwitch_0, output=""}]
   properties {outputName=UV2_derivs, sortPriority=6}
  }
  MaterialExpressionFunctionOutput_6 {
   class=MaterialExpressionFunctionOutput
   inputs [None {from=MaterialExpressionNormalize_12, output=""}]
   properties {outputName=UV2_tangent, sortPriority=7}
  }
  MaterialExpressionFunctionOutput_7 {
   class=MaterialExpressionFunctionOutput
   inputs [None {from=MaterialExpressionSaturate_4, output=""}]
   properties {outputName=Blend, sortPriority=10}
  }
  MaterialExpressionFunctionOutput_8 {
   class=MaterialExpressionFunctionOutput
   inputs [None {from=MaterialExpressionNormalize_0, output=""}]
   properties {outputName=UV2_bitangent, sortPriority=7}
  }
  MaterialExpressionIf_0 {
   class=MaterialExpressionIf
   inputs [
    A {from=MaterialExpressionConvert_6, output=""}
    B {from=MaterialExpressionConvert_6, output=""}
    "A > B" {from=MaterialExpressionNamedRerouteUsage_7, output=""}
    "A == B" {from="", output=""}
    "A < B" {from=MaterialExpressionNamedRerouteUsage_11, output=""}
   ]
   properties {}
  }
  MaterialExpressionIf_1 {
   class=MaterialExpressionIf
   inputs [
    A {from=MaterialExpressionConvert_6, output=""}
    B {from=MaterialExpressionConvert_6, output=""}
    "A > B" {from=MaterialExpressionAppendVector_0, output=""}
    "A == B" {from="", output=""}
    "A < B" {from=MaterialExpressionAppendVector_1, output=""}
   ]
   properties {}
  }
  MaterialExpressionIf_4 {
   class=MaterialExpressionIf
   inputs [
    A {from=MaterialExpressionConvert_1, output=""}
    B {from=MaterialExpressionConvert_1, output=""}
    "A > B" {from=MaterialExpressionNamedRerouteUsage_5, output=""}
    "A == B" {from="", output=""}
    "A < B" {from=MaterialExpressionNamedRerouteUsage_9, output=""}
   ]
   properties {}
  }
  MaterialExpressionIf_5 {
   class=MaterialExpressionIf
   inputs [
    A {from=MaterialExpressionConvert_1, output=""}
    B {from=MaterialExpressionConvert_1, output=""}
    "A > B" {from=MaterialExpressionAppendVector_4, output=""}
    "A == B" {from="", output=""}
    "A < B" {from=MaterialExpressionAppendVector_5, output=""}
   ]
   properties {}
  }
  MaterialExpressionIf_7 {
   class=MaterialExpressionIf
   inputs [
    A {from=MaterialExpressionConvert_34, output=""}
    B {from=MaterialExpressionConvert_34, output=""}
    "A > B" {from=MaterialExpressionAppendVector_7, output=""}
    "A == B" {from="", output=""}
    "A < B" {from=MaterialExpressionAppendVector_8, output=""}
   ]
   properties {}
  }
  MaterialExpressionIf_8 {
   class=MaterialExpressionIf
   inputs [
    A {from=MaterialExpressionAdd_6, output=""}
    B {from=MaterialExpressionSubtract_8, output=""}
    "A > B" {from=MaterialExpressionAppendVector_9, output=""}
    "A == B" {from="", output=""}
    "A < B" {from=MaterialExpressionAppendVector_10, output=""}
   ]
   properties {}
  }
  MaterialExpressionLength_1 {class=MaterialExpressionLength, inputs [None {from=MaterialExpressionTransformPosition_0, output=""}], properties {}}
  MaterialExpressionLength_2 {class=MaterialExpressionLength, inputs [None {from=MaterialExpressionTransformPosition_1, output=""}], properties {}}
  MaterialExpressionLength_3 {class=MaterialExpressionLength, inputs [None {from=MaterialExpressionTransformPosition_2, output=""}], properties {}}
  MaterialExpressionLinearInterpolate_4 {
   class=MaterialExpressionLinearInterpolate
   inputs [
    A {from=MaterialExpressionConvert_16}
    B {from=MaterialExpressionConvert_18}
    Alpha {from=MaterialExpressionConvert_15}
   ]
   properties {desc=xaa}
  }
  MaterialExpressionLinearInterpolate_5 {
   class=MaterialExpressionLinearInterpolate
   inputs [
    A {from=MaterialExpressionConvert_17}
    B {from=MaterialExpressionConvert_19}
    Alpha {from=MaterialExpressionConvert_15}
   ]
   properties {desc=yaa}
  }
  MaterialExpressionMaterialFunctionCall_0 {
   class=MaterialExpressionMaterialFunctionCall
   inputs [
    RotationAmount {from=MaterialExpressionFunctionInput_3}
    Coords {from=MaterialExpressionIf_5}
    Pivot {from=MaterialExpressionConstant2Vector_0}
   ]
   properties {materialFunction=&/MF_Rotate2D}
  }
  MaterialExpressionMaterialFunctionCall_1 {
   class=MaterialExpressionMaterialFunctionCall
   inputs [
    RotationAmount {from=MaterialExpressionFunctionInput_3}
    Coords {from=MaterialExpressionReroute_10}
    Pivot {from=MaterialExpressionConstant2Vector_0}
   ]
   properties {materialFunction=&MF_Rotate2D}
  }
  MaterialExpressionMaterialFunctionCall_4 {
   class=MaterialExpressionMaterialFunctionCall
   inputs [Multiplier {from=MaterialExpressionConstant_2}]
   properties {materialFunction=&Pi}
  }
  MaterialExpressionMaterialFunctionCall_6 {
   class=MaterialExpressionMaterialFunctionCall
   inputs [Multiplier {from=MaterialExpressionConstant_4}]
   properties {materialFunction=&Pi}
  }
  MaterialExpressionMaterialFunctionCall_9 {
   class=MaterialExpressionMaterialFunctionCall
   inputs [Multiplier {from=MaterialExpressionConstant_5}]
   properties {materialFunction=&Pi}
  }
  MaterialExpressionMax_0 {
   class=MaterialExpressionMax
   inputs [
    A {from=MaterialExpressionConvert_6}
    B {from=MaterialExpressionConvert_6}
   ]
  }
  MaterialExpressionMax_1 {
   class=MaterialExpressionMax
   inputs [
    A {from=MaterialExpressionConvert_1}
    B {from=MaterialExpressionConvert_1}
   ]
  }
  MaterialExpressionMax_2 {
   class=MaterialExpressionMax
   inputs [
    A {from=MaterialExpressionReroute_11}
    B {from=MaterialExpressionReroute_12}
   ]
  }
  MaterialExpressionMultiply_17 {
   class=MaterialExpressionMultiply
   inputs [
    A {from=MaterialExpressionFunctionInput_4}
    B {from=MaterialExpressionMaterialFunctionCall_4, output=Result}
   ]
   properties {desc=theta}
  }
  MaterialExpressionMultiply_18 {
   class=MaterialExpressionMultiply
   inputs [
    A {from=MaterialExpressionCosine_2}
    B {from=""}
   ]
   properties {constB=-1}
  }
  MaterialExpressionMultiply_22 {
   class=MaterialExpressionMultiply
   inputs [
    A {from=MaterialExpressionReroute_38}
    B {from=MaterialExpressionMultiply_23}
   ]
  }
  MaterialExpressionMultiply_23 {
   class=MaterialExpressionMultiply
   inputs [
    A {from=MaterialExpressionDotProduct_18}
    B {from=""}
   ]
   properties {constB=-1}
  }
  MaterialExpressionMultiply_24 {
   class=MaterialExpressionMultiply
   inputs [
    A {from=MaterialExpressionIf_4}
    B {from=""}
   ]
   properties {constB=-1}
  }
  MaterialExpressionMultiply_25 {
   class=MaterialExpressionMultiply
   inputs [
    A {from=MaterialExpressionDotProduct_20}
    B {from=""}
   ]
   properties {constB=-1}
  }
  MaterialExpressionMultiply_26 {
   class=MaterialExpressionMultiply
   inputs [
    A {from=MaterialExpressionDotProduct_22}
    B {from=""}
   ]
   properties {constB=-1}
  }
  MaterialExpressionMultiply_27 {
   class=MaterialExpressionMultiply
   inputs [
    A {from=MaterialExpressionCosine_4}
    B {from=MaterialExpressionConvert_26}
   ]
  }
  MaterialExpressionMultiply_28 {
   class=MaterialExpressionMultiply
   inputs [
    A {from=MaterialExpressionSine_4}
    B {from=MaterialExpressionConvert_26}
   ]
  }
  MaterialExpressionMultiply_29 {
   class=MaterialExpressionMultiply
   inputs [
    A {from=MaterialExpressionCosine_4}
    B {from=MaterialExpressionConvert_26}
   ]
  }
  MaterialExpressionMultiply_30 {
   class=MaterialExpressionMultiply
   inputs [
    A {from=MaterialExpressionSine_4}
    B {from=MaterialExpressionConvert_26}
   ]
  }
  MaterialExpressionMultiply_31 {
   class=MaterialExpressionMultiply
   inputs [
    A {from=MaterialExpressionCosine_4}
    B {from=MaterialExpressionConvert_28}
   ]
  }
  MaterialExpressionMultiply_32 {
   class=MaterialExpressionMultiply
   inputs [
    A {from=MaterialExpressionSine_4}
    B {from=MaterialExpressionConvert_28}
   ]
  }
  MaterialExpressionMultiply_33 {
   class=MaterialExpressionMultiply
   inputs [
    A {from=MaterialExpressionCosine_4}
    B {from=MaterialExpressionConvert_28}
   ]
  }
  MaterialExpressionMultiply_34 {
   class=MaterialExpressionMultiply
   inputs [
    A {from=MaterialExpressionSine_4}
    B {from=MaterialExpressionConvert_28}
   ]
  }
  MaterialExpressionMultiply_38 {
   class=MaterialExpressionMultiply
   inputs [
    A {from=MaterialExpressionFunctionInput_5}
    B {from=MaterialExpressionFunctionInput_0}
   ]
  }
  MaterialExpressionMultiply_39 {
   class=MaterialExpressionMultiply
   inputs [
    A {from=MaterialExpressionConvert_33}
    B {from=MaterialExpressionFunctionInput_5}
   ]
  }
  MaterialExpressionMultiply_4 {
   class=MaterialExpressionMultiply
   inputs [
    A {from=MaterialExpressionIf_0}
    B {from=""}
   ]
   properties {constB=-1}
  }
  MaterialExpressionMultiply_40 {
   class=MaterialExpressionMultiply
   inputs [
    A {from=MaterialExpressionDotProduct_24}
    B {from=""}
   ]
   properties {constB=-1}
  }
  MaterialExpressionMultiply_41 {
   class=MaterialExpressionMultiply
   inputs [
    A {from=MaterialExpressionDotProduct_26}
    B {from=""}
   ]
   properties {constB=-1}
  }
  MaterialExpressionMultiply_42 {
   class=MaterialExpressionMultiply
   inputs [
    A {from=MaterialExpressionCosine_5}
    B {from=MaterialExpressionConvert_37}
   ]
  }
  MaterialExpressionMultiply_43 {
   class=MaterialExpressionMultiply
   inputs [
    A {from=MaterialExpressionSine_5}
    B {from=MaterialExpressionConvert_37}
   ]
  }
  MaterialExpressionMultiply_44 {
   class=MaterialExpressionMultiply
   inputs [
    A {from=MaterialExpressionCosine_5}
    B {from=MaterialExpressionConvert_37}
   ]
  }
  MaterialExpressionMultiply_45 {
   class=MaterialExpressionMultiply
   inputs [
    A {from=MaterialExpressionSine_5}
    B {from=MaterialExpressionConvert_37}
   ]
  }
  MaterialExpressionMultiply_46 {
   class=MaterialExpressionMultiply
   inputs [
    A {from=MaterialExpressionCosine_5}
    B {from=MaterialExpressionConvert_39}
   ]
  }
  MaterialExpressionMultiply_47 {
   class=MaterialExpressionMultiply
   inputs [
    A {from=MaterialExpressionSine_5}
    B {from=MaterialExpressionConvert_39}
   ]
  }
  MaterialExpressionMultiply_48 {
   class=MaterialExpressionMultiply
   inputs [
    A {from=MaterialExpressionCosine_5}
    B {from=MaterialExpressionConvert_39}
   ]
  }
  MaterialExpressionMultiply_49 {
   class=MaterialExpressionMultiply
   inputs [
    A {from=MaterialExpressionSine_5}
    B {from=MaterialExpressionConvert_39}
   ]
  }
  MaterialExpressionMultiply_5 {
   class=MaterialExpressionMultiply
   inputs [
    A {from=MaterialExpressionDotProduct_8}
    B {from=""}
   ]
   properties {constB=-1}
  }
  MaterialExpressionMultiply_50 {
   class=MaterialExpressionMultiply
   inputs [
    A {from=MaterialExpressionDotProduct_28}
    B {from=""}
   ]
   properties {constB=-1}
  }
  MaterialExpressionMultiply_51 {
   class=MaterialExpressionMultiply
   inputs [
    A {from=MaterialExpressionDotProduct_30}
    B {from=""}
   ]
   properties {constB=-1}
  }
  MaterialExpressionMultiply_6 {
   class=MaterialExpressionMultiply
   inputs [
    A {from=MaterialExpressionReroute_1}
    B {from=MaterialExpressionMultiply_5}
   ]
  }
  MaterialExpressionMultiply_7 {
   class=MaterialExpressionMultiply
   inputs [
    A {from=MaterialExpressionDotProduct_10}
    B {from=""}
   ]
   properties {constB=-1}
  }
  MaterialExpressionMultiply_8 {
   class=MaterialExpressionMultiply
   inputs [
    A {from=MaterialExpressionDotProduct_12}
    B {from=""}
   ]
   properties {constB=-1}
  }
  MaterialExpressionMultiply_9 {
   class=MaterialExpressionMultiply
   inputs [
    A {from=MaterialExpressionMultiply_38}
    B {from=""}
   ]
   properties {constB=-1}
  }
  MaterialExpressionNamedRerouteDeclaration_0 {
   class=MaterialExpressionNamedRerouteDeclaration
   inputs [None {from=MaterialExpressionMultiply_9}]
   properties {name=iPos, nodeColor=<rgba 1 0 0.1406 1>}
  }
  MaterialExpressionNamedRerouteDeclaration_2 {
   class=MaterialExpressionNamedRerouteDeclaration
   inputs [None {from=MaterialExpressionNormalize_8}]
   properties {name=xab, nodeColor=<rgba 0.7812 1 0 1>}
  }
  MaterialExpressionNamedRerouteDeclaration_3 {
   class=MaterialExpressionNamedRerouteDeclaration
   inputs [None {from=MaterialExpressionNormalize_9}]
   properties {name=yab, nodeColor=<rgba 0.8984 0 1 1>}
  }
  MaterialExpressionNamedRerouteDeclaration_4 {
   class=MaterialExpressionNamedRerouteDeclaration
   inputs [None {from=MaterialExpressionCrossProduct_12}]
   properties {name=yac, nodeColor=<rgba 0 1 0.5781 1>}
  }
  MaterialExpressionNamedRerouteDeclaration_5 {
   class=MaterialExpressionNamedRerouteDeclaration
   inputs [None {from=MaterialExpressionCrossProduct_10}]
   properties {name=xac, nodeColor=<rgba 1 0.2578 0 1>}
  }
  MaterialExpressionNamedRerouteDeclaration_6 {
   class=MaterialExpressionNamedRerouteDeclaration
   inputs [None {from=MaterialExpressionFunctionInput_1}]
   properties {name=iNorm, nodeColor=<rgba 0 0.0625 1 1>}
  }
  MaterialExpressionNamedRerouteUsage_0 {class=MaterialExpressionNamedRerouteUsage}
  MaterialExpressionNamedRerouteUsage_1 {class=MaterialExpressionNamedRerouteUsage}
  MaterialExpressionNamedRerouteUsage_10 {class=MaterialExpressionNamedRerouteUsage}
  MaterialExpressionNamedRerouteUsage_11 {class=MaterialExpressionNamedRerouteUsage}
  MaterialExpressionNamedRerouteUsage_12 {class=MaterialExpressionNamedRerouteUsage}
  MaterialExpressionNamedRerouteUsage_13 {class=MaterialExpressionNamedRerouteUsage}
  MaterialExpressionNamedRerouteUsage_14 {class=MaterialExpressionNamedRerouteUsage}
  MaterialExpressionNamedRerouteUsage_15 {class=MaterialExpressionNamedRerouteUsage}
  MaterialExpressionNamedRerouteUsage_16 {class=MaterialExpressionNamedRerouteUsage}
  MaterialExpressionNamedRerouteUsage_17 {class=MaterialExpressionNamedRerouteUsage}
  MaterialExpressionNamedRerouteUsage_18 {class=MaterialExpressionNamedRerouteUsage}
  MaterialExpressionNamedRerouteUsage_19 {class=MaterialExpressionNamedRerouteUsage}
  MaterialExpressionNamedRerouteUsage_2 {class=MaterialExpressionNamedRerouteUsage}
  MaterialExpressionNamedRerouteUsage_20 {class=MaterialExpressionNamedRerouteUsage}
  MaterialExpressionNamedRerouteUsage_21 {class=MaterialExpressionNamedRerouteUsage}
  MaterialExpressionNamedRerouteUsage_22 {class=MaterialExpressionNamedRerouteUsage}
  MaterialExpressionNamedRerouteUsage_23 {class=MaterialExpressionNamedRerouteUsage}
  MaterialExpressionNamedRerouteUsage_3 {class=MaterialExpressionNamedRerouteUsage}
  MaterialExpressionNamedRerouteUsage_4 {class=MaterialExpressionNamedRerouteUsage}
  MaterialExpressionNamedRerouteUsage_5 {class=MaterialExpressionNamedRerouteUsage}
  MaterialExpressionNamedRerouteUsage_6 {class=MaterialExpressionNamedRerouteUsage}
  MaterialExpressionNamedRerouteUsage_7 {class=MaterialExpressionNamedRerouteUsage}
  MaterialExpressionNamedRerouteUsage_8 {class=MaterialExpressionNamedRerouteUsage}
  MaterialExpressionNamedRerouteUsage_9 {class=MaterialExpressionNamedRerouteUsage}
  MaterialExpressionNormalize_0 {class=MaterialExpressionNormalize, inputs [VectorInput {from=MaterialExpressionCrossProduct_0, output=""}]}
  MaterialExpressionNormalize_12 {class=MaterialExpressionNormalize, inputs [VectorInput {from=MaterialExpressionAdd_2, output=""}]}
  MaterialExpressionNormalize_2 {class=MaterialExpressionNormalize, inputs [VectorInput {from=MaterialExpressionAdd_0, output=""}]}
  MaterialExpressionNormalize_3 {class=MaterialExpressionNormalize, inputs [VectorInput {from=MaterialExpressionCrossProduct_4, output=""}]}
  MaterialExpressionNormalize_8 {class=MaterialExpressionNormalize, inputs [VectorInput {from=MaterialExpressionCrossProduct_11, output=""}], properties {desc=xab}}
  MaterialExpressionNormalize_9 {class=MaterialExpressionNormalize, inputs [VectorInput {from=MaterialExpressionCrossProduct_13, output=""}], properties {desc=yab}}
  MaterialExpressionReroute_0 {class=MaterialExpressionReroute, inputs [None {from=MaterialExpressionNormalize_8}]}
  MaterialExpressionReroute_1 {class=MaterialExpressionReroute, inputs [None {from=MaterialExpressionReroute_6}]}
  MaterialExpressionReroute_10 {class=MaterialExpressionReroute, inputs [None {from=MaterialExpressionIf_1}]}
  MaterialExpressionReroute_11 {class=MaterialExpressionReroute, inputs [None {from=MaterialExpressionMax_1}]}
  MaterialExpressionReroute_12 {class=MaterialExpressionReroute, inputs [None {from=MaterialExpressionMax_0}]}
  MaterialExpressionReroute_2 {class=MaterialExpressionReroute, inputs [None {from=MaterialExpressionDDX_1}]}
  MaterialExpressionReroute_3 {class=MaterialExpressionReroute, inputs [None {from=MaterialExpressionDDY_1}]}
  MaterialExpressionReroute_32 {class=MaterialExpressionReroute, inputs [None {from=MaterialExpressionNormalize_9}]}
  MaterialExpressionReroute_38 {class=MaterialExpressionReroute, inputs [None {from=MaterialExpressionReroute_39}]}
  MaterialExpressionReroute_39 {class=MaterialExpressionReroute, inputs [None {from=MaterialExpressionConvert_25}]}
  MaterialExpressionReroute_4 {class=MaterialExpressionReroute, inputs [None {from=MaterialExpressionCrossProduct_12}]}
  MaterialExpressionReroute_40 {class=MaterialExpressionReroute, inputs [None {from=MaterialExpressionConvert_27}]}
  MaterialExpressionReroute_46 {class=MaterialExpressionReroute, inputs [None {from=MaterialExpressionConvert_26}]}
  MaterialExpressionReroute_47 {class=MaterialExpressionReroute, inputs [None {from=MaterialExpressionReroute_46}]}
  MaterialExpressionReroute_48 {class=MaterialExpressionReroute, inputs [None {from=MaterialExpressionReroute_49}]}
  MaterialExpressionReroute_49 {class=MaterialExpressionReroute, inputs [None {from=MaterialExpressionConvert_28}]}
  MaterialExpressionReroute_5 {class=MaterialExpressionReroute, inputs [None {from=MaterialExpressionNamedRerouteUsage_3}]}
  MaterialExpressionReroute_54 {class=MaterialExpressionReroute, inputs [None {from=MaterialExpressionFunctionInput_1}]}
  MaterialExpressionReroute_55 {class=MaterialExpressionReroute, inputs [None {from=MaterialExpressionMultiply_39}]}
  MaterialExpressionReroute_6 {class=MaterialExpressionReroute, inputs [None {from=MaterialExpressionNamedRerouteUsage_20}]}
  MaterialExpressionReroute_63 {class=MaterialExpressionReroute, inputs [None {from=MaterialExpressionNamedRerouteUsage_18}]}
  MaterialExpressionReroute_67 {class=MaterialExpressionReroute, inputs [None {from=MaterialExpressionReroute_68}]}
  MaterialExpressionReroute_68 {class=MaterialExpressionReroute, inputs [None {from=MaterialExpressionConvert_37}]}
  MaterialExpressionReroute_69 {class=MaterialExpressionReroute, inputs [None {from=MaterialExpressionConvert_36}]}
  MaterialExpressionReroute_7 {class=MaterialExpressionReroute, inputs [None {from=MaterialExpressionCrossProduct_10}]}
  MaterialExpressionReroute_8 {class=MaterialExpressionReroute, inputs [None {from=MaterialExpressionConvert_35}], properties {desc=vd1}}
  MaterialExpressionReroute_9 {class=MaterialExpressionReroute, inputs [None {from=MaterialExpressionConvert_40}], properties {desc=vd2}}
  MaterialExpressionSaturate_2 {class=MaterialExpressionSaturate, inputs [None {from=MaterialExpressionSign_3}]}
  MaterialExpressionSaturate_4 {class=MaterialExpressionSaturate, inputs [None {from=MaterialExpressionDivide_2}]}
  MaterialExpressionShaderStageSwitch_0 {
   class=MaterialExpressionShaderStageSwitch
   inputs [
    PixelShader {from=MaterialExpressionConvert_9, output=""}
    VertexShader {from=MaterialExpressionConstant4Vector_0, output=RGBA}
   ]
   properties {}
  }
  MaterialExpressionShaderStageSwitch_1 {
   class=MaterialExpressionShaderStageSwitch
   inputs [
    PixelShader {from=MaterialExpressionConvert_8, output=""}
    VertexShader {from=MaterialExpressionConstant4Vector_0, output=RGBA}
   ]
   properties {}
  }
  MaterialExpressionSign_3 {class=MaterialExpressionSign, inputs [None {from=MaterialExpressionNamedRerouteUsage_21, output=""}], properties {}}
  MaterialExpressionSine_2 {class=MaterialExpressionSine, inputs [None {from=MaterialExpressionMultiply_17}]}
  MaterialExpressionSine_4 {class=MaterialExpressionSine, inputs [None {from=MaterialExpressionMaterialFunctionCall_6, output=Result}], properties {period=6.2832}}
  MaterialExpressionSine_5 {class=MaterialExpressionSine, inputs [None {from=MaterialExpressionMaterialFunctionCall_9, output=Result}], properties {period=6.2832}}
  MaterialExpressionSubtract_2 {
   class=MaterialExpressionSubtract
   inputs [
    A {from=MaterialExpressionMultiply_27}
    B {from=MaterialExpressionMultiply_28}
   ]
  }
  MaterialExpressionSubtract_3 {
   class=MaterialExpressionSubtract
   inputs [
    A {from=MaterialExpressionMultiply_31}
    B {from=MaterialExpressionMultiply_32}
   ]
  }
  MaterialExpressionSubtract_4 {
   class=MaterialExpressionSubtract
   inputs [
    A {from=MaterialExpressionMax_2}
    B {from=MaterialExpressionAppendVector_6}
   ]
  }
  MaterialExpressionSubtract_6 {
   class=MaterialExpressionSubtract
   inputs [
    A {from=MaterialExpressionDivide_0}
    B {from=MaterialExpressionSubtract_4}
   ]
  }
  MaterialExpressionSubtract_7 {
   class=MaterialExpressionSubtract
   inputs [
    A {from=MaterialExpressionMultiply_44}
    B {from=MaterialExpressionMultiply_45}
   ]
  }
  MaterialExpressionSubtract_8 {
   class=MaterialExpressionSubtract
   inputs [
    A {from=MaterialExpressionMultiply_48}
    B {from=MaterialExpressionMultiply_49}
   ]
  }
  MaterialExpressionTransformPosition_0 {
   class=MaterialExpressionTransformPosition
   inputs [None {from=MaterialExpressionConstant3Vector_8, output=""}]
   properties {transformSourceType=TRANSFORMPOSSOURCE_Instance, transformType=TRANSFORMPOSSOURCE_World}
  }
  MaterialExpressionTransformPosition_1 {
   class=MaterialExpressionTransformPosition
   inputs [None {from=MaterialExpressionConstant3Vector_9, output=""}]
   properties {transformSourceType=TRANSFORMPOSSOURCE_Instance, transformType=TRANSFORMPOSSOURCE_World}
  }
  MaterialExpressionTransformPosition_2 {
   class=MaterialExpressionTransformPosition
   inputs [None {from=MaterialExpressionConstant3Vector_10, output=""}]
   properties {transformSourceType=TRANSFORMPOSSOURCE_Instance, transformType=TRANSFORMPOSSOURCE_World}
  }
  MaterialExpressionVertexNormalWS_0 {class=MaterialExpressionVertexNormalWS, inputs [], properties {}}
  MaterialExpressionWorldPosition_1 {class=MaterialExpressionWorldPosition, inputs [], properties {}}
 }
 properties {bExposeToLibrary=true, libraryCategoriesText [Base Material]}
}
