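# Material function that generates dithered octaplanar UVs: keystoned X/Y projections plus a 45°-rotated set, with no top-down Z projection.
# Position and normal are chosen in world, local, periodic or preskinned space, then fed to MF_Coordinate_Keystone for two candidate frames and a blend weight.
# Interleaved gradient noise, overridable, converts that weight to an index selecting one frame, approximating blending cheaply.
# Outputs UV, derivatives, rotation and space-corrected tangent and bitangent.
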MaterialFunction /BaseMaterial/Materials/Functions/MF_ResolveCoordinateFrame_Octaplanar_Dither
graph {
 expressions {
  MaterialExpressionAdd_10 {
   class=MaterialExpressionAdd
   inputs [
    A {from=MaterialExpressionMaterialFunctionCall_2, output=Selected Vector}
    B {from=MaterialExpressionFunctionInput_13}
   ]
  }
  MaterialExpressionAppendVector_0 {
   class=MaterialExpressionAppendVector
   inputs [
    A {from=MaterialExpressionMaterialFunctionCall_10, output=Blend}
    B {from=MaterialExpressionConstant_0, output=""}
   ]
  }
  MaterialExpressionConstant3Vector_2 {class=MaterialExpressionConstant3Vector, properties {constant=<rgba 0 0 1 0>}}
  MaterialExpressionConstant3Vector_3 {class=MaterialExpressionConstant3Vector, properties {constant=<rgba 0 1 0 0>}}
  MaterialExpressionConstant3Vector_4 {class=MaterialExpressionConstant3Vector, properties {constant=<rgba 1 0 0 0>}}
  MaterialExpressionConstant_0 {class=MaterialExpressionConstant}
  MaterialExpressionConstant_1 {class=MaterialExpressionConstant, properties {r=1}}
  MaterialExpressionConvert_13 {
   class=MaterialExpressionConvert
   inputs [
    None {from=MaterialExpressionLength_3}
    None {from=MaterialExpressionLength_2}
    None {from=MaterialExpressionLength_0}
   ]
   properties {
    desc=maintain scale
    convertInputs [
     {
      type=Scalar
      expressionInput {
       expression=&self:MaterialExpressionLength_3
       inputName=None
       mask=0
       maskA=0
       maskB=0
       maskG=0
       maskR=0
       outputIndex=0
      }
     }
     {
      type=Scalar
      expressionInput {
       expression=&self:MaterialExpressionLength_2
       inputName=None
       mask=0
       maskA=0
       maskB=0
       maskG=0
       maskR=0
       outputIndex=0
      }
     }
     {
      type=Scalar
      expressionInput {
       expression=&self:MaterialExpressionLength_0
       inputName=None
       mask=0
       maskA=0
       maskB=0
       maskG=0
       maskR=0
       outputIndex=0
      }
     }
    ]
    convertOutputs [{type=Vector3}]
   }
  }
  MaterialExpressionDivide_0 {
   class=MaterialExpressionDivide
   inputs [
    A {from=""}
    B {from=MaterialExpressionNamedRerouteUsage_3}
   ]
  }
  MaterialExpressionFunctionInput_0 {
   class=MaterialExpressionFunctionInput
   inputs [Preview {from="", output=""}]
   properties {
    bUsePreviewValueAsDefault=true
    description="Sets how many times the projected coordinates repeat along each axis. Use values above 1 for denser tiling, negative values for mirroring, and avoid zero."
    inputName=CoordinateRepeat
    sortPriority=-10
    previewValue=<0.01 0.01 0.01 1>
   }
  }
  MaterialExpressionFunctionInput_1 {
   class=MaterialExpressionFunctionInput
   inputs [Preview {from="", output=""}]
   properties {
    bUsePreviewValueAsDefault=true
    description="Rotates the projected UVs around the selected axis. Units are in turns, where 1.0 = 360° and 0.25 = 90°."
    inputName="Rotation (Turns)"
    inputType=FunctionInput_Scalar
    sortPriority=-8
   }
  }
  MaterialExpressionFunctionInput_12 {
   class=MaterialExpressionFunctionInput
   inputs [Preview {from="", output=""}]
   properties {
    bUsePreviewValueAsDefault=true
    description="Controls the keystoning amount for the X/Y projection axes. Adjusting this angle shears the projection planes, effectively changing how steeply surfaces blend into the secondary, 45°-rotated projection directions."
    inputName="ProjectionAngle (degrees)"
    inputType=FunctionInput_Scalar
    sortPriority=-6
   }
  }
  MaterialExpressionFunctionInput_13 {
   class=MaterialExpressionFunctionInput
   inputs [Preview {from="", output=""}]
   properties {
    bUsePreviewValueAsDefault=true
    description=Applies an additional per-axis offset to the coordinates before projection. Useful for correctly positioning a texture.
    inputName=CoordinateTransform
    sortPriority=-9
   }
  }
  MaterialExpressionFunctionInput_14 {
   class=MaterialExpressionFunctionInput
   inputs [Preview {from="", output=""}]
   properties {
    bUsePreviewValueAsDefault=true
    description="Defines the space used for projection: World (0) keeps textures fixed in world space, Local (1) moves with the object, Periodic (2) enforces a repeating pattern, and Preskinned (3) locks to preskinned positions for stable results on animated meshes."
    inputName=CoordinateSpace
    inputType=FunctionInput_Scalar
    sortPriority=-11
   }
  }
  MaterialExpressionFunctionInput_16 {
   class=MaterialExpressionFunctionInput
   inputs [Preview {from="", output=""}]
   properties {
    bUsePreviewValueAsDefault=true
    description=Controls the sharpness of the transition between coordinate sets.
    inputName=BlendTolerance
    inputType=FunctionInput_Scalar
    sortPriority=8
    previewValue=<0.5 0 0 1>
   }
  }
  MaterialExpressionFunctionInput_3 {
   class=MaterialExpressionFunctionInput
   inputs [Preview {from=MaterialExpressionSubtract_0, output=""}]
   properties {
    bUsePreviewValueAsDefault=true
    description="Provides the dithering sequence used for stochastic projection blending. By default, the function generates an interleaved gradient noise pattern based on the golden ratio, which produces stable, slow-converging results at low cost. Supplying a custom value lets you override this with your own dithering scheme, useful for synchronizing with other systems or experimenting with alternate noise sources."
    inputName=Dither
    inputType=FunctionInput_Scalar
   }
  }
  MaterialExpressionFunctionOutput_1 {
   class=MaterialExpressionFunctionOutput
   inputs [None {from=MaterialExpressionLinearInterpolate_0, output=""}]
   properties {
    description="The primary 2D coordinates generated by the projection. Feed these into texture samplers for surface detail.\r\n"
    outputName=UV
   }
  }
  MaterialExpressionFunctionOutput_14 {
   class=MaterialExpressionFunctionOutput
   inputs [None {from=MaterialExpressionNamedRerouteUsage_4, output=""}]
   properties {
    description="The rotation applied around the projection axis, expressed in turns (1.0 = 360°). Used along with Tangent/Bitangent for reorienting normal maps via MF_TangentNormal_OrientWorld.\r\n"
    outputName="Rotation (Turns)"
    sortPriority=-5
   }
  }
  MaterialExpressionFunctionOutput_2 {
   class=MaterialExpressionFunctionOutput
   inputs [None {from=MaterialExpressionLinearInterpolate_1, output=""}]
   properties {
    description="Packed derivatives of the projected UVs (ddx/ddy). Use with SampleGrad to ensure textures sample with correct mip levels and anisotropic filtering.\r\n"
    outputName=DXY
    sortPriority=3
   }
  }
  MaterialExpressionFunctionOutput_8 {
   class=MaterialExpressionFunctionOutput
   inputs [None {from=MaterialExpressionMaterialFunctionCall_3, output=Selected Vector}]
   properties {
    description="The world-space tangent vector of the projection frame. Use with Bitangent and Rotation to correctly reorient tangent-space normals.\r\n"
    outputName=Tangent
    sortPriority=1
   }
  }
  MaterialExpressionFunctionOutput_9 {
   class=MaterialExpressionFunctionOutput
   inputs [None {from=MaterialExpressionMaterialFunctionCall_4, output=Selected Vector}]
   properties {
    description="The world-space bitangent vector of the projection frame. Complements Tangent to form the basis for tangent-space normal reorientation.\r\n"
    outputName=Bitangent
    sortPriority=2
   }
  }
  MaterialExpressionLength_0 {class=MaterialExpressionLength, inputs [None {from=MaterialExpressionTransform_0, output=""}], properties {}}
  MaterialExpressionLength_2 {class=MaterialExpressionLength, inputs [None {from=MaterialExpressionTransform_12, output=""}], properties {}}
  MaterialExpressionLength_3 {class=MaterialExpressionLength, inputs [None {from=MaterialExpressionTransform_13, output=""}], properties {}}
  MaterialExpressionLinearInterpolate_0 {
   class=MaterialExpressionLinearInterpolate
   inputs [
    A {from=MaterialExpressionNamedRerouteUsage_82}
    B {from=MaterialExpressionNamedRerouteUsage_83}
    Alpha {from=MaterialExpressionSaturate_0}
   ]
  }
  MaterialExpressionLinearInterpolate_1 {
   class=MaterialExpressionLinearInterpolate
   inputs [
    A {from=MaterialExpressionNamedRerouteUsage_64}
    B {from=MaterialExpressionNamedRerouteUsage_90}
    Alpha {from=MaterialExpressionSaturate_1}
   ]
  }
  MaterialExpressionLinearInterpolate_2 {
   class=MaterialExpressionLinearInterpolate
   inputs [
    A {from=MaterialExpressionNamedRerouteUsage_68}
    B {from=MaterialExpressionNamedRerouteUsage_72}
    Alpha {from=MaterialExpressionSaturate_2}
   ]
  }
  MaterialExpressionLinearInterpolate_3 {
   class=MaterialExpressionLinearInterpolate
   inputs [
    A {from=MaterialExpressionNamedRerouteUsage_63}
    B {from=MaterialExpressionNamedRerouteUsage_71}
    Alpha {from=MaterialExpressionSaturate_3}
   ]
  }
  MaterialExpressionLocalPosition_0 {class=MaterialExpressionLocalPosition, inputs [], properties {includedOffsets=ExcludeOffsets, localOrigin=InstancePreSkinning}}
  MaterialExpressionLocalPosition_2 {class=MaterialExpressionLocalPosition, inputs [], properties {}}
  MaterialExpressionMaterialFunctionCall_0 {
   class=MaterialExpressionMaterialFunctionCall
   inputs [
    Selector {from=MaterialExpressionNamedRerouteUsage_41}
    VectorA {from=MaterialExpressionReroute_1}
    VectorB {from=MaterialExpressionNormalize_1}
    VectorC {from=MaterialExpressionReroute_1}
    VectorD {from=MaterialExpressionShaderStageSwitch_0}
   ]
   properties {materialFunction=&/MF_Switch4_Vec3}
  }
  MaterialExpressionMaterialFunctionCall_1 {
   class=MaterialExpressionMaterialFunctionCall
   inputs [PeriodXYZ {from=MaterialExpressionDivide_0}]
   properties {materialFunction=&/MF_PrepareLWC}
  }
  MaterialExpressionMaterialFunctionCall_10 {
   class=MaterialExpressionMaterialFunctionCall
   inputs [
    Position {from=MaterialExpressionNamedRerouteUsage_40}
    Tiling {from=MaterialExpressionNamedRerouteDeclaration_2}
    RefNorm {from=MaterialExpressionNamedRerouteUsage_36}
    "Rotation (Turns)" {from=MaterialExpressionNamedRerouteDeclaration_11}
    BlendFactor {from=MaterialExpressionFunctionInput_16}
    KeystoneAngle {from=MaterialExpressionReroute_0}
   ]
   properties {materialFunction=&/MF_Coordinate_Keystone}
  }
  MaterialExpressionMaterialFunctionCall_2 {
   class=MaterialExpressionMaterialFunctionCall
   inputs [
    Selector {from=MaterialExpressionFunctionInput_14}
    VectorA {from=MaterialExpressionReroute_6}
    VectorB {from=MaterialExpressionMultiply_0}
    VectorC {from=MaterialExpressionReroute_5}
    VectorD {from=MaterialExpressionMultiply_1}
   ]
   properties {materialFunction=&MF_Switch4_Vec3}
  }
  MaterialExpressionMaterialFunctionCall_3 {
   class=MaterialExpressionMaterialFunctionCall
   inputs [
    Selector {from=MaterialExpressionNamedRerouteUsage_42}
    VectorA {from=MaterialExpressionReroute_3}
    VectorB {from=MaterialExpressionTransform_4}
    VectorC {from=MaterialExpressionReroute_3}
    VectorD {from=MaterialExpressionTransform_6}
   ]
   properties {materialFunction=&MF_Switch4_Vec3}
  }
  MaterialExpressionMaterialFunctionCall_4 {
   class=MaterialExpressionMaterialFunctionCall
   inputs [
    Selector {from=MaterialExpressionNamedRerouteUsage_43}
    VectorA {from=MaterialExpressionReroute_4}
    VectorB {from=MaterialExpressionTransform_7}
    VectorC {from=MaterialExpressionReroute_4}
    VectorD {from=MaterialExpressionTransform_8}
   ]
   properties {materialFunction=&MF_Switch4_Vec3}
  }
  MaterialExpressionMaterialFunctionCall_7 {
   class=MaterialExpressionMaterialFunctionCall
   inputs [
    InputBlend {from=MaterialExpressionNamedRerouteUsage_48}
    TriplanarDitherPattern {from=MaterialExpressionFunctionInput_3}
   ]
   properties {materialFunction=&/MF_Blend_RGBToIndex}
  }
  MaterialExpressionMaterialFunctionCall_8 {
   class=MaterialExpressionMaterialFunctionCall
   inputs [
    TemporalSampleIndex {from=""}
    PixelPosition {from=""}
   ]
   properties {materialFunction=&/MF_Noise_InterleavedGradientGolden_1d}
  }
  MaterialExpressionMultiply_0 {
   class=MaterialExpressionMultiply
   inputs [
    A {from=MaterialExpressionNamedRerouteUsage_35}
    B {from=MaterialExpressionLocalPosition_2, output=XYZ}
   ]
  }
  MaterialExpressionMultiply_1 {
   class=MaterialExpressionMultiply
   inputs [
    A {from=MaterialExpressionNamedRerouteUsage_1}
    B {from=MaterialExpressionShaderStageSwitch_2}
   ]
  }
  MaterialExpressionNamedRerouteDeclaration_10 {
   class=MaterialExpressionNamedRerouteDeclaration
   inputs [None {from=MaterialExpressionAppendVector_0}]
   properties {name=Blendi, nodeColor=<rgba 1 1 1 1>}
  }
  MaterialExpressionNamedRerouteDeclaration_11 {
   class=MaterialExpressionNamedRerouteDeclaration
   inputs [None {from=MaterialExpressionFunctionInput_1}]
   properties {name=turns, nodeColor=<rgba 0.8963 0.8963 0.0091 1>}
  }
  MaterialExpressionNamedRerouteDeclaration_12 {
   class=MaterialExpressionNamedRerouteDeclaration
   inputs [None {from=MaterialExpressionMaterialFunctionCall_10, output=UV1_bitangent}]
   properties {name=UV1B, nodeColor=<rgba 0.9 0.009 0.9 1>}
  }
  MaterialExpressionNamedRerouteDeclaration_18 {
   class=MaterialExpressionNamedRerouteDeclaration
   inputs [None {from=MaterialExpressionMaterialFunctionCall_0, output=Selected Vector}]
   properties {name=RefNorm, nodeColor=<rgba 0.8963 0.8963 0.0091 1>}
  }
  MaterialExpressionNamedRerouteDeclaration_19 {
   class=MaterialExpressionNamedRerouteDeclaration
   inputs [None {from=MaterialExpressionAdd_10}]
   properties {name=RefPos, nodeColor=<rgba 0.8963 0.8963 0.0091 1>}
  }
  MaterialExpressionNamedRerouteDeclaration_2 {
   class=MaterialExpressionNamedRerouteDeclaration
   inputs [None {from=MaterialExpressionFunctionInput_0}]
   properties {name=Repeat, nodeColor=<rgba 0.7812 0 1 1>}
  }
  MaterialExpressionNamedRerouteDeclaration_20 {
   class=MaterialExpressionNamedRerouteDeclaration
   inputs [None {from=MaterialExpressionFunctionInput_14}]
   properties {name=Selector, nodeColor=<rgba 0 0.3437 1 1>}
  }
  MaterialExpressionNamedRerouteDeclaration_22 {
   class=MaterialExpressionNamedRerouteDeclaration
   inputs [None {from=MaterialExpressionMaterialFunctionCall_10, output=UV1}]
   properties {name=UV1i, nodeColor=<rgba 0.9 0.009 0.9 1>}
  }
  MaterialExpressionNamedRerouteDeclaration_23 {
   class=MaterialExpressionNamedRerouteDeclaration
   inputs [None {from=MaterialExpressionMaterialFunctionCall_10, output=UV2_bitangent}]
   properties {name=UV2B, nodeColor=<rgba 0.009 0.9 0.9 1>}
  }
  MaterialExpressionNamedRerouteDeclaration_24 {
   class=MaterialExpressionNamedRerouteDeclaration
   inputs [None {from=MaterialExpressionMaterialFunctionCall_10, output=UV2}]
   properties {name=UV2i, nodeColor=<rgba 0.009 0.9 0.9 1>}
  }
  MaterialExpressionNamedRerouteDeclaration_26 {
   class=MaterialExpressionNamedRerouteDeclaration
   inputs [None {from=MaterialExpressionMaterialFunctionCall_10, output=UV2_tangent}]
   properties {name=UV2T, nodeColor=<rgba 0.009 0.9 0.9 1>}
  }
  MaterialExpressionNamedRerouteDeclaration_27 {
   class=MaterialExpressionNamedRerouteDeclaration
   inputs [None {from=MaterialExpressionMaterialFunctionCall_10, output=UV1_tangent}]
   properties {name=UV1T, nodeColor=<rgba 0.9 0.009 0.9 1>}
  }
  MaterialExpressionNamedRerouteDeclaration_32 {
   class=MaterialExpressionNamedRerouteDeclaration
   inputs [None {from=MaterialExpressionConstant_1}]
   properties {name=MeshVecScale, nodeColor=<rgba 0.6641 0 1 1>}
  }
  MaterialExpressionNamedRerouteDeclaration_7 {
   class=MaterialExpressionNamedRerouteDeclaration
   inputs [None {from=MaterialExpressionMaterialFunctionCall_10, output=UV1_derivs}]
   properties {name=DXY1i, nodeColor=<rgba 0.9 0.009 0.9 1>}
  }
  MaterialExpressionNamedRerouteDeclaration_8 {
   class=MaterialExpressionNamedRerouteDeclaration
   inputs [None {from=MaterialExpressionMaterialFunctionCall_10, output=UV2_derivs}]
   properties {name=DXY2i, nodeColor=<rgba 0.009 0.9 0.9 1>}
  }
  MaterialExpressionNamedRerouteDeclaration_9 {
   class=MaterialExpressionNamedRerouteDeclaration
   inputs [None {from=MaterialExpressionMaterialFunctionCall_7, output=AxisIndex}]
   properties {name=AxisIndex, nodeColor=<rgba 0.5469 0 1 1>}
  }
  MaterialExpressionNamedRerouteUsage_1 {class=MaterialExpressionNamedRerouteUsage}
  MaterialExpressionNamedRerouteUsage_3 {class=MaterialExpressionNamedRerouteUsage}
  MaterialExpressionNamedRerouteUsage_35 {class=MaterialExpressionNamedRerouteUsage}
  MaterialExpressionNamedRerouteUsage_36 {class=MaterialExpressionNamedRerouteUsage}
  MaterialExpressionNamedRerouteUsage_4 {class=MaterialExpressionNamedRerouteUsage}
  MaterialExpressionNamedRerouteUsage_40 {class=MaterialExpressionNamedRerouteUsage}
  MaterialExpressionNamedRerouteUsage_41 {class=MaterialExpressionNamedRerouteUsage}
  MaterialExpressionNamedRerouteUsage_42 {class=MaterialExpressionNamedRerouteUsage}
  MaterialExpressionNamedRerouteUsage_43 {class=MaterialExpressionNamedRerouteUsage}
  MaterialExpressionNamedRerouteUsage_48 {class=MaterialExpressionNamedRerouteUsage}
  MaterialExpressionNamedRerouteUsage_63 {class=MaterialExpressionNamedRerouteUsage}
  MaterialExpressionNamedRerouteUsage_64 {class=MaterialExpressionNamedRerouteUsage}
  MaterialExpressionNamedRerouteUsage_68 {class=MaterialExpressionNamedRerouteUsage}
  MaterialExpressionNamedRerouteUsage_71 {class=MaterialExpressionNamedRerouteUsage}
  MaterialExpressionNamedRerouteUsage_72 {class=MaterialExpressionNamedRerouteUsage}
  MaterialExpressionNamedRerouteUsage_79 {class=MaterialExpressionNamedRerouteUsage}
  MaterialExpressionNamedRerouteUsage_80 {class=MaterialExpressionNamedRerouteUsage}
  MaterialExpressionNamedRerouteUsage_81 {class=MaterialExpressionNamedRerouteUsage}
  MaterialExpressionNamedRerouteUsage_82 {class=MaterialExpressionNamedRerouteUsage}
  MaterialExpressionNamedRerouteUsage_83 {class=MaterialExpressionNamedRerouteUsage}
  MaterialExpressionNamedRerouteUsage_84 {class=MaterialExpressionNamedRerouteUsage}
  MaterialExpressionNamedRerouteUsage_90 {class=MaterialExpressionNamedRerouteUsage}
  MaterialExpressionNormalize_1 {class=MaterialExpressionNormalize, inputs [VectorInput {from=MaterialExpressionTransform_14, output=""}]}
  MaterialExpressionPreSkinnedNormal_0 {class=MaterialExpressionPreSkinnedNormal, inputs [], properties {}}
  MaterialExpressionReroute_0 {class=MaterialExpressionReroute, inputs [None {from=MaterialExpressionFunctionInput_12}]}
  MaterialExpressionReroute_1 {class=MaterialExpressionReroute, inputs [None {from=MaterialExpressionVertexNormalWS_1}]}
  MaterialExpressionReroute_3 {class=MaterialExpressionReroute, inputs [None {from=MaterialExpressionLinearInterpolate_3}]}
  MaterialExpressionReroute_4 {class=MaterialExpressionReroute, inputs [None {from=MaterialExpressionLinearInterpolate_2}]}
  MaterialExpressionReroute_5 {class=MaterialExpressionReroute, inputs [None {from=MaterialExpressionReroute_8}]}
  MaterialExpressionReroute_6 {class=MaterialExpressionReroute, inputs [None {from=MaterialExpressionMaterialFunctionCall_1, output=Per-Axis Period}]}
  MaterialExpressionReroute_8 {class=MaterialExpressionReroute, inputs [None {from=MaterialExpressionMaterialFunctionCall_1, output=Max Axis Period}]}
  MaterialExpressionSaturate_0 {class=MaterialExpressionSaturate, inputs [None {from=MaterialExpressionNamedRerouteUsage_81}]}
  MaterialExpressionSaturate_1 {class=MaterialExpressionSaturate, inputs [None {from=MaterialExpressionNamedRerouteUsage_84}]}
  MaterialExpressionSaturate_2 {class=MaterialExpressionSaturate, inputs [None {from=MaterialExpressionNamedRerouteUsage_80}]}
  MaterialExpressionSaturate_3 {class=MaterialExpressionSaturate, inputs [None {from=MaterialExpressionNamedRerouteUsage_79}]}
  MaterialExpressionShaderStageSwitch_0 {
   class=MaterialExpressionShaderStageSwitch
   inputs [
    PixelShader {from=MaterialExpressionVertexInterpolator_0, output=PS}
    VertexShader {from=MaterialExpressionPreSkinnedNormal_0, output=""}
   ]
   properties {}
  }
  MaterialExpressionShaderStageSwitch_2 {
   class=MaterialExpressionShaderStageSwitch
   inputs [
    PixelShader {from=MaterialExpressionVertexInterpolator_2, output=PS}
    VertexShader {from=MaterialExpressionLocalPosition_0, output=XYZ}
   ]
   properties {}
  }
  MaterialExpressionSubtract_0 {
   class=MaterialExpressionSubtract
   inputs [
    A {from=MaterialExpressionMaterialFunctionCall_8, output=Unsigned}
    B {from=""}
   ]
   properties {constB=0.5}
  }
  MaterialExpressionTransform_0 {
   class=MaterialExpressionTransform
   inputs [None {from=MaterialExpressionConstant3Vector_2}]
   properties {transformSourceType=TRANSFORMSOURCE_Instance}
  }
  MaterialExpressionTransform_12 {
   class=MaterialExpressionTransform
   inputs [None {from=MaterialExpressionConstant3Vector_3}]
   properties {transformSourceType=TRANSFORMSOURCE_Instance}
  }
  MaterialExpressionTransform_13 {
   class=MaterialExpressionTransform
   inputs [None {from=MaterialExpressionConstant3Vector_4}]
   properties {transformSourceType=TRANSFORMSOURCE_Instance}
  }
  MaterialExpressionTransform_14 {
   class=MaterialExpressionTransform
   inputs [None {from=MaterialExpressionVertexNormalWS_1}]
   properties {transformSourceType=TRANSFORMSOURCE_World, transformType=TRANSFORM_Instance}
  }
  MaterialExpressionTransform_4 {
   class=MaterialExpressionTransform
   inputs [None {from=MaterialExpressionLinearInterpolate_3}]
   properties {transformSourceType=TRANSFORMSOURCE_Instance}
  }
  MaterialExpressionTransform_6 {
   class=MaterialExpressionTransform
   inputs [None {from=MaterialExpressionLinearInterpolate_3}]
   properties {transformSourceType=TRANSFORMSOURCE_Local}
  }
  MaterialExpressionTransform_7 {
   class=MaterialExpressionTransform
   inputs [None {from=MaterialExpressionLinearInterpolate_2}]
   properties {transformSourceType=TRANSFORMSOURCE_Instance}
  }
  MaterialExpressionTransform_8 {
   class=MaterialExpressionTransform
   inputs [None {from=MaterialExpressionLinearInterpolate_2}]
   properties {transformSourceType=TRANSFORMSOURCE_Local}
  }
  MaterialExpressionVertexInterpolator_0 {class=MaterialExpressionVertexInterpolator, inputs [VS {from=MaterialExpressionPreSkinnedNormal_0, output=""}], properties {}}
  MaterialExpressionVertexInterpolator_2 {class=MaterialExpressionVertexInterpolator, inputs [VS {from=MaterialExpressionLocalPosition_0, output=XYZ}], properties {}}
  MaterialExpressionVertexNormalWS_1 {class=MaterialExpressionVertexNormalWS, inputs [], properties {}}
 }
 properties {
  bExposeToLibrary=true
  description="Generates UVs from an octaplanar projection, which extends biplanar mapping by supporting keystoning of the X/Y projection axes and adding an additional set of projections rotated 45° from the first. The top-down (Z) projection axis is not included, so this function is often paired with an Axis Projection for horizontal surfaces. Uses dithered sampling to approximate blending at very low cost (single sample per axis pair). Fastest variant, but dithering can introduce visible artifacts and does not support post-sample modulation."
  userExposedCaption="CoordinateBasis: Octaplanar (Dithered)"
  libraryCategoriesText [Base Material]
 }
}
