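# Material function that computes spherical UVs, tangent, bitangent and packed UV derivatives from instance-space position, with a Scale input for ellipsoid shapes.
# The normalized, scale-divided position drives atan2/acos mapping and Z-axis cross products for the tangent frame.
# Analytical UV gradients are dotted with screen-space DDX/DDY and divided by object radius, while the vertex shader receives a constant to force the highest mip.
# An unconnected Custom node holds reference HLSL.
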MaterialFunction /BaseMaterial/Materials/Functions/MF_Coordinate_AnalyticalSphere
graph {
 expressions {
  MaterialExpressionAdd_1 {
   class=MaterialExpressionAdd
   inputs [
    A {from=MaterialExpressionMultiply_3}
    B {from=MaterialExpressionMultiply_4}
   ]
  }
  MaterialExpressionAdd_2 {
   class=MaterialExpressionAdd
   inputs [
    A {from=MaterialExpressionMaterialFunctionCall_3, output=Result}
    B {from=MaterialExpressionConstant_3}
   ]
   properties {constB=0, desc=denom}
  }
  MaterialExpressionAdd_3 {
   class=MaterialExpressionAdd
   inputs [
    A {from=MaterialExpressionOneMinus_0}
    B {from=MaterialExpressionConstant_4}
   ]
   properties {constB=0}
  }
  MaterialExpressionArccosineFast_0 {class=MaterialExpressionArccosineFast, inputs [None {from=MaterialExpressionConvert_2, output=""}], properties {}}
  MaterialExpressionArctangent2Fast_0 {
   class=MaterialExpressionArctangent2Fast
   inputs [
    Y {from=MaterialExpressionConvert_2, output=""}
    X {from=MaterialExpressionConvert_2, output=""}
   ]
   properties {}
  }
  MaterialExpressionConstant3Vector_0 {class=MaterialExpressionConstant3Vector, properties {constant=<rgba 0 0 1 0>}}
  MaterialExpressionConstant4Vector_0 {class=MaterialExpressionConstant4Vector, inputs [], properties {desc=force highest mip in VS}}
  MaterialExpressionConstant_1 {class=MaterialExpressionConstant, properties {r=2}}
  MaterialExpressionConstant_3 {class=MaterialExpressionConstant, properties {r=0}}
  MaterialExpressionConstant_4 {class=MaterialExpressionConstant, properties {r=0}}
  MaterialExpressionConvert_0 {
   class=MaterialExpressionConvert
   inputs [None {from=MaterialExpressionNormalize_0}]
   properties {
    nodeName=Break Float3
    convertInputs [{
      type=Vector3
      expressionInput {
       expression=&self:MaterialExpressionNormalize_0
       inputName=None
       mask=0
       maskA=0
       maskB=0
       maskG=0
       maskR=0
       outputIndex=0
      }
     }]
    convertOutputs [{type=Scalar}, {type=Scalar}, {type=Scalar}]
   }
  }
  MaterialExpressionConvert_1 {
   class=MaterialExpressionConvert
   inputs [
    None {from=MaterialExpressionDivide_0}
    None {from=MaterialExpressionDivide_2}
   ]
   properties {
    desc=UV
    nodeName=Make Vector2
    convertInputs [
     {
      type=Scalar
      expressionInput {
       expression=&self:MaterialExpressionDivide_0
       inputName=None
       mask=0
       maskA=0
       maskB=0
       maskG=0
       maskR=0
       outputIndex=0
      }
     }
     {
      type=Scalar
      expressionInput {
       expression=&self:MaterialExpressionDivide_2
       inputName=None
       mask=0
       maskA=0
       maskB=0
       maskG=0
       maskR=0
       outputIndex=0
      }
     }
    ]
    convertOutputs [{type=Vector2}]
   }
  }
  MaterialExpressionConvert_2 {
   class=MaterialExpressionConvert
   inputs [None {from=MaterialExpressionNormalize_0}]
   properties {
    convertInputs [{
      type=Vector3
      expressionInput {
       expression=&self:MaterialExpressionNormalize_0
       inputName=None
       mask=0
       maskA=0
       maskB=0
       maskG=0
       maskR=0
       outputIndex=0
      }
     }]
    convertOutputs [{type=Scalar}, {type=Scalar}, {type=Scalar}]
   }
  }
  MaterialExpressionConvert_4 {
   class=MaterialExpressionConvert
   inputs [
    None {from=MaterialExpressionDivide_3}
    None {from=MaterialExpressionDivide_4}
    None {from=""}
   ]
   properties {
    desc=dudp
    nodeName=Make Vector3
    convertInputs [
     {
      type=Scalar
      expressionInput {
       expression=&self:MaterialExpressionDivide_3
       inputName=None
       mask=0
       maskA=0
       maskB=0
       maskG=0
       maskR=0
       outputIndex=0
      }
     }
     {
      type=Scalar
      expressionInput {
       expression=&self:MaterialExpressionDivide_4
       inputName=None
       mask=0
       maskA=0
       maskB=0
       maskG=0
       maskR=0
       outputIndex=0
      }
     }
     {type=Scalar, expressionInput {expression=null, inputName=None, mask=0, maskA=0, maskB=0, maskG=0, maskR=0, outputIndex=0}}
    ]
    convertOutputs [{type=Vector3}]
   }
  }
  MaterialExpressionConvert_5 {
   class=MaterialExpressionConvert
   inputs [
    None {from=""}
    None {from=""}
    None {from=MaterialExpressionDivide_5}
   ]
   properties {
    desc=dvdp
    nodeName=Make Vector3
    convertInputs [
     {type=Scalar, expressionInput {expression=null, inputName=None, mask=0, maskA=0, maskB=0, maskG=0, maskR=0, outputIndex=0}}
     {type=Scalar, expressionInput {expression=null, inputName=None, mask=0, maskA=0, maskB=0, maskG=0, maskR=0, outputIndex=0}}
     {
      type=Scalar
      expressionInput {
       expression=&self:MaterialExpressionDivide_5
       inputName=None
       mask=0
       maskA=0
       maskB=0
       maskG=0
       maskR=0
       outputIndex=0
      }
     }
    ]
    convertOutputs [{type=Vector3}]
   }
  }
  MaterialExpressionConvert_6 {
   class=MaterialExpressionConvert
   inputs [
    None {from=MaterialExpressionDotProduct_0}
    None {from=MaterialExpressionDotProduct_1}
    None {from=MaterialExpressionDotProduct_2}
    None {from=MaterialExpressionDotProduct_3}
   ]
   properties {
    nodeName=Make Vector4
    convertInputs [
     {
      type=Scalar
      expressionInput {
       expression=&self:MaterialExpressionDotProduct_0
       inputName=None
       mask=0
       maskA=0
       maskB=0
       maskG=0
       maskR=0
       outputIndex=0
      }
     }
     {
      type=Scalar
      expressionInput {
       expression=&self:MaterialExpressionDotProduct_1
       inputName=None
       mask=0
       maskA=0
       maskB=0
       maskG=0
       maskR=0
       outputIndex=0
      }
     }
     {
      type=Scalar
      expressionInput {
       expression=&self:MaterialExpressionDotProduct_2
       inputName=None
       mask=0
       maskA=0
       maskB=0
       maskG=0
       maskR=0
       outputIndex=0
      }
     }
     {
      type=Scalar
      expressionInput {
       expression=&self:MaterialExpressionDotProduct_3
       inputName=None
       mask=0
       maskA=0
       maskB=0
       maskG=0
       maskR=0
       outputIndex=0
      }
     }
    ]
    convertOutputs [{type=Vector4, defaultValue=<rgba 0 0 0 1>}]
   }
  }
  MaterialExpressionCrossProduct_0 {
   class=MaterialExpressionCrossProduct
   inputs [
    A {from=MaterialExpressionConstant3Vector_0, output=""}
    B {from=MaterialExpressionReroute_0, output=""}
   ]
   properties {}
  }
  MaterialExpressionCrossProduct_1 {
   class=MaterialExpressionCrossProduct
   inputs [
    A {from=MaterialExpressionCrossProduct_0, output=""}
    B {from=MaterialExpressionReroute_0, output=""}
   ]
   properties {}
  }
  MaterialExpressionCustom_0 {
   class=MaterialExpressionCustom
   inputs [
    p {from="", output=""}
    n {from="", output=""}
    size {from="", output=""}
    scale {from="", output=""}
   ]
   properties {
    code="// HLSL for ref\r\nfloat3 dpdx = ddx(p);\r\nfloat3 dpdy = ddy(p);\r\n\r\nfloat3 invScale = rcp(scale);\r\n\r\np*= invScale;\r\nn = normalize(n * invScale);\r\n\r\n//Spherical UV Mapping\r\nfloat2 uv = float2(atan2(n.y, n.x)/(2.f * PI), acos(n.z)/PI);\r\n\r\n//analytical derivatives\r\nfloat denom =  2.f * PI * (n.x*n.x + n.y*n.y + 1e-6);\r\nfloat3 dudp = float3( -n.y/denom, n.x/denom, 0.f);\r\nfloat3 dvdp = float3(  0., 0., -1./(PI * sqrt(1. - n.z*n.z + 1e-6)) );\r\n\r\nfloat3 tangent = cross(float3(0,0,1), n);\r\nfloat3 bitangent = normalize( cross(tangent, n) );\r\n\r\n//transform derivatives to screen space\r\nfloat4 uv_derivs = float4( dot(dudp,dpdx), dot(dvdp, dpdx),\r\n                           dot(dudp,dpdy), dot(dvdp, dpdy) );\r\n\r\n//scale derivatives\r\nfloat2 derivSize = (MaterialFloat2)View.MaterialTextureDerivativeMultiply;\r\nuv_derivs *= rcp(size);\r\nuv_derivs *= derivSize.xyxy;\r\n\r\n// Outputs\r\nout_tangent = tangent;\r\nout_bitangent = bitangent;\r\nout_derivs = uv_derivs;\r\n\r\nreturn uv;"
    description=MakeEllipsoid
    outputType=CMOT_Float2
    showCode=true
    additionalOutputs [
     {outputName=out_tangent, outputType=CMOT_Float3}
     {outputName=out_bitangent, outputType=CMOT_Float3}
     {outputName=out_derivs, outputType=CMOT_Float4}
    ]
    inputs [
     {inputName=p, input {expression=null, inputName=None, mask=0, maskA=0, maskB=0, maskG=0, maskR=0, outputIndex=0}}
     {inputName=n, input {expression=null, inputName=None, mask=0, maskA=0, maskB=0, maskG=0, maskR=0, outputIndex=0}}
     {inputName=size, input {expression=null, inputName=None, mask=0, maskA=0, maskB=0, maskG=0, maskR=0, outputIndex=0}}
     {inputName=scale, input {expression=null, inputName=None, mask=0, maskA=0, maskB=0, maskG=0, maskR=0, outputIndex=0}}
    ]
   }
  }
  MaterialExpressionDDX_0 {class=MaterialExpressionDDX, inputs [Value {from=MaterialExpressionNamedRerouteUsage_1, output=""}], properties {}}
  MaterialExpressionDDY_0 {class=MaterialExpressionDDY, inputs [Value {from=MaterialExpressionNamedRerouteUsage_1, output=""}], properties {}}
  MaterialExpressionDivide_0 {
   class=MaterialExpressionDivide
   inputs [
    A {from=MaterialExpressionArctangent2Fast_0}
    B {from=MaterialExpressionMaterialFunctionCall_0, output=Result}
   ]
  }
  MaterialExpressionDivide_1 {
   class=MaterialExpressionDivide
   inputs [
    A {from=MaterialExpressionTransformPosition_2}
    B {from=MaterialExpressionFunctionInput_0}
   ]
  }
  MaterialExpressionDivide_2 {
   class=MaterialExpressionDivide
   inputs [
    A {from=MaterialExpressionArccosineFast_0}
    B {from=MaterialExpressionMaterialFunctionCall_2, output=Result}
   ]
  }
  MaterialExpressionDivide_3 {
   class=MaterialExpressionDivide
   inputs [
    A {from=MaterialExpressionMultiply_6}
    B {from=MaterialExpressionAdd_2}
   ]
  }
  MaterialExpressionDivide_4 {
   class=MaterialExpressionDivide
   inputs [
    A {from=MaterialExpressionConvert_0}
    B {from=MaterialExpressionAdd_2}
   ]
  }
  MaterialExpressionDivide_5 {
   class=MaterialExpressionDivide
   inputs [
    A {from=""}
    B {from=MaterialExpressionMaterialFunctionCall_4, output=Result}
   ]
   properties {constA=-1}
  }
  MaterialExpressionDivide_6 {
   class=MaterialExpressionDivide
   inputs [
    A {from=MaterialExpressionConvert_6}
    B {from=MaterialExpressionObjectRadius_0}
   ]
  }
  MaterialExpressionDivide_7 {
   class=MaterialExpressionDivide
   inputs [
    A {from=MaterialExpressionConvert_4}
    B {from=MaterialExpressionNamedRerouteUsage_2}
   ]
  }
  MaterialExpressionDivide_8 {
   class=MaterialExpressionDivide
   inputs [
    A {from=MaterialExpressionConvert_5}
    B {from=MaterialExpressionNamedRerouteUsage_2}
   ]
  }
  MaterialExpressionDotProduct_0 {
   class=MaterialExpressionDotProduct
   inputs [
    A {from=MaterialExpressionDivide_7, output=""}
    B {from=MaterialExpressionDDX_0, output=""}
   ]
   properties {}
  }
  MaterialExpressionDotProduct_1 {
   class=MaterialExpressionDotProduct
   inputs [
    A {from=MaterialExpressionDivide_8, output=""}
    B {from=MaterialExpressionDDX_0, output=""}
   ]
   properties {}
  }
  MaterialExpressionDotProduct_2 {
   class=MaterialExpressionDotProduct
   inputs [
    A {from=MaterialExpressionDivide_7, output=""}
    B {from=MaterialExpressionDDY_0, output=""}
   ]
   properties {}
  }
  MaterialExpressionDotProduct_3 {
   class=MaterialExpressionDotProduct
   inputs [
    A {from=MaterialExpressionDivide_8, output=""}
    B {from=MaterialExpressionDDY_0, output=""}
   ]
   properties {}
  }
  MaterialExpressionFunctionInput_0 {
   class=MaterialExpressionFunctionInput
   inputs [Preview {from="", output=""}]
   properties {bUsePreviewValueAsDefault=true, inputName=Scale, previewValue=<1 1 1 1>}
  }
  MaterialExpressionFunctionOutput_1 {class=MaterialExpressionFunctionOutput, inputs [None {from=MaterialExpressionConvert_1, output=""}], properties {outputName=UV}}
  MaterialExpressionFunctionOutput_2 {
   class=MaterialExpressionFunctionOutput
   inputs [None {from=MaterialExpressionShaderStageSwitch_0, output=""}]
   properties {outputName=packedDXY, sortPriority=3}
  }
  MaterialExpressionFunctionOutput_3 {
   class=MaterialExpressionFunctionOutput
   inputs [None {from=MaterialExpressionTransform_1, output=""}]
   properties {outputName=Bitangent, sortPriority=2}
  }
  MaterialExpressionFunctionOutput_4 {
   class=MaterialExpressionFunctionOutput
   inputs [None {from=MaterialExpressionTransform_0, output=""}]
   properties {outputName=Tangent, sortPriority=1}
  }
  MaterialExpressionMaterialFunctionCall_0 {
   class=MaterialExpressionMaterialFunctionCall
   inputs [Multiplier {from=MaterialExpressionConstant_1}]
   properties {materialFunction=&Pi}
  }
  MaterialExpressionMaterialFunctionCall_2 {
   class=MaterialExpressionMaterialFunctionCall
   inputs [Multiplier {from=""}]
   properties {materialFunction=&Pi}
  }
  MaterialExpressionMaterialFunctionCall_3 {
   class=MaterialExpressionMaterialFunctionCall
   inputs [Multiplier {from=MaterialExpressionMultiply_5}]
   properties {materialFunction=&Pi}
  }
  MaterialExpressionMaterialFunctionCall_4 {
   class=MaterialExpressionMaterialFunctionCall
   inputs [Multiplier {from=MaterialExpressionSquareRoot_0}]
   properties {materialFunction=&Pi}
  }
  MaterialExpressionMultiply_3 {
   class=MaterialExpressionMultiply
   inputs [
    A {from=MaterialExpressionConvert_0}
    B {from=MaterialExpressionConvert_0}
   ]
  }
  MaterialExpressionMultiply_4 {
   class=MaterialExpressionMultiply
   inputs [
    A {from=MaterialExpressionConvert_0}
    B {from=MaterialExpressionConvert_0}
   ]
  }
  MaterialExpressionMultiply_5 {
   class=MaterialExpressionMultiply
   inputs [
    A {from=MaterialExpressionAdd_1}
    B {from=""}
   ]
   properties {constB=2}
  }
  MaterialExpressionMultiply_6 {
   class=MaterialExpressionMultiply
   inputs [
    A {from=MaterialExpressionConvert_0}
    B {from=""}
   ]
   properties {constB=-1}
  }
  MaterialExpressionMultiply_8 {
   class=MaterialExpressionMultiply
   inputs [
    A {from=MaterialExpressionConvert_0}
    B {from=MaterialExpressionConvert_0}
   ]
  }
  MaterialExpressionNamedRerouteDeclaration_0 {
   class=MaterialExpressionNamedRerouteDeclaration
   inputs [None {from=MaterialExpressionFunctionInput_0}]
   properties {name=scale, nodeColor=<rgba 0 1 0.7422 1>}
  }
  MaterialExpressionNamedRerouteDeclaration_1 {
   class=MaterialExpressionNamedRerouteDeclaration
   inputs [None {from=MaterialExpressionNormalize_0}]
   properties {name=iPos, nodeColor=<rgba 0 1 0.9062 1>}
  }
  MaterialExpressionNamedRerouteUsage_1 {class=MaterialExpressionNamedRerouteUsage}
  MaterialExpressionNamedRerouteUsage_2 {class=MaterialExpressionNamedRerouteUsage}
  MaterialExpressionNormalize_0 {class=MaterialExpressionNormalize, inputs [VectorInput {from=MaterialExpressionDivide_1, output=""}]}
  MaterialExpressionNormalize_1 {class=MaterialExpressionNormalize, inputs [VectorInput {from=MaterialExpressionCrossProduct_1, output=""}]}
  MaterialExpressionObjectRadius_0 {class=MaterialExpressionObjectRadius, inputs [], properties {}}
  MaterialExpressionOneMinus_0 {class=MaterialExpressionOneMinus, inputs [None {from=MaterialExpressionMultiply_8}]}
  MaterialExpressionReroute_0 {class=MaterialExpressionReroute, inputs [None {from=MaterialExpressionReroute_1}]}
  MaterialExpressionReroute_1 {class=MaterialExpressionReroute, inputs [None {from=MaterialExpressionNormalize_0}]}
  MaterialExpressionShaderStageSwitch_0 {
   class=MaterialExpressionShaderStageSwitch
   inputs [
    PixelShader {from=MaterialExpressionDivide_6, output=""}
    VertexShader {from=MaterialExpressionConstant4Vector_0, output=RGBA}
   ]
   properties {}
  }
  MaterialExpressionSquareRoot_0 {class=MaterialExpressionSquareRoot, inputs [None {from=MaterialExpressionAdd_3, output=""}], properties {}}
  MaterialExpressionTransformPosition_2 {
   class=MaterialExpressionTransformPosition
   inputs [None {from=MaterialExpressionWorldPosition_1, output=XYZ}]
   properties {transformSourceType=TRANSFORMPOSSOURCE_TranslatedWorld, transformType=TRANSFORMPOSSOURCE_Instance}
  }
  MaterialExpressionTransform_0 {
   class=MaterialExpressionTransform
   inputs [None {from=MaterialExpressionCrossProduct_0}]
   properties {transformSourceType=TRANSFORMSOURCE_Instance}
  }
  MaterialExpressionTransform_1 {
   class=MaterialExpressionTransform
   inputs [None {from=MaterialExpressionNormalize_1}]
   properties {transformSourceType=TRANSFORMSOURCE_Instance}
  }
  MaterialExpressionWorldPosition_1 {class=MaterialExpressionWorldPosition, inputs [], properties {worldPositionShaderOffset=WPT_CameraRelative}}
 }
 properties {bExposeToLibrary=true, libraryCategoriesText [Base Material]}
}
