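# Library material function that builds ellipsoidal-projection UVs, a tangent frame and packed derivatives for SampleGrad and normal reorientation.
# MF_Coordinate_AnalyticalSphere projects using the ellipsoid scale, then MF_CoordinateTransform_2D applies offset, repeat and rotation about a (0, 0.5) pivot.
# Tangent, Bitangent and DXY come straight from the sphere function, bypassing that 2D transform.
# The rotation output is one minus the input turns.
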MaterialFunction /BaseMaterial/Materials/Functions/MF_ResolveCoordinateFrame_SphereUV_Basic
graph {
 properties {
  bExposeToLibrary=true
  description="Generates UVs from an ellipsoidal projection. Provides UVs, tangent/bitangent vectors with rotation for tangent-space normal reorientation, and packed derivatives for stable mip selection with SampleGrad. The projection can be scaled along each axis using the Scale Ellipsoid input, enabling ellipsoidal distortion of the generated UVs."
  userExposedCaption="CoordinateBasis: Ellipsoidal UV (Standard)"
  libraryCategoriesText [Base Material]
 }
 nodes [
  n1 = FunctionInput() {
   bUsePreviewValueAsDefault=true
   description="Scales the ellipsoidal projection along each axis. Values greater than 1 stretch the projection; values less than 1 compress it. Controls the ellipsoidal distortion of the UVs."
   inputName=ScaleEllipsoid
   sortPriority=1
   previewValue=<1 1 1 1>
  }
  n2 = MaterialFunctionCall!(Scale=@n1) {materialFunction=&/MF_Coordinate_AnalyticalSphere}
  n3 = FunctionInput() {
   bUsePreviewValueAsDefault=true
   description=Applies an additional per-axis offset to the coordinates before projection. Useful for correctly positioning a texture.
   inputName=CoordinateTransform
   inputType=FunctionInput_Vector2
   sortPriority=-9
  }
  n4 = FunctionInput() {
   bUsePreviewValueAsDefault=true
   description="Rotates the projected UVs around the selected axis. Units are in turns, where 1.0 = 360° and 0.25 = 90°."
   inputName="Rotation (Turns)"
   inputType=FunctionInput_Scalar
   sortPriority=-8
  }
  n5 = FunctionInput() {
   bUsePreviewValueAsDefault=true
   description="Sets how many times the projected coordinates repeat along each axis. Use values above 1 for denser tiling, negative values for mirroring, and avoid zero."
   inputName=CoordinateRepeat
   inputType=FunctionInput_Vector2
   sortPriority=-10
   previewValue=<1 1 1 1>
  }
  FunctionOutput(None=MaterialFunctionCall!(Coords=@n2.UV, UVOffset=@n3, UVRotation=@n4, RotationPivot=Constant2Vector() {g=0.5}, UVTile=@n5) {materialFunction=&/MF_CoordinateTransform_2D}.geometry_UV) {
   description="The primary 2D coordinates generated by the projection. Feed these into texture samplers for surface detail.\r\n"
   outputName=UV
  }
  FunctionOutput(None=OneMinus(None=@n4)) {
   description="The rotation applied around the projection axis, expressed in turns (1.0 = 360°). Used along with Tangent/Bitangent for reorienting normal maps via MF_TangentNormal_OrientWorld.\r\n"
   outputName="Rotation (Turns)"
   sortPriority=-5
  }
  FunctionOutput(None=@n2.packedDXY) {
   description="Packed derivatives of the projected UVs (ddx/ddy). Use with SampleGrad to ensure textures sample with correct mip levels and anisotropic filtering.\r\n"
   outputName=DXY
   sortPriority=3
  }
  FunctionOutput(None=@n2.Tangent) {
   description="The world-space tangent vector of the projection frame. Use with Bitangent and Rotation to correctly reorient tangent-space normals.\r\n"
   outputName=Tangent
   sortPriority=1
  }
  FunctionOutput(None=@n2.Bitangent) {
   description="The world-space bitangent vector of the projection frame. Complements Tangent to form the basis for tangent-space normal reorientation.\r\n"
   outputName=Bitangent
   sortPriority=2
  }
 ]
}
