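# Material function in the Base Material library that turns a noise texture into two pattern ID vectors and a blend mask, for varying UV coordinates.
# It samples a chosen noise channel with packed derivatives, defaulting to DDX/DDY of the coordinates, and scales it by truncated half VariationSteps.
# Floors staggered by half a step give IDs on separate axes; a triangle wave raised by BlendFactor forms the mask.
# An unconnected Custom node holds reference HLSL.
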MaterialFunction /BaseMaterial/Materials/Functions/MF_Coordinate_VariationUV
graph {
 properties {bExposeToLibrary=true, libraryCategoriesText [Base Material]}
 nodes [
  Custom!(inUV, inDDX, inDDY, inTex, TexChan, inSteps, k) {
   code="// HLSL for ref\r\nfloat smpTex = dot(inTex.SampleGrad( GetMaterialSharedSampler(inTexSampler,View.MaterialTextureBilinearWrapedSampler), inUV, \r\n                                     inDDX * (MaterialFloat2)View.MaterialTextureDerivativeMultiply, \r\n                                     inDDY * (MaterialFloat2)View.MaterialTextureDerivativeMultiply ), \r\n                                     TexChan);\r\n\r\ninSteps /= 2;\r\ninSteps = trunc(inSteps);\r\n\r\noffA = float2( floor(inSteps * smpTex), 0.f );\r\noffB = float2( 0.f, floor(inSteps * smpTex + 0.5) );\r\n\r\nfloat smpMsk = abs( 2.f * frac(smpTex * inSteps) - 1.f );\r\n\r\nreturn saturate( pow(smpMsk, 8.f * k) );"
   outputType=CMOT_Float1
   showCode=true
   additionalOutputs [{outputName=offA, outputType=CMOT_Float2}, {outputName=offB, outputType=CMOT_Float2}]
   inputs [
    {inputName=inUV, input {expression=null}}
    {inputName=inDDX, input {expression=null}}
    {inputName=inDDY, input {expression=null}}
    {inputName=inTex, input {expression=null}}
    {inputName=TexChan, input {expression=null}}
    {inputName=inSteps, input {expression=null}}
    {inputName=k, input {expression=null}}
   ]
  }
  n1 = FunctionInput(Preview=TextureObject() {samplerType=SAMPLERTYPE_LinearColor, texture=/BaseMaterial/Textures/Default/T_BlackLinearColor}) {bUsePreviewValueAsDefault=true, inputName=NoiseTexture, inputType=FunctionInput_Texture2D}
  n2 = FunctionInput(Preview=TextureCoordinate()) {bUsePreviewValueAsDefault=true, inputName=Coordinates, inputType=FunctionInput_Vector2, sortPriority=-1}
  MaterialExpressionFunctionInput_5 = FunctionInput(Preview=AppendVector(A=DDX(Value=@n2), B=DDY(Value=@n2))) {bUsePreviewValueAsDefault=true, inputName=packedDXY, inputType=FunctionInput_Vector4, sortPriority=1}
  n3 = Convert(None=@MaterialExpressionFunctionInput_5) {
   convertInputs [{
     type=Vector4
     defaultValue=<rgba 0 0 0 1>
     expressionInput {
      expression=&self:MaterialExpressionFunctionInput_5
      inputName=None
      mask=0
      maskA=0
      maskB=0
      maskG=0
      maskR=0
      outputIndex=0
     }
    }]
   convertOutputs [{type=Vector2}, {type=Vector2}]
  }
  n4 = FunctionInput() {bUsePreviewValueAsDefault=true, inputName=NoiseChannel, inputType=FunctionInput_Vector4, sortPriority=2, previewValue=<1 0 0 0>}
  n5 = FunctionInput() {bUsePreviewValueAsDefault=true, inputName=VariationSteps, inputType=FunctionInput_Scalar, sortPriority=3, previewValue=<7 0 0 1>}
  n6 = Multiply(
   A=DotProduct(
    A=TextureSample!(UVs=@n2, Tex=@n1, "DDX(UVs)"=@n3, "DDY(UVs)"=@n3, Apply View MipBias) {mipValueMode=TMVM_Derivative, samplerSource=SSM_Wrap_WorldGroupSettings}.RGBA
    B=@n4
   ) {desc=smpTex}
   B=Truncate!(None=Divide(A=@n5)) {is=class MaterialExpressionTruncate, desc=inSteps}
  )
  n7 = ConstantBiasScale(None=Frac(None=@n6)) {bias=-0.5, scale=2}
  n8 = FunctionInput() {bUsePreviewValueAsDefault=true, inputName=BlendFactor, inputType=FunctionInput_Scalar, sortPriority=4, previewValue=<0.5 0 0 1>}
  FunctionOutput(None=Power(Base=Abs(None=@n7), Exp=Multiply(A=@n8) {constB=8})) {outputName=Mask, sortPriority=2}
  n9 = Constant()
  FunctionOutput(None=AppendVector(A=Floor(None=@n6), B=@n9)) {outputName=pattternID1}
  n10 = Add(A=@n6, B=Constant() {r=0.5})
  FunctionOutput(None=AppendVector(A=@n9, B=Floor(None=@n10))) {outputName=patternID2, sortPriority=1}
 ]
}
