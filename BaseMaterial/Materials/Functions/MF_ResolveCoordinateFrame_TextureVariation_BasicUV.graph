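# Material function that reduces visible texture tiling by giving each band of a noise texture a random UV offset, scale and rotation.
# A selected UV channel is transformed, then MF_Coordinate_VariationUV derives two band IDs and a mask.
# Hashed IDs seed two randomized UV sets, output with rotations and a complementary two-channel blend mask.
# Both sets share derivatives and tangent basis from the pre-randomization coordinates, for gradient sampling and normal reorientation.
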MaterialFunction /BaseMaterial/Materials/Functions/MF_ResolveCoordinateFrame_TextureVariation_BasicUV
graph {
 expressions {
  MaterialExpressionConvert_0 {
   class=MaterialExpressionConvert
   inputs [
    None {from=MaterialExpressionOneMinus_0}
    None {from=MaterialExpressionNamedRerouteUsage_18}
   ]
   properties {
    nodeName=Make Float2
    convertInputs [
     {
      type=Scalar
      expressionInput {
       expression=&self:MaterialExpressionOneMinus_0
       inputName=None
       mask=0
       maskA=0
       maskB=0
       maskG=0
       maskR=0
       outputIndex=0
      }
     }
     {
      type=Scalar
      expressionInput {
       expression=&self:MaterialExpressionNamedRerouteUsage_18
       inputName=None
       mask=0
       maskA=0
       maskB=0
       maskG=0
       maskR=0
       outputIndex=0
      }
     }
    ]
    convertOutputs [{type=Vector2}]
   }
  }
  MaterialExpressionFunctionInput_0 {
   class=MaterialExpressionFunctionInput
   inputs [Preview {from="", output=""}]
   properties {
    bUsePreviewValueAsDefault=true
    description="Sets how many times the projected coordinates repeat along each axis. Use values above 1 for denser tiling, negative values for mirroring, and avoid zero."
    inputName=CoordinateRepeat
    inputType=FunctionInput_Vector2
    sortPriority=-10
    previewValue=<1 1 1 1>
   }
  }
  MaterialExpressionFunctionInput_1 {
   class=MaterialExpressionFunctionInput
   inputs [Preview {from="", output=""}]
   properties {
    bUsePreviewValueAsDefault=true
    description="Rotates the projected UVs around the selected axis. Units are in turns, where 1.0 = 360° and 0.25 = 90°."
    inputName="Rotation (Turns)"
    inputType=FunctionInput_Scalar
    sortPriority=-8
   }
  }
  MaterialExpressionFunctionInput_11 {
   class=MaterialExpressionFunctionInput
   inputs [Preview {from="", output=""}]
   properties {
    bUsePreviewValueAsDefault=true
    description=Selects which channel or combination thereof from the Band Texture drives band selection.
    inputName=Band ChannelWeight
    inputType=FunctionInput_Vector4
    sortPriority=-3
    previewValue=<1 0 0 0>
   }
  }
  MaterialExpressionFunctionInput_12 {
   class=MaterialExpressionFunctionInput
   inputs [Preview {from="", output=""}]
   properties {
    bUsePreviewValueAsDefault=true
    description=Number of variation bands across the texture.
    inputName=Band Count
    inputType=FunctionInput_Scalar
    sortPriority=-2
    previewValue=<7 0 0 1>
   }
  }
  MaterialExpressionFunctionInput_13 {
   class=MaterialExpressionFunctionInput
   inputs [Preview {from="", output=""}]
   properties {
    bUsePreviewValueAsDefault=true
    description=Applies an additional per-axis offset to the coordinates before projection. Useful for correctly positioning a texture.
    inputName=CoordinateTransform
    inputType=FunctionInput_Vector2
    sortPriority=-9
   }
  }
  MaterialExpressionFunctionInput_2 {
   class=MaterialExpressionFunctionInput
   inputs [Preview {from=MaterialExpressionTextureObject_0, output=""}]
   properties {
    bUsePreviewValueAsDefault=true
    description="The texture used to define variation bands. Each band region drives a distinct randomized transform (offset, scale, rotation)."
    inputName=BandTexture
    inputType=FunctionInput_Texture2D
    sortPriority=-4
   }
  }
  MaterialExpressionFunctionInput_22 {
   class=MaterialExpressionFunctionInput
   inputs [Preview {from="", output=""}]
   properties {
    bUsePreviewValueAsDefault=true
    description="Controls the contrast of the band blend mask. Internally applied as pow(mask, 8 × BlendFactor). A default of 0.5 corresponds to pow4 contrast."
    inputName=BlendTolerance
    inputType=FunctionInput_Scalar
    sortPriority=8
    previewValue=<0.5 0 0 1>
   }
  }
  MaterialExpressionFunctionInput_3 {
   class=MaterialExpressionFunctionInput
   inputs [Preview {from="", output=""}]
   properties {
    bUsePreviewValueAsDefault=true
    description="Tiling of the band texture. Larger values create more, smaller features across the surface; smaller values create fewer, larger features."
    inputName=Band PatternSize
    inputType=FunctionInput_Scalar
    sortPriority=-1
    previewValue=<1 0 0 1>
   }
  }
  MaterialExpressionFunctionInput_4 {
   class=MaterialExpressionFunctionInput
   inputs [Preview {from="", output=""}]
   properties {
    bUsePreviewValueAsDefault=true
    description="Range of random scale applied per band. Values below 1 enlarge band samples; values above 1 shrink them."
    inputName=Band Min/MaxRand Scale
    inputType=FunctionInput_Vector2
    sortPriority=3
    previewValue=<1 1 0 1>
   }
  }
  MaterialExpressionFunctionInput_5 {
   class=MaterialExpressionFunctionInput
   inputs [Preview {from="", output=""}]
   properties {
    bUsePreviewValueAsDefault=true
    description="Range of random offset applied to each band’s UVs. Offsets break up visual repetition by shifting band samples."
    inputName=Band Min/MaxRand Offset
    inputType=FunctionInput_Vector2
    sortPriority=4
   }
  }
  MaterialExpressionFunctionInput_6 {
   class=MaterialExpressionFunctionInput
   inputs [Preview {from="", output=""}]
   properties {
    bUsePreviewValueAsDefault=true
    description="Maximum random per-band rotation, in turns (1.0 = 360°). Each band’s UVs are rotated by a random angle up to this value."
    inputName="Band Rand Rotation (turns)"
    inputType=FunctionInput_Scalar
    sortPriority=5
   }
  }
  MaterialExpressionFunctionInput_7 {
   class=MaterialExpressionFunctionInput
   inputs [Preview {from="", output=""}]
   properties {
    bUsePreviewValueAsDefault=true
    description="Selects which mesh UV channel (0–3) to use as the base coordinate source."
    inputName=UVIndex
    inputType=FunctionInput_Scalar
    sortPriority=-5
   }
  }
  MaterialExpressionFunctionOutput_1 {
   class=MaterialExpressionFunctionOutput
   inputs [None {from=MaterialExpressionMaterialFunctionCall_17, output=UV}]
   properties {description="First randomized UV coordinate set generated by the function.\r\n", outputName=UV1}
  }
  MaterialExpressionFunctionOutput_10 {
   class=MaterialExpressionFunctionOutput
   inputs [None {from=MaterialExpressionMaterialFunctionCall_4, output=Tangent}]
   properties {description="World-space tangent vector for UV2. Use with Bitangent2WS for normal reorientation.\r\n", outputName=Tangent2, sortPriority=6}
  }
  MaterialExpressionFunctionOutput_11 {
   class=MaterialExpressionFunctionOutput
   inputs [None {from=MaterialExpressionMaterialFunctionCall_4, output=Bitangent}]
   properties {
    description="World-space bitangent vector for UV2. Complements Tangent2WS to form the basis for tangent-space normal reorientation.\r\n"
    outputName=Bitangent2
    sortPriority=7
   }
  }
  MaterialExpressionFunctionOutput_14 {
   class=MaterialExpressionFunctionOutput
   inputs [None {from=MaterialExpressionMaterialFunctionCall_17, output=SampleRotation}]
   properties {
    description="Rotation applied to UV1, expressed in turns (1.0 = 360°). Used for correct tangent-space normal reorientation."
    outputName="Rotation1 (Turns)"
    sortPriority=-6
   }
  }
  MaterialExpressionFunctionOutput_2 {
   class=MaterialExpressionFunctionOutput
   inputs [None {from=MaterialExpressionNamedRerouteDeclaration_7, output=""}]
   properties {
    description="Packed derivatives (ddx/ddy) for UV1. Use with SampleGrad for correct mip level selection during texture sampling.\r\n"
    outputName=DXY1
    sortPriority=3
   }
  }
  MaterialExpressionFunctionOutput_3 {
   class=MaterialExpressionFunctionOutput
   inputs [None {from=MaterialExpressionMaterialFunctionCall_1, output=UV}]
   properties {description=Second randomized UV coordinate set generated by the function., outputName=UV2, sortPriority=5}
  }
  MaterialExpressionFunctionOutput_4 {
   class=MaterialExpressionFunctionOutput
   inputs [None {from=MaterialExpressionNamedRerouteDeclaration_16, output=""}]
   properties {
    description="Packed derivatives (ddx/ddy) for UV2. Use with SampleGrad for correct mip level selection during texture sampling.\r\n"
    outputName=DXY2
    sortPriority=8
   }
  }
  MaterialExpressionFunctionOutput_5 {
   class=MaterialExpressionFunctionOutput
   inputs [None {from=MaterialExpressionMaterialFunctionCall_1, output=SampleRotation}]
   properties {
    description="Rotation applied to UV2, expressed in turns (1.0 = 360°). Used for correct tangent-space normal reorientation.\r\n"
    outputName="Rotation2 (Turns)"
    sortPriority=-5
   }
  }
  MaterialExpressionFunctionOutput_7 {
   class=MaterialExpressionFunctionOutput
   inputs [None {from=MaterialExpressionConvert_0, output=""}]
   properties {
    description="Outputs the repeating band pattern used to blend between the two randomized coordinate sets (UV1 and UV2), where the RG channels provide the weights for each sample. \r\n"
    outputName=BlendMask
    sortPriority=15
   }
  }
  MaterialExpressionFunctionOutput_8 {
   class=MaterialExpressionFunctionOutput
   inputs [None {from=MaterialExpressionMaterialFunctionCall_4, output=Tangent}]
   properties {description="World-space tangent vector for UV1. Use with Bitangent1WS for normal reorientation.\r\n", outputName=Tangent1, sortPriority=1}
  }
  MaterialExpressionFunctionOutput_9 {
   class=MaterialExpressionFunctionOutput
   inputs [None {from=MaterialExpressionMaterialFunctionCall_4, output=Bitangent}]
   properties {
    description="World-space bitangent vector for UV1. Complements Tangent1WS to form the basis for tangent-space normal reorientation.\r\n"
    outputName=Bitangent1
    sortPriority=2
   }
  }
  MaterialExpressionMaterialFunctionCall_0 {
   class=MaterialExpressionMaterialFunctionCall
   inputs [
    Selector {from=MaterialExpressionFunctionInput_7}
    VectorA {from=MaterialExpressionTextureCoordinate_4}
    VectorB {from=MaterialExpressionTextureCoordinate_5}
    VectorC {from=MaterialExpressionTextureCoordinate_6}
    VectorD {from=MaterialExpressionTextureCoordinate_7}
   ]
   properties {materialFunction=&/MF_Switch4_Vec2}
  }
  MaterialExpressionMaterialFunctionCall_1 {
   class=MaterialExpressionMaterialFunctionCall
   inputs [
    Coords {from=MaterialExpressionNamedRerouteUsage_5}
    "TileRand (Min/Max)" {from=MaterialExpressionNamedRerouteUsage_10}
    "OffsetRand (Min/Max)" {from=MaterialExpressionNamedRerouteUsage_9}
    TileStrengthRand {from=""}
    OffsetStrengthRand {from=""}
    RotateStrengthRand {from=MaterialExpressionNamedRerouteUsage_13}
    SeedA {from=MaterialExpressionNamedRerouteUsage_21}
   ]
   properties {materialFunction=&/MF_Coordinate_Rand1}
  }
  MaterialExpressionMaterialFunctionCall_11 {
   class=MaterialExpressionMaterialFunctionCall
   inputs [seed {from=MaterialExpressionMaterialFunctionCall_5, output=patternID2}]
   properties {materialFunction=&/MF_Hash42}
  }
  MaterialExpressionMaterialFunctionCall_17 {
   class=MaterialExpressionMaterialFunctionCall
   inputs [
    Coords {from=MaterialExpressionNamedRerouteUsage_28}
    "TileRand (Min/Max)" {from=MaterialExpressionNamedRerouteUsage_31}
    "OffsetRand (Min/Max)" {from=MaterialExpressionNamedRerouteUsage_30}
    TileStrengthRand {from=""}
    OffsetStrengthRand {from=""}
    RotateStrengthRand {from=MaterialExpressionNamedRerouteUsage_32}
    SeedA {from=MaterialExpressionNamedRerouteUsage_7}
   ]
   properties {materialFunction=&MF_Coordinate_Rand1}
  }
  MaterialExpressionMaterialFunctionCall_3 {
   class=MaterialExpressionMaterialFunctionCall
   inputs [
    Coords {from=MaterialExpressionMaterialFunctionCall_0, output=Selected Vector}
    UVOffset {from=MaterialExpressionFunctionInput_13}
    UVRotation {from=MaterialExpressionFunctionInput_1}
    RotationPivot {from=""}
    UVTile {from=MaterialExpressionFunctionInput_0}
   ]
   properties {materialFunction=&/MF_CoordinateTransform_2D}
  }
  MaterialExpressionMaterialFunctionCall_4 {
   class=MaterialExpressionMaterialFunctionCall
   inputs [packedDXY {from=MaterialExpressionFunctionOutput_2}]
   properties {materialFunction=&/MF_DeriveTangentBasis}
  }
  MaterialExpressionMaterialFunctionCall_5 {
   class=MaterialExpressionMaterialFunctionCall
   inputs [
    Coordinates {from=MaterialExpressionMultiply_7}
    NoiseTexture {from=MaterialExpressionReroute_5}
    packedDXY {from=MaterialExpressionMultiply_0}
    NoiseChannel {from=MaterialExpressionFunctionInput_11}
    VariationSteps {from=MaterialExpressionReroute_0}
    BlendFactor {from=MaterialExpressionReroute_1}
   ]
   properties {materialFunction=&/MF_Coordinate_VariationUV}
  }
  MaterialExpressionMaterialFunctionCall_6 {
   class=MaterialExpressionMaterialFunctionCall
   inputs [seed {from=MaterialExpressionMaterialFunctionCall_5, output=pattternID1}]
   properties {materialFunction=&MF_Hash42}
  }
  MaterialExpressionMultiply_0 {
   class=MaterialExpressionMultiply
   inputs [
    A {from=MaterialExpressionFunctionInput_3}
    B {from=MaterialExpressionMaterialFunctionCall_3, output=geometry_packedDXY}
   ]
  }
  MaterialExpressionMultiply_7 {
   class=MaterialExpressionMultiply
   inputs [
    A {from=MaterialExpressionFunctionInput_3}
    B {from=MaterialExpressionMaterialFunctionCall_3, output=geometry_UV}
   ]
  }
  MaterialExpressionNamedRerouteDeclaration_10 {
   class=MaterialExpressionNamedRerouteDeclaration
   inputs [None {from=MaterialExpressionMaterialFunctionCall_5, output=Mask}]
   properties {name=Mask, nodeColor=<rgba 0 1 0.3906 1>}
  }
  MaterialExpressionNamedRerouteDeclaration_12 {
   class=MaterialExpressionNamedRerouteDeclaration
   inputs [None {from=MaterialExpressionFunctionInput_5}]
   properties {name=BandOffset, nodeColor=<rgba 0.9 0.3505 0.009 1>}
  }
  MaterialExpressionNamedRerouteDeclaration_13 {
   class=MaterialExpressionNamedRerouteDeclaration
   inputs [None {from=MaterialExpressionFunctionInput_6}]
   properties {name=BandRot, nodeColor=<rgba 0.9 0.3505 0.009 1>}
  }
  MaterialExpressionNamedRerouteDeclaration_14 {
   class=MaterialExpressionNamedRerouteDeclaration
   inputs [None {from=MaterialExpressionFunctionInput_4}]
   properties {name=BandScale, nodeColor=<rgba 0.9 0.3505 0.009 1>}
  }
  MaterialExpressionNamedRerouteDeclaration_16 {
   class=MaterialExpressionNamedRerouteDeclaration
   inputs [None {from=MaterialExpressionMaterialFunctionCall_3, output=geometry_packedDXY}]
   properties {name=DXY2, nodeColor=<rgba 0.009 0.9 0.009 1>}
  }
  MaterialExpressionNamedRerouteDeclaration_2 {
   class=MaterialExpressionNamedRerouteDeclaration
   inputs [None {from=MaterialExpressionMaterialFunctionCall_6, output=Hash4D}]
   properties {name=RandA, nodeColor=<rgba 1 0 0.75 1>}
  }
  MaterialExpressionNamedRerouteDeclaration_6 {
   class=MaterialExpressionNamedRerouteDeclaration
   inputs [None {from=MaterialExpressionMaterialFunctionCall_3, output=geometry_UV}]
   properties {name=UVi, nodeColor=<rgba 0.9 0.009 0.009 1>}
  }
  MaterialExpressionNamedRerouteDeclaration_7 {
   class=MaterialExpressionNamedRerouteDeclaration
   inputs [None {from=MaterialExpressionMaterialFunctionCall_3, output=geometry_packedDXY}]
   properties {name=DXY1, nodeColor=<rgba 0.9 0.009 0.009 1>}
  }
  MaterialExpressionNamedRerouteDeclaration_8 {
   class=MaterialExpressionNamedRerouteDeclaration
   inputs [None {from=MaterialExpressionMaterialFunctionCall_11, output=Hash4D}]
   properties {name=RandB, nodeColor=<rgba 1 0 0.75 1>}
  }
  MaterialExpressionNamedRerouteUsage_1 {class=MaterialExpressionNamedRerouteUsage}
  MaterialExpressionNamedRerouteUsage_10 {class=MaterialExpressionNamedRerouteUsage}
  MaterialExpressionNamedRerouteUsage_13 {class=MaterialExpressionNamedRerouteUsage}
  MaterialExpressionNamedRerouteUsage_18 {class=MaterialExpressionNamedRerouteUsage}
  MaterialExpressionNamedRerouteUsage_21 {class=MaterialExpressionNamedRerouteUsage}
  MaterialExpressionNamedRerouteUsage_28 {class=MaterialExpressionNamedRerouteUsage}
  MaterialExpressionNamedRerouteUsage_30 {class=MaterialExpressionNamedRerouteUsage}
  MaterialExpressionNamedRerouteUsage_31 {class=MaterialExpressionNamedRerouteUsage}
  MaterialExpressionNamedRerouteUsage_32 {class=MaterialExpressionNamedRerouteUsage}
  MaterialExpressionNamedRerouteUsage_5 {class=MaterialExpressionNamedRerouteUsage}
  MaterialExpressionNamedRerouteUsage_7 {class=MaterialExpressionNamedRerouteUsage}
  MaterialExpressionNamedRerouteUsage_9 {class=MaterialExpressionNamedRerouteUsage}
  MaterialExpressionOneMinus_0 {class=MaterialExpressionOneMinus, inputs [None {from=MaterialExpressionNamedRerouteUsage_1}]}
  MaterialExpressionReroute_0 {class=MaterialExpressionReroute, inputs [None {from=MaterialExpressionFunctionInput_12}]}
  MaterialExpressionReroute_1 {class=MaterialExpressionReroute, inputs [None {from=MaterialExpressionFunctionInput_22}]}
  MaterialExpressionReroute_5 {class=MaterialExpressionReroute, inputs [None {from=MaterialExpressionFunctionInput_2}]}
  MaterialExpressionTextureCoordinate_4 {class=MaterialExpressionTextureCoordinate}
  MaterialExpressionTextureCoordinate_5 {class=MaterialExpressionTextureCoordinate, properties {coordinateIndex=1}}
  MaterialExpressionTextureCoordinate_6 {class=MaterialExpressionTextureCoordinate, properties {coordinateIndex=2}}
  MaterialExpressionTextureCoordinate_7 {class=MaterialExpressionTextureCoordinate, properties {coordinateIndex=3}}
  MaterialExpressionTextureObject_0 {
   class=MaterialExpressionTextureObject
   inputs []
   properties {samplerType=SAMPLERTYPE_LinearColor, texture=/BaseMaterial/Textures/Noises/T_Variation_1k_RGB_nonVT}
  }
 }
 properties {
  bExposeToLibrary=true
  description="Applies randomized UV variation per-band of the input texture to reduce visible tiling. Each band can vary in offset, scale, and rotation. Outputs two coordinate sets (UV1 and UV2), each with tangent/bitangent/rotation/derivatives, plus a BlendMask (RG) for post-sample blending. Supports blend weight modulation by albedo or height, producing smoother, more perceptual transitions between bands."
  userExposedCaption="CoordinateBasis: Texture Variation (Standard)"
  libraryCategoriesText [Base Material]
 }
}
