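# Material function in the Base Material library that produces two seed-randomized 2D UV transforms and selects one by MaskIndex.
# SeedA and SeedB each drive offset and tiling lerps within shared min/max ranges, plus signed rotation wrapped by Frac, all weighted by per-variant strengths.
# Each set feeds MF_CoordinateTransform_2D.
# Switches on the truncated, epsilon-nudged index output the chosen UV, packed derivatives (packedDXY) and sample rotation.
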MaterialFunction /BaseMaterial/Materials/Functions/MF_Coordinate_Rand2
graph {
 properties {bExposeToLibrary=true, libraryCategoriesText [Base Material]}
 nodes [
  n1 = Truncate!(None=Add(A=FunctionInput() {bUsePreviewValueAsDefault=true, inputName=MaskIndex, inputType=FunctionInput_Scalar, sortPriority=20}) {constB=0.0033}) {is=class MaterialExpressionTruncate}
  n2 = FunctionInput(Preview=TextureCoordinate()) {bUsePreviewValueAsDefault=true, inputName=Coords, inputType=FunctionInput_Vector2, sortPriority=-10}
  n3 = NamedRerouteUsage()
  MaterialExpressionFunctionInput_0 = FunctionInput() {bUsePreviewValueAsDefault=true, inputName=SeedA, inputType=FunctionInput_Vector4, sortPriority=10, previewValue=<1 1 1 1>}
  n4 = Convert(None=@MaterialExpressionFunctionInput_0) {
   convertInputs [{
     type=Vector4
     defaultValue=<rgba 0 0 0 1>
     expressionInput {
      expression=&self:MaterialExpressionFunctionInput_0
      inputName=None
      mask=0
      maskA=0
      maskB=0
      maskG=0
      maskR=0
      outputIndex=0
     }
    }]
   convertOutputs [{type=Vector2}, {type=Scalar}, {type=Scalar}]
  }
  n5 = Multiply(A=ConstantBiasScale(None=@n4) {bias=-0.5, scale=2}, B=NamedRerouteUsage())
  n6 = NamedRerouteUsage()
  MaterialExpressionMaterialFunctionCall_10 = MaterialFunctionCall!(
   Coords=@n2
   UVOffset=LinearInterpolate(A=NamedRerouteUsage(), B=NamedRerouteUsage(), Alpha=Multiply(A=@n3, B=@n4))
   UVRotation=NamedRerouteDeclaration(None=Frac(None=@n5)) {name=turnsa, nodeColor=<rgba 0.8963 0.8963 0.0091 1>}
   RotationPivot
   UVTile=LinearInterpolate(A=NamedRerouteUsage(), B=NamedRerouteUsage(), Alpha=Multiply(A=@n4, B=@n6))
  ) {materialFunction=&/MF_CoordinateTransform_2D}
  n7 = NamedRerouteUsage()
  MaterialExpressionFunctionInput_1 = FunctionInput() {bUsePreviewValueAsDefault=true, inputName=SeedB, inputType=FunctionInput_Vector4, sortPriority=11, previewValue=<1 1 1 1>}
  n8 = Convert(None=@MaterialExpressionFunctionInput_1) {
   convertInputs [{
     type=Vector4
     defaultValue=<rgba 0 0 0 1>
     expressionInput {
      expression=&self:MaterialExpressionFunctionInput_1
      inputName=None
      mask=0
      maskA=0
      maskB=0
      maskG=0
      maskR=0
      outputIndex=0
     }
    }]
   convertOutputs [{type=Vector2}, {type=Scalar}, {type=Scalar}]
  }
  n9 = Multiply(A=ConstantBiasScale(None=@n8) {bias=-0.5, scale=2}, B=NamedRerouteUsage())
  n10 = NamedRerouteUsage()
  MaterialExpressionMaterialFunctionCall_14 = MaterialFunctionCall(
   Coords=@n2
   UVOffset=LinearInterpolate(A=NamedRerouteUsage(), B=NamedRerouteUsage(), Alpha=Multiply(A=@n7, B=@n8))
   UVRotation=NamedRerouteDeclaration(None=Frac(None=@n9)) {name=turnsb, nodeColor=<rgba 0.8963 0.8963 0.0091 1>}
   UVTile=LinearInterpolate(A=NamedRerouteUsage(), B=NamedRerouteUsage(), Alpha=Multiply(A=@n8, B=@n10))
  ) {materialFunction=&MF_CoordinateTransform_2D}
  n11 = Switch!(
   SwitchValue=@n1
   Default=@MaterialExpressionMaterialFunctionCall_10.geometry_UV
   UVa=@MaterialExpressionMaterialFunctionCall_10.geometry_UV
   UVb=@MaterialExpressionMaterialFunctionCall_14.geometry_UV
  ) {inputs [
    {
     inputName=UVa
     input {
      expression=&self:MaterialExpressionMaterialFunctionCall_10
      inputName=None
      mask=0
      maskA=0
      maskB=0
      maskG=0
      maskR=0
      outputIndex=0
     }
    }
    {
     inputName=UVb
     input {
      expression=&self:MaterialExpressionMaterialFunctionCall_14
      inputName=None
      mask=0
      maskA=0
      maskB=0
      maskG=0
      maskR=0
      outputIndex=0
     }
    }
   ]}
  FunctionOutput(None=@n11) {outputName=UV}
  n12 = Switch!(
   SwitchValue=@n1
   Default=@MaterialExpressionMaterialFunctionCall_10.geometry_packedDXY
   adxy=@MaterialExpressionMaterialFunctionCall_10.geometry_packedDXY
   bdxy=@MaterialExpressionMaterialFunctionCall_14.geometry_packedDXY
  ) {inputs [
    {
     inputName=adxy
     input {
      expression=&self:MaterialExpressionMaterialFunctionCall_10
      inputName=None
      mask=0
      maskA=0
      maskB=0
      maskG=0
      maskR=0
      outputIndex=1
     }
    }
    {
     inputName=bdxy
     input {
      expression=&self:MaterialExpressionMaterialFunctionCall_14
      inputName=None
      mask=0
      maskA=0
      maskB=0
      maskG=0
      maskR=0
      outputIndex=1
     }
    }
   ]}
  FunctionOutput(None=@n12) {outputName=packedDXY, sortPriority=10}
  MaterialExpressionNamedRerouteUsage_34 = NamedRerouteUsage()
  MaterialExpressionNamedRerouteUsage_35 = NamedRerouteUsage()
  n13 = Switch!(
   SwitchValue=@n1
   Default=@MaterialExpressionNamedRerouteUsage_34
   rotA=@MaterialExpressionNamedRerouteUsage_34
   rotB=@MaterialExpressionNamedRerouteUsage_35
  ) {inputs [
    {
     inputName=rotA
     input {
      expression=&self:MaterialExpressionNamedRerouteUsage_34
      inputName=None
      mask=0
      maskA=0
      maskB=0
      maskG=0
      maskR=0
      outputIndex=0
     }
    }
    {
     inputName=rotB
     input {
      expression=&self:MaterialExpressionNamedRerouteUsage_35
      inputName=None
      mask=0
      maskA=0
      maskB=0
      maskG=0
      maskR=0
      outputIndex=0
     }
    }
   ]}
  FunctionOutput(None=@n13) {outputName=SampleRotation, sortPriority=20}
  MaterialExpressionFunctionInput_5 = FunctionInput() {bUsePreviewValueAsDefault=true, inputName=OffsetStrengthRand, inputType=FunctionInput_Vector2, sortPriority=6, previewValue=<1 1 1 1>}
  n14 = Convert(None=@MaterialExpressionFunctionInput_5) {
   nodeName=Break Float2
   convertInputs [{
     type=Vector2
     expressionInput {
      expression=&self:MaterialExpressionFunctionInput_5
      inputName=None
      mask=0
      maskA=0
      maskB=0
      maskG=0
      maskR=0
      outputIndex=0
     }
    }]
   convertOutputs [{type=Scalar}, {type=Scalar}]
  }
  NamedRerouteDeclaration(None=@n14) {name=OffA, nodeColor=<rgba 0.6875 1 0 1>}
  NamedRerouteDeclaration(None=@n14) {name=OffB, nodeColor=<rgba 0.6875 1 0 1>}
  MaterialExpressionFunctionInput_7 = FunctionInput() {bUsePreviewValueAsDefault=true, inputName="TileRand (Min/Max)", inputType=FunctionInput_Vector2, previewValue=<1 1 0 1>}
  n15 = Convert(None=@MaterialExpressionFunctionInput_7) {
   nodeName=Break Float2
   convertInputs [{
     type=Vector2
     expressionInput {
      expression=&self:MaterialExpressionFunctionInput_7
      inputName=None
      mask=0
      maskA=0
      maskB=0
      maskG=0
      maskR=0
      outputIndex=0
     }
    }]
   convertOutputs [{type=Scalar}, {type=Scalar}]
  }
  NamedRerouteDeclaration(None=@n15) {name=MinTile, nodeColor=<rgba 0.6875 1 0 1>}
  MaterialExpressionFunctionInput_8 = FunctionInput() {bUsePreviewValueAsDefault=true, inputName="OffsetRand (Min/Max)", inputType=FunctionInput_Vector2, sortPriority=1}
  n16 = Convert(None=@MaterialExpressionFunctionInput_8) {
   nodeName=Break Float2
   convertInputs [{
     type=Vector2
     expressionInput {
      expression=&self:MaterialExpressionFunctionInput_8
      inputName=None
      mask=0
      maskA=0
      maskB=0
      maskG=0
      maskR=0
      outputIndex=0
     }
    }]
   convertOutputs [{type=Scalar}, {type=Scalar}]
  }
  NamedRerouteDeclaration(None=@n16) {name=MaxOffset, nodeColor=<rgba 0.6875 1 0 1>}
  MaterialExpressionFunctionInput_9 = FunctionInput() {bUsePreviewValueAsDefault=true, inputName=TileStrengthRand, inputType=FunctionInput_Vector2, sortPriority=5, previewValue=<1 1 1 1>}
  n17 = Convert(None=@MaterialExpressionFunctionInput_9) {
   nodeName=Break Float2
   convertInputs [{
     type=Vector2
     expressionInput {
      expression=&self:MaterialExpressionFunctionInput_9
      inputName=None
      mask=0
      maskA=0
      maskB=0
      maskG=0
      maskR=0
      outputIndex=0
     }
    }]
   convertOutputs [{type=Scalar}, {type=Scalar}]
  }
  NamedRerouteDeclaration(None=@n17) {name=TileA, nodeColor=<rgba 0.6875 1 0 1>}
  NamedRerouteDeclaration(None=@n17) {name=TileB, nodeColor=<rgba 0.6875 1 0 1>}
  MaterialExpressionFunctionInput_6 = FunctionInput() {bUsePreviewValueAsDefault=true, inputName=RotateStrengthRand, inputType=FunctionInput_Vector2, sortPriority=7}
  n18 = Convert(None=@MaterialExpressionFunctionInput_6) {
   nodeName=Break Float2
   convertInputs [{
     type=Vector2
     expressionInput {
      expression=&self:MaterialExpressionFunctionInput_6
      inputName=None
      mask=0
      maskA=0
      maskB=0
      maskG=0
      maskR=0
      outputIndex=0
     }
    }]
   convertOutputs [{type=Scalar}, {type=Scalar}]
  }
  NamedRerouteDeclaration(None=@n18) {name=RotA, nodeColor=<rgba 0.6875 1 0 1>}
  NamedRerouteDeclaration(None=@n18) {name=RotB, nodeColor=<rgba 0.6875 1 0 1>}
  NamedRerouteDeclaration(None=@n15) {name=MaxTile, nodeColor=<rgba 0.6875 1 0 1>}
  NamedRerouteDeclaration(None=@n16) {name=MinOffset, nodeColor=<rgba 0.6875 1 0 1>}
 ]
}
