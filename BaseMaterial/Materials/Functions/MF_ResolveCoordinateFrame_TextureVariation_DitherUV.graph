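# Library material function that breaks up texture tiling by giving noise-defined bands random UV offset, scale and rotation.
# A selected, transformed mesh UV samples the band texture, yielding two hashed pattern IDs and a contrast-adjusted mask.
# Dither rounds that mask to a binary choice between both random transforms, approximating blending with one texture sample.
# Outputs UV, rotation, packed derivatives and derived tangent/bitangent for gradient sampling and normal reorientation.
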MaterialFunction /BaseMaterial/Materials/Functions/MF_ResolveCoordinateFrame_TextureVariation_DitherUV
graph {
 properties {
  bExposeToLibrary=true
  description="Applies randomized UV variation per-band of the input texture to break up tiling and repetition. Each band can vary in offset, scale, and rotation. Uses dithered sampling to approximate blending at very low cost (single sample). Fastest option, but dithering can introduce artifacts and does not support post-sample blend modulation."
  userExposedCaption="CoordinateBasis: Texture Variation (Dithered)"
  libraryCategoriesText [Base Material]
 }
 nodes [
  n1 = MaterialFunctionCall!(
   Coords=NamedRerouteUsage()
   "TileRand (Min/Max)"=NamedRerouteUsage()
   "OffsetRand (Min/Max)"=NamedRerouteUsage()
   TileStrengthRand
   OffsetStrengthRand
   RotateStrengthRand=NamedRerouteUsage()
   SeedA=NamedRerouteUsage()
   SeedB=NamedRerouteUsage()
   MaskIndex=NamedRerouteUsage()
  ) {materialFunction=&/MF_Coordinate_Rand2}
  FunctionOutput(None=@n1.UV) {
   description="The primary 2D coordinates generated by the projection. Feed these into texture samplers for surface detail.\r\n"
   outputName=UV
  }
  FunctionOutput(None=@n1.SampleRotation) {
   description="The rotation applied around the projection axis, expressed in turns (1.0 = 360°). Used along with Tangent/Bitangent for reorienting normal maps via MF_TangentNormal_OrientWorld.\r\n"
   outputName="Rotation (Turns)"
   sortPriority=-6
  }
  n2 = FunctionOutput(None=NamedRerouteUsage()) {
   description="Packed derivatives of the projected UVs (ddx/ddy). Use with SampleGrad to ensure textures sample with correct mip levels and anisotropic filtering.\r\n"
   outputName=DXY
   sortPriority=3
  }
  n3 = MaterialFunctionCall!(packedDXY=@n2) {materialFunction=&/MF_DeriveTangentBasis}
  FunctionOutput(None=@n3.Tangent) {
   description="The world-space tangent vector of the projection frame. Use with Bitangent and Rotation to correctly reorient tangent-space normals.\r\n"
   outputName=Tangent
   sortPriority=1
  }
  FunctionOutput(None=@n3.Bitangent) {
   description="The world-space bitangent vector of the projection frame. Complements Tangent to form the basis for tangent-space normal reorientation.\r\n"
   outputName=Bitangent
   sortPriority=2
  }
  n4 = FunctionInput() {
   bUsePreviewValueAsDefault=true
   description="Tiling of the band texture. Larger values create more, smaller features across the surface; smaller values create fewer, larger features."
   inputName=Band PatternSize
   inputType=FunctionInput_Scalar
   sortPriority=-1
   previewValue=<1 0 0 1>
  }
  n5 = FunctionInput() {
   bUsePreviewValueAsDefault=true
   description="Selects which mesh UV channel (0–3) to use as the base coordinate source."
   inputName=UVIndex
   inputType=FunctionInput_Scalar
   sortPriority=-5
  }
  n6 = FunctionInput() {
   bUsePreviewValueAsDefault=true
   description=Applies an additional per-axis offset to the coordinates before projection. Useful for correctly positioning a texture.
   inputName=CoordinateTransform
   inputType=FunctionInput_Vector2
   sortPriority=-9
  }
  n7 = FunctionInput() {
   bUsePreviewValueAsDefault=true
   description="Rotates the projected UVs around the selected axis. Units are in turns, where 1.0 = 360° and 0.25 = 90°."
   inputName="Rotation (Turns)"
   inputType=FunctionInput_Scalar
   sortPriority=-8
  }
  n8 = FunctionInput() {
   bUsePreviewValueAsDefault=true
   description="Sets how many times the projected coordinates repeat along each axis. Use values above 1 for denser tiling, negative values for mirroring, and avoid zero."
   inputName=CoordinateRepeat
   inputType=FunctionInput_Vector2
   sortPriority=-10
   previewValue=<1 1 1 1>
  }
  n9 = MaterialFunctionCall!(
   Coords=MaterialFunctionCall!(
    Selector=@n5
    VectorA=TextureCoordinate()
    VectorB=TextureCoordinate() {coordinateIndex=1}
    VectorC=TextureCoordinate() {coordinateIndex=2}
    VectorD=TextureCoordinate() {coordinateIndex=3}
   ) {materialFunction=&/MF_Switch4_Vec2}.Selected Vector
   UVOffset=@n6
   UVRotation=@n7
   RotationPivot
   UVTile=@n8
  ) {materialFunction=&/MF_CoordinateTransform_2D}
  n10 = FunctionInput() {
   bUsePreviewValueAsDefault=true
   description=Selects which channel or combination thereof from the Band Texture drives band selection.
   inputName=Band ChannelWeight
   inputType=FunctionInput_Vector4
   sortPriority=-3
   previewValue=<1 0 0 0>
  }
  n11 = FunctionInput() {
   bUsePreviewValueAsDefault=true
   description=Number of variation bands across the texture.
   inputName=Band Count
   inputType=FunctionInput_Scalar
   sortPriority=-2
   previewValue=<7 0 0 1>
  }
  n12 = FunctionInput(Preview=TextureObject() {samplerType=SAMPLERTYPE_LinearColor, texture=/BaseMaterial/Textures/Noises/T_Variation_1k_RGB_nonVT}) {
   bUsePreviewValueAsDefault=true
   description="The texture used to define variation bands. Each band region drives a distinct randomized transform (offset, scale, rotation)."
   inputName=BandTexture
   inputType=FunctionInput_Texture2D
   sortPriority=-4
  }
  n13 = MaterialFunctionCall!(
   Coordinates=Multiply(A=@n4, B=@n9.geometry_UV)
   NoiseTexture=@n12
   packedDXY=Multiply(A=@n4, B=NamedRerouteDeclaration(None=@n9.geometry_packedDXY) {name=DXY1, nodeColor=<rgba 0.9 0.009 0.009 1>})
   NoiseChannel=@n10
   VariationSteps=@n11
   BlendFactor
  ) {materialFunction=&/MF_Coordinate_VariationUV}
  n14 = FunctionInput() {
   bUsePreviewValueAsDefault=true
   description="Controls the contrast of the band blend mask. Internally applied as pow(mask, 8 × BlendFactor). A default of 0.5 corresponds to pow4 contrast."
   inputName=BlendFactor
   inputType=FunctionInput_Scalar
   sortPriority=7
   previewValue=<0.5 0 0 1>
  }
  n15 = Saturate(None=Power(Base=@n13.Mask, Exp=Multiply(B=@n14) {constA=10}) {constExponent=1})
  n16 = FunctionInput(Preview=MaterialFunctionCall!(TemporalSampleIndex, PixelPosition) {materialFunction=&/MF_Noise_InterleavedGradientGolden_1d}.Unsigned) {
   bUsePreviewValueAsDefault=true
   description="Provides the dithering sequence used for stochastic projection blending. By default, the function generates an interleaved gradient noise pattern based on the golden ratio, which produces stable, slow-converging results at low cost. Supplying a custom value lets you override this with your own dithering scheme, useful for synchronizing with other systems or experimenting with alternate noise sources."
   inputName=Dither
   inputType=FunctionInput_Scalar
  }
  NamedRerouteDeclaration(None=Floor(None=Add(A=@n15, B=@n16))) {name=Blend, nodeColor=<rgba 1 1 1 1>}
  n17 = FunctionInput() {
   bUsePreviewValueAsDefault=true
   description="Range of random offset applied to each band’s UVs. Offsets break up visual repetition by shifting band samples."
   inputName=Band Min/MaxRand Offset
   inputType=FunctionInput_Vector2
   sortPriority=4
  }
  NamedRerouteDeclaration(None=@n17) {name=BandOffset, nodeColor=<rgba 0.9 0.3505 0.009 1>}
  n18 = FunctionInput() {
   bUsePreviewValueAsDefault=true
   description="Maximum random per-band rotation, in turns (1.0 = 360°). Each band’s UVs are rotated by a random angle up to this value."
   inputName="Band Rand Rotation (turns)"
   inputType=FunctionInput_Scalar
   sortPriority=5
  }
  NamedRerouteDeclaration(None=@n18) {name=BandRot, nodeColor=<rgba 0.9 0.3505 0.009 1>}
  n19 = FunctionInput() {
   bUsePreviewValueAsDefault=true
   description="Range of random scale applied per band. Values below 1 enlarge band samples; values above 1 shrink them."
   inputName=Band Min/MaxRand Scale
   inputType=FunctionInput_Vector2
   sortPriority=3
   previewValue=<1 1 0 1>
  }
  NamedRerouteDeclaration(None=@n19) {name=BandScale, nodeColor=<rgba 0.9 0.3505 0.009 1>}
  NamedRerouteDeclaration(None=MaterialFunctionCall!(seed=@n13.pattternID1) {materialFunction=&/MF_Hash42}.Hash4D) {name=RandA, nodeColor=<rgba 1 0 0.75 1>}
  NamedRerouteDeclaration(None=@n9.geometry_UV) {name=UVi, nodeColor=<rgba 0.9 0.009 0.009 1>}
  NamedRerouteDeclaration(None=MaterialFunctionCall(seed=@n13.patternID2) {materialFunction=&MF_Hash42}.Hash4D) {name=RandB, nodeColor=<rgba 1 0 0.75 1>}
 ]
}
