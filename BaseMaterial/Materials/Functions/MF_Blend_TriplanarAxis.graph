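# Material function, exposed to the library under Base Material, that computes triplanar axis blend weights from a reference normal, defaulting to vertex world normal.
# It takes the absolute normal, adds a small Z bias, and raises it to a power scaled by BlendFactor.
# Dividing by the component sum normalises the weights, which are saturated and output as AxisMask.
# An unconnected AxisBlend Custom node holds equivalent HLSL.
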MaterialFunction /BaseMaterial/Materials/Functions/MF_Blend_TriplanarAxis
graph {
 properties {bExposeToLibrary=true, libraryCategoriesText [Base Material]}
 nodes [
  Custom!(n, k) {
   code="// blend factors\r\nn = abs(n) + float3(0.f,0.f, 0.0033);\r\nfloat3 w  = pow( n, 8. * k );\r\nw /= dot( w, 1 );\r\nreturn saturate(w);"
   description=AxisBlend
   showCode=true
   inputs [{inputName=n, input {expression=null}}, {inputName=k, input {expression=null}}]
  }
  n1 = FunctionInput(Preview=VertexNormalWS()) {bUsePreviewValueAsDefault=true, inputName=RefNorm}
  n2 = FunctionInput() {bUsePreviewValueAsDefault=true, inputName=BlendFactor, inputType=FunctionInput_Scalar, sortPriority=1, previewValue=<0.5 0 0 1>}
  n3 = Power(Base=Add(A=Constant3Vector() {constant=<rgba 0 0 0.0033 0>}, B=Abs(None=@n1)), Exp=Multiply(A=@n2) {constB=8})
  n4 = DotProduct(A=@n3, B=Constant3Vector() {constant=<rgba 1 1 1 0>})
  FunctionOutput(None=Saturate(None=Divide(A=@n3, B=@n4))) {outputName=AxisMask}
 ]
}
