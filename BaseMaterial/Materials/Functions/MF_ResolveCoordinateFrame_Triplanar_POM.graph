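# Material function that builds triplanar coordinates with parallax occlusion mapping, projected in a selectable World, Local, Periodic or Preskinned space.
# Switches select position, normal and camera vector per space, feeding three per-axis MF_Coordinate_Triplanar calls and an MF_Blend_TriplanarAxis mask.
# MF_Coordinate_Parallax ray-marches the height texture on each axis, with view-angle-scaled steps.
# It outputs displaced UVs, depths, packed derivatives for SampleGrad, world-space tangent bases, rotation and blend weights.
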MaterialFunction /BaseMaterial/Materials/Functions/MF_ResolveCoordinateFrame_Triplanar_POM
graph {
 expressions {
  MaterialExpressionAdd_14 {
   class=MaterialExpressionAdd
   inputs [
    A {from=MaterialExpressionMaterialFunctionCall_2, output=Selected Vector}
    B {from=MaterialExpressionFunctionInput_13}
   ]
  }
  MaterialExpressionCameraVectorWS_3 {class=MaterialExpressionCameraVectorWS, inputs [], properties {}}
  MaterialExpressionConstant3Vector_2 {class=MaterialExpressionConstant3Vector, properties {constant=<rgba 0 0 1 0>}}
  MaterialExpressionConstant3Vector_3 {class=MaterialExpressionConstant3Vector, properties {constant=<rgba 0 1 0 0>}}
  MaterialExpressionConstant3Vector_4 {class=MaterialExpressionConstant3Vector, properties {constant=<rgba 1 0 0 0>}}
  MaterialExpressionConstant_0 {class=MaterialExpressionConstant}
  MaterialExpressionConstant_1 {class=MaterialExpressionConstant, properties {r=1}}
  MaterialExpressionConstant_2 {class=MaterialExpressionConstant, properties {r=2}}
  MaterialExpressionConstant_3 {class=MaterialExpressionConstant, properties {r=1}}
  MaterialExpressionConvert_19 {
   class=MaterialExpressionConvert
   inputs [
    None {from=MaterialExpressionLength_3}
    None {from=MaterialExpressionLength_2}
    None {from=MaterialExpressionLength_0}
   ]
   properties {
    desc=maintain scale
    convertInputs [
     {
      type=Scalar
      expressionInput {
       expression=&self:MaterialExpressionLength_3
       inputName=None
       mask=0
       maskA=0
       maskB=0
       maskG=0
       maskR=0
       outputIndex=0
      }
     }
     {
      type=Scalar
      expressionInput {
       expression=&self:MaterialExpressionLength_2
       inputName=None
       mask=0
       maskA=0
       maskB=0
       maskG=0
       maskR=0
       outputIndex=0
      }
     }
     {
      type=Scalar
      expressionInput {
       expression=&self:MaterialExpressionLength_0
       inputName=None
       mask=0
       maskA=0
       maskB=0
       maskG=0
       maskR=0
       outputIndex=0
      }
     }
    ]
    convertOutputs [{type=Vector3}]
   }
  }
  MaterialExpressionDivide_0 {
   class=MaterialExpressionDivide
   inputs [
    A {from=""}
    B {from=MaterialExpressionNamedRerouteUsage_3}
   ]
  }
  MaterialExpressionDivide_1 {
   class=MaterialExpressionDivide
   inputs [
    A {from=MaterialExpressionMaterialFunctionCall_0, output=Selected Vector}
    B {from=MaterialExpressionNamedRerouteUsage_28}
   ]
  }
  MaterialExpressionDivide_5 {
   class=MaterialExpressionDivide
   inputs [
    A {from=MaterialExpressionTransform_5}
    B {from=MaterialExpressionNamedRerouteUsage_89}
   ]
  }
  MaterialExpressionFunctionInput_0 {
   class=MaterialExpressionFunctionInput
   inputs [Preview {from="", output=""}]
   properties {
    bUsePreviewValueAsDefault=true
    description="Sets how many times the projected coordinates repeat along each axis. Use values above 1 for denser tiling, negative values for mirroring, and avoid zero."
    inputName=CoordinateRepeat
    sortPriority=-10
    previewValue=<0.01 0.01 0.01 1>
   }
  }
  MaterialExpressionFunctionInput_1 {
   class=MaterialExpressionFunctionInput
   inputs [Preview {from="", output=""}]
   properties {
    bUsePreviewValueAsDefault=true
    description="Rotates the projected UVs around the selected axis. Units are in turns, where 1.0 = 360° and 0.25 = 90°."
    inputName="Rotation (Turns)"
    inputType=FunctionInput_Scalar
    sortPriority=-8
   }
  }
  MaterialExpressionFunctionInput_11 {
   class=MaterialExpressionFunctionInput
   inputs [Preview {from="", output=""}]
   properties {
    bUsePreviewValueAsDefault=true
    description=Defines the reference plane for parallax. This value shifts where the heightmap appears anchored relative to the surface.
    inputName=POM RefPlane
    inputType=FunctionInput_Scalar
    sortPriority=14
    previewValue=<1 0 0 1>
   }
  }
  MaterialExpressionFunctionInput_13 {
   class=MaterialExpressionFunctionInput
   inputs [Preview {from="", output=""}]
   properties {
    bUsePreviewValueAsDefault=true
    description=Applies an additional per-axis offset to the coordinates before projection. Useful for correctly positioning a texture.
    inputName=CoordinateTransform
    sortPriority=-9
   }
  }
  MaterialExpressionFunctionInput_14 {
   class=MaterialExpressionFunctionInput
   inputs [Preview {from="", output=""}]
   properties {
    bUsePreviewValueAsDefault=true
    description="Defines the space used for projection: World (0) keeps textures fixed in world space, Local (1) moves with the object, Periodic (2) enforces a repeating pattern, and Preskinned (3) locks to preskinned positions for stable results on animated meshes."
    inputName=CoordinateSpace
    inputType=FunctionInput_Scalar
    sortPriority=-11
   }
  }
  MaterialExpressionFunctionInput_16 {
   class=MaterialExpressionFunctionInput
   inputs [Preview {from="", output=""}]
   properties {
    bUsePreviewValueAsDefault=true
    description="Controls the contrast of the blend mask. Internally applied as pow(mask, 8 × BlendFactor). A default of 0.5 corresponds to pow4 contrast."
    inputName=BlendTolerance
    inputType=FunctionInput_Scalar
    sortPriority=8
    previewValue=<0.5 0 0 1>
   }
  }
  MaterialExpressionFunctionInput_17 {
   class=MaterialExpressionFunctionInput
   inputs [Preview {from=MaterialExpressionTextureObject_3, output=""}]
   properties {
    bUsePreviewValueAsDefault=true
    description="The heightmap texture used for parallax occlusion mapping. Grayscale values define surface depth, with brighter values typically pushed outward and darker values pushed inward."
    inputName=HeightTex
    inputType=FunctionInput_Texture2D
    sortPriority=-20
   }
  }
  MaterialExpressionFunctionInput_2 {
   class=MaterialExpressionFunctionInput
   inputs [Preview {from="", output=""}]
   properties {
    bUsePreviewValueAsDefault=true
    description=Control the U/V Tiling of the height texture relative to the input coordinates.
    inputName=POM HeightTile
    inputType=FunctionInput_Vector2
    sortPriority=12
    previewValue=<1 1 0 1>
   }
  }
  MaterialExpressionFunctionInput_3 {
   class=MaterialExpressionFunctionInput
   inputs [Preview {from="", output=""}]
   properties {
    bUsePreviewValueAsDefault=true
    description="Specifies which channel of the heightmap texture to use for depth information. Assumes the Alpha channel is used, but any channel can be selected if your texture packs data differently."
    inputName=POM HeightChannel
    inputType=FunctionInput_Scalar
    sortPriority=10
    previewValue=<3 0 0 1>
   }
  }
  MaterialExpressionFunctionInput_4 {
   class=MaterialExpressionFunctionInput
   inputs [Preview {from="", output=""}]
   properties {
    bUsePreviewValueAsDefault=true
    description="Sets the maximum number of depth sampling steps. This is the upper bound, with actual steps scaled by the Quality setting."
    inputName=POM MaxSteps
    inputType=FunctionInput_Scalar
    sortPriority=12
    previewValue=<16 0 0 1>
   }
  }
  MaterialExpressionFunctionInput_5 {
   class=MaterialExpressionFunctionInput
   inputs [Preview {from="", output=""}]
   properties {
    bUsePreviewValueAsDefault=true
    description="Controls the strength of the parallax effect. Higher values exaggerate depth; lower values make surfaces appear flatter."
    inputName=POM Intensity
    inputType=FunctionInput_Scalar
    sortPriority=11
    previewValue=<1 0 0 1>
   }
  }
  MaterialExpressionFunctionInput_6 {
   class=MaterialExpressionFunctionInput
   inputs [Preview {from="", output=""}]
   properties {
    bUsePreviewValueAsDefault=true
    description="Scales step count by view angle. At 1, MaxSteps are always used; at 0, only one step is used when looking straight on, with more steps added at grazing angles."
    inputName=POM Quality
    inputType=FunctionInput_Scalar
    sortPriority=13
    previewValue=<0.25 0 0 1>
   }
  }
  MaterialExpressionFunctionOutput_0 {
   class=MaterialExpressionFunctionOutput
   inputs [None {from=MaterialExpressionMaterialFunctionCall_16, output=Pixel Depth Offset}]
   properties {
    description="Depth at the POM intersection (after Intensity/RefPlane). Higher (whiter) = deeper—the farther the ray travels into the material, the larger the value."
    outputName=xDepth
    sortPriority=4
   }
  }
  MaterialExpressionFunctionOutput_1 {
   class=MaterialExpressionFunctionOutput
   inputs [None {from=MaterialExpressionMaterialFunctionCall_16, output=POM UVs}]
   properties {description="UV coordinates projected along the X axis.\r\n", outputName=xUV}
  }
  MaterialExpressionFunctionOutput_10 {
   class=MaterialExpressionFunctionOutput
   inputs [None {from=MaterialExpressionMaterialFunctionCall_7, output=Selected Vector}]
   properties {description="World-space tangent vector for yUV.\r\n", outputName=yTangent, sortPriority=6}
  }
  MaterialExpressionFunctionOutput_11 {
   class=MaterialExpressionFunctionOutput
   inputs [None {from=MaterialExpressionMaterialFunctionCall_5, output=Selected Vector}]
   properties {description="World-space bitangent vector for yUV.\r\n", outputName=yBitangent, sortPriority=7}
  }
  MaterialExpressionFunctionOutput_12 {
   class=MaterialExpressionFunctionOutput
   inputs [None {from=MaterialExpressionMaterialFunctionCall_8, output=Selected Vector}]
   properties {description="World-space tangent vector for zUV.\r\n", outputName=zTangent, sortPriority=11}
  }
  MaterialExpressionFunctionOutput_13 {
   class=MaterialExpressionFunctionOutput
   inputs [None {from=MaterialExpressionMaterialFunctionCall_6, output=Selected Vector}]
   properties {description="World-space bitangent vector for zUV.\r\n", outputName=zBitangent, sortPriority=12}
  }
  MaterialExpressionFunctionOutput_14 {
   class=MaterialExpressionFunctionOutput
   inputs [None {from=MaterialExpressionFunctionInput_1, output=""}]
   properties {
    description="The rotation applied around the projection axis, expressed in turns (1.0 = 360°). Used along with Tangent/Bitangent for reorienting normal maps via MF_TangentNormal_OrientWorld.\r\n"
    outputName="Rotation (Turns)"
    sortPriority=-5
   }
  }
  MaterialExpressionFunctionOutput_15 {
   class=MaterialExpressionFunctionOutput
   inputs [None {from=MaterialExpressionMaterialFunctionCall_9, output=Pixel Depth Offset}]
   properties {
    description="Depth at the POM intersection (after Intensity/RefPlane). Higher (whiter) = deeper—the farther the ray travels into the material, the larger the value."
    outputName=yDepth
    sortPriority=9
   }
  }
  MaterialExpressionFunctionOutput_16 {
   class=MaterialExpressionFunctionOutput
   inputs [None {from=MaterialExpressionMaterialFunctionCall_10, output=Pixel Depth Offset}]
   properties {
    description="Depth at the POM intersection (after Intensity/RefPlane). Higher (whiter) = deeper—the farther the ray travels into the material, the larger the value."
    outputName=zDepth
    sortPriority=14
   }
  }
  MaterialExpressionFunctionOutput_2 {
   class=MaterialExpressionFunctionOutput
   inputs [None {from=MaterialExpressionMaterialFunctionCall_19, output=packedDXY}]
   properties {description="Packed derivatives (ddx/ddy) for xUV, used with SampleGrad.\r\n", outputName=xDXY, sortPriority=3}
  }
  MaterialExpressionFunctionOutput_3 {
   class=MaterialExpressionFunctionOutput
   inputs [None {from=MaterialExpressionMaterialFunctionCall_9, output=POM UVs}]
   properties {description="UV coordinates projected along the Y axis.\r\n", outputName=yUV, sortPriority=5}
  }
  MaterialExpressionFunctionOutput_4 {
   class=MaterialExpressionFunctionOutput
   inputs [None {from=MaterialExpressionMaterialFunctionCall_18, output=packedDXY}]
   properties {description="Packed derivatives (ddx/ddy) for yUV, used with SampleGrad.\r\n", outputName=yDXY, sortPriority=8}
  }
  MaterialExpressionFunctionOutput_5 {
   class=MaterialExpressionFunctionOutput
   inputs [None {from=MaterialExpressionMaterialFunctionCall_10, output=POM UVs}]
   properties {description="UV coordinates projected along the Z axis.\r\n", outputName=zUV, sortPriority=10}
  }
  MaterialExpressionFunctionOutput_6 {
   class=MaterialExpressionFunctionOutput
   inputs [None {from=MaterialExpressionMaterialFunctionCall_14, output=packedDXY}]
   properties {description="Packed derivatives (ddx/ddy) for zUV, used with SampleGrad.\r\n", outputName=zDXY, sortPriority=13}
  }
  MaterialExpressionFunctionOutput_7 {
   class=MaterialExpressionFunctionOutput
   inputs [None {from=MaterialExpressionMaterialFunctionCall_15, output=AxisMask}]
   properties {
    description="RGB weights for blending between the X, Y, and Z axis projections. These weights can also be modulated by albedo or height for smoother, perceptually correct transitions.\r\n"
    outputName=BlendMask
    sortPriority=15
   }
  }
  MaterialExpressionFunctionOutput_8 {
   class=MaterialExpressionFunctionOutput
   inputs [None {from=MaterialExpressionMaterialFunctionCall_3, output=Selected Vector}]
   properties {description="World-space tangent vector for xUV.\r\n", outputName=xTangent, sortPriority=1}
  }
  MaterialExpressionFunctionOutput_9 {
   class=MaterialExpressionFunctionOutput
   inputs [None {from=MaterialExpressionMaterialFunctionCall_4, output=Selected Vector}]
   properties {description="World-space bitangent vector for xUV.\r\n", outputName=xBitangent, sortPriority=2}
  }
  MaterialExpressionLength_0 {class=MaterialExpressionLength, inputs [None {from=MaterialExpressionTransform_0, output=""}], properties {}}
  MaterialExpressionLength_2 {class=MaterialExpressionLength, inputs [None {from=MaterialExpressionTransform_12, output=""}], properties {}}
  MaterialExpressionLength_3 {class=MaterialExpressionLength, inputs [None {from=MaterialExpressionTransform_13, output=""}], properties {}}
  MaterialExpressionLocalPosition_0 {class=MaterialExpressionLocalPosition, inputs [], properties {includedOffsets=ExcludeOffsets, localOrigin=InstancePreSkinning}}
  MaterialExpressionLocalPosition_2 {class=MaterialExpressionLocalPosition, inputs [], properties {}}
  MaterialExpressionMaterialFunctionCall_0 {
   class=MaterialExpressionMaterialFunctionCall
   inputs [
    Selector {from=MaterialExpressionNamedRerouteUsage_38}
    VectorA {from=MaterialExpressionReroute_14}
    VectorB {from=MaterialExpressionNormalize_0}
    VectorC {from=MaterialExpressionReroute_14}
    VectorD {from=MaterialExpressionShaderStageSwitch_0}
   ]
   properties {materialFunction=&/MF_Switch4_Vec3}
  }
  MaterialExpressionMaterialFunctionCall_1 {
   class=MaterialExpressionMaterialFunctionCall
   inputs [PeriodXYZ {from=MaterialExpressionDivide_0}]
   properties {materialFunction=&/MF_PrepareLWC}
  }
  MaterialExpressionMaterialFunctionCall_10 {
   class=MaterialExpressionMaterialFunctionCall
   inputs [
    POM_HeightTexture {from=MaterialExpressionReroute_9}
    Coordinates {from=MaterialExpressionNamedRerouteUsage_54}
    coordinateDXY {from=MaterialExpressionNamedRerouteUsage_73}
    HeightStrength {from=MaterialExpressionNamedRerouteUsage_74}
    Reference Plane {from=MaterialExpressionNamedRerouteUsage_76}
    MaxSteps {from=MaterialExpressionNamedRerouteUsage_75}
    Tangent {from=MaterialExpressionNamedRerouteUsage_77}
    Bitangent {from=MaterialExpressionNamedRerouteUsage_78}
    "Rotation (Turns)" {from=MaterialExpressionReroute_20}
    HeightChannel {from=MaterialExpressionReroute_8}
    QualityBias {from=MaterialExpressionReroute_21}
    CameraVector {from=MaterialExpressionNamedRerouteUsage_83}
    Height Tile {from=MaterialExpressionNamedRerouteUsage_2}
   ]
   properties {materialFunction=&/MF_Coordinate_Parallax}
  }
  MaterialExpressionMaterialFunctionCall_11 {
   class=MaterialExpressionMaterialFunctionCall
   inputs [
    Selector {from=MaterialExpressionNamedRerouteUsage_81}
    VectorA {from=MaterialExpressionReroute_1}
    VectorB {from=MaterialExpressionTransform_3}
    VectorC {from=MaterialExpressionReroute_1}
    VectorD {from=MaterialExpressionTransform_3}
   ]
   properties {materialFunction=&MF_Switch4_Vec3}
  }
  MaterialExpressionMaterialFunctionCall_14 {
   class=MaterialExpressionMaterialFunctionCall
   inputs [
    Position {from=MaterialExpressionNamedRerouteUsage_40}
    Tiling {from=MaterialExpressionReroute_10}
    RefNorm {from=MaterialExpressionNamedRerouteUsage_36}
    "Rotation (Turns)" {from=MaterialExpressionNamedRerouteDeclaration_11}
    AxisIndex {from=MaterialExpressionConstant_2}
   ]
   properties {materialFunction=&/MF_Coordinate_Triplanar}
  }
  MaterialExpressionMaterialFunctionCall_15 {
   class=MaterialExpressionMaterialFunctionCall
   inputs [
    RefNorm {from=MaterialExpressionDivide_1}
    BlendFactor {from=MaterialExpressionFunctionInput_16}
   ]
   properties {materialFunction=&/MF_Blend_TriplanarAxis}
  }
  MaterialExpressionMaterialFunctionCall_16 {
   class=MaterialExpressionMaterialFunctionCall
   inputs [
    POM_HeightTexture {from=MaterialExpressionReroute_4}
    Coordinates {from=MaterialExpressionNamedRerouteUsage_39}
    coordinateDXY {from=MaterialExpressionNamedRerouteUsage_58}
    HeightStrength {from=MaterialExpressionNamedRerouteUsage_60}
    Reference Plane {from=MaterialExpressionNamedRerouteUsage_64}
    MaxSteps {from=MaterialExpressionNamedRerouteUsage_61}
    Tangent {from=MaterialExpressionNamedRerouteUsage_55}
    Bitangent {from=MaterialExpressionNamedRerouteUsage_56}
    "Rotation (Turns)" {from=MaterialExpressionNamedRerouteUsage_59}
    HeightChannel {from=MaterialExpressionReroute_6}
    QualityBias {from=MaterialExpressionNamedRerouteUsage_62}
    CameraVector {from=MaterialExpressionNamedRerouteUsage_85}
    Height Tile {from=MaterialExpressionNamedRerouteUsage_5}
   ]
   properties {materialFunction=&MF_Coordinate_Parallax}
  }
  MaterialExpressionMaterialFunctionCall_18 {
   class=MaterialExpressionMaterialFunctionCall
   inputs [
    Position {from=MaterialExpressionNamedRerouteUsage_40}
    Tiling {from=MaterialExpressionReroute_10}
    RefNorm {from=MaterialExpressionNamedRerouteUsage_36}
    "Rotation (Turns)" {from=MaterialExpressionNamedRerouteDeclaration_11}
    AxisIndex {from=MaterialExpressionConstant_1}
   ]
   properties {materialFunction=&MF_Coordinate_Triplanar}
  }
  MaterialExpressionMaterialFunctionCall_19 {
   class=MaterialExpressionMaterialFunctionCall
   inputs [
    Position {from=MaterialExpressionNamedRerouteUsage_40}
    Tiling {from=MaterialExpressionReroute_10}
    RefNorm {from=MaterialExpressionNamedRerouteUsage_36}
    "Rotation (Turns)" {from=MaterialExpressionNamedRerouteDeclaration_11}
    AxisIndex {from=MaterialExpressionConstant_0}
   ]
   properties {materialFunction=&MF_Coordinate_Triplanar}
  }
  MaterialExpressionMaterialFunctionCall_2 {
   class=MaterialExpressionMaterialFunctionCall
   inputs [
    Selector {from=MaterialExpressionFunctionInput_14}
    VectorA {from=MaterialExpressionReroute_16}
    VectorB {from=MaterialExpressionMultiply_0}
    VectorC {from=MaterialExpressionReroute_18}
    VectorD {from=MaterialExpressionMultiply_1}
   ]
   properties {materialFunction=&MF_Switch4_Vec3}
  }
  MaterialExpressionMaterialFunctionCall_3 {
   class=MaterialExpressionMaterialFunctionCall
   inputs [
    Selector {from=MaterialExpressionNamedRerouteUsage_42}
    VectorA {from=MaterialExpressionReroute_0}
    VectorB {from=MaterialExpressionTransform_4}
    VectorC {from=MaterialExpressionReroute_0}
    VectorD {from=MaterialExpressionTransform_6}
   ]
   properties {materialFunction=&MF_Switch4_Vec3}
  }
  MaterialExpressionMaterialFunctionCall_4 {
   class=MaterialExpressionMaterialFunctionCall
   inputs [
    Selector {from=MaterialExpressionNamedRerouteUsage_43}
    VectorA {from=MaterialExpressionReroute_2}
    VectorB {from=MaterialExpressionTransform_7}
    VectorC {from=MaterialExpressionReroute_2}
    VectorD {from=MaterialExpressionTransform_8}
   ]
   properties {materialFunction=&MF_Switch4_Vec3}
  }
  MaterialExpressionMaterialFunctionCall_5 {
   class=MaterialExpressionMaterialFunctionCall
   inputs [
    Selector {from=MaterialExpressionNamedRerouteUsage_45}
    VectorA {from=MaterialExpressionReroute_3}
    VectorB {from=MaterialExpressionTransform_11}
    VectorC {from=MaterialExpressionReroute_3}
    VectorD {from=MaterialExpressionTransform_18}
   ]
   properties {materialFunction=&MF_Switch4_Vec3}
  }
  MaterialExpressionMaterialFunctionCall_6 {
   class=MaterialExpressionMaterialFunctionCall
   inputs [
    Selector {from=MaterialExpressionNamedRerouteUsage_47}
    VectorA {from=MaterialExpressionReroute_13}
    VectorB {from=MaterialExpressionTransform_21}
    VectorC {from=MaterialExpressionReroute_13}
    VectorD {from=MaterialExpressionTransform_22}
   ]
   properties {materialFunction=&MF_Switch4_Vec3}
  }
  MaterialExpressionMaterialFunctionCall_7 {
   class=MaterialExpressionMaterialFunctionCall
   inputs [
    Selector {from=MaterialExpressionNamedRerouteUsage_44}
    VectorA {from=MaterialExpressionReroute_11}
    VectorB {from=MaterialExpressionTransform_9}
    VectorC {from=MaterialExpressionReroute_11}
    VectorD {from=MaterialExpressionTransform_10}
   ]
   properties {materialFunction=&MF_Switch4_Vec3}
  }
  MaterialExpressionMaterialFunctionCall_8 {
   class=MaterialExpressionMaterialFunctionCall
   inputs [
    Selector {from=MaterialExpressionNamedRerouteUsage_46}
    VectorA {from=MaterialExpressionReroute_12}
    VectorB {from=MaterialExpressionTransform_19}
    VectorC {from=MaterialExpressionReroute_12}
    VectorD {from=MaterialExpressionTransform_20}
   ]
   properties {materialFunction=&MF_Switch4_Vec3}
  }
  MaterialExpressionMaterialFunctionCall_9 {
   class=MaterialExpressionMaterialFunctionCall
   inputs [
    POM_HeightTexture {from=MaterialExpressionReroute_7}
    Coordinates {from=MaterialExpressionNamedRerouteUsage_41}
    coordinateDXY {from=MaterialExpressionNamedRerouteUsage_65}
    HeightStrength {from=MaterialExpressionNamedRerouteUsage_66}
    Reference Plane {from=MaterialExpressionNamedRerouteUsage_68}
    MaxSteps {from=MaterialExpressionNamedRerouteUsage_67}
    Tangent {from=MaterialExpressionNamedRerouteUsage_69}
    Bitangent {from=MaterialExpressionNamedRerouteUsage_70}
    "Rotation (Turns)" {from=MaterialExpressionNamedRerouteUsage_71}
    HeightChannel {from=MaterialExpressionReroute_5}
    QualityBias {from=MaterialExpressionNamedRerouteUsage_72}
    CameraVector {from=MaterialExpressionNamedRerouteUsage_84}
    Height Tile {from=MaterialExpressionNamedRerouteUsage_4}
   ]
   properties {materialFunction=&MF_Coordinate_Parallax}
  }
  MaterialExpressionMultiply_0 {
   class=MaterialExpressionMultiply
   inputs [
    A {from=MaterialExpressionNamedRerouteUsage_0}
    B {from=MaterialExpressionLocalPosition_2, output=XYZ}
   ]
  }
  MaterialExpressionMultiply_1 {
   class=MaterialExpressionMultiply
   inputs [
    A {from=MaterialExpressionNamedRerouteUsage_1}
    B {from=MaterialExpressionShaderStageSwitch_2}
   ]
  }
  MaterialExpressionNamedRerouteDeclaration_0 {
   class=MaterialExpressionNamedRerouteDeclaration
   inputs [None {from=MaterialExpressionFunctionOutput_1}]
   properties {name=xUV, nodeColor=<rgba 0.9 0.009 0.009 1>}
  }
  MaterialExpressionNamedRerouteDeclaration_1 {
   class=MaterialExpressionNamedRerouteDeclaration
   inputs [None {from=MaterialExpressionFunctionOutput_3}]
   properties {name=yUV, nodeColor=<rgba 0.009 0.9 0.009 1>}
  }
  MaterialExpressionNamedRerouteDeclaration_10 {
   class=MaterialExpressionNamedRerouteDeclaration
   inputs [None {from=MaterialExpressionMaterialFunctionCall_15, output=AxisMask}]
   properties {name=Blendi, nodeColor=<rgba 1 1 1 1>}
  }
  MaterialExpressionNamedRerouteDeclaration_11 {
   class=MaterialExpressionNamedRerouteDeclaration
   inputs [None {from=MaterialExpressionFunctionOutput_14}]
   properties {name=turns, nodeColor=<rgba 0.8963 0.8963 0.0091 1>}
  }
  MaterialExpressionNamedRerouteDeclaration_12 {
   class=MaterialExpressionNamedRerouteDeclaration
   inputs [None {from=MaterialExpressionMaterialFunctionCall_19, output=UV}]
   properties {name=xUVi, nodeColor=<rgba 0.9 0.009 0.009 1>}
  }
  MaterialExpressionNamedRerouteDeclaration_13 {
   class=MaterialExpressionNamedRerouteDeclaration
   inputs [None {from=MaterialExpressionFunctionOutput_9}]
   properties {name=xB, nodeColor=<rgba 0.9 0.009 0.009 1>}
  }
  MaterialExpressionNamedRerouteDeclaration_14 {
   class=MaterialExpressionNamedRerouteDeclaration
   inputs [None {from=MaterialExpressionFunctionOutput_10}]
   properties {name=yT, nodeColor=<rgba 0.009 0.9 0.009 1>}
  }
  MaterialExpressionNamedRerouteDeclaration_15 {
   class=MaterialExpressionNamedRerouteDeclaration
   inputs [None {from=MaterialExpressionFunctionOutput_11}]
   properties {name=yB, nodeColor=<rgba 0.009 0.9 0.009 1>}
  }
  MaterialExpressionNamedRerouteDeclaration_16 {
   class=MaterialExpressionNamedRerouteDeclaration
   inputs [None {from=MaterialExpressionFunctionOutput_12}]
   properties {name=zT, nodeColor=<rgba 0.009 0.009 0.9 1>}
  }
  MaterialExpressionNamedRerouteDeclaration_17 {
   class=MaterialExpressionNamedRerouteDeclaration
   inputs [None {from=MaterialExpressionFunctionOutput_13}]
   properties {name=zB, nodeColor=<rgba 0.009 0.009 0.9 1>}
  }
  MaterialExpressionNamedRerouteDeclaration_18 {
   class=MaterialExpressionNamedRerouteDeclaration
   inputs [None {from=MaterialExpressionDivide_1}]
   properties {name=RefNorm, nodeColor=<rgba 0.8963 0.8963 0.0091 1>}
  }
  MaterialExpressionNamedRerouteDeclaration_19 {
   class=MaterialExpressionNamedRerouteDeclaration
   inputs [None {from=MaterialExpressionAdd_14}]
   properties {name=RefPos, nodeColor=<rgba 0.8963 0.8963 0.0091 1>}
  }
  MaterialExpressionNamedRerouteDeclaration_2 {
   class=MaterialExpressionNamedRerouteDeclaration
   inputs [None {from=MaterialExpressionFunctionOutput_5}]
   properties {name=zUV, nodeColor=<rgba 0.009 0.009 0.9 1>}
  }
  MaterialExpressionNamedRerouteDeclaration_20 {
   class=MaterialExpressionNamedRerouteDeclaration
   inputs [None {from=MaterialExpressionFunctionInput_14}]
   properties {name=Selector, nodeColor=<rgba 0 0.3437 1 1>}
  }
  MaterialExpressionNamedRerouteDeclaration_21 {
   class=MaterialExpressionNamedRerouteDeclaration
   inputs [None {from=MaterialExpressionMaterialFunctionCall_19, output=Tangent}]
   properties {name=xTi, nodeColor=<rgba 0.9 0.009 0.009 1>}
  }
  MaterialExpressionNamedRerouteDeclaration_22 {
   class=MaterialExpressionNamedRerouteDeclaration
   inputs [None {from=MaterialExpressionMaterialFunctionCall_19, output=Bitangent}]
   properties {name=xBi, nodeColor=<rgba 0.9 0.009 0.009 1>}
  }
  MaterialExpressionNamedRerouteDeclaration_23 {
   class=MaterialExpressionNamedRerouteDeclaration
   inputs [None {from=MaterialExpressionMaterialFunctionCall_18, output=UV}]
   properties {name=yUVi, nodeColor=<rgba 0.009 0.9 0.009 1>}
  }
  MaterialExpressionNamedRerouteDeclaration_24 {
   class=MaterialExpressionNamedRerouteDeclaration
   inputs [None {from=MaterialExpressionMaterialFunctionCall_18, output=Tangent}]
   properties {name=yTi, nodeColor=<rgba 0.009 0.9 0.009 1>}
  }
  MaterialExpressionNamedRerouteDeclaration_25 {
   class=MaterialExpressionNamedRerouteDeclaration
   inputs [None {from=MaterialExpressionMaterialFunctionCall_18, output=Bitangent}]
   properties {name=yBi, nodeColor=<rgba 0.009 0.9 0.009 1>}
  }
  MaterialExpressionNamedRerouteDeclaration_26 {
   class=MaterialExpressionNamedRerouteDeclaration
   inputs [None {from=MaterialExpressionMaterialFunctionCall_14, output=Tangent}]
   properties {name=zTi, nodeColor=<rgba 0.009 0.009 0.9 1>}
  }
  MaterialExpressionNamedRerouteDeclaration_27 {
   class=MaterialExpressionNamedRerouteDeclaration
   inputs [None {from=MaterialExpressionMaterialFunctionCall_14, output=Bitangent}]
   properties {name=zBi, nodeColor=<rgba 0.009 0.009 0.9 1>}
  }
  MaterialExpressionNamedRerouteDeclaration_28 {
   class=MaterialExpressionNamedRerouteDeclaration
   inputs [None {from=MaterialExpressionMaterialFunctionCall_14, output=UV}]
   properties {name=zUVi, nodeColor=<rgba 0.009 0.009 0.9 1>}
  }
  MaterialExpressionNamedRerouteDeclaration_3 {
   class=MaterialExpressionNamedRerouteDeclaration
   inputs [None {from=MaterialExpressionFunctionInput_0}]
   properties {name=Repeat, nodeColor=<rgba 0.7812 0 1 1>}
  }
  MaterialExpressionNamedRerouteDeclaration_30 {
   class=MaterialExpressionNamedRerouteDeclaration
   inputs [None {from=MaterialExpressionConstant_3}]
   properties {name=MeshVecScale, nodeColor=<rgba 0.6641 0 1 1>}
  }
  MaterialExpressionNamedRerouteDeclaration_33 {
   class=MaterialExpressionNamedRerouteDeclaration
   inputs [None {from=MaterialExpressionFunctionInput_5}]
   properties {name=POM_Height, nodeColor=<rgba 0.9 0.306 0.009 1>}
  }
  MaterialExpressionNamedRerouteDeclaration_34 {
   class=MaterialExpressionNamedRerouteDeclaration
   inputs [None {from=MaterialExpressionFunctionInput_4}]
   properties {name=POM_MaxSteps, nodeColor=<rgba 0.9 0.306 0.009 1>}
  }
  MaterialExpressionNamedRerouteDeclaration_35 {
   class=MaterialExpressionNamedRerouteDeclaration
   inputs [None {from=MaterialExpressionFunctionInput_6}]
   properties {name=POM_Noise, nodeColor=<rgba 0.9 0.306 0.009 1>}
  }
  MaterialExpressionNamedRerouteDeclaration_36 {
   class=MaterialExpressionNamedRerouteDeclaration
   inputs [None {from=MaterialExpressionFunctionInput_11}]
   properties {name=POM_RefPlane, nodeColor=<rgba 0.9 0.306 0.009 1>}
  }
  MaterialExpressionNamedRerouteDeclaration_37 {
   class=MaterialExpressionNamedRerouteDeclaration
   inputs [None {from=MaterialExpressionMaterialFunctionCall_11, output=Selected Vector}]
   properties {name=POM_CameraVec, nodeColor=<rgba 0.9 0.306 0.009 1>}
  }
  MaterialExpressionNamedRerouteDeclaration_4 {
   class=MaterialExpressionNamedRerouteDeclaration
   inputs [None {from=MaterialExpressionFunctionInput_2}]
   properties {name=POM_HeightTile, nodeColor=<rgba 0.9 0.306 0.009 1>}
  }
  MaterialExpressionNamedRerouteDeclaration_6 {
   class=MaterialExpressionNamedRerouteDeclaration
   inputs [None {from=MaterialExpressionFunctionOutput_8}]
   properties {name=xT, nodeColor=<rgba 0.9 0.009 0.009 1>}
  }
  MaterialExpressionNamedRerouteDeclaration_7 {
   class=MaterialExpressionNamedRerouteDeclaration
   inputs [None {from=MaterialExpressionFunctionOutput_2}]
   properties {name=xDXY, nodeColor=<rgba 0.9 0.009 0.009 1>}
  }
  MaterialExpressionNamedRerouteDeclaration_8 {
   class=MaterialExpressionNamedRerouteDeclaration
   inputs [None {from=MaterialExpressionFunctionOutput_4}]
   properties {name=yDXY, nodeColor=<rgba 0.009 0.9 0.009 1>}
  }
  MaterialExpressionNamedRerouteDeclaration_9 {
   class=MaterialExpressionNamedRerouteDeclaration
   inputs [None {from=MaterialExpressionFunctionOutput_6}]
   properties {name=zDXY, nodeColor=<rgba 0.009 0.009 0.9 1>}
  }
  MaterialExpressionNamedRerouteUsage_0 {class=MaterialExpressionNamedRerouteUsage}
  MaterialExpressionNamedRerouteUsage_1 {class=MaterialExpressionNamedRerouteUsage}
  MaterialExpressionNamedRerouteUsage_2 {class=MaterialExpressionNamedRerouteUsage}
  MaterialExpressionNamedRerouteUsage_28 {class=MaterialExpressionNamedRerouteUsage}
  MaterialExpressionNamedRerouteUsage_3 {class=MaterialExpressionNamedRerouteUsage}
  MaterialExpressionNamedRerouteUsage_36 {class=MaterialExpressionNamedRerouteUsage}
  MaterialExpressionNamedRerouteUsage_38 {class=MaterialExpressionNamedRerouteUsage}
  MaterialExpressionNamedRerouteUsage_39 {class=MaterialExpressionNamedRerouteUsage}
  MaterialExpressionNamedRerouteUsage_4 {class=MaterialExpressionNamedRerouteUsage}
  MaterialExpressionNamedRerouteUsage_40 {class=MaterialExpressionNamedRerouteUsage}
  MaterialExpressionNamedRerouteUsage_41 {class=MaterialExpressionNamedRerouteUsage}
  MaterialExpressionNamedRerouteUsage_42 {class=MaterialExpressionNamedRerouteUsage}
  MaterialExpressionNamedRerouteUsage_43 {class=MaterialExpressionNamedRerouteUsage}
  MaterialExpressionNamedRerouteUsage_44 {class=MaterialExpressionNamedRerouteUsage}
  MaterialExpressionNamedRerouteUsage_45 {class=MaterialExpressionNamedRerouteUsage}
  MaterialExpressionNamedRerouteUsage_46 {class=MaterialExpressionNamedRerouteUsage}
  MaterialExpressionNamedRerouteUsage_47 {class=MaterialExpressionNamedRerouteUsage}
  MaterialExpressionNamedRerouteUsage_48 {class=MaterialExpressionNamedRerouteUsage}
  MaterialExpressionNamedRerouteUsage_49 {class=MaterialExpressionNamedRerouteUsage}
  MaterialExpressionNamedRerouteUsage_5 {class=MaterialExpressionNamedRerouteUsage}
  MaterialExpressionNamedRerouteUsage_50 {class=MaterialExpressionNamedRerouteUsage}
  MaterialExpressionNamedRerouteUsage_51 {class=MaterialExpressionNamedRerouteUsage}
  MaterialExpressionNamedRerouteUsage_52 {class=MaterialExpressionNamedRerouteUsage}
  MaterialExpressionNamedRerouteUsage_53 {class=MaterialExpressionNamedRerouteUsage}
  MaterialExpressionNamedRerouteUsage_54 {class=MaterialExpressionNamedRerouteUsage}
  MaterialExpressionNamedRerouteUsage_55 {class=MaterialExpressionNamedRerouteUsage}
  MaterialExpressionNamedRerouteUsage_56 {class=MaterialExpressionNamedRerouteUsage}
  MaterialExpressionNamedRerouteUsage_58 {class=MaterialExpressionNamedRerouteUsage}
  MaterialExpressionNamedRerouteUsage_59 {class=MaterialExpressionNamedRerouteUsage}
  MaterialExpressionNamedRerouteUsage_60 {class=MaterialExpressionNamedRerouteUsage}
  MaterialExpressionNamedRerouteUsage_61 {class=MaterialExpressionNamedRerouteUsage}
  MaterialExpressionNamedRerouteUsage_62 {class=MaterialExpressionNamedRerouteUsage}
  MaterialExpressionNamedRerouteUsage_64 {class=MaterialExpressionNamedRerouteUsage}
  MaterialExpressionNamedRerouteUsage_65 {class=MaterialExpressionNamedRerouteUsage}
  MaterialExpressionNamedRerouteUsage_66 {class=MaterialExpressionNamedRerouteUsage}
  MaterialExpressionNamedRerouteUsage_67 {class=MaterialExpressionNamedRerouteUsage}
  MaterialExpressionNamedRerouteUsage_68 {class=MaterialExpressionNamedRerouteUsage}
  MaterialExpressionNamedRerouteUsage_69 {class=MaterialExpressionNamedRerouteUsage}
  MaterialExpressionNamedRerouteUsage_70 {class=MaterialExpressionNamedRerouteUsage}
  MaterialExpressionNamedRerouteUsage_71 {class=MaterialExpressionNamedRerouteUsage}
  MaterialExpressionNamedRerouteUsage_72 {class=MaterialExpressionNamedRerouteUsage}
  MaterialExpressionNamedRerouteUsage_73 {class=MaterialExpressionNamedRerouteUsage}
  MaterialExpressionNamedRerouteUsage_74 {class=MaterialExpressionNamedRerouteUsage}
  MaterialExpressionNamedRerouteUsage_75 {class=MaterialExpressionNamedRerouteUsage}
  MaterialExpressionNamedRerouteUsage_76 {class=MaterialExpressionNamedRerouteUsage}
  MaterialExpressionNamedRerouteUsage_77 {class=MaterialExpressionNamedRerouteUsage}
  MaterialExpressionNamedRerouteUsage_78 {class=MaterialExpressionNamedRerouteUsage}
  MaterialExpressionNamedRerouteUsage_79 {class=MaterialExpressionNamedRerouteUsage}
  MaterialExpressionNamedRerouteUsage_80 {class=MaterialExpressionNamedRerouteUsage}
  MaterialExpressionNamedRerouteUsage_81 {class=MaterialExpressionNamedRerouteUsage}
  MaterialExpressionNamedRerouteUsage_83 {class=MaterialExpressionNamedRerouteUsage}
  MaterialExpressionNamedRerouteUsage_84 {class=MaterialExpressionNamedRerouteUsage}
  MaterialExpressionNamedRerouteUsage_85 {class=MaterialExpressionNamedRerouteUsage}
  MaterialExpressionNamedRerouteUsage_89 {class=MaterialExpressionNamedRerouteUsage}
  MaterialExpressionNormalize_0 {class=MaterialExpressionNormalize, inputs [VectorInput {from=MaterialExpressionTransform_14, output=""}]}
  MaterialExpressionNormalize_2 {class=MaterialExpressionNormalize, inputs [VectorInput {from=MaterialExpressionDivide_5, output=""}]}
  MaterialExpressionPreSkinnedNormal_0 {class=MaterialExpressionPreSkinnedNormal, inputs [], properties {}}
  MaterialExpressionReroute_0 {class=MaterialExpressionReroute, inputs [None {from=MaterialExpressionNamedRerouteUsage_48}]}
  MaterialExpressionReroute_1 {class=MaterialExpressionReroute, inputs [None {from=MaterialExpressionCameraVectorWS_3}]}
  MaterialExpressionReroute_10 {class=MaterialExpressionReroute, inputs [None {from=MaterialExpressionNamedRerouteDeclaration_3}]}
  MaterialExpressionReroute_11 {class=MaterialExpressionReroute, inputs [None {from=MaterialExpressionNamedRerouteUsage_50}]}
  MaterialExpressionReroute_12 {class=MaterialExpressionReroute, inputs [None {from=MaterialExpressionNamedRerouteUsage_52}]}
  MaterialExpressionReroute_13 {class=MaterialExpressionReroute, inputs [None {from=MaterialExpressionNamedRerouteUsage_53}]}
  MaterialExpressionReroute_14 {class=MaterialExpressionReroute, inputs [None {from=MaterialExpressionVertexNormalWS_1}]}
  MaterialExpressionReroute_16 {class=MaterialExpressionReroute, inputs [None {from=MaterialExpressionMaterialFunctionCall_1, output=Per-Axis Period}]}
  MaterialExpressionReroute_17 {class=MaterialExpressionReroute, inputs [None {from=MaterialExpressionFunctionInput_17}]}
  MaterialExpressionReroute_18 {class=MaterialExpressionReroute, inputs [None {from=MaterialExpressionReroute_19}]}
  MaterialExpressionReroute_19 {class=MaterialExpressionReroute, inputs [None {from=MaterialExpressionMaterialFunctionCall_1, output=Max Axis Period}]}
  MaterialExpressionReroute_2 {class=MaterialExpressionReroute, inputs [None {from=MaterialExpressionNamedRerouteUsage_49}]}
  MaterialExpressionReroute_20 {class=MaterialExpressionReroute, inputs [None {from=MaterialExpressionNamedRerouteUsage_79}]}
  MaterialExpressionReroute_21 {class=MaterialExpressionReroute, inputs [None {from=MaterialExpressionNamedRerouteUsage_80}]}
  MaterialExpressionReroute_3 {class=MaterialExpressionReroute, inputs [None {from=MaterialExpressionNamedRerouteUsage_51}]}
  MaterialExpressionReroute_4 {class=MaterialExpressionReroute, inputs [None {from=MaterialExpressionReroute_7}]}
  MaterialExpressionReroute_5 {class=MaterialExpressionReroute, inputs [None {from=MaterialExpressionReroute_8}]}
  MaterialExpressionReroute_6 {class=MaterialExpressionReroute, inputs [None {from=MaterialExpressionReroute_5}]}
  MaterialExpressionReroute_7 {class=MaterialExpressionReroute, inputs [None {from=MaterialExpressionReroute_9}]}
  MaterialExpressionReroute_8 {class=MaterialExpressionReroute, inputs [None {from=MaterialExpressionFunctionInput_3}]}
  MaterialExpressionReroute_9 {class=MaterialExpressionReroute, inputs [None {from=MaterialExpressionReroute_17}]}
  MaterialExpressionShaderStageSwitch_0 {
   class=MaterialExpressionShaderStageSwitch
   inputs [
    PixelShader {from=MaterialExpressionVertexInterpolator_1, output=PS}
    VertexShader {from=MaterialExpressionPreSkinnedNormal_0, output=""}
   ]
   properties {}
  }
  MaterialExpressionShaderStageSwitch_2 {
   class=MaterialExpressionShaderStageSwitch
   inputs [
    PixelShader {from=MaterialExpressionVertexInterpolator_2, output=PS}
    VertexShader {from=MaterialExpressionLocalPosition_0, output=XYZ}
   ]
   properties {}
  }
  MaterialExpressionTextureObject_3 {class=MaterialExpressionTextureObject, inputs [], properties {texture=/BaseMaterial/Textures/Default/T_Gray_sRGB}}
  MaterialExpressionTransform_0 {
   class=MaterialExpressionTransform
   inputs [None {from=MaterialExpressionConstant3Vector_2}]
   properties {transformSourceType=TRANSFORMSOURCE_Instance}
  }
  MaterialExpressionTransform_10 {
   class=MaterialExpressionTransform
   inputs [None {from=MaterialExpressionNamedRerouteUsage_50}]
   properties {transformSourceType=TRANSFORMSOURCE_Local}
  }
  MaterialExpressionTransform_11 {
   class=MaterialExpressionTransform
   inputs [None {from=MaterialExpressionNamedRerouteUsage_51}]
   properties {transformSourceType=TRANSFORMSOURCE_Instance}
  }
  MaterialExpressionTransform_12 {
   class=MaterialExpressionTransform
   inputs [None {from=MaterialExpressionConstant3Vector_3}]
   properties {transformSourceType=TRANSFORMSOURCE_Instance}
  }
  MaterialExpressionTransform_13 {
   class=MaterialExpressionTransform
   inputs [None {from=MaterialExpressionConstant3Vector_4}]
   properties {transformSourceType=TRANSFORMSOURCE_Instance}
  }
  MaterialExpressionTransform_14 {
   class=MaterialExpressionTransform
   inputs [None {from=MaterialExpressionVertexNormalWS_1}]
   properties {transformSourceType=TRANSFORMSOURCE_World, transformType=TRANSFORM_Instance}
  }
  MaterialExpressionTransform_18 {
   class=MaterialExpressionTransform
   inputs [None {from=MaterialExpressionNamedRerouteUsage_51}]
   properties {transformSourceType=TRANSFORMSOURCE_Local}
  }
  MaterialExpressionTransform_19 {
   class=MaterialExpressionTransform
   inputs [None {from=MaterialExpressionNamedRerouteUsage_52}]
   properties {transformSourceType=TRANSFORMSOURCE_Instance}
  }
  MaterialExpressionTransform_20 {
   class=MaterialExpressionTransform
   inputs [None {from=MaterialExpressionNamedRerouteUsage_52}]
   properties {transformSourceType=TRANSFORMSOURCE_Local}
  }
  MaterialExpressionTransform_21 {
   class=MaterialExpressionTransform
   inputs [None {from=MaterialExpressionNamedRerouteUsage_53}]
   properties {transformSourceType=TRANSFORMSOURCE_Instance}
  }
  MaterialExpressionTransform_22 {
   class=MaterialExpressionTransform
   inputs [None {from=MaterialExpressionNamedRerouteUsage_53}]
   properties {transformSourceType=TRANSFORMSOURCE_Local}
  }
  MaterialExpressionTransform_3 {
   class=MaterialExpressionTransform
   inputs [None {from=MaterialExpressionNormalize_2}]
   properties {transformSourceType=TRANSFORMSOURCE_Instance}
  }
  MaterialExpressionTransform_4 {
   class=MaterialExpressionTransform
   inputs [None {from=MaterialExpressionNamedRerouteUsage_48}]
   properties {transformSourceType=TRANSFORMSOURCE_Instance}
  }
  MaterialExpressionTransform_5 {
   class=MaterialExpressionTransform
   inputs [None {from=MaterialExpressionCameraVectorWS_3}]
   properties {transformSourceType=TRANSFORMSOURCE_World, transformType=TRANSFORM_Instance}
  }
  MaterialExpressionTransform_6 {
   class=MaterialExpressionTransform
   inputs [None {from=MaterialExpressionNamedRerouteUsage_48}]
   properties {transformSourceType=TRANSFORMSOURCE_Local}
  }
  MaterialExpressionTransform_7 {
   class=MaterialExpressionTransform
   inputs [None {from=MaterialExpressionNamedRerouteUsage_49}]
   properties {transformSourceType=TRANSFORMSOURCE_Instance}
  }
  MaterialExpressionTransform_8 {
   class=MaterialExpressionTransform
   inputs [None {from=MaterialExpressionNamedRerouteUsage_49}]
   properties {transformSourceType=TRANSFORMSOURCE_Local}
  }
  MaterialExpressionTransform_9 {
   class=MaterialExpressionTransform
   inputs [None {from=MaterialExpressionNamedRerouteUsage_50}]
   properties {transformSourceType=TRANSFORMSOURCE_Instance}
  }
  MaterialExpressionVertexInterpolator_1 {class=MaterialExpressionVertexInterpolator, inputs [VS {from=MaterialExpressionPreSkinnedNormal_0, output=""}], properties {}}
  MaterialExpressionVertexInterpolator_2 {class=MaterialExpressionVertexInterpolator, inputs [VS {from=MaterialExpressionLocalPosition_0, output=XYZ}], properties {}}
  MaterialExpressionVertexNormalWS_1 {class=MaterialExpressionVertexNormalWS, inputs [], properties {}}
 }
 properties {
  bExposeToLibrary=true
  description="Standard Triplanar with Parallax Occlusion Mapping (POM) for added depth and occlusion. Outputs xUV, yUV, and zUV with full bases and derivatives, plus a BlendMask (RGB) for post-sample blending. Supports post-sample modulation of blend weights by albedo or height, producing smoother, perceptual transitions while adding parallax depth to triplanar blends."
  userExposedCaption="CoordinateBasis: Triplanar (Standard POM)"
  libraryCategoriesText [Base Material]
 }
}
