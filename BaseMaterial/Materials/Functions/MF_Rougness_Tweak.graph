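# Material function that adjusts the roughness of an incoming Material Attributes set by remapping its value range.
# It extracts Roughness and feeds it to RemapValueRange, driven by scalar inputs for input and output minimum and maximum, whose defaults leave the 0–1 range unchanged.
# The result is saturated and written back, so the other attributes pass through untouched.
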MaterialFunction /BaseMaterial/Materials/Functions/MF_Rougness_Tweak
graph {nodes [
  n1 = GetMaterialAttributes(None=FunctionInput() {bUsePreviewValueAsDefault=true, inputName=Input MA, inputType=FunctionInput_MaterialAttributes}) {attributeGetTypes [#1]}
  n2 = FunctionInput() {bUsePreviewValueAsDefault=true, inputName=Input Maximum, inputType=FunctionInput_Scalar, sortPriority=2, previewValue=<1 0 0 1>}
  n3 = FunctionInput() {bUsePreviewValueAsDefault=true, inputName=Output Maximum, inputType=FunctionInput_Scalar, sortPriority=4, previewValue=<1 0 0 1>}
  n4 = MaterialFunctionCall(
   Input=@n1.Roughness
   Input Low=FunctionInput() {bUsePreviewValueAsDefault=true, inputName=Input Minimum, inputType=FunctionInput_Scalar, sortPriority=1}
   Input High=@n2
   Target Low=FunctionInput() {bUsePreviewValueAsDefault=true, inputName=Output Minimum, inputType=FunctionInput_Scalar, sortPriority=3}
   Target High=@n3
  ) {materialFunction=&RemapValueRange}
  FunctionOutput(None=SetMaterialAttributes!(MaterialAttributes=@n1.MaterialAttributes, Roughness=Saturate(None=@n4.Result)) {attributeSetTypes [#1]}) {outputName=Output MA}
 ]}
