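# Material function in the Base Material library that samples one albedo, normal and packed texture set into Material Attributes.
# All three fetches share UV and explicit DDX/DDY packed in DXY, defaulting to TexCoord derivatives.
# Albedo supplies BaseColor and Displacement, while the packed map supplies AO, Roughness and Metallic, overriding PassthroughMA.
# Normals are oriented by Tangent, Bitangent and rotation, then a Custom node picks tangent or world space via MATERIAL_TANGENTSPACENORMAL.
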MaterialFunction /BaseMaterial/Materials/Functions/MF_SampleMaterial_TexFetchSingle
graph {
 expressions {
  MaterialExpressionConvert_18 {
   class=MaterialExpressionConvert
   inputs [
    None {from=MaterialExpressionDDX_0}
    None {from=MaterialExpressionDDY_0}
   ]
   properties {
    convertInputs [
     {
      type=Vector2
      expressionInput {
       expression=&self:MaterialExpressionDDX_0
       inputName=None
       mask=0
       maskA=0
       maskB=0
       maskG=0
       maskR=0
       outputIndex=0
      }
     }
     {
      type=Vector2
      expressionInput {
       expression=&self:MaterialExpressionDDY_0
       inputName=None
       mask=0
       maskA=0
       maskB=0
       maskG=0
       maskR=0
       outputIndex=0
      }
     }
    ]
    convertOutputs [{type=Vector4, defaultValue=<rgba 0 0 0 1>}]
   }
  }
  MaterialExpressionConvert_19 {
   class=MaterialExpressionConvert
   inputs [None {from=MaterialExpressionMaterialFunctionCall_7, output=RGBA}]
   properties {
    convertInputs [{
      type=Vector4
      defaultValue=<rgba 0 0 0 1>
      expressionInput {
       expression=&self:MaterialExpressionMaterialFunctionCall_7
       inputName=None
       mask=0
       maskA=0
       maskB=0
       maskG=0
       maskR=0
       outputIndex=4
      }
     }]
    convertOutputs [{type=Vector3}, {type=Scalar}]
   }
  }
  MaterialExpressionConvert_20 {
   class=MaterialExpressionConvert
   inputs [None {from=MaterialExpressionNamedRerouteUsage_34}]
   properties {
    convertInputs [{
      type=Vector4
      defaultValue=<rgba 0 0 0 1>
      expressionInput {
       expression=&self:MaterialExpressionNamedRerouteUsage_34
       inputName=None
       mask=0
       maskA=0
       maskB=0
       maskG=0
       maskR=0
       outputIndex=0
      }
     }]
    convertOutputs [{type=Vector2}, {type=Vector2}]
   }
  }
  MaterialExpressionConvert_21 {
   class=MaterialExpressionConvert
   inputs [None {from=MaterialExpressionMaterialFunctionCall_8, output=RGBA}]
   properties {
    convertInputs [{
      type=Vector4
      defaultValue=<rgba 0 0 0 1>
      expressionInput {
       expression=&self:MaterialExpressionMaterialFunctionCall_8
       inputName=None
       mask=0
       maskA=0
       maskB=0
       maskG=0
       maskR=0
       outputIndex=4
      }
     }]
    convertOutputs [{type=Scalar}, {type=Scalar}, {type=Scalar}, {type=Scalar}]
   }
  }
  MaterialExpressionConvert_22 {
   class=MaterialExpressionConvert
   inputs [None {from=MaterialExpressionNamedRerouteUsage_27}]
   properties {
    convertInputs [{
      type=Vector4
      defaultValue=<rgba 0 0 0 1>
      expressionInput {
       expression=&self:MaterialExpressionNamedRerouteUsage_27
       inputName=None
       mask=0
       maskA=0
       maskB=0
       maskG=0
       maskR=0
       outputIndex=0
      }
     }]
    convertOutputs [{type=Vector2}, {type=Vector2}]
   }
  }
  MaterialExpressionConvert_23 {
   class=MaterialExpressionConvert
   inputs [None {from=MaterialExpressionNamedRerouteUsage_37}]
   properties {
    convertInputs [{
      type=Vector4
      defaultValue=<rgba 0 0 0 1>
      expressionInput {
       expression=&self:MaterialExpressionNamedRerouteUsage_37
       inputName=None
       mask=0
       maskA=0
       maskB=0
       maskG=0
       maskR=0
       outputIndex=0
      }
     }]
    convertOutputs [{type=Vector2}, {type=Vector2}]
   }
  }
  MaterialExpressionCustom_0 {
   class=MaterialExpressionCustom
   inputs [
    WSN {from=MaterialExpressionMaterialFunctionCall_10, output=WSN}
    TSN {from=MaterialExpressionMaterialFunctionCall_10, output=WSN}
   ]
   properties {
    code="{\r\n##if MATERIAL_TANGENTSPACENORMAL\r\n\tconst float3 N = TSN;\r\n##else\r\n\tconst float3 N = WSN;\r\n##endif\r\n\treturn N;\r\n}"
    description=Default Normal
    inputs [
     {
      inputName=WSN
      input {
       expression=&self:MaterialExpressionMaterialFunctionCall_10
       inputName=None
       mask=0
       maskA=0
       maskB=0
       maskG=0
       maskR=0
       outputIndex=0
      }
     }
     {
      inputName=TSN
      input {
       expression=&self:MaterialExpressionMaterialFunctionCall_10
       inputName=None
       mask=0
       maskA=0
       maskB=0
       maskG=0
       maskR=0
       outputIndex=1
      }
     }
    ]
   }
  }
  MaterialExpressionDDX_0 {class=MaterialExpressionDDX, inputs [Value {from=MaterialExpressionTextureCoordinate_0, output=""}], properties {}}
  MaterialExpressionDDY_0 {class=MaterialExpressionDDY, inputs [Value {from=MaterialExpressionTextureCoordinate_0, output=""}], properties {}}
  MaterialExpressionFunctionInput_0 {
   class=MaterialExpressionFunctionInput
   inputs [Preview {from="", output=""}]
   properties {bUsePreviewValueAsDefault=true, inputName=PassthroughMA, inputType=FunctionInput_MaterialAttributes, sortPriority=100}
  }
  MaterialExpressionFunctionInput_10 {
   class=MaterialExpressionFunctionInput
   inputs [Preview {from=MaterialExpressionTextureObject_0, output=""}]
   properties {bUsePreviewValueAsDefault=true, inputName=PackedTex, inputType=FunctionInput_Texture2D, sortPriority=-8}
  }
  MaterialExpressionFunctionInput_12 {
   class=MaterialExpressionFunctionInput
   inputs [Preview {from="", output=""}]
   properties {bUsePreviewValueAsDefault=true, inputName="Rotation (Turns)", inputType=FunctionInput_Scalar, sortPriority=-5}
  }
  MaterialExpressionFunctionInput_3 {
   class=MaterialExpressionFunctionInput
   inputs [Preview {from=MaterialExpressionTextureCoordinate_0, output=""}]
   properties {bUsePreviewValueAsDefault=true, inputName=UV, inputType=FunctionInput_Vector2}
  }
  MaterialExpressionFunctionInput_4 {
   class=MaterialExpressionFunctionInput
   inputs [Preview {from="", output=""}]
   properties {bUsePreviewValueAsDefault=true, inputName=Tangent, sortPriority=1, previewValue=<1 0 0 0>}
  }
  MaterialExpressionFunctionInput_5 {
   class=MaterialExpressionFunctionInput
   inputs [Preview {from="", output=""}]
   properties {bUsePreviewValueAsDefault=true, inputName=Bitangent, sortPriority=2, previewValue=<0 1 0 0>}
  }
  MaterialExpressionFunctionInput_6 {
   class=MaterialExpressionFunctionInput
   inputs [Preview {from=MaterialExpressionConvert_18, output=""}]
   properties {bUsePreviewValueAsDefault=true, inputName=DXY, inputType=FunctionInput_Vector4, sortPriority=3}
  }
  MaterialExpressionFunctionInput_8 {
   class=MaterialExpressionFunctionInput
   inputs [Preview {from=MaterialExpressionTextureObject_2, output=""}]
   properties {bUsePreviewValueAsDefault=true, inputName=AlbedoTex, inputType=FunctionInput_Texture2D, sortPriority=-10}
  }
  MaterialExpressionFunctionInput_9 {
   class=MaterialExpressionFunctionInput
   inputs [Preview {from=MaterialExpressionTextureObject_1, output=""}]
   properties {bUsePreviewValueAsDefault=true, inputName=NormalTex, inputType=FunctionInput_Texture2D, sortPriority=-9}
  }
  MaterialExpressionFunctionOutput_0 {
   class=MaterialExpressionFunctionOutput
   inputs [None {from=MaterialExpressionSetMaterialAttributes_1, output=""}]
   properties {outputName=Out_MA}
  }
  MaterialExpressionMaterialFunctionCall_10 {
   class=MaterialExpressionMaterialFunctionCall
   inputs [
    TSN {from=MaterialExpressionMaterialFunctionCall_9, output=RGB}
    Tangent {from=MaterialExpressionNamedRerouteUsage_39}
    Bitangent {from=MaterialExpressionNamedRerouteUsage_40}
    "Rotation (Turns)" {from=MaterialExpressionFunctionInput_12}
   ]
   properties {materialFunction=&/MF_TangentNormal_OrientWorld}
  }
  MaterialExpressionMaterialFunctionCall_7 {
   class=MaterialExpressionMaterialFunctionCall
   inputs [
    Coordinates {from=MaterialExpressionNamedRerouteUsage_11}
    TexObject {from=MaterialExpressionNamedRerouteUsage_26}
    DDX {from=MaterialExpressionConvert_22}
    DDY {from=MaterialExpressionConvert_22}
    Mask {from=""}
    DefaultUV {from=""}
   ]
   properties {materialFunction=&/MF_TexPointSample_RGB}
  }
  MaterialExpressionMaterialFunctionCall_8 {
   class=MaterialExpressionMaterialFunctionCall
   inputs [
    Coordinates {from=MaterialExpressionNamedRerouteUsage_28}
    TexObject {from=MaterialExpressionNamedRerouteUsage_35}
    DDX {from=MaterialExpressionConvert_20}
    DDY {from=MaterialExpressionConvert_20}
    Mask {from=""}
    DefaultUV {from=""}
   ]
   properties {materialFunction=&MF_TexPointSample_RGB}
  }
  MaterialExpressionMaterialFunctionCall_9 {
   class=MaterialExpressionMaterialFunctionCall
   inputs [
    Coordinates {from=MaterialExpressionNamedRerouteUsage_36}
    TexObject {from=MaterialExpressionNamedRerouteUsage_38}
    DDX {from=MaterialExpressionConvert_23}
    DDY {from=MaterialExpressionConvert_23}
    Mask {from=""}
    DefaultUV {from=""}
   ]
   properties {materialFunction=&/MF_TexPointSample_NRM}
  }
  MaterialExpressionNamedRerouteDeclaration_0 {
   class=MaterialExpressionNamedRerouteDeclaration
   inputs [None {from=MaterialExpressionFunctionInput_3}]
   properties {name=UV, nodeColor=<rgba 0.9 0.009 0.009 1>}
  }
  MaterialExpressionNamedRerouteDeclaration_13 {
   class=MaterialExpressionNamedRerouteDeclaration
   inputs [None {from=MaterialExpressionFunctionInput_5}]
   properties {name=B, nodeColor=<rgba 0.9 0.009 0.009 1>}
  }
  MaterialExpressionNamedRerouteDeclaration_3 {
   class=MaterialExpressionNamedRerouteDeclaration
   inputs [None {from=MaterialExpressionFunctionInput_8}]
   properties {name=AlbedoMap, nodeColor=<rgba 1 0 0.4922 1>}
  }
  MaterialExpressionNamedRerouteDeclaration_4 {
   class=MaterialExpressionNamedRerouteDeclaration
   inputs [None {from=MaterialExpressionFunctionInput_9}]
   properties {name=NormalMap, nodeColor=<rgba 1 0 0.4922 1>}
  }
  MaterialExpressionNamedRerouteDeclaration_5 {
   class=MaterialExpressionNamedRerouteDeclaration
   inputs [None {from=MaterialExpressionFunctionInput_10}]
   properties {name=PackMap, nodeColor=<rgba 1 0 0.4922 1>}
  }
  MaterialExpressionNamedRerouteDeclaration_6 {
   class=MaterialExpressionNamedRerouteDeclaration
   inputs [None {from=MaterialExpressionFunctionInput_4}]
   properties {name=T, nodeColor=<rgba 0.9 0.009 0.009 1>}
  }
  MaterialExpressionNamedRerouteDeclaration_7 {
   class=MaterialExpressionNamedRerouteDeclaration
   inputs [None {from=MaterialExpressionFunctionInput_6}]
   properties {name=DXY, nodeColor=<rgba 0.9 0.009 0.009 1>}
  }
  MaterialExpressionNamedRerouteUsage_11 {class=MaterialExpressionNamedRerouteUsage}
  MaterialExpressionNamedRerouteUsage_26 {class=MaterialExpressionNamedRerouteUsage}
  MaterialExpressionNamedRerouteUsage_27 {class=MaterialExpressionNamedRerouteUsage}
  MaterialExpressionNamedRerouteUsage_28 {class=MaterialExpressionNamedRerouteUsage}
  MaterialExpressionNamedRerouteUsage_34 {class=MaterialExpressionNamedRerouteUsage}
  MaterialExpressionNamedRerouteUsage_35 {class=MaterialExpressionNamedRerouteUsage}
  MaterialExpressionNamedRerouteUsage_36 {class=MaterialExpressionNamedRerouteUsage}
  MaterialExpressionNamedRerouteUsage_37 {class=MaterialExpressionNamedRerouteUsage}
  MaterialExpressionNamedRerouteUsage_38 {class=MaterialExpressionNamedRerouteUsage}
  MaterialExpressionNamedRerouteUsage_39 {class=MaterialExpressionNamedRerouteUsage}
  MaterialExpressionNamedRerouteUsage_40 {class=MaterialExpressionNamedRerouteUsage}
  MaterialExpressionSetMaterialAttributes_1 {
   class=MaterialExpressionSetMaterialAttributes
   inputs [
    MaterialAttributes {from=MaterialExpressionFunctionInput_0, output=""}
    BaseColor {from=MaterialExpressionConvert_19, output=""}
    Displacement {from=MaterialExpressionConvert_19, output=""}
    AmbientOcclusion {from=MaterialExpressionConvert_21, output=""}
    Roughness {from=MaterialExpressionConvert_21, output=""}
    Metallic {from=MaterialExpressionConvert_21, output=""}
    Normal {from=MaterialExpressionCustom_0, output=""}
   ]
   properties {attributeSetTypes [
     #1
     #2
     #3
     #4
     #5
     #6
    ]}
  }
  MaterialExpressionTextureCoordinate_0 {class=MaterialExpressionTextureCoordinate}
  MaterialExpressionTextureObject_0 {
   class=MaterialExpressionTextureObject
   inputs []
   properties {samplerType=SAMPLERTYPE_LinearColor, texture=/BaseMaterial/Textures/Default/T_Packed_BC7}
  }
  MaterialExpressionTextureObject_1 {
   class=MaterialExpressionTextureObject
   inputs []
   properties {samplerType=SAMPLERTYPE_Normal, texture=/BaseMaterial/Textures/Default/T_FlatNorm_BC5}
  }
  MaterialExpressionTextureObject_2 {class=MaterialExpressionTextureObject, inputs [], properties {texture=/BaseMaterial/Textures/Default/T_Gray_sRGB}}
 }
 properties {bExposeToLibrary=true, libraryCategoriesText [Base Material]}
}
