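# Material function exposed as "CoordinateBasis: Simple UV (Standard)" that builds a texture coordinate frame from a selected mesh UV channel (0–3).
# The chosen channel passes through MF_CoordinateTransform_2D for offset, rotation in turns and tiling, yielding UVs and packed derivatives for SampleGrad.
# MF_DeriveTangentBasis derives tangent and bitangent from those derivatives, allowing normal maps on UV sets without a stored basis.
# It also outputs the negated rotation for normal reorientation.
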MaterialFunction /BaseMaterial/Materials/Functions/MF_ResolveCoordinateFrame_SimpleUV_Basic
graph {
 properties {
  bExposeToLibrary=true
  description="Generates UVs directly from the mesh’s authored UV channels. Provides UVs, tangent/bitangent vectors with rotation for tangent-space normal reorientation, and packed derivatives for stable mip selection with SampleGrad. Because tangents and bitangents are derived from the selected index, you can apply normal maps in alternate UV channels that don’t have a stored basis in the mesh, enabling detail maps and layered normals on secondary UV sets."
  userExposedCaption="CoordinateBasis: Simple UV (Standard)"
  libraryCategoriesText [Base Material]
 }
 nodes [
  n1 = FunctionInput() {
   bUsePreviewValueAsDefault=true
   description="Selects which mesh UV channel (0–3) to use as the base coordinate source."
   inputName=UVIndex
   inputType=FunctionInput_Scalar
   sortPriority=6
  }
  n2 = FunctionInput() {
   bUsePreviewValueAsDefault=true
   description=Applies an additional per-axis offset to the coordinates before projection. Useful for correctly positioning a texture.
   inputName=CoordinateTransform
   inputType=FunctionInput_Vector2
   sortPriority=-9
  }
  n3 = FunctionInput() {
   bUsePreviewValueAsDefault=true
   description="Rotates the projected UVs around the selected axis. Units are in turns, where 1.0 = 360° and 0.25 = 90°."
   inputName="Rotation (Turns)"
   inputType=FunctionInput_Scalar
   sortPriority=-8
  }
  n4 = NamedRerouteDeclaration(None=@n3) {name=turns, nodeColor=<rgba 0.8963 0.8963 0.0091 1>}
  n5 = FunctionInput() {
   bUsePreviewValueAsDefault=true
   description="Sets how many times the projected coordinates repeat along each axis. Use values above 1 for denser tiling, negative values for mirroring, and avoid zero."
   inputName=CoordinateRepeat
   inputType=FunctionInput_Vector2
   sortPriority=-10
   previewValue=<1 1 1 1>
  }
  n6 = MaterialFunctionCall!(
   Coords=MaterialFunctionCall!(
    Selector=@n1
    VectorA=TextureCoordinate()
    VectorB=TextureCoordinate() {coordinateIndex=1}
    VectorC=TextureCoordinate() {coordinateIndex=2}
    VectorD=TextureCoordinate() {coordinateIndex=3}
   ) {materialFunction=&/MF_Switch4_Vec2}.Selected Vector
   UVOffset=@n2
   UVRotation=@n4
   RotationPivot
   UVTile=@n5
  ) {materialFunction=&/MF_CoordinateTransform_2D}
  FunctionOutput(None=@n6.geometry_UV) {
   description="The primary 2D coordinates generated by the projection. Feed these into texture samplers for surface detail.\r\n"
   outputName=UV
  }
  FunctionOutput(None=Multiply(A=@n4) {constB=-1}) {
   description="The rotation applied around the projection axis, expressed in turns (1.0 = 360°). Used along with Tangent/Bitangent for reorienting normal maps via MF_TangentNormal_OrientWorld.\r\n"
   outputName="Rotation (Turns)"
   sortPriority=-5
  }
  FunctionOutput(None=@n6.geometry_packedDXY) {
   description="Packed derivatives of the projected UVs (ddx/ddy). Use with SampleGrad to ensure textures sample with correct mip levels and anisotropic filtering.\r\n"
   outputName=DXY
   sortPriority=3
  }
  n7 = MaterialFunctionCall!(packedDXY=@n6.geometry_packedDXY) {materialFunction=&/MF_DeriveTangentBasis}
  FunctionOutput(None=@n7.Tangent) {
   description="The world-space tangent vector of the projection frame. Use with Bitangent and Rotation to correctly reorient tangent-space normals.\r\n"
   outputName=Tangent
   sortPriority=1
  }
  FunctionOutput(None=@n7.Bitangent) {
   description="The world-space bitangent vector of the projection frame. Complements Tangent to form the basis for tangent-space normal reorientation.\r\n"
   outputName=Bitangent
   sortPriority=2
  }
 ]
}
