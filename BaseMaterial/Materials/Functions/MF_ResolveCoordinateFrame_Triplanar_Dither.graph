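# Material function that resolves a dithered triplanar projection into one UV set, packed derivatives, rotation, and tangent and bitangent for normal reorientation.
# A selector chooses world, local, periodic or preskinned position and normal, with LWC-prepared periodic positions.
# The normal yields a contrast-adjustable axis mask, which golden-ratio interleaved gradient noise, or custom dither, collapses to one axis index per pixel.
# This avoids multi-sample blending but can show dither artifacts.
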MaterialFunction /BaseMaterial/Materials/Functions/MF_ResolveCoordinateFrame_Triplanar_Dither
graph {
 expressions {
  MaterialExpressionAdd_8 {
   class=MaterialExpressionAdd
   inputs [
    A {from=MaterialExpressionMaterialFunctionCall_5, output=Selected Vector}
    B {from=MaterialExpressionFunctionInput_13}
   ]
  }
  MaterialExpressionConstant3Vector_0 {class=MaterialExpressionConstant3Vector, properties {constant=<rgba 0 0 1 0>}}
  MaterialExpressionConstant3Vector_1 {class=MaterialExpressionConstant3Vector, properties {constant=<rgba 0 1 0 0>}}
  MaterialExpressionConstant3Vector_5 {class=MaterialExpressionConstant3Vector, properties {constant=<rgba 1 0 0 0>}}
  MaterialExpressionConstant_0 {class=MaterialExpressionConstant, properties {r=1}}
  MaterialExpressionConvert_10 {
   class=MaterialExpressionConvert
   inputs [
    None {from=MaterialExpressionLength_5}
    None {from=MaterialExpressionLength_4}
    None {from=MaterialExpressionLength_1}
   ]
   properties {
    desc=maintain scale
    convertInputs [
     {
      type=Scalar
      expressionInput {
       expression=&self:MaterialExpressionLength_5
       inputName=None
       mask=0
       maskA=0
       maskB=0
       maskG=0
       maskR=0
       outputIndex=0
      }
     }
     {
      type=Scalar
      expressionInput {
       expression=&self:MaterialExpressionLength_4
       inputName=None
       mask=0
       maskA=0
       maskB=0
       maskG=0
       maskR=0
       outputIndex=0
      }
     }
     {
      type=Scalar
      expressionInput {
       expression=&self:MaterialExpressionLength_1
       inputName=None
       mask=0
       maskA=0
       maskB=0
       maskG=0
       maskR=0
       outputIndex=0
      }
     }
    ]
    convertOutputs [{type=Vector3}]
   }
  }
  MaterialExpressionDivide_0 {
   class=MaterialExpressionDivide
   inputs [
    A {from=""}
    B {from=MaterialExpressionNamedRerouteUsage_3}
   ]
  }
  MaterialExpressionDivide_1 {
   class=MaterialExpressionDivide
   inputs [
    A {from=MaterialExpressionMaterialFunctionCall_4, output=Selected Vector}
    B {from=MaterialExpressionNamedRerouteUsage_28}
   ]
  }
  MaterialExpressionFunctionInput_0 {
   class=MaterialExpressionFunctionInput
   inputs [Preview {from="", output=""}]
   properties {
    bUsePreviewValueAsDefault=true
    description="Sets how many times the projected coordinates repeat along each axis. Use values above 1 for denser tiling, negative values for mirroring, and avoid zero."
    inputName=CoordinateRepeat
    sortPriority=-10
    previewValue=<0.01 0.01 0.01 1>
   }
  }
  MaterialExpressionFunctionInput_1 {
   class=MaterialExpressionFunctionInput
   inputs [Preview {from="", output=""}]
   properties {
    bUsePreviewValueAsDefault=true
    description="Rotates the projected UVs around the selected axis. Units are in turns, where 1.0 = 360° and 0.25 = 90°."
    inputName="Rotation (Turns)"
    inputType=FunctionInput_Scalar
    sortPriority=-8
   }
  }
  MaterialExpressionFunctionInput_13 {
   class=MaterialExpressionFunctionInput
   inputs [Preview {from="", output=""}]
   properties {
    bUsePreviewValueAsDefault=true
    description=Applies an additional per-axis offset to the coordinates before projection. Useful for correctly positioning a texture.
    inputName=CoordinateTransform
    sortPriority=-9
   }
  }
  MaterialExpressionFunctionInput_14 {
   class=MaterialExpressionFunctionInput
   inputs [Preview {from="", output=""}]
   properties {
    bUsePreviewValueAsDefault=true
    description="Defines the space used for projection: World (0) keeps textures fixed in world space, Local (1) moves with the object, Periodic (2) enforces a repeating pattern, and Preskinned (3) locks to preskinned positions for stable results on animated meshes."
    inputName=CoordinateSpace
    inputType=FunctionInput_Scalar
    sortPriority=-11
   }
  }
  MaterialExpressionFunctionInput_2 {
   class=MaterialExpressionFunctionInput
   inputs [Preview {from=MaterialExpressionSubtract_0, output=""}]
   properties {
    bUsePreviewValueAsDefault=true
    description="Provides the dithering sequence used for stochastic projection blending. By default, the function generates an interleaved gradient noise pattern based on the golden ratio, which produces stable, slow-converging results at low cost. Supplying a custom value lets you override this with your own dithering scheme, useful for synchronizing with other systems or experimenting with alternate noise sources."
    inputName=Dither
    inputType=FunctionInput_Scalar
   }
  }
  MaterialExpressionFunctionInput_7 {
   class=MaterialExpressionFunctionInput
   inputs [Preview {from="", output=""}]
   properties {
    bUsePreviewValueAsDefault=true
    description="Controls the contrast of the blend mask. Internally applied as pow(mask, 8 × BlendFactor). A default of 0.5 corresponds to pow4 contrast."
    inputName=BlendFactor
    inputType=FunctionInput_Scalar
    sortPriority=6
    previewValue=<0.5 0 0 1>
   }
  }
  MaterialExpressionFunctionOutput_1 {
   class=MaterialExpressionFunctionOutput
   inputs [None {from=MaterialExpressionMaterialFunctionCall_19, output=UV}]
   properties {
    description="The primary 2D coordinates generated by the projection. Feed these into texture samplers for surface detail.\r\n"
    outputName=UV
   }
  }
  MaterialExpressionFunctionOutput_14 {
   class=MaterialExpressionFunctionOutput
   inputs [None {from=MaterialExpressionFunctionInput_1, output=""}]
   properties {
    description="The rotation applied around the projection axis, expressed in turns (1.0 = 360°). Used along with Tangent/Bitangent for reorienting normal maps via MF_TangentNormal_OrientWorld.\r\n"
    outputName="Rotation (Turns)"
    sortPriority=-5
   }
  }
  MaterialExpressionFunctionOutput_2 {
   class=MaterialExpressionFunctionOutput
   inputs [None {from=MaterialExpressionMaterialFunctionCall_19, output=packedDXY}]
   properties {
    description="Packed derivatives of the projected UVs (ddx/ddy). Use with SampleGrad to ensure textures sample with correct mip levels and anisotropic filtering.\r\n"
    outputName=DXY
    sortPriority=3
   }
  }
  MaterialExpressionFunctionOutput_8 {
   class=MaterialExpressionFunctionOutput
   inputs [None {from=MaterialExpressionMaterialFunctionCall_0, output=Selected Vector}]
   properties {
    description="The world-space tangent vector of the projection frame. Use with Bitangent and Rotation to correctly reorient tangent-space normals.\r\n"
    outputName=Tangent
    sortPriority=1
   }
  }
  MaterialExpressionFunctionOutput_9 {
   class=MaterialExpressionFunctionOutput
   inputs [None {from=MaterialExpressionMaterialFunctionCall_2, output=Selected Vector}]
   properties {
    description="The world-space bitangent vector of the projection frame. Complements Tangent to form the basis for tangent-space normal reorientation.\r\n"
    outputName=Bitangent
    sortPriority=2
   }
  }
  MaterialExpressionLength_1 {class=MaterialExpressionLength, inputs [None {from=MaterialExpressionTransform_1, output=""}], properties {}}
  MaterialExpressionLength_4 {class=MaterialExpressionLength, inputs [None {from=MaterialExpressionTransform_2, output=""}], properties {}}
  MaterialExpressionLength_5 {class=MaterialExpressionLength, inputs [None {from=MaterialExpressionTransform_5, output=""}], properties {}}
  MaterialExpressionLocalPosition_1 {class=MaterialExpressionLocalPosition, inputs [], properties {includedOffsets=ExcludeOffsets, localOrigin=InstancePreSkinning}}
  MaterialExpressionLocalPosition_2 {class=MaterialExpressionLocalPosition, inputs [], properties {}}
  MaterialExpressionMaterialFunctionCall_0 {
   class=MaterialExpressionMaterialFunctionCall
   inputs [
    Selector {from=MaterialExpressionNamedRerouteUsage_42}
    VectorA {from=MaterialExpressionReroute_6}
    VectorB {from=MaterialExpressionTransform_4}
    VectorC {from=MaterialExpressionReroute_6}
    VectorD {from=MaterialExpressionTransform_6}
   ]
   properties {materialFunction=&/MF_Switch4_Vec3}
  }
  MaterialExpressionMaterialFunctionCall_1 {
   class=MaterialExpressionMaterialFunctionCall
   inputs [PeriodXYZ {from=MaterialExpressionDivide_0}]
   properties {materialFunction=&/MF_PrepareLWC}
  }
  MaterialExpressionMaterialFunctionCall_19 {
   class=MaterialExpressionMaterialFunctionCall
   inputs [
    Position {from=MaterialExpressionNamedRerouteUsage_40}
    Tiling {from=MaterialExpressionReroute_1}
    RefNorm {from=MaterialExpressionNamedRerouteUsage_36}
    "Rotation (Turns)" {from=MaterialExpressionNamedRerouteDeclaration_11}
    AxisIndex {from=MaterialExpressionMaterialFunctionCall_3, output=AxisIndex}
   ]
   properties {materialFunction=&/MF_Coordinate_Triplanar}
  }
  MaterialExpressionMaterialFunctionCall_2 {
   class=MaterialExpressionMaterialFunctionCall
   inputs [
    Selector {from=MaterialExpressionNamedRerouteUsage_43}
    VectorA {from=MaterialExpressionReroute_0}
    VectorB {from=MaterialExpressionTransform_7}
    VectorC {from=MaterialExpressionReroute_0}
    VectorD {from=MaterialExpressionTransform_8}
   ]
   properties {materialFunction=&MF_Switch4_Vec3}
  }
  MaterialExpressionMaterialFunctionCall_25 {
   class=MaterialExpressionMaterialFunctionCall
   inputs [
    RefNorm {from=MaterialExpressionDivide_1}
    BlendFactor {from=MaterialExpressionFunctionInput_7}
   ]
   properties {materialFunction=&/MF_Blend_TriplanarAxis}
  }
  MaterialExpressionMaterialFunctionCall_3 {
   class=MaterialExpressionMaterialFunctionCall
   inputs [
    InputBlend {from=MaterialExpressionNamedRerouteUsage_27}
    TriplanarDitherPattern {from=MaterialExpressionFunctionInput_2}
   ]
   properties {materialFunction=&/MF_Blend_RGBToIndex}
  }
  MaterialExpressionMaterialFunctionCall_4 {
   class=MaterialExpressionMaterialFunctionCall
   inputs [
    Selector {from=MaterialExpressionNamedRerouteUsage_41}
    VectorA {from=MaterialExpressionReroute_3}
    VectorB {from=MaterialExpressionNormalize_11}
    VectorC {from=MaterialExpressionReroute_3}
    VectorD {from=MaterialExpressionShaderStageSwitch_0}
   ]
   properties {materialFunction=&MF_Switch4_Vec3}
  }
  MaterialExpressionMaterialFunctionCall_5 {
   class=MaterialExpressionMaterialFunctionCall
   inputs [
    Selector {from=MaterialExpressionFunctionInput_14}
    VectorA {from=MaterialExpressionReroute_2}
    VectorB {from=MaterialExpressionMultiply_0}
    VectorC {from=MaterialExpressionReroute_4}
    VectorD {from=MaterialExpressionMultiply_1}
   ]
   properties {materialFunction=&MF_Switch4_Vec3}
  }
  MaterialExpressionMaterialFunctionCall_7 {
   class=MaterialExpressionMaterialFunctionCall
   inputs [
    TemporalSampleIndex {from=""}
    PixelPosition {from=""}
   ]
   properties {materialFunction=&/MF_Noise_InterleavedGradientGolden_1d}
  }
  MaterialExpressionMultiply_0 {
   class=MaterialExpressionMultiply
   inputs [
    A {from=MaterialExpressionNamedRerouteUsage_0}
    B {from=MaterialExpressionLocalPosition_2, output=XYZ}
   ]
  }
  MaterialExpressionMultiply_1 {
   class=MaterialExpressionMultiply
   inputs [
    A {from=MaterialExpressionNamedRerouteUsage_1}
    B {from=MaterialExpressionShaderStageSwitch_2}
   ]
  }
  MaterialExpressionNamedRerouteDeclaration_0 {
   class=MaterialExpressionNamedRerouteDeclaration
   inputs [None {from=MaterialExpressionFunctionOutput_1}]
   properties {name=UV, nodeColor=<rgba 0.9 0.009 0.009 1>}
  }
  MaterialExpressionNamedRerouteDeclaration_10 {
   class=MaterialExpressionNamedRerouteDeclaration
   inputs [None {from=MaterialExpressionMaterialFunctionCall_25, output=AxisMask}]
   properties {name=Blend, nodeColor=<rgba 1 1 1 1>}
  }
  MaterialExpressionNamedRerouteDeclaration_11 {
   class=MaterialExpressionNamedRerouteDeclaration
   inputs [None {from=MaterialExpressionFunctionOutput_14}]
   properties {name=turns, nodeColor=<rgba 0.8963 0.8963 0.0091 1>}
  }
  MaterialExpressionNamedRerouteDeclaration_12 {
   class=MaterialExpressionNamedRerouteDeclaration
   inputs [None {from=MaterialExpressionConstant_0}]
   properties {name=MeshVecScale, nodeColor=<rgba 0.6641 0 1 1>}
  }
  MaterialExpressionNamedRerouteDeclaration_13 {
   class=MaterialExpressionNamedRerouteDeclaration
   inputs [None {from=MaterialExpressionFunctionOutput_9}]
   properties {name=B, nodeColor=<rgba 0.9 0.009 0.009 1>}
  }
  MaterialExpressionNamedRerouteDeclaration_18 {
   class=MaterialExpressionNamedRerouteDeclaration
   inputs [None {from=MaterialExpressionDivide_1}]
   properties {name=RefNorm, nodeColor=<rgba 0.8963 0.8963 0.0091 1>}
  }
  MaterialExpressionNamedRerouteDeclaration_19 {
   class=MaterialExpressionNamedRerouteDeclaration
   inputs [None {from=MaterialExpressionAdd_8}]
   properties {name=RefPos, nodeColor=<rgba 0.8963 0.8963 0.0091 1>}
  }
  MaterialExpressionNamedRerouteDeclaration_2 {
   class=MaterialExpressionNamedRerouteDeclaration
   inputs [None {from=MaterialExpressionFunctionInput_0}]
   properties {name=Repeat, nodeColor=<rgba 0.7812 0 1 1>}
  }
  MaterialExpressionNamedRerouteDeclaration_20 {
   class=MaterialExpressionNamedRerouteDeclaration
   inputs [None {from=MaterialExpressionFunctionInput_14}]
   properties {name=Selector, nodeColor=<rgba 0 0.3437 1 1>}
  }
  MaterialExpressionNamedRerouteDeclaration_21 {
   class=MaterialExpressionNamedRerouteDeclaration
   inputs [None {from=MaterialExpressionMaterialFunctionCall_19, output=Tangent}]
   properties {name=Ti, nodeColor=<rgba 0.9 0.009 0.009 1>}
  }
  MaterialExpressionNamedRerouteDeclaration_22 {
   class=MaterialExpressionNamedRerouteDeclaration
   inputs [None {from=MaterialExpressionMaterialFunctionCall_19, output=Bitangent}]
   properties {name=Bi, nodeColor=<rgba 0.9 0.009 0.009 1>}
  }
  MaterialExpressionNamedRerouteDeclaration_6 {
   class=MaterialExpressionNamedRerouteDeclaration
   inputs [None {from=MaterialExpressionFunctionOutput_8}]
   properties {name=T, nodeColor=<rgba 0.9 0.009 0.009 1>}
  }
  MaterialExpressionNamedRerouteDeclaration_7 {
   class=MaterialExpressionNamedRerouteDeclaration
   inputs [None {from=MaterialExpressionFunctionOutput_2}]
   properties {name=DXY, nodeColor=<rgba 0.9 0.009 0.009 1>}
  }
  MaterialExpressionNamedRerouteUsage_0 {class=MaterialExpressionNamedRerouteUsage}
  MaterialExpressionNamedRerouteUsage_1 {class=MaterialExpressionNamedRerouteUsage}
  MaterialExpressionNamedRerouteUsage_27 {class=MaterialExpressionNamedRerouteUsage}
  MaterialExpressionNamedRerouteUsage_28 {class=MaterialExpressionNamedRerouteUsage}
  MaterialExpressionNamedRerouteUsage_3 {class=MaterialExpressionNamedRerouteUsage}
  MaterialExpressionNamedRerouteUsage_36 {class=MaterialExpressionNamedRerouteUsage}
  MaterialExpressionNamedRerouteUsage_40 {class=MaterialExpressionNamedRerouteUsage}
  MaterialExpressionNamedRerouteUsage_41 {class=MaterialExpressionNamedRerouteUsage}
  MaterialExpressionNamedRerouteUsage_42 {class=MaterialExpressionNamedRerouteUsage}
  MaterialExpressionNamedRerouteUsage_43 {class=MaterialExpressionNamedRerouteUsage}
  MaterialExpressionNamedRerouteUsage_48 {class=MaterialExpressionNamedRerouteUsage}
  MaterialExpressionNamedRerouteUsage_49 {class=MaterialExpressionNamedRerouteUsage}
  MaterialExpressionNormalize_11 {class=MaterialExpressionNormalize, inputs [VectorInput {from=MaterialExpressionTransform_3, output=""}]}
  MaterialExpressionPreSkinnedNormal_1 {class=MaterialExpressionPreSkinnedNormal, inputs [], properties {}}
  MaterialExpressionReroute_0 {class=MaterialExpressionReroute, inputs [None {from=MaterialExpressionNamedRerouteUsage_49}]}
  MaterialExpressionReroute_1 {class=MaterialExpressionReroute, inputs [None {from=MaterialExpressionNamedRerouteDeclaration_2}]}
  MaterialExpressionReroute_2 {class=MaterialExpressionReroute, inputs [None {from=MaterialExpressionMaterialFunctionCall_1, output=Per-Axis Period}]}
  MaterialExpressionReroute_3 {class=MaterialExpressionReroute, inputs [None {from=MaterialExpressionVertexNormalWS_0}]}
  MaterialExpressionReroute_4 {class=MaterialExpressionReroute, inputs [None {from=MaterialExpressionReroute_8}]}
  MaterialExpressionReroute_6 {class=MaterialExpressionReroute, inputs [None {from=MaterialExpressionNamedRerouteUsage_48}]}
  MaterialExpressionReroute_8 {class=MaterialExpressionReroute, inputs [None {from=MaterialExpressionMaterialFunctionCall_1, output=Max Axis Period}]}
  MaterialExpressionShaderStageSwitch_0 {
   class=MaterialExpressionShaderStageSwitch
   inputs [
    PixelShader {from=MaterialExpressionVertexInterpolator_3, output=PS}
    VertexShader {from=MaterialExpressionPreSkinnedNormal_1, output=""}
   ]
   properties {}
  }
  MaterialExpressionShaderStageSwitch_2 {
   class=MaterialExpressionShaderStageSwitch
   inputs [
    PixelShader {from=MaterialExpressionVertexInterpolator_0, output=PS}
    VertexShader {from=MaterialExpressionLocalPosition_1, output=XYZ}
   ]
   properties {}
  }
  MaterialExpressionSubtract_0 {
   class=MaterialExpressionSubtract
   inputs [
    A {from=MaterialExpressionMaterialFunctionCall_7, output=Unsigned}
    B {from=""}
   ]
   properties {constB=0.5}
  }
  MaterialExpressionTransform_1 {
   class=MaterialExpressionTransform
   inputs [None {from=MaterialExpressionConstant3Vector_0}]
   properties {transformSourceType=TRANSFORMSOURCE_Instance}
  }
  MaterialExpressionTransform_2 {
   class=MaterialExpressionTransform
   inputs [None {from=MaterialExpressionConstant3Vector_1}]
   properties {transformSourceType=TRANSFORMSOURCE_Instance}
  }
  MaterialExpressionTransform_3 {
   class=MaterialExpressionTransform
   inputs [None {from=MaterialExpressionVertexNormalWS_0}]
   properties {transformSourceType=TRANSFORMSOURCE_World, transformType=TRANSFORM_Instance}
  }
  MaterialExpressionTransform_4 {
   class=MaterialExpressionTransform
   inputs [None {from=MaterialExpressionNamedRerouteUsage_48}]
   properties {transformSourceType=TRANSFORMSOURCE_Instance}
  }
  MaterialExpressionTransform_5 {
   class=MaterialExpressionTransform
   inputs [None {from=MaterialExpressionConstant3Vector_5}]
   properties {transformSourceType=TRANSFORMSOURCE_Instance}
  }
  MaterialExpressionTransform_6 {
   class=MaterialExpressionTransform
   inputs [None {from=MaterialExpressionNamedRerouteUsage_48}]
   properties {transformSourceType=TRANSFORMSOURCE_Local}
  }
  MaterialExpressionTransform_7 {
   class=MaterialExpressionTransform
   inputs [None {from=MaterialExpressionNamedRerouteUsage_49}]
   properties {transformSourceType=TRANSFORMSOURCE_Instance}
  }
  MaterialExpressionTransform_8 {
   class=MaterialExpressionTransform
   inputs [None {from=MaterialExpressionNamedRerouteUsage_49}]
   properties {transformSourceType=TRANSFORMSOURCE_Local}
  }
  MaterialExpressionVertexInterpolator_0 {class=MaterialExpressionVertexInterpolator, inputs [VS {from=MaterialExpressionLocalPosition_1, output=XYZ}], properties {}}
  MaterialExpressionVertexInterpolator_3 {class=MaterialExpressionVertexInterpolator, inputs [VS {from=MaterialExpressionPreSkinnedNormal_1, output=""}], properties {}}
  MaterialExpressionVertexNormalWS_0 {class=MaterialExpressionVertexNormalWS, inputs [], properties {}}
 }
 properties {
  bExposeToLibrary=true
  description="Generates UVs from a triplanar projection (blending X, Y, and Z axes). Uses dithered sampling to approximate axis blending at very low cost (single sample per axis). Fastest option, but dithering can produce visible artifacts and does not support post-sample modulation."
  userExposedCaption="CoordinateBasis: Triplanar (Dithered)"
  libraryCategoriesText [Base Material]
 }
}
